using { /Fortnite.com/Devices }
using { /Verse.org/Simulation }
using { /UnrealEngine.com/Temporary/Diagnostics }
using { /UnrealEngine.com/Temporary/SpatialMath }
using { /Verse.org/Random }
using { /Fortnite.com/Characters }
using { /Fortnite.com/Vehicles }
using { /Fortnite.com/Game }
using { /UnrealEngine.com/Temporary/UI }

using { HavocVerse.Combat }
using { HavocVerse.Debug }
using { HavocVerse.Devices }
using { HavocVerse.UI }
using { HavocVerse.Logging }
using { HavocVerse.Util }

boss_battle_device := class(creative_device):

    @editable BossProp: creative_prop = creative_prop{}
    @editable BossPropManipulator : prop_manipulator_device = prop_manipulator_device{}

    @editable BossFightWaitTimer: timer_device = timer_device{}
    @editable JoinFightDialog: popup_dialog_device = popup_dialog_device{}
    @editable EjectAndCloseMenusTrigger: trigger_device = trigger_device{}

    @editable BossPregameTimer: timer_device = timer_device{}

    @editable FullArenaVolume: volume_device = volume_device{}
    @editable TargetableAreaVolume: volume_device = volume_device{}

    var InPregame: logic = false
    @editable PregameRespawnVolume: volume_device = volume_device{}

    var m_PlayersInFight: [player]void = map{}
    var m_PlayersTargetable: [player]void = map{}

    var IsBossAlive: logic = false

    @editable MusicSwitch: switch_device = switch_device{}

    @editable PlayerJoinFightTrigger: trigger_device = trigger_device{}
    @editable FightStartTrigger: trigger_device = trigger_device{}
    @editable PlayerLeaveFightTrigger: trigger_device = trigger_device{}
    @editable BossDeathTrigger: trigger_device = trigger_device{}
    @editable BossDeathTriggerAll: trigger_device = trigger_device{}
    @editable FightEndTrigger: trigger_device = trigger_device{}
    @editable ClearMinionDamageVolume: damage_volume_device = damage_volume_device{}
    
    @editable RespawnTeleportersTrigger: trigger_device = trigger_device{}

    @editable SpawnTeleporters: []teleporter_device = array{}
    var TeleporterIndex: int = 0

    @editable IntroCameraCinematics: []cinematic_sequence_device = array{}
    @editable IntroCinematic: cinematic_sequence_device = cinematic_sequence_device{}
    @editable IntroCinematicLength: float = 5.0
    @editable ExtraIntroDelay: float = 1.0
    var IntroCinematicIndex: int = 0

    @editable BossStartingHealth1Player: float = 1000.0
    @editable BossStartingHealth2Players: float = 1000.0
    @editable BossStartingHealth3Players: float = 1000.0
    @editable BossStartingHealth4Players: float = 1000.0
    var BossStartingHealth: float = 100.0
    var BossHealth: float = 100.0

    @editable MoveCinematic: cinematic_sequence_device = cinematic_sequence_device{}
    var MovementSpeed: float = 500.0
    var RotateSpeed: float = 0.5
    @editable BossDamageCinematic: cinematic_sequence_device = cinematic_sequence_device{}
    @editable BossWinsCinematic: cinematic_sequence_device = cinematic_sequence_device{}
    @editable BossWinsCinematicLength: float = 5.0
    @editable BossDeathCinematic: cinematic_sequence_device = cinematic_sequence_device{}
    @editable BossDeathCinematicLength: float = 5.0
    @editable BossLocationResetCinematic: cinematic_sequence_device = cinematic_sequence_device{}

    @editable ArenaCenterPoint: creative_prop = creative_prop{}


    @editable DoCyclicDamage: logic = false
    @editable PermDamageVolumes: []volume_device = array{}
    @editable PermDamageAmount: float = 100.0
    @editable PermDamageSwitch: switch_device = switch_device{}

    @editable UseRandomAttacks: logic = false
    @editable Stages: []boss_stage_manager = array{}
    @editable NewStageAttack: new_stage_attack = new_stage_attack{}

    @editable Weapons: []boss_weapon_manager = array{}
    @editable DamageMultiplier: float = 1.0
    @editable ResourceMultiplier: float = 1.0
    @editable AttackExtraDelay: float = 1.0

    @editable p_ResourceManager: resource_manager_device = resource_manager_device{}
    @editable ResourceID: string = ""

    @editable p_HUDHealthBarConfig: hud_health_bar_config = hud_health_bar_config{}
    @editable p_HUDHealthBarPlacement: ui_canvas_placement_config = ui_canvas_placement_config{}

    HealthBar: combat_health_bar = combat_health_bar{
        LastDamageRetainTime := 1.0
        Health := 100.0
        MaxHealth := 100.0
    }
    @editable InFightAccessControl: access_control_device = access_control_device{
        p_Mode := access_control_mode.Whitelist
    }

    OnBegin<override>()<suspends>:void=
        BossFightWaitTimer.Start()
        JoinFightDialog.RespondingButtonEvent.Subscribe(OnJoinPopUpSelected)
        BossFightWaitTimer.SuccessEvent.Subscribe(OnWaitTimerSuccess)
        BossPregameTimer.SuccessEvent.Subscribe(OnPregameTimerSuccess)

        FullArenaVolume.AgentEntersEvent.Subscribe(OnPlayerEnterArena)
        FullArenaVolume.AgentExitsEvent.Subscribe(OnPlayerExitsArena)
        PregameRespawnVolume.AgentEntersEvent.Subscribe(OnPlayerEnterRespawnVolume)
        TargetableAreaVolume.AgentEntersEvent.Subscribe(OnPlayerEnterTargetableArena)
        TargetableAreaVolume.AgentExitsEvent.Subscribe(OnPlayerExitTargetableArena)

        BossPropManipulator.DamagedEvent.Subscribe(OnBossDamaged)

        for(Stage:Stages):
            Stage.AttacksSubscribe()
 
        NewStageAttack.AttacksSubscribe()

        for(DVolumes:PermDamageVolumes):
            DVolumes.AgentEntersEvent.Subscribe(OnPlayerDamaged)

        HudManager := boss_battle_hud_manager{
            AccessControl := Some(InFightAccessControl)
            RefreshRate := 0.5
            Device := Self
        }
        set HealthBar.RefreshHudManager = Some(HudManager)
        spawn:
            HudManager.Run(GetPlayspace())

    var m_LastActivityAt: float = -1.0
    BossDamageCheckLoop()<suspends>: void =
        loop {
            Sleep(0.0)
            if(not IsBossAlive?) {
                break
            }
            Timeout := if (GetPlayspace().GetPlayers().Length = 1). 300.0 else 120.0
            if (m_LastActivityAt > 0.0, GetSimulationElapsedTime() - m_LastActivityAt > Timeout){
                for (Player:m_PlayersInFight.Keys()) {
                    RespawnTeleportersTrigger.Trigger(Player)
                }
                break
            }
        }

    GetPlayerCount<public>()<transacts>:int=
        m_PlayersInFight.Keys().Length

    OnWaitTimerSuccess(Agent : ?agent): void =
        set InPregame = true
        JoinFightDialog.Show()
        BossPregameTimer.Start()

    OnPregameTimerSuccess(Agent : ?agent): void =
        var PlayerCount: int = GetPlayerCount()
        set InPregame = false

        if(PlayerCount>0):
            spawn:
                StartBossFight()
        else:
            spawn:
                BossFightCanceled()

    OnPlayerEnterArena(Agent : agent): void =
        if (Player := player[Agent]) {
            if. set m_PlayersInFight[Player] = void
            PlayerJoinFightTrigger.Trigger(Agent)
        }

    OnPlayerEnterRespawnVolume(Agent : agent): void =
        if(InPregame?):
            TeleportPlayerToArena(Agent)

    OnPlayerExitsArena(Agent : agent): void =
        if (Player := player[Agent]) {
            set m_PlayersInFight = m_PlayersInFight.RemoveByKey(Player)(0)
            PlayerLeaveFightTrigger.Trigger(Agent)
            spawn:
                DelayedRestartMusic(Agent)
        }

    DelayedRestartMusic(Agent:agent)<suspends>:void =
        Sleep(5.0)
        
        MusicSwitch.CheckState(Agent)

    OnPlayerEnterTargetableArena(Agent : agent): void =
        if:
            Player := player[Agent]
            set m_PlayersTargetable[Player] = void

    OnPlayerExitTargetableArena(Agent : agent): void =
        if:
            Player := player[Agent]
            set m_PlayersTargetable = m_PlayersTargetable.RemoveByKey(Player)(0)

    OnJoinPopUpSelected(Agent:agent, ResponseIndex:int):void=
        if(ResponseIndex = 0):
            spawn:
                OnJoinPopUpSelected0(Agent)

    OnJoinPopUpSelected0(Agent:agent)<suspends> :void=
        EjectAndCloseMenusTrigger.Trigger(Agent)

        if (
            Player := player[Agent]
            FortPlayer := Player.GetFortCharacter[]
            Vehicle := FortPlayer.GetVehicle[]
        ) {
            Log("Ejecting '{Player}' from their vehicle", ?Level := log_level.Debug)
            Vehicle.Damage(damage_args{
                Instigator := Some(FortPlayer)
                Amount := 10000.0
            })
        }

        Sleep(0.25)
        TeleportPlayerToArena(Agent)

    PlayIntroCinematic(Player : player):void=
        IntroCinematic.Play()
        if (Cinematic := IntroCameraCinematics[IntroCinematicIndex]):
            Cinematic.Play(Player)
            set IntroCinematicIndex += 1
        else if (Cinematic := IntroCameraCinematics[0]):
            Cinematic.Play(Player)
            set IntroCinematicIndex = 1

    TeleportPlayerToArena(Agent : agent):void=
        if (Teleporter := SpawnTeleporters[TeleporterIndex]):
            Teleporter.Teleport(Agent)
            set TeleporterIndex += 1
        else if (Teleporter := SpawnTeleporters[0]):
            Teleporter.Teleport(Agent)
            set TeleporterIndex = 1

    OnPlayerDamaged(Agent : agent): void =
        if(IsBossAlive?):
            if(not PermDamageSwitch.GetCurrentState[Agent]):
                PermDamageSwitch.SetState(Agent, true)
                if (FortCharacter := Agent.GetFortCharacter[]):
                    FortCharacter.Damage(PermDamageAmount)
                spawn:
                    DelayedDamageAttackCooldown(Agent) 
    
    DelayedDamageAttackCooldown(Agent : agent)<suspends>: void =
        Sleep(3.0)
        PermDamageSwitch.SetState(Agent, false)

    StartCyclicDamage()<suspends>:void=
        loop:
            if(IsBossAlive?):
                for(Player:m_PlayersInFight.Keys()):
                    if (FortCharacter := Player.GetFortCharacter[]):
                        FortCharacter.Damage(0.1)
                Sleep(5.0)
            else:
                break

    GetTargettablePlayerCount<public>()<transacts>:int=
        m_PlayersTargetable.Keys().Length

    OnBossDamaged(Agent : agent) : void =
        var DamageAmount: float = 0.0
        var ResourceAmount: int = 1

        set m_LastActivityAt = GetSimulationElapsedTime()

        if(IsBossAlive?):
            for(Weapon:Weapons):
                var ItemTest: logic = Weapon.CheckForItem(Agent)
                if(ItemTest?):
                    set DamageAmount = Weapon.GetDamageAmount(Agent)*DamageMultiplier
                    set BossHealth = BossHealth - DamageAmount
                    BossDamageCinematic.Play()
                    HealthBar.Damage(DamageAmount)
                    set ResourceAmount = Weapon.GetResourceAmount(Agent)
                    p_ResourceManager.Give(Agent, ResourceID, ResourceAmount, ?SendFeedback := true)


    RotateTowardsLocation(Location : vector3, Time : float)<suspends>:void=
        BossLocation := BossProp.GetTransform().Translation
        var Roll : float = 0.0

        if (LookDirection := (Location - BossLocation).MakeUnitVector[]):

            Yaw := RadiansToDegrees(ArcTan(LookDirection.Y, LookDirection.X)) - 90.0
            Pitch := RadiansToDegrees(ArcTan(LookDirection.Z, Sqrt((LookDirection.X * LookDirection.X) + (LookDirection.Y * LookDirection.Y))))
            set Roll = RadiansToDegrees(ArcTan(LookDirection.X, LookDirection.Z)) - 90.0

            if (Roll <= 0.0):
                set Roll = Roll * -1

            if (Roll >= 10.0):
                set Roll = 10.0

            NewRotation := MakeRotationFromYawPitchRollDegrees(Yaw, 0.0, 0.0)

            MoveCinematic.Play()
    
            BossProp.MoveTo(BossProp.GetTransform().Translation, NewRotation, Max(Time, 0.1))

            MoveCinematic.Stop()

    MoveTowardsRandomPlayer()<suspends>:void=
        var TargettablePlayerCount: int = GetTargettablePlayerCount()
        if(TargettablePlayerCount>0):
            var TargettablePlayers:[]player = array{}
            for (Player:m_PlayersInFight.Keys()) {
                set TargettablePlayers = TargettablePlayers + array{Player}
            }
            var RandomIndex: int = GetRandomInt(0,TargettablePlayers.Length-1)
            
            if (Player := TargettablePlayers[RandomIndex], Agent := agent[Player], FortCharacter := Agent.GetFortCharacter[]):
                BossLocation := BossProp.GetTransform().Translation
                PlayerLocation := FortCharacter.GetTransform().Translation

                DistanceX : float = BossLocation.X - PlayerLocation.X
                DistanceY : float = BossLocation.Y - PlayerLocation.Y
        
                BossDistance := Sqrt((DistanceX * DistanceX) + (DistanceY * DistanceY))

                RotateTowardsLocation(PlayerLocation, RotateSpeed)

                var NewPlayerLocation : vector3 = vector3{X:=PlayerLocation.X, Y:=PlayerLocation.Y, Z:=BossLocation.Z}

                MoveCinematic.Play()
                BossProp.MoveTo(NewPlayerLocation,BossProp.GetTransform().Rotation,Max(BossDistance/MovementSpeed, 0.1))
                MoveCinematic.Stop()

    MoveTowardsCenter()<suspends>:void=
        BossLocation := BossProp.GetTransform().Translation
        CenterLocation := ArenaCenterPoint.GetTransform().Translation
            
        DistanceX : float = BossLocation.X - CenterLocation.X
        DistanceY : float = BossLocation.Y - CenterLocation.Y
                    
        BossDistance := Sqrt((DistanceX * DistanceX) + (DistanceY * DistanceY))
            
        RotateTowardsLocation(CenterLocation, RotateSpeed)
            
        var NewCenterLocation : vector3 = vector3{X:=CenterLocation.X, Y:=CenterLocation.Y, Z:=BossLocation.Z}
            
        MoveCinematic.Play()
        BossProp.MoveTo(NewCenterLocation,BossProp.GetTransform().Rotation,Max(BossDistance/MovementSpeed, 0.1))
        MoveCinematic.Stop()

    StartBossFight()<suspends>:void=
        var PlayerCount: int = GetPlayerCount()

        if(PlayerCount>=4):
            set BossStartingHealth = BossStartingHealth4Players
        else if(PlayerCount>=3):
            set BossStartingHealth = BossStartingHealth3Players
        else if(PlayerCount>=2):
            set BossStartingHealth = BossStartingHealth2Players
        else if(PlayerCount>=1):
            set BossStartingHealth = BossStartingHealth1Player

        set BossHealth = BossStartingHealth

        for(Stage:Stages):
            Stage.SetHealthCutoffAmount(BossStartingHealth*Stage.GetHealthCutoff())

        ClearMinionDamageVolume.Disable()

        set HealthBar.Health = BossHealth
        set HealthBar.MaxHealth = BossHealth
        HealthBar.PostHealthChange()

        for (Player:m_PlayersInFight.Keys()) {
            PlayIntroCinematic(Player)
            FightStartTrigger.Trigger(Player)
        }
        Sleep(IntroCinematicLength)
        for (Player:m_PlayersInFight.Keys()) {
            MusicSwitch.CheckState(Player)
        }
        set IsBossAlive = true
        Sleep(ExtraIntroDelay)

        BossPregameTimer.Reset()

        if(DoCyclicDamage?):
            spawn:
                StartCyclicDamage()

        set m_LastActivityAt = GetSimulationElapsedTime()
        spawn:
            BossDamageCheckLoop()
        
        AttackLoop()
        

    BossFightCanceled()<suspends>:void=
        BossPregameTimer.Reset()
        BossFightWaitTimer.Reset()
        BossFightWaitTimer.Start()

    BossFightGameOver()<suspends>:void=
        set IsBossAlive = false
        FightEndTrigger.Trigger()
        ClearMinionDamageVolume.Enable()
        BossWinsCinematic.Play()
        Sleep(BossWinsCinematicLength)
        BossLocationResetCinematic.Play()
        BossPregameTimer.Reset()
        BossFightWaitTimer.Reset()
        BossFightWaitTimer.Start()

    BossDefeated()<suspends>:void=
        set IsBossAlive = false
        FightEndTrigger.Trigger()
        ClearMinionDamageVolume.Enable()

        for (Player:m_PlayersInFight.Keys()) {
            BossDeathTrigger.Trigger(Player)
        }
        
        BossDeathTriggerAll.Trigger()

        BossDeathCinematic.Play()
        Sleep(BossDeathCinematicLength)
        Sleep(5.0)

        for (Player:m_PlayersInFight.Keys()) {
            RespawnTeleportersTrigger.Trigger(Player)
        }

        BossLocationResetCinematic.Play()
        BossPregameTimer.Reset()
        BossFightWaitTimer.Reset()
        BossFightWaitTimer.Start()

    SelectStage()<suspends>:int=
        var Index: int = 0
        loop:
            if(Index>=Stages.Length):
                return 0

            if(Stage:=Stages[Index]):
                var StageCutoffAmount: float = Stage.GetHealthCutoffAmount()
                if(BossHealth>StageCutoffAmount):
                    set MovementSpeed = Stage.GetMovementSpeed()
                    set RotateSpeed = Stage.GetRotateSpeed()
                    return Index
            
            set Index = Index + 1

    AttackLoop()<suspends>:void=
        var MaxWeight: int = 0
        var AttackWeight: int = 0
        var CurrentStageIndex: int = 0
        var CurrentAttackIndex: int = 0
        var LastStageIndex: int = -1

        NewStageAttack.Attack.ResetTriggers()
        for(Stage:Stages):
            for(Attack:Stage.Attacks):
                Attack.ResetTriggers()

        if(FirstAttack:=Stages[0].Attacks[0]):

            loop:
                set CurrentStageIndex = SelectStage()

                if(not CurrentStageIndex = LastStageIndex):
                    set LastStageIndex = CurrentStageIndex
                    set CurrentAttackIndex = 0

                    if(not CurrentStageIndex = 0):
                        if(NewStageAttack.Attack.IsBossNearPlayerRequired[]):
                            MoveTowardsRandomPlayer()
                            NewStageAttack.Attack.StartAttack()
                        
                        if(NewStageAttack.Attack.IsBossAtCenterRequired[]):
                            MoveTowardsCenter()

                        if(NewStageAttack.Attack.IsLookTowardsTriggerRequired[]):
                            AttackTriggerLocationInt := NewStageAttack.Attack.GetRandomTriggerIndex()
                            if(AttackTrigger := NewStageAttack.Attack.RandomAttackTriggers[AttackTriggerLocationInt]):
                                AttackTriggerLocation := AttackTrigger.Trigger.GetTransform().Translation
                                RotateTowardsLocation(AttackTriggerLocation, RotateSpeed)
                                        
                                AttackTrigger.TriggerCustomTrigger()
                                NewStageAttack.Attack.StartAttack()

                    if(UseRandomAttacks?):
                        for(Attack:Stages[CurrentStageIndex].Attacks):
                            set MaxWeight = MaxWeight + Attack.GetWeight()

                if(UseRandomAttacks?):
                    var SelectedWeight: int = GetRandomInt(0,MaxWeight)
                    for(Attack:Stages[CurrentStageIndex].Attacks):
                        
                        set AttackWeight = Attack.GetWeight()
                        if(SelectedWeight<AttackWeight and SelectedWeight>0):
                            if(Attack.IsBossNearPlayerRequired[]):
                                MoveTowardsRandomPlayer()
                                Attack.StartAttack()
                            else if(Attack.IsLookTowardsTriggerRequired[]):
                                AttackTriggerLocationInt := Attack.GetRandomTriggerIndex()
                                if(AttackTrigger := Attack.RandomAttackTriggers[AttackTriggerLocationInt]):
                                    AttackTriggerLocation := AttackTrigger.Trigger.GetTransform().Translation
                                    RotateTowardsLocation(AttackTriggerLocation, RotateSpeed)
                                    
                                    AttackTrigger.TriggerCustomTrigger()
                                    Attack.StartAttack()

                            
                        else:
                            set SelectedWeight = SelectedWeight - AttackWeight
                
                else:
                    if(AttackOrderIndex:=Stages[CurrentStageIndex].AttackOrder[CurrentAttackIndex]):
                        if(Attack:=Stages[CurrentStageIndex].Attacks[AttackOrderIndex]):
                            if(Attack.IsBossNearPlayerRequired[]):
                                MoveTowardsRandomPlayer()

                            if(Attack.IsBossAtCenterRequired[]):
                                MoveTowardsCenter()

                            if(Attack.IsLookTowardsTriggerRequired[]):
                                AttackTriggerLocationInt := Attack.GetRandomTriggerIndex()
                                if(AttackTrigger := Attack.RandomAttackTriggers[AttackTriggerLocationInt]):
                                    AttackTriggerLocation := AttackTrigger.Trigger.GetTransform().Translation
                                    RotateTowardsLocation(AttackTriggerLocation, RotateSpeed)
                                        
                                    AttackTrigger.TriggerCustomTrigger()
                            Attack.StartAttack()

                            set CurrentAttackIndex = CurrentAttackIndex + 1
                            if(CurrentAttackIndex>=Stages[CurrentStageIndex].AttackOrder.Length):
                                set CurrentAttackIndex = 0

                var PlayerCount: int = GetPlayerCount()
                if(PlayerCount=0):
                    spawn:
                        BossFightGameOver()
                    break
                if(BossHealth<=0.0):
                    spawn:
                        BossDefeated()
                    break
                
                Sleep(AttackExtraDelay)

boss_stage_manager<internal> := class<final><concrete>:
    
    @editable HealthCutoff: float = 0.5
    var HealthCutoffAmount: float = 100.0
    @editable Attacks: []boss_attack = array{}
    @editable AttackOrder: []int = array{}
    @editable MovementSpeed: float = 500.0
    @editable RotateSpeed: float = 0.5

    GetHealthCutoff():float=
        return HealthCutoff

    GetMovementSpeed():float=
        return MovementSpeed
        
    GetRotateSpeed():float=
        return RotateSpeed

    GetHealthCutoffAmount():float=
        return HealthCutoffAmount
        
    SetHealthCutoffAmount(Amount:float):void=
        set HealthCutoffAmount = Amount

    AttacksSubscribe():void=
        for(Attack:Attacks):
            Attack.DamageVolumesSubscibe()

new_stage_attack<internal> := class<final><concrete>:
    @editable Attack: boss_attack = boss_attack{}

    AttacksSubscribe():void=
        Attack.DamageVolumesSubscibe()

    DoAttack()<suspends>:void=
        Attack.StartAttack()

boss_weapon_manager<internal> := class<final><concrete>:      
    @editable ItemButton : conditional_button_device = conditional_button_device{}
    @editable Item1Damage: float = 0.0
    @editable Item2Damage: float = 0.0
    @editable Item3Damage: float = 0.0
    @editable Item1Resource: float = 0.0
    @editable Item2Resource: float = 0.0
    @editable Item3Resource: float = 0.0
    @editable ResourceMultiplier: float = 1.0

    CheckForItem(Agent:agent) : logic =
        if (ItemButton.IsHoldingItem[Agent, 0]): 
            return true
        if (ItemButton.IsHoldingItem[Agent, 1]): 
            return true
        if (ItemButton.IsHoldingItem[Agent, 2]): 
            return true
        return false

    GetDamageAmount(Agent:agent) : float =
        if (ItemButton.IsHoldingItem[Agent, 0]): 
            return Item1Damage
        if (ItemButton.IsHoldingItem[Agent, 1]): 
            return Item2Damage
        if (ItemButton.IsHoldingItem[Agent, 2]): 
            return Item3Damage
        return 0.0
    
    GetResourceAmount(Agent:agent) : int =
        var ResourceAmount : int = 0
        var WholeNumber : int = 0
        var Decimal : float = 0.0
        var CurNumber : float = 0.0

        if (ItemButton.IsHoldingItem[Agent, 0]): 
            set CurNumber = Item1Resource*ResourceMultiplier
        if (ItemButton.IsHoldingItem[Agent, 1]): 
            set CurNumber = Item2Resource*ResourceMultiplier
        if (ItemButton.IsHoldingItem[Agent, 2]): 
            set CurNumber = Item3Resource*ResourceMultiplier

        loop:
            if(CurNumber < 1.0):
                set Decimal = CurNumber * 100.0
                break
            else:
                set WholeNumber = WholeNumber + 1
                set CurNumber = CurNumber - 1.0

        var DecimalConvert : int = 0
        if:
            value := Round[Decimal]
            set DecimalConvert = value

        if (GetRandomInt(0,100) < DecimalConvert):
            set ResourceAmount = WholeNumber + 1
        else:
            set ResourceAmount = WholeNumber    

        return ResourceAmount   

boss_attack<internal> := class<final><concrete>:
    var AttackActive: logic = false
    @editable Weight: int = 1

    @editable AttackCinematic: cinematic_sequence_device = cinematic_sequence_device{}
    @editable AttackTime: float = 3.0

    @editable RequiresBossNearPlayer: logic = true
    @editable RequiresBossAtCenter: logic = false
    @editable RequiresLookTowardsPlayer: logic = true
    @editable RequiresLookTowardsRandomTrigger: logic = false
    @editable RandomAttackTriggers: []custom_trigger = array{}

    @editable DamageVolumes: []volume_device = array{}
    @editable DamageSwitch: switch_device = switch_device{}
    @editable DamageAmount: float = 25.0
    @editable DamageDelay: float = 1.0

    GetWeight():int=
        return Weight

    IsBossNearPlayerRequired()<decides><transacts>:void=
        RequiresBossNearPlayer?

    IsBossAtCenterRequired()<decides><transacts>:void=
        RequiresBossAtCenter?

    IsLookTowardsRequired()<decides><transacts>:void=
        RequiresLookTowardsPlayer?

    IsLookTowardsTriggerRequired()<decides><transacts>:void=
        RequiresLookTowardsRandomTrigger?

    GetRandomTriggerIndex():int=
        var RandomIndex: int = 0
        var Index: int = 0
        loop:
            set RandomIndex = GetRandomInt(0,RandomAttackTriggers.Length-1)
            if(Index > 10):
                return 0
            
            set Index = Index + 1

            if(RandTrigger := RandomAttackTriggers[RandomIndex]):
                RandTrigActive := RandTrigger.GetIsActive()
                if(not RandTrigActive?):
                    return RandomIndex

    ResetTriggers():void=
        for(Trigger:RandomAttackTriggers):
            Trigger.ResetTrigger()

    GetAttackTime():float=
        return AttackTime

    IsAttackActive()<decides><transacts>:void=
        AttackActive?

    DamageVolumesSubscibe():void=
        for(Volume:DamageVolumes):
            Volume.AgentEntersEvent.Subscribe(OnPlayerHit)

    StartAttack()<suspends>:void=
        set AttackActive = true
        AttackCinematic.Play()
        Sleep(AttackTime)
        set AttackActive = false

    OnPlayerHit(Agent : agent): void =
        if(AttackActive?):
            if(not DamageSwitch.GetCurrentState[Agent]):
                DamageSwitch.SetState(Agent, true)
                if (FortCharacter := Agent.GetFortCharacter[]):
                    FortCharacter.Damage(DamageAmount)
                spawn:
                    DelayedAttackCooldown(Agent) 

    DelayedAttackCooldown(Agent : agent)<suspends>:void=
        Sleep(DamageDelay)
        DamageSwitch.SetState(Agent, false)

custom_trigger<internal> := class<final><concrete>:
    var IsActive: logic = false
    @editable Trigger: trigger_device = trigger_device{}

    TriggerCustomTrigger():void=
        Trigger.Trigger()
        set IsActive = true

    GetIsActive():logic=
        return IsActive 
    
    ResetTrigger():void=
        set IsActive = false


boss_battle_hud_manager := class(hud_manager_base):
    Device: boss_battle_device
        
    CreateHUD<override>(Player:player):?hud_manager_player=
        Element := CreateHUDHealthBar(Device.p_HUDHealthBarConfig, Device.HealthBar)
        Canvas := canvas{
            Slots := array{
                Device.p_HUDHealthBarPlacement.ToCanvasSlot(Element.Widget)
            }
        }
        option{MakeSingleElementHUDPlayer(Player, Canvas, Element)}