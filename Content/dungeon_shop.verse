using { /Fortnite.com/Devices }
using { /Verse.org/Simulation }
using { /Verse.org/Assets }
using { /Verse.org/Colors }
using { /UnrealEngine.com/Temporary/Diagnostics }
using { /UnrealEngine.com/Temporary/SpatialMath }
using { /UnrealEngine.com/Temporary/UI }
using { /Fortnite.com/UI }

using { HavocVerse.Debug }
using { HavocVerse.Devices }
using { HavocVerse.Resource }
using { HavocVerse.Storage }
using { HavocVerse.UI }
using { HavocVerse.Util }
using { HavocVerse.Logging }

dungeon_shop_item_mode := enum:
    Sell
    BuyResource
    BuyItemGranter

price_text_alignment_mode := enum:
    Left
    Right
    Center

dungeon_shop_item_config := class<concrete>:
    @editable Mode: dungeon_shop_item_mode = dungeon_shop_item_mode.Sell
    @editable NoPriceText: logic = false
    @editable AccessControl: []access_control_device = array{}
    @editable OriginResourceId: string = "N/A"
    @editable SellAmount: int = -1
    @editable PriceFormatter: []number_formatter_config = array{number_formatter_config{
        Template := "%Val%"
        Style := number_format_style.TenKMBT
    }}
    @editable TakeFeedback: logic = true
    @editable TransformedResourceId: string = "N/A"
    @editable BuyAmount: int = 1
    @editable BuyItemGranters: []item_granter_device = array{}
    @editable ButtonTextOverride: string = ""

    @editable TransactionTrigger: []trigger_device = array{}
    @editable TransactionTriggerPerAmount: logic = false

    @editable SinglePurchase: logic = false
    @editable SinglePurchaseChunkId: string = "shop1:item1"
    @editable SinglePurchaseRecoverTrigger: []trigger_device = array{}

    @editable UnlockedTexture: string = ""
    @editable LockedTexture: string = ""
    @editable SinglePuchasedTexture: string = ""

    @editable BasePrice: float = 50.0
    @editable Variance: float = 20.0
    @editable TimeA: float = 0.05
    @editable TimeB: float = 0.02

    @editable QuickSellTrigger: []trigger_device = array{}

    var m_AlreadyPurchased: [player]logic = map{}

    ComputePrice(Now:float)<transacts>:float=
        # Now is the current time in seconds elapsed since the start of the game
        (Variance / 2.0) * Sin(2 * PiFloat * TimeA * Now) + (Variance / 2.0) * Sin(2 * PiFloat * TimeB * Now) + BasePrice

    GetFractionOfRange(Val:float)<transacts>:float=
        Minimum := BasePrice - Variance
        Maximum := BasePrice + Variance
        if (Maximum = Minimum) {
            if (Val = Maximum). 1.0 else 0.0
        } else {
            Frac := (Val - Minimum) / (Maximum - Minimum)
            Max(0.0, Min(1.0, Frac))
        }

    IsUnlocked<public>(Player:player)<decides><transacts>:void=
        not AccessControl[0] or AccessControl[0].IsAllowed[Player]

dungeon_shop_price_color_mode := enum:
    Threshold
    Trend

dungeon_shop_price_color_threshold := class<concrete>:
    @editable RangePercent: float = 0.0
    @editable Color: color = NamedColors.White

dungeon_shop_menu_config := class<concrete>:
    @editable MenuYOffset<public>: float = 0.0
    @editable MenuXOffset<public>: float = 0.0
    @editable BackgroundTexture<public>: string = ""
    @editable BackgroundSize<public>: vector2 = vector2{X := 1280.0, Y := 720.0}
    @editable ItemsPerRow<public>: int = 6
    @editable GridTopOffset<public>: float = 120.0
    @editable PriceColorDefault<public>: color = NamedColors.White
    @editable PriceColorMode<public>: dungeon_shop_price_color_mode = dungeon_shop_price_color_mode.Threshold
    @editable PriceColorTrendUp<public>: color = NamedColors.Green
    @editable PriceColorTrendDown<public>: color = NamedColors.Red
    @editable PriceColoringThresholds<public>: []dungeon_shop_price_color_threshold = array{}
    @editable PriceTextPlacement<public>: vector2 = vector2{X := 0.0, Y := 6.0}
    @editable PriceTextAlignment<public>: price_text_alignment_mode = price_text_alignment_mode.Center
    @editable ActionButtonSizeHint<public>: vector2 = vector2{X := 160.0, Y := 50.0}
    @editable ItemBackgroundSize<public>: vector2 = vector2{X := 212.0, Y := 308.0}
    @editable ActionButtonOffsetToImage<public>: float = 10.0
    @editable GridGaps<public>: vector2 = vector2{X := 10.0, Y := 10.0}
    @editable SingleSellButton<public>: logic = false
    @editable SingleSellButtonText<public>: string = "Sell All"
    @editable SingleSellButtonYOffset<public>: float = 20.0
    @editable SingleSellButtonSizeHint<public>: vector2 = vector2{X := 434.0, Y := 50.0}
    @editable CloseButtonOffsetToGrid<public>: float = 20.0
    @editable CloseButtonSizeHint<public>: vector2 = vector2{X := 434.0, Y := 50.0}
    @editable TrendImagePlacement<public>: vector2 = vector2{X := 14.0, Y := 6.0}
    @editable TrendImageSizeHint<public>: vector2 = vector2{X := 40.0, Y := 40.0}
    @editable UpTrendTexture<public>: string = ""
    @editable DownTrendTexture<public>: string = ""

    var m_PriceColoringOrderedThresholds: []dungeon_shop_price_color_threshold = array{}

dungeon_shop_device := class(creative_device, aware_device, persistence_recordable):
    @editable p_ResourceManager: resource_manager_device = resource_manager_device{}
    @editable p_Items: []dungeon_shop_item_config = array{}
    @editable p_MenuConfig: dungeon_shop_menu_config = dungeon_shop_menu_config{}
    @editable p_RefreshRate: float = 1.0
    @editable p_NotEnoughFeedback: logic = true
    @editable p_HideLockedItemsPrices: logic = true

    @editable p_OpenButtons: []button_device = array{}
    @editable p_OpenTriggers: []trigger_device = array{}
    @editable p_CloseTriggers: []trigger_device = array{}

    @editable p_QuickSellAllTrigger: []trigger_device = array{}
    @editable p_QuickSellTakeFeedback: logic = true
    
    @editable MenuOpenedTrigger: trigger_device = trigger_device{}

    @editable MenuClosedTrigger: trigger_device = trigger_device{}

    var m_LastUpdateAt: float = 0.0

    var m_KnownMenuInstances: [player]dungeon_shop_menu_instance = map{}

    var m_PersistenceControl: persistence_recordable_control = persistence_recordable_control{}

    DeviceMeta<override>()<transacts>:aware_device_meta=
        aware_device_meta{
            DeviceName := "dungeon_shop"
        }

    OnBegin<override>()<suspends>:void=
        AwareInitDevice(Self, OnInit)

    OnInit<private>()<suspends>:void=
        set p_MenuConfig.m_PriceColoringOrderedThresholds = p_MenuConfig.PriceColoringThresholds.QuickSortWithOpts(CmpDungeonShopColorThreshold, quick_sort_opts{
            Order := sort_order.Descending
        })

        for (Button:p_OpenButtons) {
            Button.InteractedWithEvent.Subscribe(OnOpenButtonInteractedWith)
        }
        for (Trigger:p_OpenTriggers) {
            Trigger.TriggeredEvent.Subscribe(OnOpenTriggerTriggered)
        }
        for (Trigger:p_CloseTriggers) {
            Trigger.TriggeredEvent.Subscribe(OnCloseTriggerTriggered)
        }

        for (Trigger:p_QuickSellAllTrigger) {
            Trigger.TriggeredEvent.Subscribe(OnQuickSellAllTriggered)
        }

        for (Index->Item:p_Items, AccessControl := Item.AccessControl[0]) {
            Closure1 := dungeon_shop_item_access_control_closure{
                Device := Self
                ItemIndex := Index
            }
            AccessControl.SubscribeToPlayerJoinEvents(Closure1.OnPlayerAdded, Closure1.OnPlayerRemoved)

            if (Item.QuickSellTrigger.Length > 0) {
                Closure2 := dungeon_shop_item_trigger_closure{
                    Device := Self
                    Item := Item
                }
                for (Trigger:Item.QuickSellTrigger) {
                    Trigger.TriggeredEvent.Subscribe(Closure2.OnQuickSellTriggered)
                }
            }
        }
        GetPlayspace().PlayerRemovedEvent().Subscribe(OnPlayerQuitGame)

        spawn:
            StartUpdateLoop()

    StartUpdateLoop<private>()<suspends>:void=
        loop {
            Now := GetSimulationElapsedTime()
            set m_LastUpdateAt = Now
            for (__->MenuInstance:m_KnownMenuInstances, MenuInstance.IsActive[]) {
                MenuInstance.UpdatePrices()
            }
            Sleep(p_RefreshRate)
        }

    TriggerMenuClosedTrigger(Agent:agent):void=
        MenuClosedTrigger.Trigger(Agent)

    OnOpenButtonInteractedWith(Agent:agent):void=
        HandleMenuOpen(Some(Agent))

    OnOpenTriggerTriggered(Agent:?agent):void=
        HandleMenuOpen(Agent)

    OnQuickSellAllTriggered(Agent:?agent):void=
        if (Player := player[Agent?]) {
            for (Item:p_Items, Item.Mode = dungeon_shop_item_mode.Sell, Item.IsUnlocked[Player]) {
                QuickSell(Item, Player)
            }
        }

    QuickSell<internal>(Item:dungeon_shop_item_config, Player:player):void=
        Price := Item.ComputePrice(m_LastUpdateAt)
        SellAmount := if (Item.SellAmount > 0) {
            Item.SellAmount
        } else {
            Max(p_ResourceManager.GetAmount(Player, Item.OriginResourceId), 1)
        }
        Success := p_ResourceManager.Take(Player, Item.OriginResourceId, SellAmount,
            ?NotifyNotEnough := false, ?SendFeedback := p_QuickSellTakeFeedback)
        if (Success?) {
            IntPrice := DefaultingFloor(Price * SellAmount)
            p_ResourceManager.Give(Player, Item.TransformedResourceId, IntPrice, ?SendFeedback := true)
            SellAmount
        }

    OnCloseTriggerTriggered(Agent:?agent):void=
        if (l_Agent := Agent?) {
            if (Player := player[l_Agent], Instance := m_KnownMenuInstances[Player]) {
                Instance.Close()
            }
        } else {
            for (__->Instance:m_KnownMenuInstances) {
                Instance.Close()

            }
        }

    OnPlayerQuitGame(Player:player):void=
        for (Index->Item:p_Items) {
            set Item.m_AlreadyPurchased = Item.m_AlreadyPurchased.RemoveByKey(Player)(0)
        }

    HandleMenuOpen<private>(Agent:?agent):void=
        if (Player := player[Agent?], not m_KnownMenuInstances[Player], AssetsProvider := AwareDefaultAssetsProvider[]) {

            MenuOpenedTrigger.Trigger(Player)

            Menu := dungeon_shop_menu{
                AssetsProvider := AssetsProvider
                Device := Self
                Config := p_MenuConfig
            }
            Res := Menu.Open(Player)
            if:
                Instance := dungeon_shop_menu_instance[Res?]
                set m_KnownMenuInstances[Player] = Instance
        }

    GetPersistenceControl<override>()<transacts><decides>:persistence_recordable_control=
        m_PersistenceControl

    PersistenceReserve<override>(StorageMem:storage_memory)<transacts><decides>:dungeon_shop_persistence_state=
        var ItemChunks: [string]storage_memory_chunk = map{}
        for (Item:p_Items, Item.SinglePurchase?, ChunkId := Item.SinglePurchaseChunkId) {
            if:
                MemChunk := StorageMem.Reserve[1, ChunkId]
                set ItemChunks[ChunkId] = MemChunk
        }
        dungeon_shop_persistence_state{
            ItemChunks := ItemChunks
        }
    
    PersistenceSave<override>(i_State:persistence_recordable_state, Agent:agent, SaveCause:persistence_save_cause):void=
        if (State := dungeon_shop_persistence_state[i_State]) {
            for (Item:p_Items, Item.SinglePurchase?, MemChunk := State.ItemChunks[Item.SinglePurchaseChunkId]) {
                Purchased := option{Item.m_AlreadyPurchased[Agent]}.UnwrapOr(false)
                MemChunk.Write(Agent, array{Purchased})
            }
        }

    PersistenceLoad<override>(i_State:persistence_recordable_state, Agent:agent, Query:persistence_load_query):void=
        if (State := dungeon_shop_persistence_state[i_State], Player := player[Agent]) {
            for (Item:p_Items, Item.SinglePurchase?, MemChunk := State.ItemChunks[Item.SinglePurchaseChunkId]) {
                Data := MemChunk.Read(Agent)
                Purchased := option{Data[0]}.UnwrapOr(false)
                MayRecover := if (Query.Overwrite?) {
                    if (set Item.m_AlreadyPurchased[Player] = Purchased) {}
                    Purchased
                } else if (
                    Purchased?
                    PrevState := option{Item.m_AlreadyPurchased[Player]}.UnwrapOr(false)
                    PrevState = false
                    set Item.m_AlreadyPurchased[Player] = Purchased
                ) {
                    true
                } else {
                    false
                }
                if (MayRecover? and Query.FirstLoad?) {
                    for (Trigger:Item.SinglePurchaseRecoverTrigger) {
                        Trigger.Trigger(Player)
                    }
                }
            }
        }

dungeon_shop_persistence_state := class(persistence_recordable_state):
    ItemChunks: [string]storage_memory_chunk

    Reset<override>(Agent:agent):void=
        for (__->Chunk:ItemChunks) {
            Chunk.Reset(Agent)
        }

dungeon_shop_menu := class(ui_menu):
    Device: dungeon_shop_device
    Config: dungeon_shop_menu_config

    CreateInstance<override>(Player:player):tuple(ui_menu_instance,canvas)=
        Instance := dungeon_shop_menu_instance{
            Menu := Self
            Player := Player
            Device := Device
            AssetsProvider := AssetsProvider
        }

        RowSlots := for (RowIndex->Row:Device.p_Items.Chunks(Config.ItemsPerRow)) {
            ColumnSlots := for (Index->Item:Row) {
                Closure := dungeon_shop_menu_item_action_closure{
                    Instance := Instance
                    Item := Item
                }
                ButtonPlaceholder := if (Config.SingleSellButton? and Item.Mode = dungeon_shop_item_mode.Sell) {
                    None()
                } else {
                    Some of PlaceholderSizeHint(Config.ActionButtonSizeHint, true)
                }
                BackgroundImage := CreateDefaultingTextureBlock(Config.ItemBackgroundSize, None())
                Price := DefaultingFloor(Item.ComputePrice(Device.m_LastUpdateAt))
                PriceFormatter := CreateNumberFormatter(option{Item.PriceFormatter[0]})
                PriceTextBlock := CreateTextBlock(
                    "",
                    ?TextColor := Config.PriceColorDefault,
                    ?Justification :=
                        case(Config.PriceTextAlignment):
                            price_text_alignment_mode.Left => text_justification.Left
                            price_text_alignment_mode.Right => text_justification.Right
                            price_text_alignment_mode.Center => text_justification.Center
                )
                PriceTrendImage := CreateDefaultingTextureBlock(Config.TrendImageSizeHint, None())
                InstanceItem := dungeon_shop_menu_instance_item{
                    Instance := Instance
                    ItemConfig := Item
                    BackgroundImage := BackgroundImage
                    PriceTrendImage := PriceTrendImage
                    PriceFormatter := PriceFormatter
                    PriceTextBlock := MakeCachedTextBlock(PriceTextBlock)
                    ActionButtonClosure := Closure
                    ActionButtonPlaceholder := ButtonPlaceholder
                }
                InstanceItem.UpdateLockState(Player)
                InstanceItem.UpdatePrice(Player)
                set Instance.m_Items += array{InstanceItem}

                var OverlaySlots: []overlay_slot = array{
                    overlay_slot{
                        Widget := BackgroundImage.Placeholder.ParentWidget()
                    }
                }

                if (not (Item.NoPriceText?)) {
                    set OverlaySlots += array{
                        overlay_slot{
                            Widget := PriceTrendImage.Placeholder.ParentWidget()
                            VerticalAlignment := vertical_alignment.Top
                            HorizontalAlignment := if (Config.TrendImagePlacement.X >= 0.0) {
                                horizontal_alignment.Left
                            } else {
                                horizontal_alignment.Right
                            }
                            Padding := margin{
                                Top := Config.PriceTextPlacement.Y
                                Left := if (Config.TrendImagePlacement.X > 0.0) {
                                    Config.TrendImagePlacement.X
                                } else {
                                    0.0
                                }
                                Right := if (Config.TrendImagePlacement.X < 0.0) {
                                    -Config.TrendImagePlacement.X
                                } else {
                                    0.0
                                }
                            }
                        }

                        overlay_slot{
                            Widget := PriceTextBlock
                            VerticalAlignment := vertical_alignment.Top
                            HorizontalAlignment := if (Config.PriceTextPlacement.X > 0.0) {
                                horizontal_alignment.Left
                            } else if (Config.PriceTextPlacement.X < 0.0) {
                                horizontal_alignment.Right
                            } else {
                                horizontal_alignment.Center
                            }
                            Padding := margin{
                                Top := Config.PriceTextPlacement.Y
                                Left := if (Config.PriceTextPlacement.X > 0.0) {
                                    Config.PriceTextPlacement.X
                                } else {
                                    0.0
                                }
                                Right := if (Config.PriceTextPlacement.X < 0.0) {
                                    -Config.PriceTextPlacement.X
                                } else {
                                    0.0
                                }
                            }
                        }
                    }
                }

                ItemWidget := stack_box{
                    Orientation := orientation.Vertical
                    Slots := array{
                        stack_box_slot{
                            Widget := overlay{
                                Slots := OverlaySlots
                            }
                            VerticalAlignment := vertical_alignment.Top
                            HorizontalAlignment := horizontal_alignment.Center
                        }
                    }
                }
                if (l_ButtonPlaceholder := ButtonPlaceholder?) {
                    ItemWidget.AddWidget(stack_box_slot{
                        Widget := l_ButtonPlaceholder.ParentWidget()
                        VerticalAlignment := vertical_alignment.Top
                        HorizontalAlignment := horizontal_alignment.Center
                        Padding := margin{
                            Top := Config.ActionButtonOffsetToImage
                        }
                    })
                }
                stack_box_slot{
                    Widget := ItemWidget
                    HorizontalAlignment := horizontal_alignment.Left
                    Padding := margin{
                        Left := if (Index > 0). Config.GridGaps.X else 0.0
                    }
                }
            }

            RowWidget := stack_box{
                Orientation := orientation.Horizontal
                Slots := ColumnSlots
            }
            stack_box_slot{
                Widget := RowWidget
                VerticalAlignment := vertical_alignment.Top
                HorizontalAlignment := horizontal_alignment.Left
                Padding := margin{
                    Top := if (RowIndex > 0). Config.GridGaps.Y else 0.0
                }
            }
        }

        ItemsGrid := stack_box{
            Orientation := orientation.Vertical
            Slots := RowSlots
        }

        Widget := stack_box{
            Orientation := orientation.Vertical
            Slots := array{
                stack_box_slot{
                    Widget := ItemsGrid
                    VerticalAlignment := vertical_alignment.Top
                    HorizontalAlignment := horizontal_alignment.Center
                    Padding := margin{
                        Top := Config.GridTopOffset
                    }
                }
            }
        }

        var AnySell: logic = false
        for (Item:Device.p_Items, Item.Mode = dungeon_shop_item_mode.Sell, not AnySell?) {
            set AnySell = true
        }
        if (Config.SingleSellButton? and AnySell?) {
            SellAllButton := button_loud{
                DefaultText := ToMessage(Config.SingleSellButtonText)
            }
            SellAllButton.OnClick().Subscribe(Instance.OnSellAllClick)
            Widget.AddWidget(stack_box_slot{
                Widget := ApplySizeHint(SellAllButton, Config.SingleSellButtonSizeHint)
                VerticalAlignment := vertical_alignment.Top
                HorizontalAlignment := horizontal_alignment.Center
                Padding := margin{
                    Top := Config.SingleSellButtonYOffset
                }
            })
        }

        CloseButton := button_regular{
            DefaultText := ToMessage("Close")
        }
        CloseButton.OnClick().Subscribe(Instance.OnCloseClick)
        Widget.AddWidget(stack_box_slot{
            Widget := ApplySizeHint(CloseButton, Config.CloseButtonSizeHint)
            VerticalAlignment := vertical_alignment.Top
            HorizontalAlignment := horizontal_alignment.Center
            Padding := margin{
                Top := Config.CloseButtonOffsetToGrid
            }
        })

        Offsets := margin{
            Top := Config.MenuYOffset
            Left := Config.MenuXOffset
        }
        Canvas := CommonCanvas(Widget,
            ?BackgroundTexture := Config.BackgroundTexture, ?BackgroundSize := Config.BackgroundSize,
            ?Offsets := Some(Offsets))
        (Instance, Canvas)

dungeon_shop_menu_instance := class(ui_menu_instance):
    Device: dungeon_shop_device
    AssetsProvider: assets_provider

    var m_Items: []dungeon_shop_menu_instance_item = array{}

    UpdatePrices():void=
        for (InstanceItem:m_Items) {
            InstanceItem.UpdatePrice(Player)
        }

    OnClose<override>():void=
        Device.TriggerMenuClosedTrigger(Player)
        set Device.m_KnownMenuInstances = Device.m_KnownMenuInstances.RemoveByKey(Player)(0)

    OnSellAllClick<internal>(Data:widget_message):void=
        for (Item:Device.p_Items, Item.Mode = dungeon_shop_item_mode.Sell, Item.IsUnlocked[Player]) {
            Device.QuickSell(Item, Player)
        }

dungeon_shop_menu_instance_item := class:
    Instance: dungeon_shop_menu_instance
    ItemConfig: dungeon_shop_item_config
    BackgroundImage: defaulting_texture_block
    PriceTrendImage: defaulting_texture_block
    PriceFormatter: number_formatter
    PriceTextBlock: cached_text_block

    ActionButtonClosure: dungeon_shop_menu_item_action_closure
    ActionButtonPlaceholder: ?widget_placeholder
    var m_ActionButtonSubscription: ?cancelable = false

    UpdatePrice(Player:player):void=
        LastUpdateAt := Instance.Device.m_LastUpdateAt

        if (Instance.Device.p_HideLockedItemsPrices?, not ItemConfig.IsUnlocked[Player]) {
            PriceTextBlock.SetText("")
            PriceTrendImage.SetTexture(None())
            return
        }

        Price := DefaultingFloor(ItemConfig.ComputePrice(LastUpdateAt))
        PriceTextBlock.SetText(PriceFormatter.Format(Price))

        PrevPrice := if (LastUpdateAt > Instance.Device.p_RefreshRate) {
            DefaultingFloor(ItemConfig.ComputePrice(LastUpdateAt - Instance.Device.p_RefreshRate))
        } else {
            Price
        }
        MenuConfig := Instance.Device.p_MenuConfig
        case(MenuConfig.PriceColorMode):
            dungeon_shop_price_color_mode.Threshold =>
                RangePercent := ItemConfig.GetFractionOfRange(ToFloat(Price)) * 100.0
                var FoundThreshold: logic = false
                for (Index->Threshold:MenuConfig.m_PriceColoringOrderedThresholds, not (FoundThreshold?)) {
                    if (RangePercent >= Threshold.RangePercent) {
                        PriceTextBlock.SetTextColor(Threshold.Color)
                        set FoundThreshold = true
                    }
                }
                if (not (FoundThreshold?)) {
                    PriceTextBlock.SetTextColor(MenuConfig.PriceColorDefault)
                }
            dungeon_shop_price_color_mode.Trend =>
                PriceTextBlock.SetTextColor(if (Price > PrevPrice) {
                    MenuConfig.PriceColorTrendUp
                } else if (Price < PrevPrice) {
                    MenuConfig.PriceColorTrendDown
                } else {
                    MenuConfig.PriceColorDefault
                })
        TextureName := if (Price > PrevPrice) {
            Some(MenuConfig.UpTrendTexture)
        } else if (Price < PrevPrice) {
            Some(MenuConfig.DownTrendTexture)
        } else {
            None()
        }
        PriceTrendImage.SetCachingTexture(Instance.AssetsProvider, TextureName)

    UpdateLockState(Player:player):void=
        TextureName := if (ItemConfig.IsUnlocked[Player]) {
            if (ItemConfig.SinglePurchase? and ItemConfig.m_AlreadyPurchased[Player] = true) {
                ItemConfig.SinglePuchasedTexture
            } else {
                ItemConfig.UnlockedTexture
            }
        } else {
            ItemConfig.LockedTexture
        }
        Texture := if (TextureName.Length > 0) {
            option{Instance.AssetsProvider.GetTextureAsset[TextureName]}
        } else {
            None()
        }
        BackgroundImage.SetTexture(Texture)
        UpdateActionButton(Player)

    UpdateActionButton(Player:player):void=
        l_ActionButtonPlaceholder := if (i_ActionButtonPlaceholder := ActionButtonPlaceholder?) {
            i_ActionButtonPlaceholder
        } else {
            return
        }
        ExpectedButtonType := if (ItemConfig.IsUnlocked[Player]) {
            if (ItemConfig.SinglePurchase? and ItemConfig.m_AlreadyPurchased[Player] = true) {
                fort_button_type.Regular
            } else {
                fort_button_type.Loud
            }
        } else {
            fort_button_type.Regular
        }
        if (text_button_base[l_ActionButtonPlaceholder.GetWidget[]].GetButtonType[] = ExpectedButtonType) {
            return
        }
        if (l_Subscription := m_ActionButtonSubscription?) {
            l_Subscription.Cancel()
            set m_ActionButtonSubscription = false
        }
        ButtonText := if (ItemConfig.ButtonTextOverride.Length > 0) {
            ItemConfig.ButtonTextOverride
        } else {
            case(ItemConfig.Mode):
                dungeon_shop_item_mode.Sell => "Sell"
                dungeon_shop_item_mode.BuyResource => "Buy"
                dungeon_shop_item_mode.BuyItemGranter => "Buy"
        }
        Button := CreateFortButton(ExpectedButtonType, ToMessage(ButtonText))
        l_ActionButtonPlaceholder.SetWidget(Button)
        Subscription := Button.OnClick().Subscribe(ActionButtonClosure.OnClick)
        set m_ActionButtonSubscription = Some(Subscription)

dungeon_shop_menu_item_action_closure := class:
    Instance: dungeon_shop_menu_instance
    Item: dungeon_shop_item_config

    OnClick(Data:widget_message):void=
        if (not Instance.IsActive[]) {
            return
        }
        if (not Item.IsUnlocked[Instance.Player]) {
            return
        }
        if (Item.SinglePurchase? and Item.m_AlreadyPurchased[Instance.Player] = true) {
            return
        }
        ResourceManager := Instance.Device.p_ResourceManager
        var Price: float = Item.ComputePrice(Instance.Device.m_LastUpdateAt)
        TransactionAmount: int = case(Item.Mode):
            dungeon_shop_item_mode.Sell =>
                SellAmount := if (Item.SellAmount > 0) {
                    Item.SellAmount
                } else {
                    Max(ResourceManager.GetAmount(Instance.Player, Item.OriginResourceId), 1)
                }
                Success := ResourceManager.Take(Instance.Player, Item.OriginResourceId, SellAmount, ?NotifyNotEnough := Instance.Device.p_NotEnoughFeedback, ?SendFeedback := Item.TakeFeedback)
                if (Success?) {
                    IntPrice := DefaultingFloor(Price * SellAmount)
                    ResourceManager.Give(Instance.Player, Item.TransformedResourceId, IntPrice, ?SendFeedback := true)
                    SellAmount
                } else {
                    0
                }
            dungeon_shop_item_mode.BuyResource =>
                BuyAmount := if (Item.BuyAmount > 0) {
                    Item.BuyAmount
                } else {
                    OwnedAmount := Max(ResourceManager.GetAmount(Instance.Player, Item.OriginResourceId), 1)
                    set Price *= ToFloat(OwnedAmount) / Price
                    OwnedAmount
                }
                IntPrice := DefaultingFloor(Price)
                Success := ResourceManager.Take(Instance.Player, Item.OriginResourceId, IntPrice, ?NotifyNotEnough := Instance.Device.p_NotEnoughFeedback, ?SendFeedback := Item.TakeFeedback)
                if (Success?) {
                    ResourceManager.Give(Instance.Player, Item.TransformedResourceId, BuyAmount, ?SendFeedback := true)
                    BuyAmount
                } else {
                    0
                }
            dungeon_shop_item_mode.BuyItemGranter =>
                IntPrice := DefaultingFloor(Price)
                Success := ResourceManager.Take(Instance.Player, Item.OriginResourceId, IntPrice, ?NotifyNotEnough := Instance.Device.p_NotEnoughFeedback, ?SendFeedback := Item.TakeFeedback)
                if (Success?) {
                    for (ItemGranter:Item.BuyItemGranters) {
                        ItemGranter.GrantItem(Instance.Player)
                    }
                    1
                } else {
                    0
                }
        if (TransactionAmount > 0) {
            if (Item.SinglePurchase?) {
                if (set Item.m_AlreadyPurchased[Instance.Player] = true) {}
            }
            for (TransactionTrigger:Item.TransactionTrigger) {
                if (Item.TransactionTriggerPerAmount?) {
                    for (Index := 1..TransactionAmount) {
                        TransactionTrigger.Trigger(Instance.Player)
                    }
                } else {
                    TransactionTrigger.Trigger(Instance.Player)
                }
            }
        }
        for (InstanceItem:Instance.m_Items) {
            InstanceItem.UpdateLockState(Instance.Player)
        }

dungeon_shop_item_access_control_closure := class:
    Device: dungeon_shop_device
    ItemIndex: int

    OnPlayerAdded(Player:player):void=
        UpdateItemsState(Player)

    OnPlayerRemoved(Player:player, Quit:logic):void=
        UpdateItemsState(Player)

    UpdateItemsState<private>(Player:player):void=
        for (MenuInstance:Device.m_KnownMenuInstances) {
            for (Index->InstanceItem:MenuInstance.m_Items, Index = ItemIndex) {
                InstanceItem.UpdateLockState(Player)
            }
        }

dungeon_shop_item_trigger_closure := class:
    Device: dungeon_shop_device
    Item: dungeon_shop_item_config

    OnQuickSellTriggered(Agent:?agent):void=
        if (Player := player[Agent?], Item.Mode = dungeon_shop_item_mode.Sell, Item.IsUnlocked[Player]) {
            Device.QuickSell(Item, Player)
        }

CmpDungeonShopColorThreshold(A:dungeon_shop_price_color_threshold, B:dungeon_shop_price_color_threshold)<computes>:int=
    CmpFloat(A.RangePercent, B.RangePercent)

ToStringDungeonShopColorThreshold(A:dungeon_shop_price_color_threshold):string=
    "\{RangePercent: {A.RangePercent}\}"
