###################################################################
# Automatically generated using `gen_local_assets_provider.py`.
#
# Generated by: Brad @ 2024-10-25 02:53:10 UTC.
###################################################################

using { /Fortnite.com/Devices }
using { /Verse.org/Assets }
using { /UnrealEngine.com/Temporary/SpatialMath }

using { HavocVerse.Util }

Assets<public> := module:
    VFX<public> := module:
        Import<public> := module:
BuyplateBillboards<public> := module:
DailyRewardsMenuUI<public> := module:
FontQuicksandBold<public> := module:
GameplayRewardsMenuUI<public> := module:
HavocVerseUI<public> := module:
TemplateCommon<public> := module:
    BackendTooltips<public> := module:
UI<public> := module:
    Levels<public> := module:
    Rebirths<public> := module:
    Shop<public> := module:
UIAssets<public> := module:
    Apoc<public> := module:
    Widgets<public> := module:

local_assets_provider_device := class(creative_device, assets_provider_device):
    CreateAssetsProvider<override>():assets_provider=
        TextureAssets := array{
            texture_metadata{Name := "Assets/VFX/level_up", Asset := Assets.VFX.level_up, Dimensions := option{vector2{X := 512.0, Y := 128.0}}}
            texture_metadata{Name := "Assets/VFX/plus_armory", Asset := Assets.VFX.plus_armory, Dimensions := option{vector2{X := 256.0, Y := 256.0}}}
            texture_metadata{Name := "Assets/VFX/plus_gold", Asset := Assets.VFX.plus_gold, Dimensions := option{vector2{X := 512.0, Y := 512.0}}}
            texture_metadata{Name := "Assets/VFX/Import/T_NeonCircle", Asset := Assets.VFX.Import.T_NeonCircle, Dimensions := option{vector2{X := 1024.0, Y := 1024.0}}}
            texture_metadata{Name := "Assets/VFX/Import/T_Spark3", Asset := Assets.VFX.Import.T_Spark3, Dimensions := option{vector2{X := 64.0, Y := 64.0}}}
            texture_metadata{Name := "BuyplateBillboards/bp_001", Asset := BuyplateBillboards.bp_001, Dimensions := option{vector2{X := 512.0, Y := 256.0}}}
            texture_metadata{Name := "BuyplateBillboards/bp_002", Asset := BuyplateBillboards.bp_002, Dimensions := option{vector2{X := 512.0, Y := 256.0}}}
            texture_metadata{Name := "BuyplateBillboards/bp_003", Asset := BuyplateBillboards.bp_003, Dimensions := option{vector2{X := 512.0, Y := 256.0}}}
            texture_metadata{Name := "BuyplateBillboards/bp_004", Asset := BuyplateBillboards.bp_004, Dimensions := option{vector2{X := 512.0, Y := 256.0}}}
            texture_metadata{Name := "BuyplateBillboards/bp_005", Asset := BuyplateBillboards.bp_005, Dimensions := option{vector2{X := 512.0, Y := 256.0}}}
            texture_metadata{Name := "BuyplateBillboards/bp_006", Asset := BuyplateBillboards.bp_006, Dimensions := option{vector2{X := 512.0, Y := 256.0}}}
            texture_metadata{Name := "BuyplateBillboards/bp_007", Asset := BuyplateBillboards.bp_007, Dimensions := option{vector2{X := 512.0, Y := 256.0}}}
            texture_metadata{Name := "BuyplateBillboards/bp_008", Asset := BuyplateBillboards.bp_008, Dimensions := option{vector2{X := 512.0, Y := 256.0}}}
            texture_metadata{Name := "BuyplateBillboards/bp_009", Asset := BuyplateBillboards.bp_009, Dimensions := option{vector2{X := 512.0, Y := 256.0}}}
            texture_metadata{Name := "BuyplateBillboards/bp_010", Asset := BuyplateBillboards.bp_010, Dimensions := option{vector2{X := 512.0, Y := 256.0}}}
            texture_metadata{Name := "BuyplateBillboards/bp_011", Asset := BuyplateBillboards.bp_011, Dimensions := option{vector2{X := 512.0, Y := 256.0}}}
            texture_metadata{Name := "BuyplateBillboards/bp_012", Asset := BuyplateBillboards.bp_012, Dimensions := option{vector2{X := 512.0, Y := 256.0}}}
            texture_metadata{Name := "BuyplateBillboards/bp_013", Asset := BuyplateBillboards.bp_013, Dimensions := option{vector2{X := 512.0, Y := 256.0}}}
            texture_metadata{Name := "BuyplateBillboards/bp_014", Asset := BuyplateBillboards.bp_014, Dimensions := option{vector2{X := 512.0, Y := 256.0}}}
            texture_metadata{Name := "BuyplateBillboards/bp_015", Asset := BuyplateBillboards.bp_015, Dimensions := option{vector2{X := 512.0, Y := 256.0}}}
            texture_metadata{Name := "BuyplateBillboards/bp_016", Asset := BuyplateBillboards.bp_016, Dimensions := option{vector2{X := 512.0, Y := 256.0}}}
            texture_metadata{Name := "BuyplateBillboards/bp_017", Asset := BuyplateBillboards.bp_017, Dimensions := option{vector2{X := 512.0, Y := 256.0}}}
            texture_metadata{Name := "BuyplateBillboards/bp_018", Asset := BuyplateBillboards.bp_018, Dimensions := option{vector2{X := 512.0, Y := 256.0}}}
            texture_metadata{Name := "BuyplateBillboards/bp_019", Asset := BuyplateBillboards.bp_019, Dimensions := option{vector2{X := 512.0, Y := 256.0}}}
            texture_metadata{Name := "BuyplateBillboards/bp_020", Asset := BuyplateBillboards.bp_020, Dimensions := option{vector2{X := 512.0, Y := 256.0}}}
            texture_metadata{Name := "BuyplateBillboards/bp_021", Asset := BuyplateBillboards.bp_021, Dimensions := option{vector2{X := 512.0, Y := 256.0}}}
            texture_metadata{Name := "BuyplateBillboards/bp_022", Asset := BuyplateBillboards.bp_022, Dimensions := option{vector2{X := 512.0, Y := 256.0}}}
            texture_metadata{Name := "BuyplateBillboards/bp_023", Asset := BuyplateBillboards.bp_023, Dimensions := option{vector2{X := 512.0, Y := 256.0}}}
            texture_metadata{Name := "BuyplateBillboards/bp_024", Asset := BuyplateBillboards.bp_024, Dimensions := option{vector2{X := 512.0, Y := 256.0}}}
            texture_metadata{Name := "BuyplateBillboards/bp_025", Asset := BuyplateBillboards.bp_025, Dimensions := option{vector2{X := 512.0, Y := 256.0}}}
            texture_metadata{Name := "BuyplateBillboards/bp_026", Asset := BuyplateBillboards.bp_026, Dimensions := option{vector2{X := 512.0, Y := 256.0}}}
            texture_metadata{Name := "BuyplateBillboards/bp_027", Asset := BuyplateBillboards.bp_027, Dimensions := option{vector2{X := 512.0, Y := 256.0}}}
            texture_metadata{Name := "BuyplateBillboards/bp_028", Asset := BuyplateBillboards.bp_028, Dimensions := option{vector2{X := 512.0, Y := 256.0}}}
            texture_metadata{Name := "BuyplateBillboards/bp_029", Asset := BuyplateBillboards.bp_029, Dimensions := option{vector2{X := 512.0, Y := 256.0}}}
            texture_metadata{Name := "BuyplateBillboards/bp_030", Asset := BuyplateBillboards.bp_030, Dimensions := option{vector2{X := 512.0, Y := 256.0}}}
            texture_metadata{Name := "BuyplateBillboards/bp_031", Asset := BuyplateBillboards.bp_031, Dimensions := option{vector2{X := 512.0, Y := 256.0}}}
            texture_metadata{Name := "BuyplateBillboards/bp_032", Asset := BuyplateBillboards.bp_032, Dimensions := option{vector2{X := 512.0, Y := 256.0}}}
            texture_metadata{Name := "BuyplateBillboards/bp_033", Asset := BuyplateBillboards.bp_033, Dimensions := option{vector2{X := 512.0, Y := 256.0}}}
            texture_metadata{Name := "BuyplateBillboards/bp_034", Asset := BuyplateBillboards.bp_034, Dimensions := option{vector2{X := 512.0, Y := 256.0}}}
            texture_metadata{Name := "BuyplateBillboards/bp_035", Asset := BuyplateBillboards.bp_035, Dimensions := option{vector2{X := 512.0, Y := 256.0}}}
            texture_metadata{Name := "BuyplateBillboards/bp_036", Asset := BuyplateBillboards.bp_036, Dimensions := option{vector2{X := 512.0, Y := 256.0}}}
            texture_metadata{Name := "BuyplateBillboards/bp_037", Asset := BuyplateBillboards.bp_037, Dimensions := option{vector2{X := 512.0, Y := 256.0}}}
            texture_metadata{Name := "BuyplateBillboards/bp_038", Asset := BuyplateBillboards.bp_038, Dimensions := option{vector2{X := 512.0, Y := 256.0}}}
            texture_metadata{Name := "BuyplateBillboards/bp_039", Asset := BuyplateBillboards.bp_039, Dimensions := option{vector2{X := 512.0, Y := 256.0}}}
            texture_metadata{Name := "BuyplateBillboards/bp_040", Asset := BuyplateBillboards.bp_040, Dimensions := option{vector2{X := 512.0, Y := 256.0}}}
            texture_metadata{Name := "BuyplateBillboards/bp_041", Asset := BuyplateBillboards.bp_041, Dimensions := option{vector2{X := 512.0, Y := 256.0}}}
            texture_metadata{Name := "BuyplateBillboards/bp_042", Asset := BuyplateBillboards.bp_042, Dimensions := option{vector2{X := 512.0, Y := 256.0}}}
            texture_metadata{Name := "BuyplateBillboards/bp_043", Asset := BuyplateBillboards.bp_043, Dimensions := option{vector2{X := 512.0, Y := 256.0}}}
            texture_metadata{Name := "BuyplateBillboards/bp_044", Asset := BuyplateBillboards.bp_044, Dimensions := option{vector2{X := 512.0, Y := 256.0}}}
            texture_metadata{Name := "BuyplateBillboards/bp_045", Asset := BuyplateBillboards.bp_045, Dimensions := option{vector2{X := 512.0, Y := 256.0}}}
            texture_metadata{Name := "BuyplateBillboards/bp_046", Asset := BuyplateBillboards.bp_046, Dimensions := option{vector2{X := 512.0, Y := 256.0}}}
            texture_metadata{Name := "BuyplateBillboards/bp_047", Asset := BuyplateBillboards.bp_047, Dimensions := option{vector2{X := 512.0, Y := 256.0}}}
            texture_metadata{Name := "BuyplateBillboards/bp_048", Asset := BuyplateBillboards.bp_048, Dimensions := option{vector2{X := 512.0, Y := 256.0}}}
            texture_metadata{Name := "BuyplateBillboards/bp_049", Asset := BuyplateBillboards.bp_049, Dimensions := option{vector2{X := 512.0, Y := 256.0}}}
            texture_metadata{Name := "BuyplateBillboards/bp_050", Asset := BuyplateBillboards.bp_050, Dimensions := option{vector2{X := 512.0, Y := 256.0}}}
            texture_metadata{Name := "BuyplateBillboards/bp_051", Asset := BuyplateBillboards.bp_051, Dimensions := option{vector2{X := 512.0, Y := 256.0}}}
            texture_metadata{Name := "BuyplateBillboards/bp_052", Asset := BuyplateBillboards.bp_052, Dimensions := option{vector2{X := 512.0, Y := 256.0}}}
            texture_metadata{Name := "BuyplateBillboards/bp_053", Asset := BuyplateBillboards.bp_053, Dimensions := option{vector2{X := 512.0, Y := 256.0}}}
            texture_metadata{Name := "BuyplateBillboards/bp_054", Asset := BuyplateBillboards.bp_054, Dimensions := option{vector2{X := 512.0, Y := 256.0}}}
            texture_metadata{Name := "BuyplateBillboards/bp_055", Asset := BuyplateBillboards.bp_055, Dimensions := option{vector2{X := 512.0, Y := 256.0}}}
            texture_metadata{Name := "BuyplateBillboards/bp_056", Asset := BuyplateBillboards.bp_056, Dimensions := option{vector2{X := 512.0, Y := 256.0}}}
            texture_metadata{Name := "BuyplateBillboards/bp_057", Asset := BuyplateBillboards.bp_057, Dimensions := option{vector2{X := 512.0, Y := 256.0}}}
            texture_metadata{Name := "BuyplateBillboards/bp_058", Asset := BuyplateBillboards.bp_058, Dimensions := option{vector2{X := 512.0, Y := 256.0}}}
            texture_metadata{Name := "BuyplateBillboards/bp_059", Asset := BuyplateBillboards.bp_059, Dimensions := option{vector2{X := 512.0, Y := 256.0}}}
            texture_metadata{Name := "BuyplateBillboards/bp_060", Asset := BuyplateBillboards.bp_060, Dimensions := option{vector2{X := 512.0, Y := 256.0}}}
            texture_metadata{Name := "BuyplateBillboards/bp_061", Asset := BuyplateBillboards.bp_061, Dimensions := option{vector2{X := 512.0, Y := 256.0}}}
            texture_metadata{Name := "BuyplateBillboards/bp_062", Asset := BuyplateBillboards.bp_062, Dimensions := option{vector2{X := 512.0, Y := 256.0}}}
            texture_metadata{Name := "BuyplateBillboards/bp_063", Asset := BuyplateBillboards.bp_063, Dimensions := option{vector2{X := 512.0, Y := 256.0}}}
            texture_metadata{Name := "BuyplateBillboards/bp_064", Asset := BuyplateBillboards.bp_064, Dimensions := option{vector2{X := 512.0, Y := 256.0}}}
            texture_metadata{Name := "BuyplateBillboards/bp_065", Asset := BuyplateBillboards.bp_065, Dimensions := option{vector2{X := 512.0, Y := 256.0}}}
            texture_metadata{Name := "BuyplateBillboards/bp_066", Asset := BuyplateBillboards.bp_066, Dimensions := option{vector2{X := 512.0, Y := 256.0}}}
            texture_metadata{Name := "BuyplateBillboards/bp_067", Asset := BuyplateBillboards.bp_067, Dimensions := option{vector2{X := 512.0, Y := 256.0}}}
            texture_metadata{Name := "BuyplateBillboards/bp_068", Asset := BuyplateBillboards.bp_068, Dimensions := option{vector2{X := 512.0, Y := 256.0}}}
            texture_metadata{Name := "BuyplateBillboards/bp_069", Asset := BuyplateBillboards.bp_069, Dimensions := option{vector2{X := 512.0, Y := 256.0}}}
            texture_metadata{Name := "BuyplateBillboards/bp_070", Asset := BuyplateBillboards.bp_070, Dimensions := option{vector2{X := 512.0, Y := 256.0}}}
            texture_metadata{Name := "BuyplateBillboards/bp_071", Asset := BuyplateBillboards.bp_071, Dimensions := option{vector2{X := 512.0, Y := 256.0}}}
            texture_metadata{Name := "BuyplateBillboards/bp_072", Asset := BuyplateBillboards.bp_072, Dimensions := option{vector2{X := 512.0, Y := 256.0}}}
            texture_metadata{Name := "BuyplateBillboards/bp_073", Asset := BuyplateBillboards.bp_073, Dimensions := option{vector2{X := 512.0, Y := 256.0}}}
            texture_metadata{Name := "BuyplateBillboards/bp_074", Asset := BuyplateBillboards.bp_074, Dimensions := option{vector2{X := 512.0, Y := 256.0}}}
            texture_metadata{Name := "BuyplateBillboards/bp_075", Asset := BuyplateBillboards.bp_075, Dimensions := option{vector2{X := 512.0, Y := 256.0}}}
            texture_metadata{Name := "BuyplateBillboards/bp_076", Asset := BuyplateBillboards.bp_076, Dimensions := option{vector2{X := 512.0, Y := 256.0}}}
            texture_metadata{Name := "BuyplateBillboards/bp_077", Asset := BuyplateBillboards.bp_077, Dimensions := option{vector2{X := 512.0, Y := 256.0}}}
            texture_metadata{Name := "BuyplateBillboards/bp_078", Asset := BuyplateBillboards.bp_078, Dimensions := option{vector2{X := 512.0, Y := 256.0}}}
            texture_metadata{Name := "BuyplateBillboards/bp_079", Asset := BuyplateBillboards.bp_079, Dimensions := option{vector2{X := 512.0, Y := 256.0}}}
            texture_metadata{Name := "BuyplateBillboards/bp_080", Asset := BuyplateBillboards.bp_080, Dimensions := option{vector2{X := 512.0, Y := 256.0}}}
            texture_metadata{Name := "BuyplateBillboards/bp_081", Asset := BuyplateBillboards.bp_081, Dimensions := option{vector2{X := 512.0, Y := 256.0}}}
            texture_metadata{Name := "BuyplateBillboards/bp_082", Asset := BuyplateBillboards.bp_082, Dimensions := option{vector2{X := 512.0, Y := 256.0}}}
            texture_metadata{Name := "BuyplateBillboards/bp_083", Asset := BuyplateBillboards.bp_083, Dimensions := option{vector2{X := 512.0, Y := 256.0}}}
            texture_metadata{Name := "BuyplateBillboards/bp_084", Asset := BuyplateBillboards.bp_084, Dimensions := option{vector2{X := 512.0, Y := 256.0}}}
            texture_metadata{Name := "BuyplateBillboards/bp_085", Asset := BuyplateBillboards.bp_085, Dimensions := option{vector2{X := 512.0, Y := 256.0}}}
            texture_metadata{Name := "BuyplateBillboards/bp_086", Asset := BuyplateBillboards.bp_086, Dimensions := option{vector2{X := 512.0, Y := 256.0}}}
            texture_metadata{Name := "BuyplateBillboards/bp_087", Asset := BuyplateBillboards.bp_087, Dimensions := option{vector2{X := 512.0, Y := 256.0}}}
            texture_metadata{Name := "BuyplateBillboards/bp_088", Asset := BuyplateBillboards.bp_088, Dimensions := option{vector2{X := 512.0, Y := 256.0}}}
            texture_metadata{Name := "BuyplateBillboards/bp_089", Asset := BuyplateBillboards.bp_089, Dimensions := option{vector2{X := 512.0, Y := 256.0}}}
            texture_metadata{Name := "BuyplateBillboards/bp_090", Asset := BuyplateBillboards.bp_090, Dimensions := option{vector2{X := 512.0, Y := 256.0}}}
            texture_metadata{Name := "BuyplateBillboards/bp_091", Asset := BuyplateBillboards.bp_091, Dimensions := option{vector2{X := 512.0, Y := 256.0}}}
            texture_metadata{Name := "BuyplateBillboards/bp_092", Asset := BuyplateBillboards.bp_092, Dimensions := option{vector2{X := 512.0, Y := 256.0}}}
            texture_metadata{Name := "BuyplateBillboards/bp_093", Asset := BuyplateBillboards.bp_093, Dimensions := option{vector2{X := 512.0, Y := 256.0}}}
            texture_metadata{Name := "BuyplateBillboards/bp_094", Asset := BuyplateBillboards.bp_094, Dimensions := option{vector2{X := 512.0, Y := 256.0}}}
            texture_metadata{Name := "BuyplateBillboards/bp_095", Asset := BuyplateBillboards.bp_095, Dimensions := option{vector2{X := 512.0, Y := 256.0}}}
            texture_metadata{Name := "BuyplateBillboards/bp_096", Asset := BuyplateBillboards.bp_096, Dimensions := option{vector2{X := 512.0, Y := 256.0}}}
            texture_metadata{Name := "BuyplateBillboards/bp_097", Asset := BuyplateBillboards.bp_097, Dimensions := option{vector2{X := 512.0, Y := 256.0}}}
            texture_metadata{Name := "BuyplateBillboards/bp_098", Asset := BuyplateBillboards.bp_098, Dimensions := option{vector2{X := 512.0, Y := 256.0}}}
            texture_metadata{Name := "BuyplateBillboards/bp_099", Asset := BuyplateBillboards.bp_099, Dimensions := option{vector2{X := 512.0, Y := 256.0}}}
            texture_metadata{Name := "BuyplateBillboards/bp_100", Asset := BuyplateBillboards.bp_100, Dimensions := option{vector2{X := 512.0, Y := 256.0}}}
            texture_metadata{Name := "BuyplateBillboards/bp_101", Asset := BuyplateBillboards.bp_101, Dimensions := option{vector2{X := 512.0, Y := 256.0}}}
            texture_metadata{Name := "BuyplateBillboards/bp_102", Asset := BuyplateBillboards.bp_102, Dimensions := option{vector2{X := 512.0, Y := 256.0}}}
            texture_metadata{Name := "BuyplateBillboards/bp_103", Asset := BuyplateBillboards.bp_103, Dimensions := option{vector2{X := 512.0, Y := 256.0}}}
            texture_metadata{Name := "BuyplateBillboards/bp_104", Asset := BuyplateBillboards.bp_104, Dimensions := option{vector2{X := 512.0, Y := 256.0}}}
            texture_metadata{Name := "BuyplateBillboards/bp_105", Asset := BuyplateBillboards.bp_105, Dimensions := option{vector2{X := 512.0, Y := 256.0}}}
            texture_metadata{Name := "BuyplateBillboards/bp_106", Asset := BuyplateBillboards.bp_106, Dimensions := option{vector2{X := 512.0, Y := 256.0}}}
            texture_metadata{Name := "BuyplateBillboards/bp_107", Asset := BuyplateBillboards.bp_107, Dimensions := option{vector2{X := 512.0, Y := 256.0}}}
            texture_metadata{Name := "BuyplateBillboards/bp_108", Asset := BuyplateBillboards.bp_108, Dimensions := option{vector2{X := 512.0, Y := 256.0}}}
            texture_metadata{Name := "BuyplateBillboards/bp_109", Asset := BuyplateBillboards.bp_109, Dimensions := option{vector2{X := 512.0, Y := 256.0}}}
            texture_metadata{Name := "BuyplateBillboards/bp_110", Asset := BuyplateBillboards.bp_110, Dimensions := option{vector2{X := 512.0, Y := 256.0}}}
            texture_metadata{Name := "BuyplateBillboards/bp_200", Asset := BuyplateBillboards.bp_200, Dimensions := option{vector2{X := 512.0, Y := 256.0}}}
            texture_metadata{Name := "BuyplateBillboards/bp_201", Asset := BuyplateBillboards.bp_201, Dimensions := option{vector2{X := 512.0, Y := 256.0}}}
            texture_metadata{Name := "BuyplateBillboards/bp_202", Asset := BuyplateBillboards.bp_202, Dimensions := option{vector2{X := 512.0, Y := 256.0}}}
            texture_metadata{Name := "BuyplateBillboards/bp_203", Asset := BuyplateBillboards.bp_203, Dimensions := option{vector2{X := 512.0, Y := 256.0}}}
            texture_metadata{Name := "BuyplateBillboards/bp_204", Asset := BuyplateBillboards.bp_204, Dimensions := option{vector2{X := 512.0, Y := 256.0}}}
            texture_metadata{Name := "BuyplateBillboards/bp_205", Asset := BuyplateBillboards.bp_205, Dimensions := option{vector2{X := 512.0, Y := 256.0}}}
            texture_metadata{Name := "BuyplateBillboards/bp_206", Asset := BuyplateBillboards.bp_206, Dimensions := option{vector2{X := 512.0, Y := 256.0}}}
            texture_metadata{Name := "BuyplateBillboards/bp_207", Asset := BuyplateBillboards.bp_207, Dimensions := option{vector2{X := 512.0, Y := 256.0}}}
            texture_metadata{Name := "BuyplateBillboards/bp_208", Asset := BuyplateBillboards.bp_208, Dimensions := option{vector2{X := 512.0, Y := 256.0}}}
            texture_metadata{Name := "BuyplateBillboards/bp_209", Asset := BuyplateBillboards.bp_209, Dimensions := option{vector2{X := 512.0, Y := 256.0}}}
            texture_metadata{Name := "BuyplateBillboards/bp_210", Asset := BuyplateBillboards.bp_210, Dimensions := option{vector2{X := 512.0, Y := 256.0}}}
            texture_metadata{Name := "BuyplateBillboards/bp_211", Asset := BuyplateBillboards.bp_211, Dimensions := option{vector2{X := 512.0, Y := 256.0}}}
            texture_metadata{Name := "BuyplateBillboards/bp_212", Asset := BuyplateBillboards.bp_212, Dimensions := option{vector2{X := 512.0, Y := 256.0}}}
            texture_metadata{Name := "BuyplateBillboards/bp_213", Asset := BuyplateBillboards.bp_213, Dimensions := option{vector2{X := 512.0, Y := 256.0}}}
            texture_metadata{Name := "BuyplateBillboards/bp_214", Asset := BuyplateBillboards.bp_214, Dimensions := option{vector2{X := 512.0, Y := 256.0}}}
            texture_metadata{Name := "BuyplateBillboards/bp_215", Asset := BuyplateBillboards.bp_215, Dimensions := option{vector2{X := 512.0, Y := 256.0}}}
            texture_metadata{Name := "BuyplateBillboards/bp_216", Asset := BuyplateBillboards.bp_216, Dimensions := option{vector2{X := 512.0, Y := 256.0}}}
            texture_metadata{Name := "BuyplateBillboards/bp_217", Asset := BuyplateBillboards.bp_217, Dimensions := option{vector2{X := 512.0, Y := 256.0}}}
            texture_metadata{Name := "BuyplateBillboards/bp_218", Asset := BuyplateBillboards.bp_218, Dimensions := option{vector2{X := 512.0, Y := 256.0}}}
            texture_metadata{Name := "BuyplateBillboards/bp_219", Asset := BuyplateBillboards.bp_219, Dimensions := option{vector2{X := 512.0, Y := 256.0}}}
            texture_metadata{Name := "BuyplateBillboards/bp_220", Asset := BuyplateBillboards.bp_220, Dimensions := option{vector2{X := 512.0, Y := 256.0}}}
            texture_metadata{Name := "BuyplateBillboards/bp_221", Asset := BuyplateBillboards.bp_221, Dimensions := option{vector2{X := 512.0, Y := 256.0}}}
            texture_metadata{Name := "BuyplateBillboards/bp_222", Asset := BuyplateBillboards.bp_222, Dimensions := option{vector2{X := 512.0, Y := 256.0}}}
            texture_metadata{Name := "BuyplateBillboards/bp_223", Asset := BuyplateBillboards.bp_223, Dimensions := option{vector2{X := 512.0, Y := 256.0}}}
            texture_metadata{Name := "BuyplateBillboards/bp_224", Asset := BuyplateBillboards.bp_224, Dimensions := option{vector2{X := 512.0, Y := 256.0}}}
            texture_metadata{Name := "BuyplateBillboards/bp_225", Asset := BuyplateBillboards.bp_225, Dimensions := option{vector2{X := 512.0, Y := 256.0}}}
            texture_metadata{Name := "BuyplateBillboards/bp_226", Asset := BuyplateBillboards.bp_226, Dimensions := option{vector2{X := 512.0, Y := 256.0}}}
            texture_metadata{Name := "BuyplateBillboards/bp_227", Asset := BuyplateBillboards.bp_227, Dimensions := option{vector2{X := 512.0, Y := 256.0}}}
            texture_metadata{Name := "BuyplateBillboards/bp_228", Asset := BuyplateBillboards.bp_228, Dimensions := option{vector2{X := 512.0, Y := 256.0}}}
            texture_metadata{Name := "BuyplateBillboards/bp_229", Asset := BuyplateBillboards.bp_229, Dimensions := option{vector2{X := 512.0, Y := 256.0}}}
            texture_metadata{Name := "BuyplateBillboards/bp_230", Asset := BuyplateBillboards.bp_230, Dimensions := option{vector2{X := 512.0, Y := 256.0}}}
            texture_metadata{Name := "BuyplateBillboards/bp_231", Asset := BuyplateBillboards.bp_231, Dimensions := option{vector2{X := 512.0, Y := 256.0}}}
            texture_metadata{Name := "BuyplateBillboards/bp_232", Asset := BuyplateBillboards.bp_232, Dimensions := option{vector2{X := 512.0, Y := 256.0}}}
            texture_metadata{Name := "BuyplateBillboards/bp_233", Asset := BuyplateBillboards.bp_233, Dimensions := option{vector2{X := 512.0, Y := 256.0}}}
            texture_metadata{Name := "BuyplateBillboards/bp_234", Asset := BuyplateBillboards.bp_234, Dimensions := option{vector2{X := 512.0, Y := 256.0}}}
            texture_metadata{Name := "BuyplateBillboards/bp_235", Asset := BuyplateBillboards.bp_235, Dimensions := option{vector2{X := 512.0, Y := 256.0}}}
            texture_metadata{Name := "BuyplateBillboards/bp_236", Asset := BuyplateBillboards.bp_236, Dimensions := option{vector2{X := 512.0, Y := 256.0}}}
            texture_metadata{Name := "BuyplateBillboards/bp_237", Asset := BuyplateBillboards.bp_237, Dimensions := option{vector2{X := 512.0, Y := 256.0}}}
            texture_metadata{Name := "BuyplateBillboards/bp_238", Asset := BuyplateBillboards.bp_238, Dimensions := option{vector2{X := 512.0, Y := 256.0}}}
            texture_metadata{Name := "BuyplateBillboards/bp_239", Asset := BuyplateBillboards.bp_239, Dimensions := option{vector2{X := 512.0, Y := 256.0}}}
            texture_metadata{Name := "BuyplateBillboards/bp_240", Asset := BuyplateBillboards.bp_240, Dimensions := option{vector2{X := 512.0, Y := 256.0}}}
            texture_metadata{Name := "BuyplateBillboards/bp_241", Asset := BuyplateBillboards.bp_241, Dimensions := option{vector2{X := 512.0, Y := 256.0}}}
            texture_metadata{Name := "BuyplateBillboards/bp_242", Asset := BuyplateBillboards.bp_242, Dimensions := option{vector2{X := 512.0, Y := 256.0}}}
            texture_metadata{Name := "BuyplateBillboards/bp_243", Asset := BuyplateBillboards.bp_243, Dimensions := option{vector2{X := 512.0, Y := 256.0}}}
            texture_metadata{Name := "BuyplateBillboards/bp_244", Asset := BuyplateBillboards.bp_244, Dimensions := option{vector2{X := 512.0, Y := 256.0}}}
            texture_metadata{Name := "BuyplateBillboards/bp_245", Asset := BuyplateBillboards.bp_245, Dimensions := option{vector2{X := 512.0, Y := 256.0}}}
            texture_metadata{Name := "BuyplateBillboards/bp_246", Asset := BuyplateBillboards.bp_246, Dimensions := option{vector2{X := 512.0, Y := 256.0}}}
            texture_metadata{Name := "BuyplateBillboards/bp_247", Asset := BuyplateBillboards.bp_247, Dimensions := option{vector2{X := 512.0, Y := 256.0}}}
            texture_metadata{Name := "BuyplateBillboards/bp_248", Asset := BuyplateBillboards.bp_248, Dimensions := option{vector2{X := 512.0, Y := 256.0}}}
            texture_metadata{Name := "BuyplateBillboards/bp_249", Asset := BuyplateBillboards.bp_249, Dimensions := option{vector2{X := 512.0, Y := 256.0}}}
            texture_metadata{Name := "BuyplateBillboards/bp_300", Asset := BuyplateBillboards.bp_300, Dimensions := option{vector2{X := 512.0, Y := 256.0}}}
            texture_metadata{Name := "BuyplateBillboards/bp_301", Asset := BuyplateBillboards.bp_301, Dimensions := option{vector2{X := 512.0, Y := 256.0}}}
            texture_metadata{Name := "BuyplateBillboards/bp_302", Asset := BuyplateBillboards.bp_302, Dimensions := option{vector2{X := 512.0, Y := 256.0}}}
            texture_metadata{Name := "BuyplateBillboards/bp_303", Asset := BuyplateBillboards.bp_303, Dimensions := option{vector2{X := 512.0, Y := 256.0}}}
            texture_metadata{Name := "BuyplateBillboards/bp_304", Asset := BuyplateBillboards.bp_304, Dimensions := option{vector2{X := 512.0, Y := 256.0}}}
            texture_metadata{Name := "BuyplateBillboards/bp_400", Asset := BuyplateBillboards.bp_400, Dimensions := option{vector2{X := 512.0, Y := 256.0}}}
            texture_metadata{Name := "BuyplateBillboards/bp_401", Asset := BuyplateBillboards.bp_401, Dimensions := option{vector2{X := 512.0, Y := 256.0}}}
            texture_metadata{Name := "BuyplateBillboards/bp_402", Asset := BuyplateBillboards.bp_402, Dimensions := option{vector2{X := 512.0, Y := 256.0}}}
            texture_metadata{Name := "DailyRewardsMenuUI/background", Asset := DailyRewardsMenuUI.background, Dimensions := option{vector2{X := 2048.0, Y := 1397.0}}}
            texture_metadata{Name := "DailyRewardsMenuUI/claimed_label", Asset := DailyRewardsMenuUI.claimed_label, Dimensions := option{vector2{X := 268.0, Y := 268.0}}}
            texture_metadata{Name := "DailyRewardsMenuUI/day_bg", Asset := DailyRewardsMenuUI.day_bg, Dimensions := option{vector2{X := 480.0, Y := 508.0}}}
            texture_metadata{Name := "DailyRewardsMenuUI/prev_day_overlay", Asset := DailyRewardsMenuUI.prev_day_overlay, Dimensions := option{vector2{X := 480.0, Y := 508.0}}}
            texture_metadata{Name := "DailyRewardsMenuUI/reward_day1", Asset := DailyRewardsMenuUI.reward_day1, Dimensions := option{vector2{X := 480.0, Y := 508.0}}}
            texture_metadata{Name := "DailyRewardsMenuUI/reward_day2", Asset := DailyRewardsMenuUI.reward_day2, Dimensions := option{vector2{X := 480.0, Y := 508.0}}}
            texture_metadata{Name := "DailyRewardsMenuUI/reward_day3", Asset := DailyRewardsMenuUI.reward_day3, Dimensions := option{vector2{X := 480.0, Y := 508.0}}}
            texture_metadata{Name := "DailyRewardsMenuUI/reward_day4", Asset := DailyRewardsMenuUI.reward_day4, Dimensions := option{vector2{X := 480.0, Y := 508.0}}}
            texture_metadata{Name := "DailyRewardsMenuUI/reward_day5", Asset := DailyRewardsMenuUI.reward_day5, Dimensions := option{vector2{X := 480.0, Y := 508.0}}}
            texture_metadata{Name := "DailyRewardsMenuUI/reward_day6", Asset := DailyRewardsMenuUI.reward_day6, Dimensions := option{vector2{X := 480.0, Y := 508.0}}}
            texture_metadata{Name := "DailyRewardsMenuUI/special_gift_claimed", Asset := DailyRewardsMenuUI.special_gift_claimed, Dimensions := option{vector2{X := 915.0, Y := 1070.0}}}
            texture_metadata{Name := "DailyRewardsMenuUI/special_gift_unclaimed", Asset := DailyRewardsMenuUI.special_gift_unclaimed, Dimensions := option{vector2{X := 915.0, Y := 1070.0}}}
            texture_metadata{Name := "DailyRewardsMenuUI/today_bg", Asset := DailyRewardsMenuUI.today_bg, Dimensions := option{vector2{X := 480.0, Y := 508.0}}}
            texture_metadata{Name := "FontQuicksandBold/a", Asset := FontQuicksandBold.a, Dimensions := option{vector2{X := 15.0, Y := 29.0}}}
            texture_metadata{Name := "FontQuicksandBold/A_upper", Asset := FontQuicksandBold.A_upper, Dimensions := option{vector2{X := 16.0, Y := 29.0}}}
            texture_metadata{Name := "FontQuicksandBold/b", Asset := FontQuicksandBold.b, Dimensions := option{vector2{X := 15.0, Y := 29.0}}}
            texture_metadata{Name := "FontQuicksandBold/B_upper", Asset := FontQuicksandBold.B_upper, Dimensions := option{vector2{X := 16.0, Y := 29.0}}}
            texture_metadata{Name := "FontQuicksandBold/c", Asset := FontQuicksandBold.c, Dimensions := option{vector2{X := 12.0, Y := 29.0}}}
            texture_metadata{Name := "FontQuicksandBold/C_upper", Asset := FontQuicksandBold.C_upper, Dimensions := option{vector2{X := 15.0, Y := 29.0}}}
            texture_metadata{Name := "FontQuicksandBold/d", Asset := FontQuicksandBold.d, Dimensions := option{vector2{X := 15.0, Y := 29.0}}}
            texture_metadata{Name := "FontQuicksandBold/D_upper", Asset := FontQuicksandBold.D_upper, Dimensions := option{vector2{X := 17.0, Y := 29.0}}}
            texture_metadata{Name := "FontQuicksandBold/e", Asset := FontQuicksandBold.e, Dimensions := option{vector2{X := 14.0, Y := 29.0}}}
            texture_metadata{Name := "FontQuicksandBold/E_upper", Asset := FontQuicksandBold.E_upper, Dimensions := option{vector2{X := 14.0, Y := 29.0}}}
            texture_metadata{Name := "FontQuicksandBold/f", Asset := FontQuicksandBold.f, Dimensions := option{vector2{X := 11.0, Y := 29.0}}}
            texture_metadata{Name := "FontQuicksandBold/F_upper", Asset := FontQuicksandBold.F_upper, Dimensions := option{vector2{X := 13.0, Y := 29.0}}}
            texture_metadata{Name := "FontQuicksandBold/g", Asset := FontQuicksandBold.g, Dimensions := option{vector2{X := 15.0, Y := 29.0}}}
            texture_metadata{Name := "FontQuicksandBold/G_upper", Asset := FontQuicksandBold.G_upper, Dimensions := option{vector2{X := 17.0, Y := 29.0}}}
            texture_metadata{Name := "FontQuicksandBold/h", Asset := FontQuicksandBold.h, Dimensions := option{vector2{X := 14.0, Y := 29.0}}}
            texture_metadata{Name := "FontQuicksandBold/H_upper", Asset := FontQuicksandBold.H_upper, Dimensions := option{vector2{X := 17.0, Y := 29.0}}}
            texture_metadata{Name := "FontQuicksandBold/i", Asset := FontQuicksandBold.i, Dimensions := option{vector2{X := 6.0, Y := 29.0}}}
            texture_metadata{Name := "FontQuicksandBold/I_upper", Asset := FontQuicksandBold.I_upper, Dimensions := option{vector2{X := 7.0, Y := 29.0}}}
            texture_metadata{Name := "FontQuicksandBold/j", Asset := FontQuicksandBold.j, Dimensions := option{vector2{X := 7.0, Y := 29.0}}}
            texture_metadata{Name := "FontQuicksandBold/J_upper", Asset := FontQuicksandBold.J_upper, Dimensions := option{vector2{X := 13.0, Y := 29.0}}}
            texture_metadata{Name := "FontQuicksandBold/k", Asset := FontQuicksandBold.k, Dimensions := option{vector2{X := 14.0, Y := 29.0}}}
            texture_metadata{Name := "FontQuicksandBold/K_upper", Asset := FontQuicksandBold.K_upper, Dimensions := option{vector2{X := 16.0, Y := 29.0}}}
            texture_metadata{Name := "FontQuicksandBold/l", Asset := FontQuicksandBold.l, Dimensions := option{vector2{X := 6.0, Y := 29.0}}}
            texture_metadata{Name := "FontQuicksandBold/L_upper", Asset := FontQuicksandBold.L_upper, Dimensions := option{vector2{X := 13.0, Y := 29.0}}}
            texture_metadata{Name := "FontQuicksandBold/m", Asset := FontQuicksandBold.m, Dimensions := option{vector2{X := 22.0, Y := 29.0}}}
            texture_metadata{Name := "FontQuicksandBold/M_upper", Asset := FontQuicksandBold.M_upper, Dimensions := option{vector2{X := 20.0, Y := 29.0}}}
            texture_metadata{Name := "FontQuicksandBold/n", Asset := FontQuicksandBold.n, Dimensions := option{vector2{X := 14.0, Y := 29.0}}}
            texture_metadata{Name := "FontQuicksandBold/N_upper", Asset := FontQuicksandBold.N_upper, Dimensions := option{vector2{X := 18.0, Y := 29.0}}}
            texture_metadata{Name := "FontQuicksandBold/num_0", Asset := FontQuicksandBold.num_0, Dimensions := option{vector2{X := 15.0, Y := 29.0}}}
            texture_metadata{Name := "FontQuicksandBold/num_1", Asset := FontQuicksandBold.num_1, Dimensions := option{vector2{X := 9.0, Y := 29.0}}}
            texture_metadata{Name := "FontQuicksandBold/num_2", Asset := FontQuicksandBold.num_2, Dimensions := option{vector2{X := 13.0, Y := 29.0}}}
            texture_metadata{Name := "FontQuicksandBold/num_3", Asset := FontQuicksandBold.num_3, Dimensions := option{vector2{X := 13.0, Y := 29.0}}}
            texture_metadata{Name := "FontQuicksandBold/num_4", Asset := FontQuicksandBold.num_4, Dimensions := option{vector2{X := 17.0, Y := 29.0}}}
            texture_metadata{Name := "FontQuicksandBold/num_5", Asset := FontQuicksandBold.num_5, Dimensions := option{vector2{X := 14.0, Y := 29.0}}}
            texture_metadata{Name := "FontQuicksandBold/num_6", Asset := FontQuicksandBold.num_6, Dimensions := option{vector2{X := 13.0, Y := 29.0}}}
            texture_metadata{Name := "FontQuicksandBold/num_7", Asset := FontQuicksandBold.num_7, Dimensions := option{vector2{X := 14.0, Y := 29.0}}}
            texture_metadata{Name := "FontQuicksandBold/num_8", Asset := FontQuicksandBold.num_8, Dimensions := option{vector2{X := 14.0, Y := 29.0}}}
            texture_metadata{Name := "FontQuicksandBold/num_9", Asset := FontQuicksandBold.num_9, Dimensions := option{vector2{X := 13.0, Y := 29.0}}}
            texture_metadata{Name := "FontQuicksandBold/o", Asset := FontQuicksandBold.o, Dimensions := option{vector2{X := 15.0, Y := 29.0}}}
            texture_metadata{Name := "FontQuicksandBold/O_upper", Asset := FontQuicksandBold.O_upper, Dimensions := option{vector2{X := 18.0, Y := 29.0}}}
            texture_metadata{Name := "FontQuicksandBold/p", Asset := FontQuicksandBold.p, Dimensions := option{vector2{X := 15.0, Y := 29.0}}}
            texture_metadata{Name := "FontQuicksandBold/P_upper", Asset := FontQuicksandBold.P_upper, Dimensions := option{vector2{X := 14.0, Y := 29.0}}}
            texture_metadata{Name := "FontQuicksandBold/q", Asset := FontQuicksandBold.q, Dimensions := option{vector2{X := 15.0, Y := 29.0}}}
            texture_metadata{Name := "FontQuicksandBold/Q_upper", Asset := FontQuicksandBold.Q_upper, Dimensions := option{vector2{X := 19.0, Y := 29.0}}}
            texture_metadata{Name := "FontQuicksandBold/r", Asset := FontQuicksandBold.r, Dimensions := option{vector2{X := 10.0, Y := 29.0}}}
            texture_metadata{Name := "FontQuicksandBold/R_upper", Asset := FontQuicksandBold.R_upper, Dimensions := option{vector2{X := 16.0, Y := 29.0}}}
            texture_metadata{Name := "FontQuicksandBold/s", Asset := FontQuicksandBold.s, Dimensions := option{vector2{X := 12.0, Y := 29.0}}}
            texture_metadata{Name := "FontQuicksandBold/S_upper", Asset := FontQuicksandBold.S_upper, Dimensions := option{vector2{X := 14.0, Y := 29.0}}}
            texture_metadata{Name := "FontQuicksandBold/symb_bang", Asset := FontQuicksandBold.symb_bang, Dimensions := option{vector2{X := 6.0, Y := 29.0}}}
            texture_metadata{Name := "FontQuicksandBold/symb_comma", Asset := FontQuicksandBold.symb_comma, Dimensions := option{vector2{X := 7.0, Y := 29.0}}}
            texture_metadata{Name := "FontQuicksandBold/symb_dot", Asset := FontQuicksandBold.symb_dot, Dimensions := option{vector2{X := 6.0, Y := 29.0}}}
            texture_metadata{Name := "FontQuicksandBold/symb_eq", Asset := FontQuicksandBold.symb_eq, Dimensions := option{vector2{X := 15.0, Y := 29.0}}}
            texture_metadata{Name := "FontQuicksandBold/symb_minus", Asset := FontQuicksandBold.symb_minus, Dimensions := option{vector2{X := 10.0, Y := 29.0}}}
            texture_metadata{Name := "FontQuicksandBold/symb_paren_close", Asset := FontQuicksandBold.symb_paren_close, Dimensions := option{vector2{X := 9.0, Y := 29.0}}}
            texture_metadata{Name := "FontQuicksandBold/symb_paren_open", Asset := FontQuicksandBold.symb_paren_open, Dimensions := option{vector2{X := 9.0, Y := 29.0}}}
            texture_metadata{Name := "FontQuicksandBold/symb_plus", Asset := FontQuicksandBold.symb_plus, Dimensions := option{vector2{X := 14.0, Y := 29.0}}}
            texture_metadata{Name := "FontQuicksandBold/symb_question", Asset := FontQuicksandBold.symb_question, Dimensions := option{vector2{X := 13.0, Y := 29.0}}}
            texture_metadata{Name := "FontQuicksandBold/symb_slash", Asset := FontQuicksandBold.symb_slash, Dimensions := option{vector2{X := 14.0, Y := 29.0}}}
            texture_metadata{Name := "FontQuicksandBold/t", Asset := FontQuicksandBold.t, Dimensions := option{vector2{X := 10.0, Y := 29.0}}}
            texture_metadata{Name := "FontQuicksandBold/T_upper", Asset := FontQuicksandBold.T_upper, Dimensions := option{vector2{X := 15.0, Y := 29.0}}}
            texture_metadata{Name := "FontQuicksandBold/u", Asset := FontQuicksandBold.u, Dimensions := option{vector2{X := 14.0, Y := 29.0}}}
            texture_metadata{Name := "FontQuicksandBold/U_upper", Asset := FontQuicksandBold.U_upper, Dimensions := option{vector2{X := 17.0, Y := 29.0}}}
            texture_metadata{Name := "FontQuicksandBold/v", Asset := FontQuicksandBold.v, Dimensions := option{vector2{X := 13.0, Y := 29.0}}}
            texture_metadata{Name := "FontQuicksandBold/V_upper", Asset := FontQuicksandBold.V_upper, Dimensions := option{vector2{X := 16.0, Y := 29.0}}}
            texture_metadata{Name := "FontQuicksandBold/w", Asset := FontQuicksandBold.w, Dimensions := option{vector2{X := 18.0, Y := 29.0}}}
            texture_metadata{Name := "FontQuicksandBold/W_upper", Asset := FontQuicksandBold.W_upper, Dimensions := option{vector2{X := 23.0, Y := 29.0}}}
            texture_metadata{Name := "FontQuicksandBold/x", Asset := FontQuicksandBold.x, Dimensions := option{vector2{X := 12.0, Y := 29.0}}}
            texture_metadata{Name := "FontQuicksandBold/X_upper", Asset := FontQuicksandBold.X_upper, Dimensions := option{vector2{X := 16.0, Y := 29.0}}}
            texture_metadata{Name := "FontQuicksandBold/y", Asset := FontQuicksandBold.y, Dimensions := option{vector2{X := 14.0, Y := 29.0}}}
            texture_metadata{Name := "FontQuicksandBold/Y_upper", Asset := FontQuicksandBold.Y_upper, Dimensions := option{vector2{X := 14.0, Y := 29.0}}}
            texture_metadata{Name := "FontQuicksandBold/z", Asset := FontQuicksandBold.z, Dimensions := option{vector2{X := 12.0, Y := 29.0}}}
            texture_metadata{Name := "FontQuicksandBold/Z_upper", Asset := FontQuicksandBold.Z_upper, Dimensions := option{vector2{X := 16.0, Y := 29.0}}}
            texture_metadata{Name := "GameplayRewardsMenuUI/background", Asset := GameplayRewardsMenuUI.background, Dimensions := option{vector2{X := 2048.0, Y := 1425.0}}}
            texture_metadata{Name := "GameplayRewardsMenuUI/card_bg", Asset := GameplayRewardsMenuUI.card_bg, Dimensions := option{vector2{X := 548.0, Y := 622.0}}}
            texture_metadata{Name := "GameplayRewardsMenuUI/reward_blue_chest", Asset := GameplayRewardsMenuUI.reward_blue_chest, Dimensions := option{vector2{X := 219.0, Y := 178.0}}}
            texture_metadata{Name := "GameplayRewardsMenuUI/reward_blue_chest_claimed", Asset := GameplayRewardsMenuUI.reward_blue_chest_claimed, Dimensions := option{vector2{X := 219.0, Y := 185.0}}}
            texture_metadata{Name := "GameplayRewardsMenuUI/reward_blue_gift", Asset := GameplayRewardsMenuUI.reward_blue_gift, Dimensions := option{vector2{X := 219.0, Y := 166.0}}}
            texture_metadata{Name := "GameplayRewardsMenuUI/reward_chest", Asset := GameplayRewardsMenuUI.reward_chest, Dimensions := option{vector2{X := 219.0, Y := 178.0}}}
            texture_metadata{Name := "GameplayRewardsMenuUI/reward_chest_claimed", Asset := GameplayRewardsMenuUI.reward_chest_claimed, Dimensions := option{vector2{X := 219.0, Y := 185.0}}}
            texture_metadata{Name := "GameplayRewardsMenuUI/reward_red_gift", Asset := GameplayRewardsMenuUI.reward_red_gift, Dimensions := option{vector2{X := 219.0, Y := 166.0}}}
            texture_metadata{Name := "GameplayRewardsMenuUI/unlocks_in_bg", Asset := GameplayRewardsMenuUI.unlocks_in_bg, Dimensions := option{vector2{X := 468.0, Y := 112.0}}}
            texture_metadata{Name := "GameplayRewardsMenuUI/up_next_label", Asset := GameplayRewardsMenuUI.up_next_label, Dimensions := option{vector2{X := 282.0, Y := 68.0}}}
            texture_metadata{Name := "HavocVerseUI/ui_black_contrast_rounded", Asset := HavocVerseUI.ui_black_contrast_rounded, Dimensions := option{vector2{X := 512.0, Y := 512.0}}}
            texture_metadata{Name := "HavocVerseUI/ui_debug_menu", Asset := HavocVerseUI.ui_debug_menu, Dimensions := option{vector2{X := 1280.0, Y := 720.0}}}
            texture_metadata{Name := "HavocVerseUI/ui_debug_menu_code_prompt", Asset := HavocVerseUI.ui_debug_menu_code_prompt, Dimensions := option{vector2{X := 450.0, Y := 542.0}}}
            texture_metadata{Name := "TemplateCommon/placeholder", Asset := TemplateCommon.placeholder, Dimensions := option{vector2{X := 512.0, Y := 256.0}}}
            texture_metadata{Name := "TemplateCommon/BackendTooltips/extra", Asset := TemplateCommon.BackendTooltips.extra, Dimensions := option{vector2{X := 2048.0, Y := 1024.0}}}
            texture_metadata{Name := "TemplateCommon/BackendTooltips/integration_debug", Asset := TemplateCommon.BackendTooltips.integration_debug, Dimensions := option{vector2{X := 2048.0, Y := 2048.0}}}
            texture_metadata{Name := "TemplateCommon/BackendTooltips/progress_line", Asset := TemplateCommon.BackendTooltips.progress_line, Dimensions := option{vector2{X := 1024.0, Y := 512.0}}}
            texture_metadata{Name := "TemplateCommon/BackendTooltips/tycoon", Asset := TemplateCommon.BackendTooltips.tycoon, Dimensions := option{vector2{X := 2048.0, Y := 1024.0}}}
            texture_metadata{Name := "UI/minimap5", Asset := UI.minimap5, Dimensions := option{vector2{X := 2048.0, Y := 2048.0}}}
            texture_metadata{Name := "UI/ui_keypad", Asset := UI.ui_keypad, Dimensions := option{vector2{X := 804.0, Y := 1566.0}}}
            texture_metadata{Name := "UI/ui_mapcode", Asset := UI.ui_mapcode, Dimensions := option{vector2{X := 584.0, Y := 112.0}}}
            texture_metadata{Name := "UI/Levels/L1", Asset := UI.Levels.L1, Dimensions := option{vector2{X := 1024.0, Y := 256.0}}}
            texture_metadata{Name := "UI/Levels/L10A", Asset := UI.Levels.L10A, Dimensions := option{vector2{X := 1024.0, Y := 256.0}}}
            texture_metadata{Name := "UI/Levels/L10B", Asset := UI.Levels.L10B, Dimensions := option{vector2{X := 1024.0, Y := 256.0}}}
            texture_metadata{Name := "UI/Levels/L10C", Asset := UI.Levels.L10C, Dimensions := option{vector2{X := 1024.0, Y := 256.0}}}
            texture_metadata{Name := "UI/Levels/L2", Asset := UI.Levels.L2, Dimensions := option{vector2{X := 1024.0, Y := 256.0}}}
            texture_metadata{Name := "UI/Levels/L3", Asset := UI.Levels.L3, Dimensions := option{vector2{X := 1024.0, Y := 256.0}}}
            texture_metadata{Name := "UI/Levels/L4", Asset := UI.Levels.L4, Dimensions := option{vector2{X := 1024.0, Y := 256.0}}}
            texture_metadata{Name := "UI/Levels/L5", Asset := UI.Levels.L5, Dimensions := option{vector2{X := 1024.0, Y := 256.0}}}
            texture_metadata{Name := "UI/Levels/L6", Asset := UI.Levels.L6, Dimensions := option{vector2{X := 1024.0, Y := 256.0}}}
            texture_metadata{Name := "UI/Levels/L7", Asset := UI.Levels.L7, Dimensions := option{vector2{X := 1024.0, Y := 256.0}}}
            texture_metadata{Name := "UI/Levels/L8", Asset := UI.Levels.L8, Dimensions := option{vector2{X := 1024.0, Y := 256.0}}}
            texture_metadata{Name := "UI/Levels/L9", Asset := UI.Levels.L9, Dimensions := option{vector2{X := 1024.0, Y := 256.0}}}
            texture_metadata{Name := "UI/Rebirths/rebirth_01", Asset := UI.Rebirths.rebirth_01, Dimensions := option{vector2{X := 548.0, Y := 128.0}}}
            texture_metadata{Name := "UI/Rebirths/rebirth_02", Asset := UI.Rebirths.rebirth_02, Dimensions := option{vector2{X := 548.0, Y := 128.0}}}
            texture_metadata{Name := "UI/Rebirths/rebirth_03", Asset := UI.Rebirths.rebirth_03, Dimensions := option{vector2{X := 548.0, Y := 128.0}}}
            texture_metadata{Name := "UI/Rebirths/rebirth_04", Asset := UI.Rebirths.rebirth_04, Dimensions := option{vector2{X := 548.0, Y := 128.0}}}
            texture_metadata{Name := "UI/Rebirths/rebirth_05", Asset := UI.Rebirths.rebirth_05, Dimensions := option{vector2{X := 548.0, Y := 128.0}}}
            texture_metadata{Name := "UI/Rebirths/rebirth_06", Asset := UI.Rebirths.rebirth_06, Dimensions := option{vector2{X := 548.0, Y := 128.0}}}
            texture_metadata{Name := "UI/Rebirths/rebirth_07", Asset := UI.Rebirths.rebirth_07, Dimensions := option{vector2{X := 548.0, Y := 128.0}}}
            texture_metadata{Name := "UI/Rebirths/rebirth_08", Asset := UI.Rebirths.rebirth_08, Dimensions := option{vector2{X := 548.0, Y := 128.0}}}
            texture_metadata{Name := "UI/Rebirths/rebirth_09", Asset := UI.Rebirths.rebirth_09, Dimensions := option{vector2{X := 548.0, Y := 128.0}}}
            texture_metadata{Name := "UI/Rebirths/rebirth_10", Asset := UI.Rebirths.rebirth_10, Dimensions := option{vector2{X := 548.0, Y := 128.0}}}
            texture_metadata{Name := "UI/Rebirths/rebirth_100", Asset := UI.Rebirths.rebirth_100, Dimensions := option{vector2{X := 548.0, Y := 128.0}}}
            texture_metadata{Name := "UI/Rebirths/rebirth_11", Asset := UI.Rebirths.rebirth_11, Dimensions := option{vector2{X := 548.0, Y := 128.0}}}
            texture_metadata{Name := "UI/Rebirths/rebirth_12", Asset := UI.Rebirths.rebirth_12, Dimensions := option{vector2{X := 548.0, Y := 128.0}}}
            texture_metadata{Name := "UI/Rebirths/rebirth_13", Asset := UI.Rebirths.rebirth_13, Dimensions := option{vector2{X := 548.0, Y := 128.0}}}
            texture_metadata{Name := "UI/Rebirths/rebirth_14", Asset := UI.Rebirths.rebirth_14, Dimensions := option{vector2{X := 548.0, Y := 128.0}}}
            texture_metadata{Name := "UI/Rebirths/rebirth_15", Asset := UI.Rebirths.rebirth_15, Dimensions := option{vector2{X := 548.0, Y := 128.0}}}
            texture_metadata{Name := "UI/Rebirths/rebirth_16", Asset := UI.Rebirths.rebirth_16, Dimensions := option{vector2{X := 548.0, Y := 128.0}}}
            texture_metadata{Name := "UI/Rebirths/rebirth_17", Asset := UI.Rebirths.rebirth_17, Dimensions := option{vector2{X := 548.0, Y := 128.0}}}
            texture_metadata{Name := "UI/Rebirths/rebirth_18", Asset := UI.Rebirths.rebirth_18, Dimensions := option{vector2{X := 548.0, Y := 128.0}}}
            texture_metadata{Name := "UI/Rebirths/rebirth_19", Asset := UI.Rebirths.rebirth_19, Dimensions := option{vector2{X := 548.0, Y := 128.0}}}
            texture_metadata{Name := "UI/Rebirths/rebirth_20", Asset := UI.Rebirths.rebirth_20, Dimensions := option{vector2{X := 548.0, Y := 128.0}}}
            texture_metadata{Name := "UI/Rebirths/rebirth_200", Asset := UI.Rebirths.rebirth_200, Dimensions := option{vector2{X := 548.0, Y := 128.0}}}
            texture_metadata{Name := "UI/Rebirths/rebirth_21", Asset := UI.Rebirths.rebirth_21, Dimensions := option{vector2{X := 548.0, Y := 128.0}}}
            texture_metadata{Name := "UI/Rebirths/rebirth_22", Asset := UI.Rebirths.rebirth_22, Dimensions := option{vector2{X := 548.0, Y := 128.0}}}
            texture_metadata{Name := "UI/Rebirths/rebirth_23", Asset := UI.Rebirths.rebirth_23, Dimensions := option{vector2{X := 548.0, Y := 128.0}}}
            texture_metadata{Name := "UI/Rebirths/rebirth_24", Asset := UI.Rebirths.rebirth_24, Dimensions := option{vector2{X := 548.0, Y := 128.0}}}
            texture_metadata{Name := "UI/Rebirths/rebirth_25", Asset := UI.Rebirths.rebirth_25, Dimensions := option{vector2{X := 548.0, Y := 128.0}}}
            texture_metadata{Name := "UI/Rebirths/rebirth_26", Asset := UI.Rebirths.rebirth_26, Dimensions := option{vector2{X := 548.0, Y := 128.0}}}
            texture_metadata{Name := "UI/Rebirths/rebirth_300", Asset := UI.Rebirths.rebirth_300, Dimensions := option{vector2{X := 548.0, Y := 128.0}}}
            texture_metadata{Name := "UI/Rebirths/rebirth_400", Asset := UI.Rebirths.rebirth_400, Dimensions := option{vector2{X := 548.0, Y := 128.0}}}
            texture_metadata{Name := "UI/Rebirths/rebirth_50", Asset := UI.Rebirths.rebirth_50, Dimensions := option{vector2{X := 548.0, Y := 128.0}}}
            texture_metadata{Name := "UI/Rebirths/rebirth_500", Asset := UI.Rebirths.rebirth_500, Dimensions := option{vector2{X := 548.0, Y := 128.0}}}
            texture_metadata{Name := "UI/Shop/shop_bg_b1", Asset := UI.Shop.shop_bg_b1, Dimensions := option{vector2{X := 2048.0, Y := 987.0}}}
            texture_metadata{Name := "UI/Shop/shop_bg_b2", Asset := UI.Shop.shop_bg_b2, Dimensions := option{vector2{X := 2048.0, Y := 987.0}}}
            texture_metadata{Name := "UI/Shop/shop_bg_b3", Asset := UI.Shop.shop_bg_b3, Dimensions := option{vector2{X := 2048.0, Y := 987.0}}}
            texture_metadata{Name := "UI/Shop/shop_bg_b4", Asset := UI.Shop.shop_bg_b4, Dimensions := option{vector2{X := 2048.0, Y := 987.0}}}
            texture_metadata{Name := "UI/Shop/shop_bg_b5", Asset := UI.Shop.shop_bg_b5, Dimensions := option{vector2{X := 2048.0, Y := 987.0}}}
            texture_metadata{Name := "UI/Shop/shop_bg_sell", Asset := UI.Shop.shop_bg_sell, Dimensions := option{vector2{X := 2048.0, Y := 987.0}}}
            texture_metadata{Name := "UI/Shop/shop_trend_down", Asset := UI.Shop.shop_trend_down, Dimensions := option{vector2{X := 128.0, Y := 128.0}}}
            texture_metadata{Name := "UI/Shop/shop_trend_up", Asset := UI.Shop.shop_trend_up, Dimensions := option{vector2{X := 128.0, Y := 128.0}}}
            texture_metadata{Name := "UIAssets/Apoc/ui_notif_bg", Asset := UIAssets.Apoc.ui_notif_bg, Dimensions := option{vector2{X := 577.0, Y := 159.0}}}
            texture_metadata{Name := "UIAssets/Apoc/x_circle", Asset := UIAssets.Apoc.x_circle, Dimensions := option{vector2{X := 128.0, Y := 128.0}}}
            texture_metadata{Name := "UIAssets/Widgets/bonus_bg", Asset := UIAssets.Widgets.bonus_bg, Dimensions := option{vector2{X := 784.0, Y := 252.0}}}
            texture_metadata{Name := "UIAssets/Widgets/bonus_frame", Asset := UIAssets.Widgets.bonus_frame, Dimensions := option{vector2{X := 784.0, Y := 252.0}}}
            texture_metadata{Name := "UIAssets/Widgets/gift", Asset := UIAssets.Widgets.gift, Dimensions := option{vector2{X := 128.0, Y := 128.0}}}
            texture_metadata{Name := "UIAssets/Widgets/rainbow_gradient", Asset := UIAssets.Widgets.rainbow_gradient, Dimensions := option{vector2{X := 512.0, Y := 512.0}}}
            texture_metadata{Name := "UIAssets/Widgets/rainbow_spiral", Asset := UIAssets.Widgets.rainbow_spiral, Dimensions := option{vector2{X := 1024.0, Y := 1024.0}}}
            texture_metadata{Name := "UIAssets/Widgets/T_gradient", Asset := UIAssets.Widgets.T_gradient, Dimensions := option{vector2{X := 256.0, Y := 256.0}}}
            texture_metadata{Name := "UIAssets/Widgets/testtrace", Asset := UIAssets.Widgets.testtrace, Dimensions := option{vector2{X := 512.0, Y := 512.0}}}
            texture_metadata{Name := "UIAssets/Widgets/testtraceopacity", Asset := UIAssets.Widgets.testtraceopacity, Dimensions := option{vector2{X := 512.0, Y := 512.0}}}
        }

        var MappedTextureAssets: [string]texture_metadata = map{}
        for (
            TextureMeta:TextureAssets
            set MappedTextureAssets[TextureMeta.Name] = TextureMeta
        ) {}

        local_assets_provider{
            TextureAssets := MappedTextureAssets
        }

local_assets_provider := class<final>(assets_provider):
    TextureAssets: [string]texture_metadata

    GetTextureMetaAssets<override>()<transacts>:[]texture_metadata=
        TextureAssets.Values()

    GetTextureAsset<override>(Name:string)<transacts><decides>:texture=
        TextureAssets[Name].Asset

    GetTextureAssetDimensions<override>(Name:string)<transacts><decides>:vector2=
        TextureAssets[Name].Dimensions?

    GetMaterialAsset<override>(Name:string)<transacts><decides>:material=
        case(Name):
            "Assets/VFX/MI_level_up" => Assets.VFX.MI_level_up
            "Assets/VFX/MI_plus_armory" => Assets.VFX.MI_plus_armory
            "Assets/VFX/MI_plus_gold" => Assets.VFX.MI_plus_gold
            "Assets/VFX/Import/M_AuraCircle" => Assets.VFX.Import.M_AuraCircle
            "Assets/VFX/Import/M_Glow_Slow" => Assets.VFX.Import.M_Glow_Slow
            "Assets/VFX/Import/MI_Glow_Slow" => Assets.VFX.Import.MI_Glow_Slow
            "Assets/VFX/Import/MI_ToolsAuraCircle" => Assets.VFX.Import.MI_ToolsAuraCircle
            "BuyplateBillboards/MI_bp_001" => BuyplateBillboards.MI_bp_001
            "BuyplateBillboards/MI_bp_002" => BuyplateBillboards.MI_bp_002
            "BuyplateBillboards/MI_bp_003" => BuyplateBillboards.MI_bp_003
            "BuyplateBillboards/MI_bp_004" => BuyplateBillboards.MI_bp_004
            "BuyplateBillboards/MI_bp_005" => BuyplateBillboards.MI_bp_005
            "BuyplateBillboards/MI_bp_006" => BuyplateBillboards.MI_bp_006
            "BuyplateBillboards/MI_bp_007" => BuyplateBillboards.MI_bp_007
            "BuyplateBillboards/MI_bp_008" => BuyplateBillboards.MI_bp_008
            "BuyplateBillboards/MI_bp_009" => BuyplateBillboards.MI_bp_009
            "BuyplateBillboards/MI_bp_010" => BuyplateBillboards.MI_bp_010
            "BuyplateBillboards/MI_bp_011" => BuyplateBillboards.MI_bp_011
            "BuyplateBillboards/MI_bp_012" => BuyplateBillboards.MI_bp_012
            "BuyplateBillboards/MI_bp_013" => BuyplateBillboards.MI_bp_013
            "BuyplateBillboards/MI_bp_014" => BuyplateBillboards.MI_bp_014
            "BuyplateBillboards/MI_bp_015" => BuyplateBillboards.MI_bp_015
            "BuyplateBillboards/MI_bp_016" => BuyplateBillboards.MI_bp_016
            "BuyplateBillboards/MI_bp_017" => BuyplateBillboards.MI_bp_017
            "BuyplateBillboards/MI_bp_018" => BuyplateBillboards.MI_bp_018
            "BuyplateBillboards/MI_bp_019" => BuyplateBillboards.MI_bp_019
            "BuyplateBillboards/MI_bp_020" => BuyplateBillboards.MI_bp_020
            "BuyplateBillboards/MI_bp_021" => BuyplateBillboards.MI_bp_021
            "BuyplateBillboards/MI_bp_022" => BuyplateBillboards.MI_bp_022
            "BuyplateBillboards/MI_bp_023" => BuyplateBillboards.MI_bp_023
            "BuyplateBillboards/MI_bp_024" => BuyplateBillboards.MI_bp_024
            "BuyplateBillboards/MI_bp_025" => BuyplateBillboards.MI_bp_025
            "BuyplateBillboards/MI_bp_026" => BuyplateBillboards.MI_bp_026
            "BuyplateBillboards/MI_bp_027" => BuyplateBillboards.MI_bp_027
            "BuyplateBillboards/MI_bp_028" => BuyplateBillboards.MI_bp_028
            "BuyplateBillboards/MI_bp_029" => BuyplateBillboards.MI_bp_029
            "BuyplateBillboards/MI_bp_030" => BuyplateBillboards.MI_bp_030
            "BuyplateBillboards/MI_bp_031" => BuyplateBillboards.MI_bp_031
            "BuyplateBillboards/MI_bp_032" => BuyplateBillboards.MI_bp_032
            "BuyplateBillboards/MI_bp_033" => BuyplateBillboards.MI_bp_033
            "BuyplateBillboards/MI_bp_034" => BuyplateBillboards.MI_bp_034
            "BuyplateBillboards/MI_bp_035" => BuyplateBillboards.MI_bp_035
            "BuyplateBillboards/MI_bp_036" => BuyplateBillboards.MI_bp_036
            "BuyplateBillboards/MI_bp_037" => BuyplateBillboards.MI_bp_037
            "BuyplateBillboards/MI_bp_038" => BuyplateBillboards.MI_bp_038
            "BuyplateBillboards/MI_bp_039" => BuyplateBillboards.MI_bp_039
            "BuyplateBillboards/MI_bp_040" => BuyplateBillboards.MI_bp_040
            "BuyplateBillboards/MI_bp_041" => BuyplateBillboards.MI_bp_041
            "BuyplateBillboards/MI_bp_042" => BuyplateBillboards.MI_bp_042
            "BuyplateBillboards/MI_bp_043" => BuyplateBillboards.MI_bp_043
            "BuyplateBillboards/MI_bp_044" => BuyplateBillboards.MI_bp_044
            "BuyplateBillboards/MI_bp_045" => BuyplateBillboards.MI_bp_045
            "BuyplateBillboards/MI_bp_046" => BuyplateBillboards.MI_bp_046
            "BuyplateBillboards/MI_bp_047" => BuyplateBillboards.MI_bp_047
            "BuyplateBillboards/MI_bp_048" => BuyplateBillboards.MI_bp_048
            "BuyplateBillboards/MI_bp_049" => BuyplateBillboards.MI_bp_049
            "BuyplateBillboards/MI_bp_050" => BuyplateBillboards.MI_bp_050
            "BuyplateBillboards/MI_bp_051" => BuyplateBillboards.MI_bp_051
            "BuyplateBillboards/MI_bp_052" => BuyplateBillboards.MI_bp_052
            "BuyplateBillboards/MI_bp_053" => BuyplateBillboards.MI_bp_053
            "BuyplateBillboards/MI_bp_054" => BuyplateBillboards.MI_bp_054
            "BuyplateBillboards/MI_bp_055" => BuyplateBillboards.MI_bp_055
            "BuyplateBillboards/MI_bp_056" => BuyplateBillboards.MI_bp_056
            "BuyplateBillboards/MI_bp_057" => BuyplateBillboards.MI_bp_057
            "BuyplateBillboards/MI_bp_058" => BuyplateBillboards.MI_bp_058
            "BuyplateBillboards/MI_bp_059" => BuyplateBillboards.MI_bp_059
            "BuyplateBillboards/MI_bp_060" => BuyplateBillboards.MI_bp_060
            "BuyplateBillboards/MI_bp_061" => BuyplateBillboards.MI_bp_061
            "BuyplateBillboards/MI_bp_062" => BuyplateBillboards.MI_bp_062
            "BuyplateBillboards/MI_bp_063" => BuyplateBillboards.MI_bp_063
            "BuyplateBillboards/MI_bp_064" => BuyplateBillboards.MI_bp_064
            "BuyplateBillboards/MI_bp_065" => BuyplateBillboards.MI_bp_065
            "BuyplateBillboards/MI_bp_066" => BuyplateBillboards.MI_bp_066
            "BuyplateBillboards/MI_bp_067" => BuyplateBillboards.MI_bp_067
            "BuyplateBillboards/MI_bp_068" => BuyplateBillboards.MI_bp_068
            "BuyplateBillboards/MI_bp_069" => BuyplateBillboards.MI_bp_069
            "BuyplateBillboards/MI_bp_070" => BuyplateBillboards.MI_bp_070
            "BuyplateBillboards/MI_bp_071" => BuyplateBillboards.MI_bp_071
            "BuyplateBillboards/MI_bp_072" => BuyplateBillboards.MI_bp_072
            "BuyplateBillboards/MI_bp_073" => BuyplateBillboards.MI_bp_073
            "BuyplateBillboards/MI_bp_074" => BuyplateBillboards.MI_bp_074
            "BuyplateBillboards/MI_bp_075" => BuyplateBillboards.MI_bp_075
            "BuyplateBillboards/MI_bp_076" => BuyplateBillboards.MI_bp_076
            "BuyplateBillboards/MI_bp_077" => BuyplateBillboards.MI_bp_077
            "BuyplateBillboards/MI_bp_078" => BuyplateBillboards.MI_bp_078
            "BuyplateBillboards/MI_bp_079" => BuyplateBillboards.MI_bp_079
            "BuyplateBillboards/MI_bp_080" => BuyplateBillboards.MI_bp_080
            "BuyplateBillboards/MI_bp_081" => BuyplateBillboards.MI_bp_081
            "BuyplateBillboards/MI_bp_082" => BuyplateBillboards.MI_bp_082
            "BuyplateBillboards/MI_bp_083" => BuyplateBillboards.MI_bp_083
            "BuyplateBillboards/MI_bp_084" => BuyplateBillboards.MI_bp_084
            "BuyplateBillboards/MI_bp_085" => BuyplateBillboards.MI_bp_085
            "BuyplateBillboards/MI_bp_086" => BuyplateBillboards.MI_bp_086
            "BuyplateBillboards/MI_bp_087" => BuyplateBillboards.MI_bp_087
            "BuyplateBillboards/MI_bp_088" => BuyplateBillboards.MI_bp_088
            "BuyplateBillboards/MI_bp_089" => BuyplateBillboards.MI_bp_089
            "BuyplateBillboards/MI_bp_090" => BuyplateBillboards.MI_bp_090
            "BuyplateBillboards/MI_bp_091" => BuyplateBillboards.MI_bp_091
            "BuyplateBillboards/MI_bp_092" => BuyplateBillboards.MI_bp_092
            "BuyplateBillboards/MI_bp_093" => BuyplateBillboards.MI_bp_093
            "BuyplateBillboards/MI_bp_094" => BuyplateBillboards.MI_bp_094
            "BuyplateBillboards/MI_bp_095" => BuyplateBillboards.MI_bp_095
            "BuyplateBillboards/MI_bp_096" => BuyplateBillboards.MI_bp_096
            "BuyplateBillboards/MI_bp_097" => BuyplateBillboards.MI_bp_097
            "BuyplateBillboards/MI_bp_098" => BuyplateBillboards.MI_bp_098
            "BuyplateBillboards/MI_bp_099" => BuyplateBillboards.MI_bp_099
            "BuyplateBillboards/MI_bp_100" => BuyplateBillboards.MI_bp_100
            "BuyplateBillboards/MI_bp_101" => BuyplateBillboards.MI_bp_101
            "BuyplateBillboards/MI_bp_102" => BuyplateBillboards.MI_bp_102
            "BuyplateBillboards/MI_bp_103" => BuyplateBillboards.MI_bp_103
            "BuyplateBillboards/MI_bp_104" => BuyplateBillboards.MI_bp_104
            "BuyplateBillboards/MI_bp_105" => BuyplateBillboards.MI_bp_105
            "BuyplateBillboards/MI_bp_106" => BuyplateBillboards.MI_bp_106
            "BuyplateBillboards/MI_bp_107" => BuyplateBillboards.MI_bp_107
            "BuyplateBillboards/MI_bp_108" => BuyplateBillboards.MI_bp_108
            "BuyplateBillboards/MI_bp_109" => BuyplateBillboards.MI_bp_109
            "BuyplateBillboards/MI_bp_110" => BuyplateBillboards.MI_bp_110
            "BuyplateBillboards/MI_bp_200" => BuyplateBillboards.MI_bp_200
            "BuyplateBillboards/MI_bp_201" => BuyplateBillboards.MI_bp_201
            "BuyplateBillboards/MI_bp_202" => BuyplateBillboards.MI_bp_202
            "BuyplateBillboards/MI_bp_203" => BuyplateBillboards.MI_bp_203
            "BuyplateBillboards/MI_bp_204" => BuyplateBillboards.MI_bp_204
            "BuyplateBillboards/MI_bp_205" => BuyplateBillboards.MI_bp_205
            "BuyplateBillboards/MI_bp_206" => BuyplateBillboards.MI_bp_206
            "BuyplateBillboards/MI_bp_207" => BuyplateBillboards.MI_bp_207
            "BuyplateBillboards/MI_bp_208" => BuyplateBillboards.MI_bp_208
            "BuyplateBillboards/MI_bp_209" => BuyplateBillboards.MI_bp_209
            "BuyplateBillboards/MI_bp_210" => BuyplateBillboards.MI_bp_210
            "BuyplateBillboards/MI_bp_211" => BuyplateBillboards.MI_bp_211
            "BuyplateBillboards/MI_bp_212" => BuyplateBillboards.MI_bp_212
            "BuyplateBillboards/MI_bp_213" => BuyplateBillboards.MI_bp_213
            "BuyplateBillboards/MI_bp_214" => BuyplateBillboards.MI_bp_214
            "BuyplateBillboards/MI_bp_215" => BuyplateBillboards.MI_bp_215
            "BuyplateBillboards/MI_bp_216" => BuyplateBillboards.MI_bp_216
            "BuyplateBillboards/MI_bp_217" => BuyplateBillboards.MI_bp_217
            "BuyplateBillboards/MI_bp_218" => BuyplateBillboards.MI_bp_218
            "BuyplateBillboards/MI_bp_219" => BuyplateBillboards.MI_bp_219
            "BuyplateBillboards/MI_bp_220" => BuyplateBillboards.MI_bp_220
            "BuyplateBillboards/MI_bp_221" => BuyplateBillboards.MI_bp_221
            "BuyplateBillboards/MI_bp_222" => BuyplateBillboards.MI_bp_222
            "BuyplateBillboards/MI_bp_223" => BuyplateBillboards.MI_bp_223
            "BuyplateBillboards/MI_bp_224" => BuyplateBillboards.MI_bp_224
            "BuyplateBillboards/MI_bp_225" => BuyplateBillboards.MI_bp_225
            "BuyplateBillboards/MI_bp_226" => BuyplateBillboards.MI_bp_226
            "BuyplateBillboards/MI_bp_227" => BuyplateBillboards.MI_bp_227
            "BuyplateBillboards/MI_bp_228" => BuyplateBillboards.MI_bp_228
            "BuyplateBillboards/MI_bp_229" => BuyplateBillboards.MI_bp_229
            "BuyplateBillboards/MI_bp_230" => BuyplateBillboards.MI_bp_230
            "BuyplateBillboards/MI_bp_231" => BuyplateBillboards.MI_bp_231
            "BuyplateBillboards/MI_bp_232" => BuyplateBillboards.MI_bp_232
            "BuyplateBillboards/MI_bp_233" => BuyplateBillboards.MI_bp_233
            "BuyplateBillboards/MI_bp_234" => BuyplateBillboards.MI_bp_234
            "BuyplateBillboards/MI_bp_235" => BuyplateBillboards.MI_bp_235
            "BuyplateBillboards/MI_bp_236" => BuyplateBillboards.MI_bp_236
            "BuyplateBillboards/MI_bp_237" => BuyplateBillboards.MI_bp_237
            "BuyplateBillboards/MI_bp_238" => BuyplateBillboards.MI_bp_238
            "BuyplateBillboards/MI_bp_239" => BuyplateBillboards.MI_bp_239
            "BuyplateBillboards/MI_bp_240" => BuyplateBillboards.MI_bp_240
            "BuyplateBillboards/MI_bp_241" => BuyplateBillboards.MI_bp_241
            "BuyplateBillboards/MI_bp_242" => BuyplateBillboards.MI_bp_242
            "BuyplateBillboards/MI_bp_243" => BuyplateBillboards.MI_bp_243
            "BuyplateBillboards/MI_bp_244" => BuyplateBillboards.MI_bp_244
            "BuyplateBillboards/MI_bp_245" => BuyplateBillboards.MI_bp_245
            "BuyplateBillboards/MI_bp_246" => BuyplateBillboards.MI_bp_246
            "BuyplateBillboards/MI_bp_247" => BuyplateBillboards.MI_bp_247
            "BuyplateBillboards/MI_bp_248" => BuyplateBillboards.MI_bp_248
            "BuyplateBillboards/MI_bp_249" => BuyplateBillboards.MI_bp_249
            "BuyplateBillboards/MI_bp_300" => BuyplateBillboards.MI_bp_300
            "BuyplateBillboards/MI_bp_301" => BuyplateBillboards.MI_bp_301
            "BuyplateBillboards/MI_bp_302" => BuyplateBillboards.MI_bp_302
            "BuyplateBillboards/MI_bp_303" => BuyplateBillboards.MI_bp_303
            "BuyplateBillboards/MI_bp_304" => BuyplateBillboards.MI_bp_304
            "BuyplateBillboards/MI_bp_400" => BuyplateBillboards.MI_bp_400
            "BuyplateBillboards/MI_bp_401" => BuyplateBillboards.MI_bp_401
            "BuyplateBillboards/MI_bp_402" => BuyplateBillboards.MI_bp_402
            "TemplateCommon/BillboardMat_Master" => TemplateCommon.BillboardMat_Master
            "TemplateCommon/BuyplateDiscMat_Master" => TemplateCommon.BuyplateDiscMat_Master
            "TemplateCommon/BuyplateDiscMatTransparent_Master" => TemplateCommon.BuyplateDiscMatTransparent_Master
            "TemplateCommon/BuyplateMatCostNotMet" => TemplateCommon.BuyplateMatCostNotMet
            "TemplateCommon/BuyplateMatMain" => TemplateCommon.BuyplateMatMain
            "TemplateCommon/BuyplateMatMainSub" => TemplateCommon.BuyplateMatMainSub
            "TemplateCommon/BuyplateMatOrange" => TemplateCommon.BuyplateMatOrange
            "TemplateCommon/BackendTooltips/extra_Mat" => TemplateCommon.BackendTooltips.extra_Mat
            "TemplateCommon/BackendTooltips/integration_debug_Mat" => TemplateCommon.BackendTooltips.integration_debug_Mat
            "TemplateCommon/BackendTooltips/progress_line_Mat" => TemplateCommon.BackendTooltips.progress_line_Mat
            "TemplateCommon/BackendTooltips/tycoon_Mat" => TemplateCommon.BackendTooltips.tycoon_Mat
            "UIAssets/M_UI" => UIAssets.M_UI
            "UIAssets/M_UI_Gradient" => UIAssets.M_UI_Gradient
            "UIAssets/M_UI_Rainbow" => UIAssets.M_UI_Rainbow
            "UIAssets/M_UI_RainbowGradient" => UIAssets.M_UI_RainbowGradient
            "UIAssets/M_UI_Shimmer" => UIAssets.M_UI_Shimmer
            "UIAssets/M_UI_Twist" => UIAssets.M_UI_Twist
            "UIAssets/MI_UI_Shimmer_Text" => UIAssets.MI_UI_Shimmer_Text
            "UIAssets/Widgets/M_UI_Rainbow_BG" => UIAssets.Widgets.M_UI_Rainbow_BG
            "UIAssets/Widgets/MI_UI_Rainbow" => UIAssets.Widgets.MI_UI_Rainbow
            "UIAssets/Widgets/UI_RainbowBox" => UIAssets.Widgets.UI_RainbowBox
            "UIAssets/Widgets/UI_RainbowBox2" => UIAssets.Widgets.UI_RainbowBox2

    GetMeshAsset<override>(Name:string)<transacts><decides>:mesh=
        case(Name):
            "TemplateCommon/BuyplateDiscMesh2" => TemplateCommon.BuyplateDiscMesh2
            "TemplateCommon/BuyplateImageBackgroundMesh" => TemplateCommon.BuyplateImageBackgroundMesh
            "TemplateCommon/SM_BuyplateDevice" => TemplateCommon.SM_BuyplateDevice
            "TemplateCommon/BackendTooltips/TooltipPlane" => TemplateCommon.BackendTooltips.TooltipPlane

    GetAnimationSequenceAsset<override>(Name:string)<transacts><decides>:animation_sequence=
       None()?

    GetParticleSystemAsset<override>(Name:string)<transacts><decides>:particle_system=
        case(Name):
            "Assets/VFX/ArmoryParticle" => Assets.VFX.ArmoryParticle
            "Assets/VFX/BuyplateUnlockEffect" => Assets.VFX.BuyplateUnlockEffect
            "Assets/VFX/LevelUp" => Assets.VFX.LevelUp
            "Assets/VFX/Sparkle" => Assets.VFX.Sparkle
            "Assets/VFX/VFX_gold" => Assets.VFX.VFX_gold
            "Assets/VFX/Import/VFX_Manhole" => Assets.VFX.Import.VFX_Manhole
            "Assets/VFX/Import/VFX_OrangeMagicSmall" => Assets.VFX.Import.VFX_OrangeMagicSmall
