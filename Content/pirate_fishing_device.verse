
using { /Fortnite.com/Devices }
using { /Verse.org/Simulation }
using { /UnrealEngine.com/Temporary/Diagnostics }

# See https://dev.epicgames.com/documentation/en-us/uefn/create-your-own-device-in-verse for how to create a verse device.

# A Verse-authored creative device that can be placed in a level
pirate_fishing_device := class(creative_device):

    @editable FishingRodConditionButton : conditional_button_device = conditional_button_device{}
    @editable FishingZones: []fishing_zone_device = array{}

    @editable FishingRodCatchTrigger : trigger_device = trigger_device{}
    @editable OtherCatchTrigger : trigger_device = trigger_device{}

    OnBegin<override>()<suspends>:void=
        for(FishingZone:FishingZones):
            FishingZone.CaughtEvent.Subscribe(OnFishCaught)

    OnFishCaught(Agent:agent):void=
        if (FishingRodConditionButton.IsHoldingItem[Agent, 0]): #Fishing Rod
            FishingRodCatchTrigger.Trigger(Agent)
        else:
            OtherCatchTrigger.Trigger(Agent)
