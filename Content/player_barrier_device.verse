
using { /Fortnite.com/Devices }
using { /Verse.org/Simulation }
using { /UnrealEngine.com/Temporary/Diagnostics }

# See https://dev.epicgames.com/documentation/en-us/uefn/create-your-own-device-in-verse for how to create a verse device.

# A Verse-authored creative device that can be placed in a level
player_barrier_device := class(creative_device):

    @editable Switches: []switch_device = array{}

    @editable Barriers: []barrier_device = array{}
    @editable DamageVolumes: []damage_volume_device = array{}

    @editable DisableTriggers: []trigger_device = array{}

    OnBegin<override>()<suspends>:void=
        for(Switch:Switches):
            Switch.TurnedOnEvent.Subscribe(EnableBarriers)
        for(Switch:Switches):
            Switch.TurnedOffEvent.Subscribe(DisableBarriers)
        for(Trigger:DisableTriggers):
            Trigger.TriggeredEvent.Subscribe(DisableBarriers2)

    EnableBarriers(Agent:agent):void=
        for(Barrier:Barriers):
            Barrier.Enable()
        for(DamageVolume:DamageVolumes):
            DamageVolume.Enable()

    DisableBarriers(Agent:agent):void=
        for(Barrier:Barriers):
            Barrier.Disable()
        for(DamageVolume:DamageVolumes):
            DamageVolume.Disable()

    DisableBarriers2(Agent:?agent):void=
        if(MaybeAgent := Agent?):
            for(Switch:Switches):
                Switch.TurnOff(MaybeAgent)

        for(Barrier:Barriers):
            Barrier.Disable()
        for(DamageVolume:DamageVolumes):
            DamageVolume.Disable()
