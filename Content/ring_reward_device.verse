using { /Fortnite.com/Devices }
using { /Verse.org/Simulation }
using { /UnrealEngine.com/Temporary/Diagnostics }
using { /Fortnite.com/Characters }
using { /Fortnite.com/Playspaces }
using { /UnrealEngine.com/Temporary/Diagnostics }
using { /UnrealEngine.com/Temporary/SpatialMath }

using { HavocVerse.Debug }
using { HavocVerse.Util }
using { HavocVerse.Logging }


# See https://dev.epicgames.com/documentation/en-us/uefn/create-your-own-device-in-verse for how to create a verse device.

# A Verse-authored creative device that can be placed in a level
RingHandler<internal> := class() {
    Device : ring_reward_device
    Parameter1 : int

    VolumeHandlerFunction(Agent : agent) : void = {
        Device.OnRingSuccess(Agent,Parameter1)
    }
}

ring_reward_device := class(creative_device, aware_device):

    @editable p_Rings: []ring_device = array{}
    @editable RewardTrigger: trigger_device = trigger_device{}

    DeviceMeta<override>()<transacts>:aware_device_meta=
        aware_device_meta{
            DeviceName := "ring_reward_device"
        }  

    OnBegin<override>()<suspends>:void=
        var Index: int = 0
        for(Ring:p_Rings):
            Ring.SuccessVolume.AgentEntersEvent.Subscribe(RingHandler{Device:=Self, Parameter1:=Index}.VolumeHandlerFunction)
            set Index = Index + 1

    OnRingSuccess(Agent:agent,Parameter1:int) : void =
        if(Ring:=p_Rings[Parameter1]):
            if(Ring.IsCurrentlyActive[]):
                RewardTrigger.Trigger(Agent)
                Ring.Collect()

ring_device<internal> := class<final><concrete>:
    
    var RespawnTime : float = 30.0
    var IsActive : logic = true
    
    @editable Ring: creative_prop = creative_prop{}
    #@editable ShowCinematic: cinematic_sequence_device = cinematic_sequence_device{}
    #@editable HideCinematic: cinematic_sequence_device = cinematic_sequence_device{}

    @editable SuccessVolume: volume_device = volume_device{}

    IsCurrentlyActive()<decides><transacts>:void=
        IsActive?

    Collect():void=
        set IsActive = false
        Ring.Hide()
        #HideCinematic.Play()
        spawn:
            Respawn()

    Respawn()<suspends>:void=
        Sleep(RespawnTime)
        set IsActive = true
        Ring.Show()
        #ShowCinematic.Play()