
using { /Fortnite.com/Devices }
using { /Verse.org/Simulation }
using { /UnrealEngine.com/Temporary/Diagnostics }

# See https://dev.epicgames.com/documentation/en-us/uefn/create-your-own-device-in-verse for how to create a verse device.

# A Verse-authored creative device that can be placed in a level
tornados_device := class(creative_device):

    @editable Tornados: []tornado = array{}

    OnBegin<override>()<suspends>:void=
        for(Tornado:Tornados):
            Tornado.TriggersSubscibe()

tornado<internal> := class<final><concrete>:
    @editable EnableTrigger: trigger_device = trigger_device{}
    @editable DisableTriggers: []trigger_device = array{}
    
    @editable DamageVolumes: []damage_volume_device = array{}
    @editable VFXSpawners: []vfx_spawner_device = array{}
    @editable AudioPlayers: []audio_player_device = array{}

    TriggersSubscibe():void=
        EnableTrigger.TriggeredEvent.Subscribe(EnableTornado)
        for(DisableTrigger:DisableTriggers):
            DisableTrigger.TriggeredEvent.Subscribe(DisableTornado)

    EnableTornado(Agent:?agent):void=
        for(DamageVolume:DamageVolumes):
            DamageVolume.Enable()
        for(VFXSpawner:VFXSpawners):
            VFXSpawner.Enable()
        for(AudioPlayer:AudioPlayers):
            AudioPlayer.Play()

    DisableTornado(Agent:?agent):void=
        for(DamageVolume:DamageVolumes):
            DamageVolume.Disable()
        for(VFXSpawner:VFXSpawners):
            VFXSpawner.Disable()
        for(AudioPlayer:AudioPlayers):
            AudioPlayer.Stop()