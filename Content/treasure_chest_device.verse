using { /Fortnite.com/Devices }
using { /Verse.org/Simulation }
using { /UnrealEngine.com/Temporary/Diagnostics }
using { /Fortnite.com/Characters }
using { /Fortnite.com/Playspaces }
using { /UnrealEngine.com/Temporary/Diagnostics }
using { /UnrealEngine.com/Temporary/SpatialMath }
using { /Verse.org/Random }

using { HavocVerse.Debug }
using { HavocVerse.Util }
using { HavocVerse.Logging }

# See https://dev.epicgames.com/documentation/en-us/uefn/create-your-own-device-in-verse for how to create a verse device.

# A Verse-authored creative device that can be placed in a level

TreasureHandler<internal> := class() {
    Device : treasure_chest_device
    Parameter1 : int

    ButtonHandlerFunction(Agent : agent) : void = {
        Device.OnButtonSelected(Agent,Parameter1)
    }

    SpawnTriggerHandlerFunction(Agent : ?agent) : void = {
        if(MaybeAgent:=Agent?):
            Device.OnSpawnTrigger(MaybeAgent,Parameter1)
    }

    DespawnTriggerHandlerFunction(Agent : ?agent) : void = {
        if(MaybeAgent:=Agent?):
            Device.OnDespawnTrigger(MaybeAgent,Parameter1)
    }
}

treasure_chest_device := class(creative_device, aware_device):

    @editable AutoRespawn: logic = true
    @editable p_TreasureChests: []treasure_device = array{}
    @editable RewardTrigger: trigger_device = trigger_device{}
    @editable TargetAmount: int = 3
    var CurrentAmount : int = 0

    DeviceMeta<override>()<transacts>:aware_device_meta=
        aware_device_meta{
            DeviceName := "treasure_chest_device"
        }  

    OnBegin<override>()<suspends>:void=
        var Index: int = 0
        for(Treasure:p_TreasureChests):
            Treasure.Collect()
            Treasure.CollectButton.InteractedWithEvent.Subscribe(TreasureHandler{Device:=Self, Parameter1:=Index}.ButtonHandlerFunction)

            if(not AutoRespawn?):
                Treasure.SpawnTrigger.TriggeredEvent.Subscribe(TreasureHandler{Device:=Self, Parameter1:=Index}.SpawnTriggerHandlerFunction)

            set Index = Index + 1
        
        if(AutoRespawn?):
            SpawnChests()

    OnSpawnTrigger(Agent:agent,Parameter1:int) : void =
        if(Treasure:=p_TreasureChests[Parameter1]):
            Treasure.InstantSpawnTreasure()

    OnDespawnTrigger(Agent:agent,Parameter1:int) : void =
        if(Treasure:=p_TreasureChests[Parameter1]):
            Treasure.Collect()           

    OnButtonSelected(Agent:agent,Parameter1:int) : void =
        if(Treasure:=p_TreasureChests[Parameter1]):
            set CurrentAmount = CurrentAmount - 1
            RewardTrigger.Trigger(Agent)
            Treasure.Collect()

            if(AutoRespawn?):
                SpawnChests()

    SpawnChests():void=
        var SpawnIndex: int = 0    
        loop:
            if(CurrentAmount>TargetAmount):
                break
            set SpawnIndex = GetRandomInt(0,p_TreasureChests.Length-1)
            if(Treasure:=p_TreasureChests[SpawnIndex]):
                if(not Treasure.IsCurrentlyActive[]):
                    Treasure.SpawnTreasure()
                    set CurrentAmount = CurrentAmount + 1



treasure_device<internal> := class<final><concrete>:
    
    @editable MinRespawnTime : float = 120.0
    @editable MaxRespawnTime : float = 600.0

    @editable SpawnTrigger: trigger_device = trigger_device{}
    @editable DespawnTrigger: trigger_device = trigger_device{}

    var IsActive : logic = false
            
    @editable CollectButton: button_device = button_device{}
    @editable Chest: creative_prop = creative_prop{}
        
    IsCurrentlyActive()<decides><transacts>:void=
        IsActive?

        
    Collect():void=
        set IsActive = false
        CollectButton.Disable()
        Chest.Hide()

    InstantSpawnTreasure():void=
        set IsActive = true
        Chest.Show()
        CollectButton.Enable()

    SpawnTreasure():void=
        spawn:
            Respawn()
        
    Respawn()<suspends>:void=
        var RespawnTime: float = 0.0
        set RespawnTime = GetRandomFloat(MinRespawnTime,MaxRespawnTime)

        Sleep(RespawnTime)
        set IsActive = true
        Chest.Show()
        CollectButton.Enable()