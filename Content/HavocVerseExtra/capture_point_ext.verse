using { /Fortnite.com/Devices }
using { /Verse.org/Simulation }
using { /UnrealEngine.com/Temporary/Diagnostics }
using { /UnrealEngine.com/Temporary/UI }

using { HavocVerse.Debug }
using { HavocVerse.Devices }
using { HavocVerse.UI }
using { HavocVerse.Logging }
using { HavocVerse.Util }

TooltipDeactivateOnGameQuit<internal><localizes>:message = "Whether to send deactivate when the player quits the game. This may not be wanted for anything dealing with persistence"

capture_point_ext_resource_booster_config_group := class<concrete>:
    @editable AccessControl: ?access_control_device = false
    @editable AutoGenerationManager: ?auto_generation_manager = false

    var m_Active: logic = false

capture_point_ext_resource_booster_config_multiplier := class<concrete>:
    @editable ResourceId: string = ""
    @editable Multiplier: float = 1.0

capture_point_ext_resource_booster_config := class<concrete>:
    @editable Groups: []capture_point_ext_resource_booster_config_group = array{}
    @editable MultiplierChannel: string = "capture_booster1"
    @editable ResourceMultipliers: []capture_point_ext_resource_booster_config_multiplier = array{}

capture_point_billboard_reflection_config := class<concrete>:
    @editable Billboards: []billboard_device = array{}
    @editable ClaimedTemplate: string =  "%Player%"
    @editable NotClaimedTemplate: string =  "UNCLAIMED"

capture_point_ext_device := class(creative_device, aware_device):
    @editable p_CapturePoint: capture_area_device = capture_area_device{}

    @editable p_ResourceMultiplier: ?capture_point_ext_resource_booster_config = false

    @editable p_ActivateTriggers: []trigger_device = array{}
    @editable p_DeactivateTriggers: []trigger_device = array{}
    @editable{ToolTip := TooltipDeactivateOnGameQuit} p_DeactivateOnGameQuit: logic = true

    @editable p_BillboardReflection: ?capture_point_billboard_reflection_config = false

    @editable p_HUDEnabled: logic = false
    @editable p_HUDAccessControl: ?access_control_device = false
    @editable p_HUDZOrder: type{_X:int where 0 <= _X, _X <= 2147483647} = 0
    @editable p_HUDImagePlacement: ui_canvas_placement_config = ui_canvas_placement_config{}
    @editable p_HUDCapturedImage: string = ""
    @editable p_HUDNotCapturedImage: string = ""
    @editable p_HUDImageScale: float = 1.0

    var m_HudManager: ?capture_point_ext_hud_manager = false

    var m_ActiveHolder: ?agent = false

    DeviceMeta<override>()<transacts>:aware_device_meta=
        aware_device_meta{
            DeviceName := "capture_point_ext"
        }

    OnBegin<override>()<suspends>:void=
        AwareInitDevice(Self, OnInit)
    
    OnInit<private>()<suspends>:void=
        p_CapturePoint.ControlChangeEvent.Subscribe(OnControlChanged)

        GetPlayspace().PlayerRemovedEvent().Subscribe(OnPlayerRemoved)

        HandleStateChange()
        if (p_HUDEnabled?) {
            HudManager := capture_point_ext_hud_manager{
                AccessControl := p_HUDAccessControl
                RefreshRate := -1.0
                ZOrder := p_HUDZOrder
                Device := Self
            }
            set m_HudManager = Some(HudManager)
            HudManager.Init(GetPlayspace())
        }

        if (ResourceMultiplierConfig := p_ResourceMultiplier?) {
            for (
                Group:ResourceMultiplierConfig.Groups
                AccessControl := Group.AccessControl?
            ) {
                AccessControl.SubscribeToPlayerJoinEvents(OnInnerAccessControlJoin, OnInnerAccessControlQuit,
                    ?ReplayAllowed := false)
            }
        }

    OnPlayerRemoved<private>(Player:player):void=
        if (ActiveHolder := m_ActiveHolder?, ActiveHolder = Player) {
            set m_ActiveHolder = None()
            ClearResourceMultiplier(ActiveHolder)
            if (p_DeactivateOnGameQuit?) {
                for (Trigger:p_DeactivateTriggers) {
                    Trigger.Trigger(ActiveHolder)
                }
            }
            HandleStateChange()
        }

    OnInnerAccessControlJoin<private>(Player:player):void=
        RecheckHolder()

    OnInnerAccessControlQuit<private>(Player:player, GameQuit:logic):void=
        RecheckHolder()

    RecheckHolder<private>():void=
        if (Holder := m_ActiveHolder?, ResourceMultiplierConfig := p_ResourceMultiplier?) {
            for (
                Group:ResourceMultiplierConfig.Groups
                AccessControl := Group.AccessControl?
                AutoGenerationManager := Group.AutoGenerationManager?
            ) {
                if (Group.m_Active?, not AccessControl.IsAllowed[Holder]) {
                    Log(MsgResourceMultiplierClear(ToString(Self), Holder, ToString(AccessControl)), ?Level := log_level.Debug)
                    for (MultiplierConfig:ResourceMultiplierConfig.ResourceMultipliers) {
                        AutoGenerationManager.GetResourceMultiplierInstance(MultiplierConfig.ResourceId)
                            .SetMultiplier(ResourceMultiplierConfig.MultiplierChannel, None())
                    }
                    set Group.m_Active = false
                } else if (not (Group.m_Active?), AccessControl.IsAllowed[Holder]) {
                    Log(MsgResourceMultiplierApply(ToString(Self), Holder, ToString(AccessControl)), ?Level := log_level.Debug)
                    for (MultiplierConfig:ResourceMultiplierConfig.ResourceMultipliers) {
                        AutoGenerationManager.GetResourceMultiplierInstance(MultiplierConfig.ResourceId)
                            .SetMultiplier(ResourceMultiplierConfig.MultiplierChannel, MultiplierConfig.Multiplier)
                    }
                    set Group.m_Active = true
                }
            }
        }

    HandleStateChange<private>():void=
        if (HudManager := m_HudManager?) {
            HudManager.Refresh()
        }
        if (BillboardReflection := p_BillboardReflection?) {
            for (Billboard:BillboardReflection.Billboards) {
                Text := if (ActiveHolder := m_ActiveHolder?) {
                    ReplaceIntoMessage(BillboardReflection.ClaimedTemplate, map{
                        "Player" => ToMessage(ActiveHolder)
                    })
                } else {
                    ToMessage(BillboardReflection.NotClaimedTemplate)
                }
                Billboard.SetText(Text)
            }
        }

    ClearResourceMultiplier<private>(Holder:agent):void=
        if (ResourceMultiplierConfig := p_ResourceMultiplier?) {
            for (
                Group:ResourceMultiplierConfig.Groups
                Group.m_Active?
                AccessControl := Group.AccessControl?
                AutoGenerationManager := Group.AutoGenerationManager?
            ) {
                Log(MsgResourceMultiplierClear(ToString(Self), Holder, ToString(AccessControl)), ?Level := log_level.Debug)
                for (MultiplierConfig:ResourceMultiplierConfig.ResourceMultipliers) {
                    AutoGenerationManager.GetResourceMultiplierInstance(MultiplierConfig.ResourceId)
                        .SetMultiplier(ResourceMultiplierConfig.MultiplierChannel, None())
                }
                set Group.m_Active = false
            }
        }

    OnControlChanged<private>(NewHolder:agent):void=
        Log(MsgNewHolder(ToString(Self), NewHolder), ?Level := log_level.Debug)
        if (PrevHolder := m_ActiveHolder?) {
            set m_ActiveHolder = None()
            ClearResourceMultiplier(PrevHolder)
            for (Trigger:p_DeactivateTriggers) {
                Trigger.Trigger(PrevHolder)
            }
        }

        set m_ActiveHolder = Some(NewHolder)
        HandleStateChange()

        for (Trigger:p_ActivateTriggers) {
            Trigger.Trigger(NewHolder)
        }

        if (ResourceMultiplierConfig := p_ResourceMultiplier?) {
            for (
                Group:ResourceMultiplierConfig.Groups
                not (Group.m_Active?)
                Group.AccessControl?.IsAllowed[NewHolder]
                AccessControl := Group.AccessControl?
                AutoGenerationManager := Group.AutoGenerationManager?
            ) {
                Log(MsgResourceMultiplierApply(ToString(Self), NewHolder, ToString(AccessControl)), ?Level := log_level.Debug)
                for (MultiplierConfig:ResourceMultiplierConfig.ResourceMultipliers) {
                    AutoGenerationManager.GetResourceMultiplierInstance(MultiplierConfig.ResourceId)
                        .SetMultiplier(ResourceMultiplierConfig.MultiplierChannel, MultiplierConfig.Multiplier)
                }
                set Group.m_Active = true
            }
        }

    MsgNewHolder<private><localizes>(DeviceStr:string, Agent:agent):message=
        "{DeviceStr} handling new holder '{Agent}'"
    MsgResourceMultiplierApply<private><localizes>(DeviceStr:string, Agent:agent, AccessControlStr:string):message=
        "{DeviceStr} applying resource multiplier(s) for '{Agent}' on {AccessControlStr}"
    MsgResourceMultiplierClear<private><localizes>(DeviceStr:string, Agent:agent, AccessControlStr:string):message=
        "{DeviceStr} clear resource multiplier(s) for '{Agent}' on {AccessControlStr}"

capture_point_ext_hud_manager := class(hud_manager_base):
    Device: capture_point_ext_device

    CreateHUD<override>(Player:player):?hud_manager_player=
        Component := capture_point_ext_hud_component:
            Device := Device
            Player := Player
        Some(MakeUIFrameworkHUDPlayer(Player, array{
            (Component, Device.p_HUDImagePlacement)
        }))

capture_point_ext_hud_component := class(ui_component):
    Device: capture_point_ext_device
    Player: player

    var m_LastState: logic = false

    TryUpdate<override>():logic=
        CurrentState := if (Device.m_ActiveHolder? = Player). true else false
        if (CurrentState = m_LastState) {
            true
        } else {
            false
        }

    Create<override>():widget=
        CurrentState := if (Device.m_ActiveHolder? = Player). true else false
        set m_LastState = CurrentState

        TextureName := if (CurrentState?) {
            Device.p_HUDCapturedImage
        } else {
            Device.p_HUDNotCapturedImage
        }
        CreateStaticImage(TextureName, ?UniformScale := Device.p_HUDImageScale)
