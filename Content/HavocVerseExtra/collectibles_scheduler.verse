using { /Fortnite.com/Devices }
using { /UnrealEngine.com/Temporary/Diagnostics }
using { /Verse.org/Concurrency }
using { /Verse.org/Simulation }

using { HavocVerse.Debug }
using { HavocVerse.Math }
using { HavocVerse.Util }
using { HavocVerse.Logging }

collectibles_scheduler_device := class(creative_device, aware_device, aware_debug_device):
    @editable p_DevId: string = "N/A"

    @editable p_Collectibles: []collectible_object_device = array{}
    @editable p_CollectiblesFromTag: string = ""

    @editable OnCollectTriggers: []trigger_device = array{}

    @editable p_RespawnTime: float_range = float_range{
        Minimum := 1.5
    }

    var m_InternalIdCounter: int = 0
    var m_Entries: []collectibles_scheduler_entry = array{}
    var m_PlayerLeaveSignals: [player]event() = map{}

    DeviceMeta<override>()<transacts>:aware_device_meta=
        aware_device_meta{
            DeviceName := "collectibles_scheduler"
            DevId := p_DevId
        }

    DebugStats<override>()<transacts>:aware_debug_stats=
        aware_debug_stats{
            Length := map{
                "entries" => m_Entries.Length
                "playerLeaveSignals" => m_PlayerLeaveSignals.Length
            }
        }

    OnBegin<override>()<suspends>:void=
        AwareInitDevice(Self, OnInit)
    
    OnInit<private>()<suspends>:void=
        for (Collectible:p_Collectibles) {
            RegisterCollectible(Collectible)
        }

        if (
            p_CollectiblesFromTag.Length > 0
            TagsProvider := AwareDefaultTagsProvider[]
            Tag := TagsProvider.GetTag[p_CollectiblesFromTag]
        ) {
            PreTagCount := m_Entries.Length
            for (
                CreativeObject:GetCreativeObjectsWithTag(Tag)
                Collectible := collectible_object_device[CreativeObject]
            ) {
                RegisterCollectible(Collectible)
            }
            TagRegisteredCount := m_Entries.Length - PreTagCount
            Log(Self, "Registered {TagRegisteredCount} collectibles from tag \"{p_CollectiblesFromTag}\"", ?Level := log_level.Debug)
        } else if (p_CollectiblesFromTag.Length > 0) {
            Log(Self, "Could not load collectibles from tag \"{p_CollectiblesFromTag}\". Not registered in tags provider?", ?Level := log_level.Error)
        }

        GetPlayspace().PlayerRemovedEvent().Subscribe(OnPlayerRemoved)

    OnPlayerRemoved<private>(Player:player):void=
        RemoveRes := m_PlayerLeaveSignals.RemoveByKey(Player)
        set m_PlayerLeaveSignals = RemoveRes(0)
        if (AbortEvent := RemoveRes(1)?) {
            AbortEvent.Signal()
        }

    RegisterCollectible<private>(Collectible:collectible_object_device):void=
        set m_InternalIdCounter += 1
        Entry := collectibles_scheduler_entry{
            DebugId := m_InternalIdCounter
            Scheduler := Self
            Collectible := Collectible
        }
        set m_Entries += array{Entry}

        Collectible.CollectedEvent.Subscribe(Entry.OnCollected)

    HandlePlayerCollected<internal>(Entry:collectibles_scheduler_entry, Player:player)<suspends>:void=
        RespawnTime := p_RespawnTime.GetRandomValue()
        Log(Self, MsgRespawnSchedule(Entry.DebugId, Player, RespawnTime), ?Level := log_level.Verbose)
        AbortEvent := m_PlayerLeaveSignals[Player] or (
            l_AbortEvent: event() := event(){}
            if. set m_PlayerLeaveSignals[Player] = l_AbortEvent
            l_AbortEvent
        )
        race:
            AbortEvent.Await()
            block:
                Sleep(RespawnTime)
                Entry.Collectible.Respawn(Player)
                Log(Self, MsgRespawned(Entry.DebugId, Player), ?Level := log_level.Verbose)
 
    MsgRespawnSchedule<private><localizes>(CollectibleId:int, Agent:agent, Time:float):message=
        "Scheduling collectible #{CollectibleId} respawn for '{Agent}' in {Time}s"
    MsgRespawned<private><localizes>(CollectibleId:int, Agent:agent):message=
        "Collectible #{CollectibleId} respawned for '{Agent}'"

collectibles_scheduler_entry<internal> := class<final>:
    DebugId: int
    Scheduler: collectibles_scheduler_device
    Collectible: collectible_object_device

    OnCollected<internal>(Agent:agent):void=
        if (Player := player[Agent]) {
            spawn:
                Scheduler.HandlePlayerCollected(Self, Player)
            for (Trigger:Scheduler.OnCollectTriggers) {
                Trigger.Trigger(Agent)
            }
        }
