using { /Fortnite.com/Devices }
using { /Fortnite.com/UI }
using { /Verse.org/Simulation }
using { /Verse.org/Colors }
using { /UnrealEngine.com/Temporary/Diagnostics }
using { /UnrealEngine.com/Temporary/SpatialMath }
using { /UnrealEngine.com/Temporary/UI }

using { HavocVerse.Debug }
using { HavocVerse.Devices }
using { HavocVerse.Devices.GameSpecific }
using { HavocVerse.UI }
using { HavocVerse.Persistence }
using { HavocVerse.Logging }
using { HavocVerse.Util }
using { HavocVerse.Util.Task }

daily_rewards_reward_config := class<concrete>:
    @editable RewardImage: string = ""
    @editable ClaimTrigger: []trigger_device = array{}
    @editable ClaimAckTrigger: []trigger_device = array{}
    @editable ResourceRewards: []daily_rewards_reward_config_resource = array{}

daily_rewards_reward_config_resource := class<concrete>:
    @editable ResourceId: string = ""
    @editable Amount: int = 1
    @editable SendFeedbackOverride: ?logic = false

daily_rewards_player_state := class<final>:
    AbortEvent: event() = event(){}
    DayChangeEvent: event() = event(){}

    var m_RewardInfo: daily_rewards_player_reward_info = daily_rewards_player_reward_info{}
    var m_LastRewardAck: logic = false
    var m_OnTycoonJoinOpenMenu: logic = false

daily_rewards_player_reward_info := class<final>:
    Streak: int = 0
    RewardIndex: int = 0
    NextDayStartAt: ?date_time = false

MakeDailyRewardsPlayerRewardInfo<constructor>(CopyOf:daily_rewards_player_reward_info) := daily_rewards_player_reward_info:
    Streak := CopyOf.Streak
    RewardIndex := CopyOf.RewardIndex
    NextDayStartAt := CopyOf.NextDayStartAt

daily_rewards_device := class(creative_device, aware_device, aware_debug_device):
    @editable p_PersistenceDocumentId: string = "daily_rewards"
    @editable p_StartCycleOffsetHours: float = 12.0
    @editable p_DayLengthHours: float = 24.0
    @editable p_LateRejoinToleranceHours: float = 12.0
    @editable p_ClaimOnAck: logic = false

    @editable p_ResourceManager: resource_manager_device = resource_manager_device{}
    @editable p_ResourceRewardsSendFeedback: logic = true
    @editable p_Rewards: []daily_rewards_reward_config = array{}
    @editable p_SpecialGiftUnclaimedImage: string = ""
    @editable p_SpecialGiftClaimedImage: string = ""
    @editable p_MenuImagesScale: float = 1.0
    @editable p_MenuRewardsScale: float = 1.0

    @editable p_TycoonClaimAsMenuJoinSource: ?tycoon_manager_device = false
    @editable p_OpenMenuOnFirstJoin: logic = false
    @editable p_OpenMenuOnSubsequentNewDays: logic = true
    @editable p_JoinAutoOpenMenuDelay: float = 0.0
    @editable p_OpenMenuOnNewDayWhileOnline: logic = true
    @editable p_OpenButtons: []button_device = array{}

    @editable p_AccostNewPlayerDelay: float = 30.0
    @editable p_AccostNewPlayerTrigger: []trigger_device = array{}

    var m_Players: [player]daily_rewards_player_state = map{}
    var m_InfiniteLoopProtection: [player]int = map{}

    DeviceMeta<override>()<transacts>:aware_device_meta=
        aware_device_meta{
            DeviceName := "daily_rewards"
        }

    DebugStats<override>()<transacts>:aware_debug_stats=
        aware_debug_stats{
            Length := map{
                "players" => m_Players.Length
            }
        }

    OnBegin<override>()<suspends>:void=
        AwareInitDevice(Self, OnInit)
    
    OnInit<private>()<suspends>:void=
        GetPlayspace().PlayerAddedEvent().Subscribe(OnPlayerAdded)
        GetPlayspace().PlayerRemovedEvent().Subscribe(OnPlayerRemoved)
        for (Player:GetPlayspace().GetPlayers()) {
            OnPlayerAdded(Player)
        }

        for (Button:p_OpenButtons) {
            Button.InteractedWithEvent.Subscribe(OpenMenu)
        }

        if (TycoonManager := p_TycoonClaimAsMenuJoinSource?) {
            SubscribeToEvent(TycoonManager.PlayerJoinProgressLineEvent, OnTycoonJoin)
        }

    OnPlayerAdded<private>(Player:player):void=
        CheckPlayer(Player, true)

    OnPlayerRemoved<private>(Player:player):void=
        RemoveRes := m_Players.RemoveByKey(Player)
        set m_Players = RemoveRes(0)
        if (PlayerState := RemoveRes(1)?) {
            PlayerState.AbortEvent.Signal()
        }
        set m_InfiniteLoopProtection = m_InfiniteLoopProtection.RemoveByKey(Player)(0)

    OpenMenu<private>(Agent:agent):void=
        if (Player := player[Agent], PlayerState := m_Players[Player]) {
            if (PlayerState.m_OnTycoonJoinOpenMenu?) {
                set PlayerState.m_OnTycoonJoinOpenMenu = false
            }
            if (IsSimpleMenuOpen["daily_rewards", Player]) {
                Log("Skipping daily rewards menu open for '{Player}' because it is already open.", ?Level := log_level.Debug)
                return
            }
            OpenSimpleMenu("daily_rewards", Agent, daily_rewards_menu_component{
                Device := Self
                PlayerState := PlayerState
                RewardIndex := PlayerState.m_RewardInfo.RewardIndex
            })
        }

    CheckPlayer<private>(Player:player, OnJoin:logic):void=
        if (not Player.IsActive[]) {
            return
        }

        if (ProtIt := m_InfiniteLoopProtection[Player], ProtIt >= 10) {
            Log("Aborting daily rewards player check for '{Player}' due to infitine looping!", ?Level := log_level.Error)
            return
        }
        set m_InfiniteLoopProtection = ConcatenateMaps(m_InfiniteLoopProtection, map{
            Player => (m_InfiniteLoopProtection[Player] or 0) + 1
        })

        PlayerState := m_Players[Player] or (
            l_PlayerState := daily_rewards_player_state{}
            if. set m_Players[Player] = l_PlayerState
            l_PlayerState
        )

        Now := UtcDateTimeNow()
        Document := PlayerDataGetDocument(Player, p_PersistenceDocumentId)
        var Streak: int = 1
        var ClaimToday: logic = false
        var FirstTime: logic = false
        var NextDayStartAt: date_time = if (NextDayStartAtEpoch := Document.GetFloat["next_day_start_at"]) {
            set Streak = Document.GetInt("streak", 1)
            DateTimeFromEpoch(NextDayStartAtEpoch)
        } else {
            set FirstTime = true
            set ClaimToday = true
            Now + (p_StartCycleOffsetHours * TimeUnits.Hour)
        }
        Log("Current time: {Now}. Login NextDayStartAt: {NextDayStartAt}", ?Level := log_level.Verbose)
        GlitchedNextDayStart := if ((NextDayStartAt - Now).TotalSeconds > Max(p_DayLengthHours, p_StartCycleOffsetHours) * TimeUnits.Hour) {
            Log("Detected glitched daily rewards next day start for '{Player}': {NextDayStartAt}", ?Level := log_level.Debug)
            true
        } else {
            false
        }
        if (NextDayStartAt.IsBefore[Now] or GlitchedNextDayStart?) {
            set ClaimToday = true
            Elapsed := (Now - NextDayStartAt).TotalSeconds
            if (not GlitchedNextDayStart?, Elapsed < (p_DayLengthHours * TimeUnits.Hour) + (p_LateRejoinToleranceHours * TimeUnits.Hour)) {
                var ItProtection: int = 0
                loop {
                    if (NextDayStartAt.IsBefore[Now]) {
                        # keep cycle
                        set NextDayStartAt = NextDayStartAt + (p_DayLengthHours * TimeUnits.Hour)
                    } else {
                        break
                    }
                    set ItProtection = ItProtection + 1
                    if (ItProtection > 2) {
                        break
                    }
                }
                if (ItProtection > 2) {
                    set NextDayStartAt = Now + (p_StartCycleOffsetHours * TimeUnits.Hour)
                    set Streak = 1
                } else {
                    set Streak = Streak + 1
                }
            } else {
                set NextDayStartAt = Now + (p_StartCycleOffsetHours * TimeUnits.Hour)
                if (GlitchedNextDayStart?) {
                    set Streak = Streak + 1
                } else {
                    set Streak = 1
                }
            }
        }

        Document.SetFloat("next_day_start_at", NextDayStartAt.ToEpoch())
        set PlayerState.m_LastRewardAck = Document.GetBool("last_reward_ack", true)
        if (PlayerState.m_LastRewardAck?) {
            set PlayerState.m_LastRewardAck = InvertLogic(ClaimToday)
            Document.SetBool("last_reward_ack", PlayerState.m_LastRewardAck)
        } else if (ClaimToday?, p_ClaimOnAck?) {
            # prev day reward wasn't claimed
            set Streak = Max(1, Streak - 1)
        }
        Document.SetInt("streak", Streak)

        RewardIndex := DefaultingMod(Streak-1, 7)
        set PlayerState.m_RewardInfo = daily_rewards_player_reward_info{
            Streak := Streak
            RewardIndex := RewardIndex
            NextDayStartAt := Some(NextDayStartAt)
        }
        PlayerState.DayChangeEvent.Signal()

        if (ClaimToday?, not p_ClaimOnAck?) {
            ClaimReward(Player, RewardIndex)
        }

        var OpenedMenu: logic = false
        if (OnJoin?) {
            if (
                (FirstTime? and p_OpenMenuOnFirstJoin?) or
                ((not FirstTime?) and p_OpenMenuOnSubsequentNewDays?)
            ) {
                set OpenedMenu = true
                if (TycoonManager := p_TycoonClaimAsMenuJoinSource?, not TycoonManager.IsPlaying[Player]) {
                    set PlayerState.m_OnTycoonJoinOpenMenu = true
                } else {
                    spawn:
                        SchedulePlayerOpenMenu(Player, 0.5 + p_JoinAutoOpenMenuDelay)
                }
            }
        } else if (
            p_OpenMenuOnNewDayWhileOnline?
            (not p_TycoonClaimAsMenuJoinSource?) or (p_TycoonClaimAsMenuJoinSource?.IsPlaying[Player])
        ) {
            set OpenedMenu = true
            spawn:
                SchedulePlayerOpenMenu(Player, 0.0)
        }

        if (
            not OpenedMenu?
            OnJoin?
            FirstTime?
            p_AccostNewPlayerTrigger.Length > 0
        ) {
            spawn:
                ScheduleNewPlayerAccost(Player)
        }

        spawn:
            SchedulePlayerCheck(PlayerState.AbortEvent, Player, (NextDayStartAt - Now).TotalSeconds + 1.0)
        Log(MsgPostCheckPlayer(Player, ToString(NextDayStartAt - Now), ToString(NextDayStartAt)), ?Level := log_level.Debug)

    MsgPostCheckPlayer<private><localizes>(Player:player, StartsInStr:string, NextDayStartAtStr:string):message=
        "Next daily rewards day of '{Player}' starts in {StartsInStr} ({NextDayStartAtStr})"

    OnTycoonJoin<private>(Player:player, __:tycoon_manager_progress_line):void=
        if (PlayerState := m_Players[Player], PlayerState.m_OnTycoonJoinOpenMenu?) {
            set PlayerState.m_OnTycoonJoinOpenMenu = false
            spawn:
                OnTycoonJoin0(Player)
        }

    OnTycoonJoin0<private>(Player:player)<suspends>:void=
        Sleep(0.0)
        SchedulePlayerOpenMenu(Player, p_JoinAutoOpenMenuDelay)

    ClaimReward<internal>(Player:player, RewardIndex:int):void=
        if (RewardConfig := p_Rewards[RewardIndex]) {
            for (Trigger:RewardConfig.ClaimTrigger) {
                Trigger.Trigger(Player)
            }
            for (Resource:RewardConfig.ResourceRewards) {
                SendFeedback := Resource.SendFeedbackOverride? or p_ResourceRewardsSendFeedback
                p_ResourceManager.Give(Player, Resource.ResourceId, Resource.Amount, ?SendFeedback := SendFeedback)
            }
        }
        Log("'{Player}' claimed daily reward {RewardIndex}", ?Level := log_level.Debug)

    SchedulePlayerOpenMenu<private>(Player:player, Delay:float)<suspends>:void=
        Sleep(Delay)
        if (
            Player.IsActive[]
            PlayerState := m_Players[Player]
            not PlayerState.m_LastRewardAck?
        ) {
            OpenMenu(Player)
        }

    ScheduleNewPlayerAccost<private>(Player:player)<suspends>:void=
        Sleep(p_AccostNewPlayerDelay)
        if (
            Player.IsActive[]
            PlayerState := m_Players[Player]
            not PlayerState.m_LastRewardAck?
        ) {
            Log("Daily rewards accosting '{Player}' after {p_AccostNewPlayerDelay}s delay", ?Level := log_level.Debug)
            for (Trigger:p_AccostNewPlayerTrigger) {
                Trigger.Trigger(Player)
            }
        }

    SchedulePlayerCheck<private>(AbortEvent:event(), Player:player, WaitFor:float)<suspends>:void=
        race:
            AbortEvent.Await()
            block:
                Sleep(Max(WaitFor, 0.0) + 1.0)
                CheckPlayer(Player, false)

## Menu {{{
daily_rewards_menu_component := class<final>(ui_menu_component):
    Device: daily_rewards_device
    PlayerState: daily_rewards_player_state
    var RewardIndex: int

    var m_CountdownTextBlock: ?custom_font_text_block = false

    TryUpdate<override>():logic=
        if (CountdownTextBlock := m_CountdownTextBlock?) {
            Now := UtcDateTimeNow()
            if (
                NextDayStartAt := PlayerState.m_RewardInfo.NextDayStartAt?
                NextDayStartAt.IsAfter[Now]
            ) {
                Duration := NextDayStartAt - Now
                CountdownText := ToHumanDuration(Duration.TotalSeconds, human_duration_parts{
                    Days := false
                    Hours := if (Duration.TotalSeconds >= TimeUnits.Hour). true else false
                    Minutes := true
                    Seconds := true
                }, human_duration_opts{
                    Delimiter := " "
                    DaysFmt := ("d", "d")
                    HoursFmt := ("hr", "hrs")
                    MinutesFmt := ("min", "mins")
                    SecondsFmt := ("s", "s")
                })
                CountdownTextBlock.SetText(CountdownText)
            } else {
                CountdownTextBlock.SetText("Now!")
            }
        }
        true

    Create<override>():widget=
        LoginStreakFont := GetCustomFontOrDefault("Quicksand-Bold")
        Background := CreateStaticImage("DailyRewardsMenuUI/background", ?Scale := vector2{X := 0.65625, Y := 0.6564065855})
        Width := GetFixedWidgetDimensions(Background).X

        CountdownTextBlock := CreateCustomFontTextBlock(LoginStreakFont, "",
            ?TextColor := MakeColorFromHex("#1cb0f6"), ?FontSize := 24.0)
        set m_CountdownTextBlock = Some(CountdownTextBlock)
        TryUpdate()

        set RewardIndex = PlayerState.m_RewardInfo.RewardIndex
        ScheduleTask(BackgroundJobs)

        overlay:
            Slots := array{
                overlay_slot:
                    Widget := Background
                    HorizontalAlignment := horizontal_alignment.Center
                    VerticalAlignment := vertical_alignment.Center
                overlay_slot:
                    Widget := CreateCustomFontTextBlock(LoginStreakFont, "{PlayerState.m_RewardInfo.Streak} days",
                        ?TextColor := MakeColorFromHex("#1cb0f6"), ?FontSize := 24.0).Widget
                    HorizontalAlignment := horizontal_alignment.Left
                    VerticalAlignment := vertical_alignment.Top
                    Padding := margin{
                        Top := 110.0
                        Left := 165.0
                    }
                overlay_slot:
                    Widget := overlay:
                        Slots := array{
                            overlay_slot:
                                Widget := color_block:
                                    DefaultDesiredSize := vector2{X := 430.0, Y := 1.0}
                                    DefaultOpacity := 0.0
                                HorizontalAlignment := horizontal_alignment.Fill
                            overlay_slot:
                                Widget := CountdownTextBlock.Widget
                                HorizontalAlignment := horizontal_alignment.Center
                        }
                    HorizontalAlignment := horizontal_alignment.Left
                    VerticalAlignment := vertical_alignment.Top
                    Padding := margin{
                        Top := 110.0
                        Left := Width - 430.0
                    }
                overlay_slot:
                    Widget := AddChild(daily_rewards_menu_main_grid_component{
                        Root := Self
                    })
                    HorizontalAlignment := horizontal_alignment.Left
                    VerticalAlignment := vertical_alignment.Top
                    Padding := margin{
                        Top := 190.0
                        Left := 95.0
                    }
                overlay_slot:
                    Widget := AddChild(daily_rewards_menu_actions_row_component{
                        Root := Self
                    })
                    HorizontalAlignment := horizontal_alignment.Center
                    VerticalAlignment := vertical_alignment.Bottom
                    Padding := margin{
                        Bottom := 90.0
                    }
            }

    BackgroundJobs<private>()<suspends>:void=
        branch:
            loop {
                Sleep(1.0)
                TryUpdate()
            }

        PlayerState.DayChangeEvent.Await()
        ForceUpdate()

    OnViewLeave<override>(Player:player):void=
        if (RewardIndex = PlayerState.m_RewardInfo.RewardIndex, not PlayerState.m_LastRewardAck?) {
            set PlayerState.m_LastRewardAck = true
            Document := PlayerDataGetDocument(Player, Device.p_PersistenceDocumentId)
            Document.SetBool("last_reward_ack", PlayerState.m_LastRewardAck)
            if (Device.p_ClaimOnAck?) {
                Device.ClaimReward(Player, RewardIndex)
            }

            if (RewardConfig := Device.p_Rewards[RewardIndex]) {
                for (Trigger:RewardConfig.ClaimAckTrigger) {
                    Trigger.Trigger(Player)
                }
            }
            Log(MsgRewardAck(Player, RewardIndex), ?Level := log_level.Debug)
        }

    MsgRewardAck<private><localizes>(Agent:agent, ClaimedRewardIndex:int):message=
        "'{Agent}' ack'd daily reward {ClaimedRewardIndex}"

### Main Grid Component {{{
daily_rewards_menu_main_grid_component := class(ui_component):
    Root: daily_rewards_menu_component

    Create<override>():widget=
        GridWidgets := for (Index := 0..5) {
            AddChild(daily_rewards_menu_day_card_component{
                Root := Root
                Index := Index
                Special := false
            })
        }.Chunks(3)
        stack_box:
            Orientation := orientation.Horizontal
            Slots := array{
                stack_box_slot:
                    Widget := CreateUIGrid(GridWidgets, ?GapX := 30.0, ?GapY := 30.0)
                    HorizontalAlignment := horizontal_alignment.Left
                    VerticalAlignment := vertical_alignment.Top
                stack_box_slot:
                    Widget := AddChild(daily_rewards_menu_day_card_component{
                        Root := Root
                        Index := 6
                        Special := true
                    })
                    HorizontalAlignment := horizontal_alignment.Left
                    VerticalAlignment := vertical_alignment.Center
                    Padding := margin{
                        Left := -20.0
                    }
            }
### }}}

### Day Card Component {{{
daily_rewards_menu_day_card_component := class(ui_component):
    Root: daily_rewards_menu_component
    Index: int
    Special: logic

    Create<override>():widget=
        DayTextFont := GetCustomFontOrDefault("Quicksand-Bold")

        ImagesScale := Root.Device.p_MenuImagesScale
        Today := if (Index = Root.RewardIndex). true else false

        DayBackgroundName := if (Special?) {
            if ((Today? and ((not Root.Device.p_ClaimOnAck?) or Root.PlayerState.m_LastRewardAck?)) or Index < Root.RewardIndex) {
                Root.Device.p_SpecialGiftClaimedImage
            } else {
                Root.Device.p_SpecialGiftUnclaimedImage
            }
        } else if (Today?) {
            "DailyRewardsMenuUI/today_bg"
        } else {
            "DailyRewardsMenuUI/day_bg"
        }
        CardOverlay := overlay:
            Slots := array{
                overlay_slot:
                    Widget := CreateStaticImage(DayBackgroundName, ?UniformScale := ImagesScale)
                    HorizontalAlignment := horizontal_alignment.Fill
                    VerticalAlignment := vertical_alignment.Fill
                overlay_slot:
                    Widget := CreateCustomFontTextBlock(DayTextFont, "DAY {Index+1}",
                        ?TextColor := NamedColors.White, ?FontSize := 24.0).Widget
                    HorizontalAlignment := horizontal_alignment.Center
                    VerticalAlignment := vertical_alignment.Top
                    Padding := margin{
                        Top := 8.0
                    }
            }
        if (Special?) {
            return CardOverlay
        }

        if (RewardConfig := Root.Device.p_Rewards[Index]) {
            CardOverlay.AddWidget(overlay_slot{
                Widget := CreateStaticImage(RewardConfig.RewardImage, ?UniformScale := Root.Device.p_MenuRewardsScale)
                HorizontalAlignment := horizontal_alignment.Center
                VerticalAlignment := vertical_alignment.Center
            })
        }
        if (Today?, (not Root.Device.p_ClaimOnAck?) or Root.PlayerState.m_LastRewardAck?) {
            CardOverlay.AddWidget(overlay_slot{
                Widget := CreateStaticImage("DailyRewardsMenuUI/claimed_label", ?UniformScale := ImagesScale)
                HorizontalAlignment := horizontal_alignment.Right
                VerticalAlignment := vertical_alignment.Bottom
            })
        } else if (Index < Root.RewardIndex) {
            CardOverlay.AddWidget(overlay_slot{
                Widget := CreateStaticImage("DailyRewardsMenuUI/prev_day_overlay", ?UniformScale := ImagesScale)
                HorizontalAlignment := horizontal_alignment.Fill
                VerticalAlignment := vertical_alignment.Fill
            })
        }
        CardOverlay
### }}}

### Actions Row Component {{{
daily_rewards_menu_actions_row_component := class(ui_menu_component):
    Root: daily_rewards_menu_component

    Create<override>():widget=
        CloseButton := if (Root.PlayerState.m_LastRewardAck?) {
            CreateFortButton(fort_button_type.Regular, ToMessage("Close"))
        } else {
            CreateFortButton(fort_button_type.Loud, ToMessage(if (Root.Device.p_ClaimOnAck?) {
                "Claim"
            } else {
                "Thanks"
            }))
        }
        BindToViewerClose(CloseButton.OnClick())

        ApplySizeHint(CloseButton, vector2{X := 950.0, Y := 50.0})
### }}}
## }}}
