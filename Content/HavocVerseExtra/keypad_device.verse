using { /Fortnite.com/Devices }
using { /Fortnite.com/UI }
using { /Verse.org/Simulation }
using { /UnrealEngine.com/Temporary/Diagnostics }
using { /UnrealEngine.com/Temporary/SpatialMath }
using { /UnrealEngine.com/Temporary/UI }

using { HavocVerse.Debug }
using { HavocVerse.UI }
using { HavocVerse.Util }

keypad_code_config := class<concrete>:
    @editable Code: string = ""
    @editable SuccessTrigger: trigger_device = trigger_device{}

keypad_device := class(creative_device, aware_device, aware_debug_device):
    @editable p_DevId: string = "N/A"

    @editable p_OpenButtons: []button_device = array{}
    @editable p_OpenTriggers: []trigger_device = array{}
    @editable p_CloseTriggers: []trigger_device = array{}

    @editable p_Codes: []keypad_code_config = array{}
    @editable p_OnFailTrigger: []trigger_device = array{}

    @editable p_MaxInputLength: int = 10
    @editable p_CodeTextStyle: ui_text_style_config = ui_text_style_config{}

    @editable p_BackgroundTexture: string = ""
    @editable p_BackgroundSize: vector2 = vector2{X := 450.0, Y := 600.0}
    @editable p_DefaultToColorBackground: logic = true
    @editable p_ButtonsSpacing: float = 20.0
    @editable p_CodeFromButtonsSpacing: float = 40.0

    var m_OpenedMenus: [player]keypad_menu_instance = map{}

    DeviceMeta<override>()<transacts>:aware_device_meta=
        aware_device_meta{
            DeviceName := "keypad_device"
            DevId := p_DevId
        }

    DebugStats<override>()<transacts>:aware_debug_stats=
        aware_debug_stats{
            Length := map{
                "openedMenus" => m_OpenedMenus.Length
            }
        }

    OnBegin<override>()<suspends>:void=
        AwareInitDevice(Self, OnInit)
    
    OnInit<private>()<suspends>:void=
        for (Button:p_OpenButtons) {
            Button.InteractedWithEvent.Subscribe(OnOpenButtonInteractedWith)
        }
        for (Trigger:p_OpenTriggers) {
            Trigger.TriggeredEvent.Subscribe(OnOpenTriggerTriggered)
        }
        for (Trigger:p_CloseTriggers) {
            Trigger.TriggeredEvent.Subscribe(OnCloseTriggerTriggered)
        }

        if (Integration := AwareGetIntegration[]) {
            Integration.SubscribeToMenuCloseRequest(OnMenuCloseRequest)
        }

    OnMenuCloseRequest<private>(Player:player):void=
        if (Instance := m_OpenedMenus[Player]) {
            Instance.Close()
        }

    OnOpenButtonInteractedWith<private>(Agent:agent):void=
        HandleMenuOpen(Some(Agent))

    OnOpenTriggerTriggered<private>(Agent:?agent):void=
        HandleMenuOpen(Agent)

    HandleMenuOpen<private>(Agent:?agent):void=
        if (Player := player[Agent?], not m_OpenedMenus[Player], AssetsProvider := AwareDefaultAssetsProvider[]) {
            Menu := keypad_menu{
                AssetsProvider := AssetsProvider
                Device := Self
            }
            Menu.Open(Player)
        }

    OnCloseTriggerTriggered<private>(Agent:?agent):void=
        if (l_Agent := Agent?) {
            if (Player := player[l_Agent], Instance := m_OpenedMenus[Player]) {
                Instance.Close()
            }
        } else {
            for (__->Instance:m_OpenedMenus) {
                Instance.Close()
            }
        }

    Submit<internal>(Player:player, Code:string):void=
        for (Config:p_Codes, Config.Code = Code) {
            Config.SuccessTrigger.Trigger(Player)
            return
        }
        for (Trigger:p_OnFailTrigger) {
            Trigger.Trigger(Player)
        }

keypad_menu := class(ui_menu):
    Device: keypad_device

    CreateInstance<override>(Player:player):tuple(ui_menu_instance,canvas)=
        Instance := keypad_menu_instance{
            Menu := Self
            Device := Device
            Player := Player
            CodeTextWidget := Device.p_CodeTextStyle.ToTextBlock("")
        }
        set Device.m_OpenedMenus = AddToMap(Device.m_OpenedMenus, Player, Instance)

        Digits := array{1, 2, 3, 4, 5, 6, 7, 8, 9, -2, 0, -1}
        DigitsStack := stack_box{
            Orientation := orientation.Vertical
            Slots := for (RowIndex->Row:Digits.Chunks(3)) {
                RowWidget := stack_box{
                    Orientation := orientation.Horizontal
                    Slots := for (Index->Digit:Row) {
                        Button := if (Digit = -1) {
                            button_loud{
                                DefaultText := ToMessage("▶")
                            }
                        } else if (Digit = -2) {
                            button_regular{
                                DefaultText := ToMessage("◀")
                            }
                        } else {
                            button_regular{
                                DefaultText := ToMessage(ToString(Digit))
                            }
                        }
                        Closure := keypad_menu_closure{
                            Instance := Instance
                            Digit := Digit
                        }
                        Button.OnClick().Subscribe(Closure.OnClick)
                        stack_box_slot{
                            Widget := ApplySizeHint(Button, vector2{X := 106.0, Y := 50.0})
                            HorizontalAlignment := horizontal_alignment.Left
                            Padding := margin{
                                Left := if (Index > 0) {
                                    Device.p_ButtonsSpacing
                                } else {
                                    0.0
                                }
                            }
                        }
                    }
                }
                stack_box_slot{
                    Widget := RowWidget
                    VerticalAlignment := vertical_alignment.Top
                    HorizontalAlignment := horizontal_alignment.Left
                    Padding := margin{
                        Top := if (RowIndex > 0) {
                            Device.p_ButtonsSpacing
                        } else {
                            0.0
                        }
                    }
                }
            }
        }

        CloseButton := button_regular{
            DefaultText := ToMessage("Close")
        }
        CloseButton.OnClick().Subscribe(Instance.OnCloseClick)

        InputStack := stack_box{
            Orientation := orientation.Vertical
            Slots := array{
                stack_box_slot{
                    Widget := DigitsStack
                    HorizontalAlignment := horizontal_alignment.Fill
                }
                stack_box_slot{
                    Widget := ApplyHeightConstraint(CloseButton, vector2{Y := 50.0})
                    HorizontalAlignment := horizontal_alignment.Fill
                    Padding := margin{
                        Top := Device.p_ButtonsSpacing
                    }
                }
            }
        }

        Widget := stack_box{
            Orientation := orientation.Vertical
            Slots := array{
                stack_box_slot{
                    Widget := Instance.CodeTextWidget
                    HorizontalAlignment := horizontal_alignment.Center
                }
                stack_box_slot{
                    Widget := InputStack
                    HorizontalAlignment := horizontal_alignment.Center
                    Padding := margin{
                        Top := Device.p_CodeFromButtonsSpacing
                    }
                }
            }
        }

        Canvas := CommonCanvas(Widget,
            ?BackgroundTexture := Device.p_BackgroundTexture, ?BackgroundSize := Device.p_BackgroundSize,
            ?DefaultToColorBlock := Device.p_DefaultToColorBackground)
        (Instance, Canvas)

keypad_menu_instance := class(ui_menu_instance):
    Device: keypad_device

    CodeTextWidget: text_block

    var m_Typed: string = ""

    OnClose<override>():void=
        set Device.m_OpenedMenus = Device.m_OpenedMenus.RemoveByKey(Player)(0)

    TypedDigit<internal>(Digit:int):void=
        if (m_Typed.Length >= Device.p_MaxInputLength) {
            return
        }
        set m_Typed += ToString(Digit)
        CodeTextWidget.SetText(ToMessage(m_Typed))

    Backspace<internal>():void=
        if (l_Slice := m_Typed.Slice[0, m_Typed.Length - 1]) {
            set m_Typed = l_Slice
        } else {
            set m_Typed = array{}
        }
        CodeTextWidget.SetText(ToMessage(m_Typed))

    Submit<internal>():void=
        Close()
        if (m_Typed.Length > 0) {
            Device.Submit(Player, m_Typed)
        }

keypad_menu_closure := class:
    Instance: keypad_menu_instance
    Digit: int

    OnClick<internal>(Data:widget_message):void=
        if (not Instance.IsActive[]) {
            return
        }
        if (Digit = -1) {
            Instance.Submit()
        } else if (Digit = -2) {
            Instance.Backspace()
        } else {
            Instance.TypedDigit(Digit)
        }
