using { /Fortnite.com/Devices }
using { /Fortnite.com/Characters }
using { /Fortnite.com/Game }
using { /UnrealEngine.com/Temporary/Diagnostics }
using { /Verse.org/Simulation }

using { HavocVerse.Debug }
using { HavocVerse.Devices }
using { HavocVerse.Util }
using { HavocVerse.Logging }

kill_steal_access_control_check_mode := enum:
    Victim
    Murderer
    Both

kill_steal_resource_entry_amount_kind := enum:
    Percent
    FixedAmount

kill_steal_resource_entry_limit_source := enum:
    MurdererGenRatePercent
    VictimGenRatePercent
    RawAmount

TooltipStrictAmount<internal><localizes>:message = "Whether to only take the target amount if the victim has enough.\nIf false, takes all the player has."

kill_steal_resource_entry_config := class<concrete><final>:
    @editable ResourceList: []string = array{}
    @editable AmountKind: kill_steal_resource_entry_amount_kind = kill_steal_resource_entry_amount_kind.Percent
    @editable Amount: float = 0.0
    @editable{ToolTip := TooltipStrictAmount} StrictAmount: logic = false
    @editable LimitSource: ?kill_steal_resource_entry_limit_source = false
    @editable LimitAt: float = 3000.0

kill_steal_resource_message_recipient := enum:
    Murderer
    Victim
    Both

kill_steal_resource_hud_messenger_config := class<concrete><final>:
    @editable SendTo: kill_steal_resource_message_recipient = kill_steal_resource_message_recipient.Murderer
    @editable HUDMessenger: hud_message_device = hud_message_device{}
    @editable_text_box{MultiLine := true}
    Template: string = "Stole $%Amount:money% from %Victim%"
    @editable NumberFormatter: number_formatter_config = number_formatter_config{}

kill_steal_device := class(creative_device, aware_device):
    @editable p_DevId: string = "N/A"

    @editable p_AccessControl: ?access_control_device = false
    @editable p_AccessControlCheckMode: kill_steal_access_control_check_mode = kill_steal_access_control_check_mode.Both

    @editable p_ResourceStealEnabled: logic = false
    @editable p_ResourceManager: ?resource_manager_device = false
    @editable p_ResourceStealEntries: []kill_steal_resource_entry_config = array{}
    @editable p_ResourceStealHudMessengers: []kill_steal_resource_hud_messenger_config = array{}

    @editable p_SendTakeFeedback: logic = true
    @editable p_SendGiveFeedback: logic = true

    @editable p_OnKillTriggers: []trigger_device = array{}
    @editable p_OnDeathTriggers: []trigger_device = array{}

    DeviceMeta<override>()<transacts>:aware_device_meta=
        aware_device_meta{
            DeviceName := "kill_steal"
            DevId := p_DevId
        }

    OnBegin<override>()<suspends>:void=
        AwareInitDevice(Self, OnInit)
    
    OnInit<private>()<suspends>:void=
        if (p_ResourceStealEnabled?, not (p_ResourceManager?)) {
            Log(Self, "Resource steal is enabled, but the resource manager isn't set", ?Level := log_level.Error)
        }

        GetPlayspace().PlayerAddedEvent().Subscribe(OnPlayerAdded)
        for (Player:GetPlayspace().GetPlayers()) {
            OnPlayerAdded(Player)
        }

    OnPlayerAdded<private>(Player:player):void=
        if (FortCharacter := Player.GetFortCharacter[]) {
            Closure := kill_steal_player_closure{
                Device := Self
                FortPlayer := FortCharacter
            }
            FortCharacter.EliminatedEvent().Subscribe(Closure.OnEliminated)
        }

    AccessControlCheck<private>(AccessControl:access_control_device, Victim:player, Murderer:player)<transacts><decides>:void=
        case(p_AccessControlCheckMode):
            kill_steal_access_control_check_mode.Victim => AccessControl.IsAllowed[Victim]
            kill_steal_access_control_check_mode.Murderer => AccessControl.IsAllowed[Murderer]
            kill_steal_access_control_check_mode.Both =>
                AccessControl.IsAllowed[Victim] and AccessControl.IsAllowed[Murderer]

    HandleElimination<internal>(Victim:player, Murderer:player):void=
        if (AccessControl := p_AccessControl?, not AccessControlCheck[AccessControl, Victim, Murderer]) {
            return
        }

        if (p_ResourceStealEnabled?, ResourceManager := p_ResourceManager?) {
            Summary := kill_steal_resource_summary{}
            for (Entry:p_ResourceStealEntries) {
                ProcessResourceStealEntry(ResourceManager, Entry, Victim, Murderer, Summary)
            }

            for (MsgConfig:p_ResourceStealHudMessengers) {
                NumberFormatter := MsgConfig.NumberFormatter.CreateFormatter()
                var Replacements: [string]message = map{
                    "Murderer" => ToMessage(Murderer)
                    "Victim" => ToMessage(Victim)
                }
                for (ResourceId->Amount:Summary.m_Amounts) {
                    set Replacements = ConcatenateMaps(Replacements, map{
                        "Amount:{ResourceId}" => ToMessage(Amount)
                    })
                }

                Message := ReplaceIntoMessage(MsgConfig.Template, Replacements)
                if (
                    MsgConfig.SendTo = kill_steal_resource_message_recipient.Murderer or
                    MsgConfig.SendTo = kill_steal_resource_message_recipient.Both
                ) {
                    MsgConfig.HUDMessenger.Show(Murderer, Message)
                }
                if (
                    MsgConfig.SendTo = kill_steal_resource_message_recipient.Victim or
                    MsgConfig.SendTo = kill_steal_resource_message_recipient.Both
                ) {
                    MsgConfig.HUDMessenger.Show(Victim, Message)
                }
            }
        }

        for (Trigger:p_OnKillTriggers) {
            Trigger.Trigger(Murderer)
        }
        for (Trigger:p_OnDeathTriggers) {
            Trigger.Trigger(Victim)
        }

    ProcessResourceStealEntry<private>(
        ResourceManager:resource_manager_device, Entry:kill_steal_resource_entry_config,
        Victim:player, Murderer:player,
        Summary:kill_steal_resource_summary
    ):void=
        for (ResourceId->__:Entry.ResourceList.ToSet()) {
            HeldAmount := ResourceManager.GetAmount(Victim, ResourceId)
            var Amount: float = case(Entry.AmountKind):
                kill_steal_resource_entry_amount_kind.Percent =>
                    ToFloat(HeldAmount) * (Entry.Amount / 100.0)
                kill_steal_resource_entry_amount_kind.FixedAmount =>
                    Entry.Amount

            if (LimitSource := Entry.LimitSource?) {
                Limit := case(LimitSource):
                    kill_steal_resource_entry_limit_source.MurdererGenRatePercent =>
                        ResourceManager.GetAutoGenerationRate(Murderer, ResourceId) * (Entry.LimitAt / 100.0)
                    kill_steal_resource_entry_limit_source.VictimGenRatePercent =>
                        ResourceManager.GetAutoGenerationRate(Victim, ResourceId) * (Entry.LimitAt / 100.0)
                    kill_steal_resource_entry_limit_source.RawAmount =>
                        Entry.LimitAt
                set Amount = Min(Amount, Limit)
            }

            var IntAmount: int = DefaultingFloor(Amount)
            if (not (Entry.StrictAmount?)) {
                set IntAmount = Min(IntAmount, HeldAmount)
            }

            Log(MsgResourceSteal(Victim, Murderer, ResourceId, IntAmount), ?Level := log_level.Verbose)
            Success := ResourceManager.Take(Victim, ResourceId, IntAmount, ?SendFeedback := p_SendTakeFeedback)
            if (Success?) {
                ResourceManager.Give(Murderer, ResourceId, IntAmount, ?SendFeedback := p_SendGiveFeedback)
                Summary.Add(ResourceId, IntAmount)
            } else {
                Summary.Add(ResourceId, 0)
                Log(MsgResourceStealFail(Victim, ResourceId, IntAmount), ?Level := log_level.Warning)
            }
        }

    MsgResourceSteal<private><localizes>(Victim:agent, Murderer:agent, ResourceId:string, Amount:int):message=
        "Kill stealing {Amount} of \"{ResourceId}\" from '{Victim}' for '{Agent}'"
    MsgResourceStealFail<private><localizes>(Agent:agent, ResourceId:string, Amount:int):message=
        "Couldn't steal {Amount} of resource \"{ResourceId}\" from '{Agent}': amount greater than held"

kill_steal_player_closure := class<final>:
    Device: kill_steal_device
    FortPlayer: fort_character

    OnEliminated(Result:elimination_result):void=
        if (
            EliminatingCharacter := Result.EliminatingCharacter?
            Murderer := player[EliminatingCharacter.GetAgent[]]
            Victim := player[FortPlayer.GetAgent[]]
            Victim <> Murderer
        ) {
            Device.HandleElimination(Victim, Murderer)
        }


kill_steal_resource_summary := class<final>:
    var m_Amounts: [string]int = map{}

    Add(ResourceId:string, Amount:int):void=
        NewAmount := (m_Amounts[ResourceId] or 0) + Amount
        if. set m_Amounts[ResourceId] = NewAmount
