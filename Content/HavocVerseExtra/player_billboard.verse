using { /Fortnite.com/Devices }
using { /Fortnite.com/Characters }
using { /Verse.org/Simulation }
using { /UnrealEngine.com/Temporary/Diagnostics }

using { HavocVerse.Debug }
using { HavocVerse.Devices }
using { HavocVerse.Devices.GameSpecific }
using { HavocVerse.Logging }
using { HavocVerse.Util }

player_billboard_device := class<final><concrete>(creative_device, aware_device, custom_device):
    @editable p_DevId: string = "N/A"

    @editable_text_box{MultiLine := true}
    p_Template: string = "%Player%\nScore: %Score%\nRebirth: %Rebirth%"
    @editable p_Billboards: []billboard_device = array{}
    @editable p_RefreshRate: float = 0.5

    @editable p_ScoreManager: ?score_manager_device = false
    @editable p_ScoreFormatter: number_formatter_config = number_formatter_config{}
    @editable p_TycoonManager: ?tycoon_manager_device = false
    @editable p_RebirthCountFormatter: number_formatter_config = number_formatter_config{}
    @editable p_IgnoreRebirth0: logic = false
    @editable p_LevelGroupsManager: ?level_groups_manager_device = false
    @editable p_LevelFormatter: number_formatter_config = number_formatter_config{}

    @editable p_HideWhenUnassigned: logic = true
    @editable p_DontUpdateWhileDead: logic = true
    @editable p_ClearPlayerOnQuit: logic = true

    @editable p_SyncFirstPlayerOf: ?access_control_device = false
    @editable p_UserActions: []player_billboard_user_action_config = array{}

    var m_RegisteredPlayer: ?tuple(player,event()) = false

    DeviceMeta<override>()<transacts>:aware_device_meta=
        aware_device_meta{
            DeviceName := "player_billboard"
            DevId := p_DevId
        }

    OnBegin<override>()<suspends>:void=
        AwareInitDevice(Self, OnInit)
    
    OnInit<private>()<suspends>:void=
        for (UserActionConfig:p_UserActions) {
            UserAction := player_billboard_user_action{
                Device := Self
                Config := UserActionConfig
            }
            UserActionConfig.OnTrigger.TriggeredEvent.Subscribe(UserAction.OnTriggered)
        }

        if (not (p_ScoreManager?), p_Template.Contains["%Score%"]) {
            Log(Self, "p_Template contains %Score% but no score manager is linked", ?Level := log_level.Warning)
        }
        if (not (p_TycoonManager?), p_Template.Contains["%Rebirth%"]) {
            Log(Self, "p_Template contains %Rebirth% but no tycoon manager is linked", ?Level := log_level.Warning)
        }

        GetPlayspace().PlayerRemovedEvent().Subscribe(OnPlayerRemoved)

        if (p_HideWhenUnassigned?, not (m_RegisteredPlayer?)) {
            for (Billboard:p_Billboards) {
                Billboard.HideText()
            }
        }

        if (AccessControl := p_SyncFirstPlayerOf?) {
            AccessControl.SubscribeToPlayerJoinEvents(OnSyncAccessControlJoin, OnSyncAccessControlQuit)
        }

    OnSyncAccessControlJoin<private>(Player:player):void=
        RegisterPlayer(Player)

    OnSyncAccessControlQuit<private>(Player:player, GameQuit:logic):void=
        if (Registered := m_RegisteredPlayer?, Registered(0) = Player) {
            ClearRegisteredPlayer()
        }

    OnPlayerRemoved<private>(Player:player):void=
        if (p_ClearPlayerOnQuit?) {
            if (Registered := m_RegisteredPlayer?, Registered(0) = Player) {
                ClearRegisteredPlayer()
            }
        }

    RegisterPlayer<public>(Player:player):void=
        if (Registered := m_RegisteredPlayer?) {
            if (Registered(0) = Player) {
                return
            }
            Registered(1).Signal()
        }
        AbortEvent := event(){}
        set m_RegisteredPlayer = Some(Player, AbortEvent)
        UpdateBillboards(Player)
        for (Billboard:p_Billboards) {
            Billboard.ShowText()
        }
        spawn:
            RunUpdateLoop(AbortEvent)

    ClearRegisteredPlayer<public>():void=
        if (Registered := m_RegisteredPlayer?) {
            Registered(1).Signal()
            set m_RegisteredPlayer = None()

            if (p_HideWhenUnassigned?) {
                for (Billboard:p_Billboards) {
                    Billboard.HideText()
                }
            }
        }

    SetRegisteredPlayer<public>(Player:?player):void=
        if (l_Player := Player?) {
            RegisterPlayer(l_Player)
        } else {
            ClearRegisteredPlayer()
        }

    RunUpdateLoop<private>(AbortEvent:event())<suspends>:void=
        race:
            AbortEvent.Await()
            loop {
                if (RegisteredPlayer := m_RegisteredPlayer?, Player := RegisteredPlayer(0)) {
                    if (p_DontUpdateWhileDead?, not Player.GetFortCharacter[].IsActive[]) {
                    } else {
                        UpdateBillboards(Player)
                    }
                } else {
                    break
                }
                Sleep(p_RefreshRate)
            }

    UpdateBillboards<private>(Player:player):void=
        Lookup := msg_var_lookup{
            Device := Self
            Player := Player
        }
        Text := ReplaceIntoMessageWithFn(p_Template, Lookup.Replace)
        for (Billboard:p_Billboards) {
            Billboard.SetText(Text)
        }

    VarLookup<internal>(Player:player, VarName:string):?message=
        Result := case(VarName):
            "Player" => ToMessage(Player)
            "Score" =>
                ToMessage of (if (ScoreManager := p_ScoreManager?) {
                    Score := ScoreManager.GetCurrentScore(Player)
                    p_ScoreFormatter.Format(Score)
                } else {
                    "-1"
                })
            "Rebirth" =>
                ToMessage of (if (TycoonManager := p_TycoonManager?) {
                    RebirthCount := TycoonManager.GetRebirthLevel[Player] or 0
                    if (RebirthCount = 0, p_IgnoreRebirth0?) {
                        ""
                    } else {
                        p_RebirthCountFormatter.Format(RebirthCount)
                    }
                } else {
                    "-1"
                })
            _ =>
                if (GroupId := VarName.StripPrefix["Level:"]) {
                    ToMessage of (if (LevelGroupsManager := p_LevelGroupsManager?) {
                        Level := LevelGroupsManager.GetLevel(Player, GroupId)
                        p_LevelFormatter.Format(Level)
                    } else {
                        "-1"
                    })
                } else {
                    return None()
                }
        Some(Result)
    
    CustomTrigger<override>(Action:string, Agent:?agent):logic=
        if (Action = "register_player", Player := player[Agent?]) {
            RegisterPlayer(Player)
            return true
        } else if (Action = "clear_registered") {
            ClearRegisteredPlayer()
            return true
        }
        false

msg_var_lookup := class<final>:
    Device: player_billboard_device
    Player: player

    Replace(VarName:string):?message=
        Device.VarLookup(Player, VarName)

player_billboard_user_action_kind := enum:
    RegisterPlayer
    ClearRegistered

player_billboard_user_action_config := class<concrete>:
    @editable Action: player_billboard_user_action_kind = player_billboard_user_action_kind.RegisterPlayer
    @editable OnTrigger: trigger_device = trigger_device{}

player_billboard_user_action := class:
    Device: player_billboard_device
    Config: player_billboard_user_action_config

    OnTriggered(Agent:?agent):void=
        case(Config.Action):
            player_billboard_user_action_kind.RegisterPlayer =>
                if (Player := player[Agent?]) {
                    Device.RegisterPlayer(Player)
                }
            player_billboard_user_action_kind.ClearRegistered => Device.ClearRegisteredPlayer()
