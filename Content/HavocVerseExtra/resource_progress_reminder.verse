using { /Fortnite.com/Devices }
using { /Verse.org/Simulation }
using { /UnrealEngine.com/Temporary/Diagnostics }

using { HavocVerse.Debug }
using { HavocVerse.Persistence }
using { HavocVerse.Devices }
using { HavocVerse.Logging }
using { HavocVerse.Util }

TooltipReminderAttachmentRef<internal><localizes>:message = "If provided, the reminder won't be sent if the attachment is already unlocked"

rprem_reminder_config := class<concrete><final> {
	@editable ResourceId: string = ""
	@editable ThresholdAmount: int = 1
	@editable{ToolTip := TooltipReminderAttachmentRef} AttachmentRef: ?attachment_ref = false
	@editable HUDMessage: ?hud_message_device = false
	@editable MessageTemplate: string = ""
}

rprem_player_state := class<final> {
	var m_Done: [int]void = map{}
}

resource_progress_reminder_device := class<concrete><final>(creative_device) {
	@editable p_ResourceManager: resource_manager_device = resource_manager_device{}
	@editable p_AttachmentsManager: ?attachments_manager_device = false
	@editable p_CheckRate: float = 1.0

    @editable p_PersistenceDocumentId: string = "rprem"
	@editable p_Reminders: []rprem_reminder_config = array{}

	var m_PlayersCache: [player]rprem_player_state = map{}

	CheckConfig<private>(): void ={
		var AnyAttachment: logic = false
		for (Reminder:p_Reminders, not AnyAttachment?, Reminder.AttachmentRef?) {
			set AnyAttachment = true
		}
		if (AnyAttachment?, not p_AttachmentsManager?) {
			Log(Self, "At least one AttachmentRef is set, but p_AttachmentsManager is missing", ?Level := log_level.Error)
		}
	}

	OnBegin<override>()<suspends>: void ={
		AwareInitDevice(Self, OnInit)
	}
	
	OnInit<private>()<suspends>: void ={
		CheckConfig()

		GetPlayspace().PlayerRemovedEvent().Subscribe(OnPlayerRemoved)

		spawn {
			StartCheckLoop()
		}
	}

	OnPlayerRemoved<private>(Player:player): void ={
		set m_PlayersCache = m_PlayersCache.RemoveByKey(Player)(0)
	}

	StartCheckLoop<private>()<suspends>: void ={
		loop {
			Sleep(p_CheckRate)
			for (Player:GetPlayspace().GetPlayers()) {
				CheckPlayer(Player)
			}
		}
	}

	CheckPlayer<private>(Player:player): void ={
		State := m_PlayersCache[Player] or (
			l_State := LoadPlayerState(Player)
			if. set m_PlayersCache[Player] = l_State
			l_State
		);
		for (Index->Reminder:p_Reminders, not State.m_Done[Index]) {
			Amount := p_ResourceManager.GetAmount(Player, Reminder.ResourceId)
			if (Amount >= Reminder.ThresholdAmount) {
				ProcessReminder(Player, Index, Reminder)
				if. set State.m_Done[Index] = void
				SavePlayerState(Player, State)
			}
		}
	}

	ProcessReminder<private>(Player:player, Index:int, Reminder:rprem_reminder_config): void ={
		if (
			Ref := Reminder.AttachmentRef?
			AttachmentsManager := p_AttachmentsManager?
			AttachmentsManager.IsUnlocked[Player, Ref.AttachmentKindId, Ref.Attachment]
		) {
			Log(Self, "Skipping reminder index {Index} for '{Player}' because the attachment {Ref} is unlocked", ?Level := log_level.Debug)
			return
		}

		Log(Self, "Sending reminder index {Index} to '{Player}", ?Level := log_level.Debug)
		if (HUDMessage := Reminder.HUDMessage?) {
			HUDMessage.Show(Player, ToMessage(Reminder.MessageTemplate))
		}
	}

	SavePlayerState<private>(Player:player, State:rprem_player_state): void ={
		Document := PlayerDataGetDocument(Player, p_PersistenceDocumentId)
		DoneList := for (ReminderIndex->__:State.m_Done). ReminderIndex
		Document.SetIntList("done", DoneList)
	}

	LoadPlayerState<private>(Player:player): rprem_player_state ={
		Document := PlayerDataGetDocument(Player, p_PersistenceDocumentId)
		DoneList := Document.GetIntList["done"] or array{}
		rprem_player_state{
			m_Done := DoneList.ToSet()
		}
	}
}
	