using { /Fortnite.com/Devices }
using { /Verse.org/Simulation }
using { /UnrealEngine.com/Temporary/Diagnostics }
using { /UnrealEngine.com/Temporary/UI }

using { HavocVerse.Debug }
using { HavocVerse.Devices }
using { HavocVerse.UI }
using { HavocVerse.Util }

static_hud_image_device := class(creative_device, aware_device):
    @editable p_AccessControl: ?access_control_device = false
    @editable p_ZOrder: type{_X:int where 0 <= _X, _X <= 2147483647} = 0
    @editable p_ImagePlacement: ui_canvas_placement_config = ui_canvas_placement_config{}
    @editable p_Image: string = ""
    @editable p_ImageScale: float = 1.0

    DeviceMeta<override>()<transacts>:aware_device_meta=
        aware_device_meta{
            DeviceName := "static_hud_image"
        }

    OnBegin<override>()<suspends>:void=
        AwareInitDevice(Self, OnInit)
    
    OnInit<private>()<suspends>:void=
        HudManager := static_hud_image_hud_manager{
            AccessControl := p_AccessControl
            RefreshRate := -1.0
            ZOrder := p_ZOrder
            Device := Self
        }
        HudManager.Init(GetPlayspace())

static_hud_image_hud_manager := class(hud_manager_base):
    Device: static_hud_image_device

    CreateHUD<override>(Player:player):?hud_manager_player=
        Component := static_hud_image_hud_component:
            Device := Device
            Player := Player
        Some(MakeUIFrameworkHUDPlayer(Player, array{
            (Component, Device.p_ImagePlacement)
        }))

static_hud_image_hud_component := class(ui_component):
    Device: static_hud_image_device
    Player: player

    Create<override>():widget=
        CreateStaticImage(Device.p_Image, ?UniformScale := Device.p_ImageScale)
