using { /Fortnite.com/Devices }
using { /Verse.org/Simulation }
using { /Verse.org/Random }
using { /UnrealEngine.com/Temporary/Diagnostics }
using { /UnrealEngine.com/Temporary/SpatialMath }

using { HavocVerse.Devices }
using { HavocVerse.Debug }
using { HavocVerse.Resource }
using { HavocVerse.Logging }
using { HavocVerse.Util }

tycoon_conveyor_item_growth_mode := enum:
    DuringFall

tycoon_conveyor_item_growth_config := class<concrete>:
    @editable Mode: tycoon_conveyor_item_growth_mode = tycoon_conveyor_item_growth_mode.DuringFall
    @editable StartScale: vector3 = vector3{}
    @editable EndScale: vector3 = vector3{X := 1.0, Y := 1.0, Z := 1.0}

tycoon_conveyor_auto_generation_mode := enum:
    PerSecondRate
    OnTargetReached

TooltipConveyorGeneratorAmount<internal><localizes>:message = "Represents either a rate per second or an amount, depending on p_AutoGenerationMode"
TooltipConveyorGeneratorVirtualRateOverride<internal><localizes>:message = "Override the automatically calculated virtual rate when p_AutoGenerationMode = OnTargetReached"

tycoon_conveyor_generator_config := class<concrete>:
    @editable ResourceId: string = ""
    @editable{ToolTip := TooltipConveyorGeneratorAmount} Amount: float = 1.0
    @editable{ToolTip := TooltipConveyorGeneratorVirtualRateOverride} VirtualRateOverride: ?float = false

TooltipConveyorGenerateTriggers<internal><localizes>:message = "Triggered when an item reaches the target"

tycoon_conveyor_dropper_config := class<concrete>:
    @editable DropperId: string = ""
    @editable ItemPropOverride: ?creative_prop_asset = false
    @editable SpawnIntervalOverride: ?float = false
    @editable SpeedOverride: ?float = false
    @editable FailsafeDelayOverride: ?float = false
    @editable Generates: []tycoon_conveyor_generator_config = array{}
    @editable{ToolTip := TooltipConveyorGenerateTriggers} GenerateTriggers: []trigger_device = array{}
    @editable SpawnAudioCueOverride: ?shared_audio_cue_pool_device = false
    @editable ReachTargetAudioCueOverride: ?shared_audio_cue_pool_device = false
    @editable FallAudioCueOverride: ?shared_audio_cue_pool_device = false

tycoon_conveyor_dropper_instance_config := class<concrete>:
    @editable DropperId: string = ""
    @editable SpawnMarker: creative_prop = creative_prop{}

tycoon_conveyor_dropper_instance := class<final>:
    DropperId: string
    ModelConfig: tycoon_conveyor_dropper_config
    InstanceConfig: tycoon_conveyor_dropper_instance_config

    var m_Active: ?event() = false

    var m_GenerationTicketsInitialized: logic = false
    var m_AutoGenerationTickets: []auto_generation_ticket = array{}

TooltipConveyorSpeed<internal><localizes>:message = "Speed in cm/s"
TooltipConveyorResourceManager<internal><localizes>:message = "Required when p_AutoGenerationMode = OnTargetReached"

tycoon_conveyor_model_device := class(creative_device, aware_device):
    @editable p_DevId: string = "N/A"
    @editable{ToolTip := TooltipRoutingIdOverride} p_RoutingIdOverride: ?string = false

    @editable p_ItemProp: creative_prop_asset = DefaultCreativePropAsset
    @editable p_ItemRandomRotation: logic = false
    @editable p_ItemGrowth: ?tycoon_conveyor_item_growth_config = false
    @editable p_Droppers: []tycoon_conveyor_dropper_config = array{}
    @editable p_AwaitTargetBeforeRespawn: logic = true

    @editable p_DefaultSpawnInterval: float = 1.0
    @editable{ToolTip := TooltipConveyorSpeed} p_DefaultSpeed: float = 100.0
    @editable{ToolTip := TooltipConveyorSpeed} p_Gravity: float = 500.0
    @editable p_DefaultFailsafeDelay: float = 10.0
    @editable p_DelayBeforeFirstSpawn: float = 1.0

    @editable p_SpawnAudioCue: ?shared_audio_cue_pool_device = false
    @editable p_FallAudioCue: ?shared_audio_cue_pool_device = false
    @editable p_ReachTargetAudioCue: ?shared_audio_cue_pool_device = false

    @editable p_AutoGenerationMode: tycoon_conveyor_auto_generation_mode = tycoon_conveyor_auto_generation_mode.PerSecondRate
    @editable p_AutoGenerationManager: ?auto_generation_manager = false
    @editable p_ResourceManager: ?resource_manager_device = false
    @editable p_ResourceContainer: ?creative_device = false

    DeviceMeta<override>()<transacts>:aware_device_meta=
        aware_device_meta{
            DeviceName := "tycoon_conveyor_model"
            DevId := p_DevId
            RoutingIdOverride := p_RoutingIdOverride
        }

    FindDropperConfigById<internal>(DropperId:string)<transacts><decides>:tycoon_conveyor_dropper_config=
        var Result: ?tycoon_conveyor_dropper_config = false
        for (DropperConfig:p_Droppers, not (Result?), DropperConfig.DropperId = DropperId) {
            set Result = Some(DropperConfig)
        }
        Result?

tycoon_conveyor_device := class(creative_device, aware_device, aware_debug_device, custom_device, resetable):
    @editable p_DevId: string = "N/A"
    @editable{ToolTip := TooltipRoutingIdOverride} p_RoutingIdOverride: ?string = false

    @editable p_Model: tycoon_conveyor_model_device = tycoon_conveyor_model_device{}

    @editable p_TargetMarker: creative_prop = creative_prop{}
    @editable p_AccessControl: ?access_control_device = false
    @editable p_DropperInstances: []tycoon_conveyor_dropper_instance_config = array{}

    @editable p_UserActions: []tycoon_conveyor_user_action_config = array{}

    var m_Droppers: []tycoon_conveyor_dropper_instance = array{}

    DeviceMeta<override>()<transacts>:aware_device_meta=
        aware_device_meta{
            DeviceName := "tycoon_conveyor"
            DevId := p_DevId
            RoutingIdOverride := p_RoutingIdOverride
        }

    DebugStats<override>()<transacts>:aware_debug_stats=
        aware_debug_stats{
            Length := map{
                "droppers" => m_Droppers.Length
            }
        }

    OnBegin<override>()<suspends>:void=
        AwareInitDevice(Self, OnInit)

    OnInit<private>()<suspends>:void=
        if (ResourceContainer := p_Model.p_ResourceContainer?, not resource_holder[ResourceContainer]) {
            Log("{p_Model} has p_ResourceContainer set to a device that is not a resource holder", ?Level := log_level.Error)
        }

        var AnyGenerates: logic = false
        for (DropperConfig:p_Model.p_Droppers, not (AnyGenerates?), DropperConfig.Generates.Length > 0) {
            set AnyGenerates = true
        }
        var FailedLoad: logic = false
        if (AnyGenerates?, not (p_Model.p_ResourceManager?), not (resource_holder[p_Model.p_ResourceContainer?])) {
            Log("{p_Model} has droppers that generate resources but no resource manager or resource container set", ?Level := log_level.Error)
            set FailedLoad = true
        }
        if (AnyGenerates?, not (p_AccessControl?)) {
            Log("{Self} has droppers that generate resources but no access control is set", ?Level := log_level.Error)
            set FailedLoad = true
        }
        if (FailedLoad?) {
            return
        }

        if (AccessControl := p_AccessControl?) {
            AccessControl.RegisterUsingDevice(Self)
        }

        if (not p_TargetMarker.IsValid[]) {
            Log(Self, "Target marker is an invalid prop", ?Level := log_level.Error)
        }
        for (InstanceConfig:p_DropperInstances) {
            if (DropperConfig := p_Model.FindDropperConfigById[InstanceConfig.DropperId]) {
                if (InstanceConfig.SpawnMarker.IsValid[]) {
                    InstanceConfig.SpawnMarker.Hide()
                    DropperInstance := tycoon_conveyor_dropper_instance{
                        DropperId := InstanceConfig.DropperId
                        ModelConfig := DropperConfig
                        InstanceConfig := InstanceConfig
                    }
                    set m_Droppers += array{DropperInstance}
                } else {
                    Log(Self, "Dropper instance [id: {InstanceConfig.DropperId}] has an invalid marker prop", ?Level := log_level.Error)
                }
            } else {
                Log(Self, "Dropper instance [id: {InstanceConfig.DropperId}] has no corresponding config in model", ?Level := log_level.Error)
            }
        }

        for (UserActionConfig:p_UserActions) {
            UserAction := tycoon_conveyor_user_action{
                Device := Self
                Config := UserActionConfig
            }
            UserActionConfig.OnTrigger.TriggeredEvent.Subscribe(UserAction.OnTriggered)
        }

    Reset<override>(Agent:?agent):void=
        for (DropperInstance:m_Droppers) {
            DisableDropper(DropperInstance)
        }

    EnableDropper<internal>(DropperInstance:tycoon_conveyor_dropper_instance):void=
        if (DropperInstance.m_Active?) {
            return
        }
        AbortEvent := event(){}
        set DropperInstance.m_Active = Some(AbortEvent)

        DropperConfig := DropperInstance.ModelConfig
        if (not (DropperInstance.m_GenerationTicketsInitialized?)) {
            VanityRate := if (p_Model.p_AutoGenerationMode = tycoon_conveyor_auto_generation_mode.OnTargetReached). true else false
            if (AutoGenerationManager := p_Model.p_AutoGenerationManager?) {
                set DropperInstance.m_AutoGenerationTickets = for (
                    Generator:DropperConfig.Generates
                    Ticket := AutoGenerationManager.CreateTicket[Generator.ResourceId, auto_generation_ticket_opts{
                        Vanity := true
                        VanityIgnoreMultipliers := false
                    }]
                ) {
                    Ticket
                }
            }
            set DropperInstance.m_GenerationTicketsInitialized = true
        }
        if (p_Model.p_AutoGenerationMode = tycoon_conveyor_auto_generation_mode.PerSecondRate) {
            for (Index->Ticket:DropperInstance.m_AutoGenerationTickets, Generator := DropperConfig.Generates[Index]) {
                Ticket.SetRate(Generator.Amount)
            }
        } else if (p_Model.p_AutoGenerationMode = tycoon_conveyor_auto_generation_mode.OnTargetReached) {
            Interval := DropperConfig.SpawnIntervalOverride? or p_Model.p_DefaultSpawnInterval
            TheoreticalTimePerGen := if (p_Model.p_AwaitTargetBeforeRespawn?) {
                CalcTheoreticalTimeToAcheminate(DropperInstance) + Interval
            } else {
                Interval
            }
            for (Index->Ticket:DropperInstance.m_AutoGenerationTickets, Generator := DropperConfig.Generates[Index]) {
                VirtualRate := Generator.VirtualRateOverride? or (Generator.Amount / TheoreticalTimePerGen)
                Ticket.SetRate(VirtualRate)
            }
        }
        spawn:
            RunDropperLoop(DropperInstance, AbortEvent)
        Log("Enabled tycoon dropper {Self}[id: {DropperInstance.DropperId}]", ?Level := log_level.Debug)

    DisableDropper<internal>(DropperInstance:tycoon_conveyor_dropper_instance):void=
        if (AbortEvent := DropperInstance.m_Active?) {
            AbortEvent.Signal()
            set DropperInstance.m_Active = false
            for (Ticket:DropperInstance.m_AutoGenerationTickets) {
                Ticket.Disable()
            }
            Log("Disabled tycoon dropper {Self}[id: {DropperInstance.DropperId}]", ?Level := log_level.Debug)
        }

    RunDropperLoop<private>(DropperInstance:tycoon_conveyor_dropper_instance, AbortEvent:event())<suspends>:void=
        Interval := DropperInstance.ModelConfig.SpawnIntervalOverride? or p_Model.p_DefaultSpawnInterval
        race:
            AbortEvent.Await()
            block:
                Sleep(p_Model.p_DelayBeforeFirstSpawn)
                loop {
                    ItemTask := spawn:
                        SpawnAndAcheminateItem(DropperInstance, AbortEvent)
                    if (p_Model.p_AwaitTargetBeforeRespawn?) {
                        ItemTask.Await()
                    }
                    Sleep(Interval)
                }

    SpawnAndAcheminateItem<private>(DropperInstance:tycoon_conveyor_dropper_instance, AbortEvent:event())<suspends>:void=
        DropperConfig := DropperInstance.ModelConfig
        ItemProp := DropperConfig.ItemPropOverride? or p_Model.p_ItemProp
        FailsafeDelay := DropperConfig.FailsafeDelayOverride? or p_Model.p_DefaultFailsafeDelay
        ItemGrowth := p_Model.p_ItemGrowth

        var SpawnTransform: transform = DropperInstance.InstanceConfig.SpawnMarker.GetTransform()
        if (ItemGrowthConfig := ItemGrowth?) {
            set SpawnTransform = SpawnTransform.WithScale(SafeScale(ItemGrowthConfig.StartScale))
        }
        if (p_Model.p_ItemRandomRotation?) {
            RandomRotDegrees := GetRandomFloat(0.0, 359.99)
            RandomRot := MakeRotation(SpawnTransform.Rotation.GetLocalUp(), DegreesToRadians(RandomRotDegrees))
            set SpawnTransform = SpawnTransform.WithRotation(SpawnTransform.Rotation.RotateBy(RandomRot))
        }

        SpawnRes := SpawnProp(ItemProp, SpawnTransform)
        if (Prop := SpawnRes(0)?) {
            SpawnAudioCuePool := option{DropperConfig.SpawnAudioCueOverride?}
                .ChainOr(p_Model.p_SpawnAudioCue)
            if (AudioCuePool := SpawnAudioCuePool?) {
                AudioCuePool.Play(SpawnTransform.Translation)
            }

            race:
                block:
                    AbortEvent.Await()
                    Prop.Dispose()
                    return
                race:
                    AcheminateItemProp(DropperConfig, Prop, ItemGrowth)
                    Sleep(FailsafeDelay)
            Prop.Dispose()
            HandleTargetReached(DropperConfig)

            ReachTargetAudioCuePool := option{DropperConfig.ReachTargetAudioCueOverride?}
                .ChainOr(p_Model.p_ReachTargetAudioCue)
            if (AudioCuePool := ReachTargetAudioCuePool?) {
                AudioCuePool.Play(p_TargetMarker.GetTransform().Translation)
            }
        } else {
            race:
                AbortEvent.Await()
                block:
                    Sleep(FailsafeDelay)
                    HandleTargetReached(DropperConfig)
        }

    AcheminateItemProp<private>(
        DropperConfig:tycoon_conveyor_dropper_config,
        Prop:creative_prop,
        ItemGrowth:?tycoon_conveyor_item_growth_config
    )<suspends>:void=
        TargetPosition := p_TargetMarker.GetTransform().Translation

        if (Prop.IsValid[]) {
            var LeveledTransform: transform = Prop.GetTransform().WithZ(TargetPosition.Z)
            if (ItemGrowthConfig := ItemGrowth?) {
                set LeveledTransform = LeveledTransform.WithScale(SafeScale(ItemGrowthConfig.EndScale))
            }
            Prop.MoveToAtFixedSpeed(LeveledTransform, p_Model.p_Gravity)
            FallAudioCuePool := option{DropperConfig.FallAudioCueOverride?}
                .ChainOr(p_Model.p_FallAudioCue)
            if (AudioCuePool := FallAudioCuePool?) {
                AudioCuePool.Play(LeveledTransform.Translation)
            }
        }

        if (Prop.IsValid[]) {
            Speed := DropperConfig.SpeedOverride? or p_Model.p_DefaultSpeed
            TargetTransform := Prop.GetTransform().WithTranslation(TargetPosition)
            Prop.MoveToAtFixedSpeed(TargetTransform, Speed)
        }

    CalcTheoreticalTimeToAcheminate<private>(DropperInstance:tycoon_conveyor_dropper_instance):float=
        TargetPosition := p_TargetMarker.GetTransform().Translation

        SpawnTransform := DropperInstance.InstanceConfig.SpawnMarker.GetTransform()
        LeveledTransform := SpawnTransform.WithZ(TargetPosition.Z)
        TimeToLeveled := Distance(SpawnTransform.Translation, LeveledTransform.Translation) / p_Model.p_Gravity

        Speed := DropperInstance.ModelConfig.SpeedOverride? or p_Model.p_DefaultSpeed
        TargetTransform := LeveledTransform.WithTranslation(TargetPosition)
        TimeToTarget := Distance(LeveledTransform.Translation, TargetTransform.Translation) / Speed

        TimeToLeveled + TimeToTarget

    HandleTargetReached<private>(DropperConfig:tycoon_conveyor_dropper_config):void=
        if (
            p_Model.p_AutoGenerationMode = tycoon_conveyor_auto_generation_mode.OnTargetReached
            AccessControl := p_AccessControl?
        ) {
            for (
                Generator:DropperConfig.Generates
                EffectiveMultiplier := if (AutoGenerationManager := p_Model.p_AutoGenerationManager?) {
                    AutoGenerationManager.GetEffectiveMultiplier(Generator.ResourceId)
                } else {
                    1.0
                }
                Amount := Int[Generator.Amount * EffectiveMultiplier]
            ) {
                if (
                    ResourceContainer := resource_holder[p_Model.p_ResourceContainer?]
                    ResourceContainer.IsResourceAllowed[Generator.ResourceId]
                ) {
                    ResourceContainer.Deposit(Generator.ResourceId, Amount)
                } else if (ResourceManager := p_Model.p_ResourceManager?) {
                    ResourceManager.Give(AccessControl, Generator.ResourceId, Amount)
                } else {
                    Log(Self, "Meant to generate resource \"{Generator.ResourceId}\" but the resource container doesn't accept it and no resource manager is set", ?Level := log_level.Error)
                }
            }
        }

        if (AccessControl := p_AccessControl?) {
            for (Agent:AccessControl.ListAllowedAgents()) {
                for (Trigger:DropperConfig.GenerateTriggers) {
                    Trigger.Trigger(Agent)
                }
            }
        } else {
            for (Trigger:DropperConfig.GenerateTriggers) {
                Trigger.Trigger()
            }
        }

    FindDropperById<internal>(DropperId:string)<transacts><decides>:tycoon_conveyor_dropper_instance=
        var Result: ?tycoon_conveyor_dropper_instance = false
        for (DropperInstance:m_Droppers, not (Result?), DropperInstance.DropperId = DropperId) {
            set Result = Some(DropperInstance)
        }
        Result?

    CustomTrigger<override>(Input:string, Agent:?agent):logic=
        if (
            SplitRes := Input.SplitOnce[':']
            Action := SplitRes(0)
            DropperInstance := FindDropperById[SplitRes(1)]
        ) {
            if (Action = "enable_dropper") {
                EnableDropper(DropperInstance)
                return true
            } else if (Action = "disable_dropper") {
                DisableDropper(DropperInstance)
                return true
            }
        }
        false

tycoon_conveyor_user_action_kind := enum:
    Enable
    Disable

tycoon_conveyor_user_action_config := class<concrete>:
    @editable Action: tycoon_conveyor_user_action_kind = tycoon_conveyor_user_action_kind.Enable
    @editable OnTrigger: trigger_device = trigger_device{}
    @editable DropperIds: []string = array{}

tycoon_conveyor_user_action := class:
    Device: tycoon_conveyor_device
    Config: tycoon_conveyor_user_action_config

    OnTriggered(Agent:?agent):void=
        for (
            DropperId:Config.DropperIds
            DropperInstance := Device.FindDropperById[DropperId]
        ) {
            if (Config.Action = tycoon_conveyor_user_action_kind.Enable) {
                Device.EnableDropper(DropperInstance)
            } else if (Config.Action = tycoon_conveyor_user_action_kind.Disable) {
                Device.DisableDropper(DropperInstance)
            }
        }
