using { /Verse.org/Simulation } 
using { /Fortnite.com/Devices } 
using { /Fortnite.com/Devices/CreativeAnimation } 
using { /Fortnite.com/Playspaces } 
using { /UnrealEngine.com/Temporary/SpatialMath }
using { /UnrealEngine.com/Temporary/Diagnostics }

using { HavocVerse.Debug }
using { HavocVerse.Logging }

(AnimationController:animation_controller).PlayNewAnimation<public>(
    Keyframes:[]keyframe_delta,
    ?Mode:animation_mode = animation_mode.OneShot
):void=
    AnimationController.Pause()
    AnimationController.SetAnimation(Keyframes, ?Mode := Mode)
    AnimationController.Play()

# Awaits until the started animation is done playing.
(AnimationController:animation_controller).AwaitDone<public>()<suspends>:void=
    loop {
        Result := AnimationController.AwaitNextKeyframe()
        if (Result <> await_next_keyframe_result.KeyframeReached) {
            break
        }
    }

animation_interpolation_kind<public> := enum:
    AIK_Linear
    AIK_Ease
    AIK_EaseIn
    AIK_EaseOut
    AIK_EaseInOut

GetInterpolationType<public>(Kind:animation_interpolation_kind)<transacts>:cubic_bezier_parameters=
    case(Kind):
        animation_interpolation_kind.AIK_Linear => InterpolationTypes.Linear
        animation_interpolation_kind.AIK_Ease => InterpolationTypes.Ease
        animation_interpolation_kind.AIK_EaseIn => InterpolationTypes.EaseIn
        animation_interpolation_kind.AIK_EaseOut => InterpolationTypes.EaseOut
        animation_interpolation_kind.AIK_EaseInOut => InterpolationTypes.EaseInOut

(InProp:creative_prop).MoveToAtFixedSpeed<public>(Transform:transform, Speed:float)<suspends>:move_to_result=
    Dist := Distance(InProp.GetTransform().Translation, Transform.Translation)
    InProp.MoveTo(Transform, Dist / Speed)

(InVolume:volume_device).AwaitAgentExit<public>(Playspace:fort_playspace, Agent:agent)<suspends>:void=
    race:
        loop {
            QuittingPlayer := Playspace.PlayerRemovedEvent().Await()
            if (QuittingPlayer = Agent) {
                break
            }
        }
        loop {
            ExitingAgent := InVolume.AgentExitsEvent.Await()
            if (ExitingAgent = Agent) {
                break
            }
        }

TweenValue<public>(
    Start:float, End:float, Time:float,
    Setter:type{__(:float):void},
    ?TickRate:float = 0.0, ?Interpolation:?type{__(:float)<computes>:float} = false
)<suspends>:void=
    Delta := End - Start
    StartedAt := GetSimulationElapsedTime()
    loop {
        ElapsedTime := GetSimulationElapsedTime() - StartedAt
        RawProgress := if (InterpolationFn := Interpolation?) {
            InterpolationFn(Clamp(ElapsedTime / Time, 0.0, 1.0))
        } else {
            ElapsedTime / Time
        }
        Progress := Clamp(RawProgress, 0.0, 1.0)
        Value := Start + (Delta * Progress)
        Setter(Value)
        if (Progress >= 1.0) {
            break
        }
        Sleep(TickRate)
    }

TweenValue<public>(
    Start:vector2, End:vector2, Time:float,
    Setter:type{__(:vector2):void},
    ?TickRate:float = 0.0, ?Interpolation:?type{__(:float)<computes>:float} = false
)<suspends>:void=
    Delta := End - Start
    StartedAt := GetSimulationElapsedTime()
    loop {
        ElapsedTime := GetSimulationElapsedTime() - StartedAt
        RawProgress := if (InterpolationFn := Interpolation?) {
            InterpolationFn(Clamp(ElapsedTime / Time, 0.0, 1.0))
        } else {
            ElapsedTime / Time
        }
        Progress := Clamp(RawProgress, 0.0, 1.0)
        Value := Start + (Delta * Progress)
        Setter(Value)
        if (Progress >= 1.0) {
            break
        }
        Sleep(TickRate)
    }

EaseIn<public>(t:float)<computes>:float=
    t * t

EaseOut<public>(t:float)<computes>:float=
    1.0 - ((1.0 - t) * (1.0 - t))

# Sets an MPC value for a player.
# It is important to have a different sequence controller available for each online player.
(Playspace:fort_playspace).SetMPCValue<public>(
    SeqControllers:[]cinematic_sequence_device,
    Player:player, Value:int,
    ?Context:string = "unknown"
):void=
    SeqController := if (
        PlayerIndex := Playspace.GetPlayers().Find[Player]
        l_SeqController := SeqControllers[PlayerIndex]
    ) {
        l_SeqController
    } else{
        Log("Unable to set MPC value ({Context}) for '{Player}': missing available sequence controller", ?Level := log_level.Error)
        return
    }

    SeqController.SetPlayRate(0.0)
    SeqController.SetPlaybackFrame(Value)
    SeqController.Play(Player)
    SeqController.Pause(Player)
