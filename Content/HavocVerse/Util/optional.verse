# Sometimes just using `false` makes the typechecker unhappy with overloaded functions.
# This util is a workaround for that.
None<public>()<reads>:?false = false

# For increased clarity.
Some<public>(Val:t where t:type)<reads>:?t = option{Val}

# Chains another optional if the first one is None.
(Input:?t where t:type).ChainOr<public>(Else:?t)<reads>:?t=
    if (Val := Input?). Input else Else

# Unwraps an optional or returns a default value.
(Input:?t where t:type).UnwrapOr<public>(Default:t)<reads>:t=
    if (Val := Input?). Val else Default
