sort_order<public> := enum:
    Ascending
    Descending

quick_sort_opts<public> := class<concrete>:
    Order<public>: sort_order = sort_order.Ascending

# Performs quicksort on an array, using a given comparison function.
# Taken from https://forums.unrealengine.com/t/currently-impossible-to-implement-polymorphic-quicksort-in-verse/1147764/3.
(Input:[]t where t:type).QuickSort<public>(CompareFn(A:t, B:t)<transacts>:int)<transacts>:[]t= 
    Input.QuickSortWithOpts(CompareFn, quick_sort_opts{})

(Input:[]t where t:type).QuickSortDesc<public>(CompareFn(A:t, B:t)<transacts>:int)<transacts>:[]t= 
    Input.QuickSortWithOpts(CompareFn, quick_sort_opts{
        Order := sort_order.Descending
    })

(Input:[]t where t:type).QuickSortWithOpts<public>(CompareFn:type{__(:t,:t)<transacts>:int}, Opts:quick_sort_opts)<transacts>:[]t= 
    if (Input.Length <= 1) {
        return Input
    }

    var Less: []t = array{}
    var Equal: []t = array{}
    var Greater: []t = array{}
    if (Pivot := Input[Int[Input.Length * 0.5]]) {
        for (Element:Input) {
            CmpRes := if (Opts.Order = sort_order.Descending) {
                -CompareFn(Element, Pivot)
            } else {
                CompareFn(Element, Pivot)
            }
            if (CmpRes < 0):
                set Less += array{Element}
            else if (CmpRes > 0):
                set Greater += array{Element}
            else:
                set Equal += array{Element}
        }
    }

    Less.QuickSortWithOpts(CompareFn, Opts) + Equal + Greater.QuickSortWithOpts(CompareFn, Opts)

# Performs quicksort on an array of integers, returning a new sorted array.
SortInts<public>(Input:[]int)<transacts>:[]int = Input.QuickSort(CmpInt)

CmpInt<public>(A:int, B:int)<computes>:int=
    if (A > B) {
        1
    } else if (A < B) {
        -1
    } else {
        0
    }

# Performs quicksort on an array of floats, returning a new sorted array.
SortFloats<public>(Input:[]float)<transacts>:[]float = Input.QuickSort(CmpFloat)

CmpFloat<public>(A:float, B:float)<computes>:int=
    if (A > B) {
        1
    } else if (A < B) {
        -1
    } else {
        0
    }

CmpStr<public>(A:string, B:string)<transacts>:int=
    var Index: int = 0
    loop {
        if (CharA := A[Index], CharB := B[Index]) {
            Cmp := CmpChar(CharA, CharB)
            if (Cmp <> 0) {
                return Cmp
            }
        } else if (A[Index]) {
            return -1
        } else if (B[Index]) {
            return 1
        } else {
            return 0
        }
        set Index += 1
    }

CmpChar<public>(A:char, B:char)<computes>:int=
    OrdA := CharOrd(A)
    OrdB := CharOrd(B)
    if (OrdA > OrdB) {
        1
    } else if (OrdA < OrdB) {
        -1
    } else {
        0
    }

CharOrd<public>(Char:char)<computes>:int=
    case(Char):
        ' ' => 0
        '_' => 1
        '(' => 2
        ')' => 3
        '0' => 10
        '1' => 11
        '2' => 12
        '3' => 13
        '4' => 14
        '5' => 15
        '6' => 16
        '7' => 17
        '8' => 18
        '9' => 19
        'a' => 20
        'b' => 21
        'c' => 22
        'd' => 23
        'e' => 24
        'f' => 25
        'g' => 26
        'h' => 27
        'i' => 28
        'j' => 29
        'k' => 30
        'l' => 31
        'm' => 32
        'n' => 33
        'o' => 34
        'p' => 35
        'q' => 36
        'r' => 37
        's' => 38
        't' => 39
        'u' => 40
        'v' => 41
        'w' => 42
        'x' => 43
        'y' => 44
        'z' => 45
        'A' => 50
        'B' => 51
        'C' => 52
        'D' => 53
        'E' => 54
        'F' => 55
        'G' => 56
        'H' => 57
        'I' => 58
        'J' => 59
        'K' => 60
        'L' => 61
        'M' => 62
        'N' => 63
        'O' => 64
        'P' => 65
        'Q' => 66
        'R' => 67
        'S' => 68
        'T' => 69
        'U' => 70
        'V' => 71
        'W' => 72
        'X' => 73
        'Y' => 74
        'Z' => 75
        _ => -1
