# Abstraction to provide real world time access.
#DEPRECATED. USE UtcDateTimeNow instead.
real_time_provider<public> := interface:
    # Returns the current date and time in UTC.
    UtcNow<public>():date_time
    
# Returns the current date and time in UTC.
UtcDateTimeNow<public>():date_time=
    DateTimeFromEpoch(GetSecondsSinceEpoch())

# No element is 0-indexed.
date_time<public> := class<final><persistable>:
    Year<public>: int
    Month<public>: int
    Day<public>: int
    Hour<public>: int
    Minute<public>: int
    Second<public>: float

    ToEpoch<public>()<transacts>:float=
        DaysSinceEpoch := TimeImpl.DaysSinceEpoch(Year, Month, Day)
        ToFloat(DaysSinceEpoch * 86400 + Hour * 3600 + Minute * 60) + Second

    IsBefore<public>(Other:date_time)<decides><transacts>:void=
        ToEpoch() < Other.ToEpoch()

    IsAfter<public>(Other:date_time)<decides><transacts>:void=
        ToEpoch() > Other.ToEpoch()

ToString<public>(DateTime:date_time)<transacts>:string=
    "{DateTime.Year}-{DateTime.Month}-{DateTime.Day} {DateTime.Hour}:{DateTime.Minute}:{DateTime.Second}"

operator'-'<public>(A:date_time, B:date_time)<transacts>:duration=
    EpochA := A.ToEpoch()
    EpochB := B.ToEpoch()
    duration{
        TotalSeconds := EpochA - EpochB
    }

operator'+'<public>(DateTime:date_time, Duration:float)<transacts>:date_time=
    NewEpoch := DateTime.ToEpoch() + Duration
    DateTimeFromEpoch(NewEpoch)

operator'+'<public>(DateTime:date_time, Duration:duration)<transacts>:date_time=
    DateTime + Duration.TotalSeconds

DateTimeFromEpoch<public>(Epoch:float)<transacts>:date_time=
    var Seconds: float = Clamp(Epoch, 0.0, 2500000000.0)
    var Days: int = DefaultingFloor(Seconds / 86400.0)
    set Seconds = Seconds - (Days * 86400.0)
    set Days = Days + 1
    Hours := DefaultingFloor(Seconds / 3600.0)
    set Seconds = Seconds - (Hours * 3600.0)
    Minutes := DefaultingFloor(Seconds / 60.0)
    set Seconds = Seconds - (Minutes * 60.0)

    var Year: int = 1970
    loop {
        DaysInYear := if (TimeImpl.IsLeapYear[Year]) {
            366
        } else {
            365
        }
        if (Days >= DaysInYear) {
            set Days = Days - DaysInYear
            set Year = Year + 1
        } else {
            break
        }
    }

    var Month: int = 1
    loop {
        DaysInMonth := TimeImpl.DaysInMonth(Year, Month)
        if (DaysInMonth > 0, Days >= DaysInMonth) {
            set Days = Days - DaysInMonth
            set Month = Month + 1
        } else {
            break
        }
    }

    date_time{
        Year := Year
        Month := Month
        Day := Days
        Hour := Hours
        Minute := Minutes
        Second := Seconds
    }

duration<public> := class<final>:
    TotalSeconds<public>: float

    ToDetailedDuration<public>()<transacts>:detailed_duration=
        var w_Seconds: float = TotalSeconds

        Days := DefaultingFloor(w_Seconds / 86400.0)
        set w_Seconds -= Days * 86400.0
        Hours := DefaultingFloor(w_Seconds / 3600.0)
        set w_Seconds -= Hours * 3600.0
        Minutes := DefaultingFloor(w_Seconds / 60.0)
        set w_Seconds -= Minutes * 60.0

        detailed_duration{
            Days := Days
            Hours := Hours
            Minutes := Minutes
            Seconds := DefaultingFloor(w_Seconds)
        }

MakeDuration<public><constructor>(TotalSeconds:float) := duration:
    TotalSeconds := TotalSeconds

ToString<public>(Duration:duration)<transacts>:string=
    ToHumanDurationAutoScale(Duration.TotalSeconds, 4, human_duration_opts{
        Delimiter := ""
        DaysFmt := ("d", "d")
        HoursFmt := ("h", "h")
        MinutesFmt := ("m", "m")
        SecondsFmt := ("s", "s")
    })

detailed_duration<public> := class<final>:
    Days<public>: int
    Hours<public>: int
    Minutes<public>: int
    Seconds<public>: int

TimeUnits<public> := module:
    Day<public>: float = 86400.0
    Hour<public>: float = 3600.0
    Minute<public>: float = 60.0

TimeImpl<internal> := module:
    IsLeapYear<scoped {HavocVerse.Util}>(Year:int)<decides><transacts>:void=
        Mod[Year, 4] = 0
        (not Mod[Year, 100] = 0) or (Mod[Year, 400] = 0)

    DaysInMonth<scoped {HavocVerse.Util}>(Year:int, Month:int)<transacts>:int=
        case(Month):
            1 => 31
            2 =>
                if (IsLeapYear[Year]) {
                    29
                } else {
                    28
                }
            3 => 31
            4 => 30
            5 => 31
            6 => 30
            7 => 31
            8 => 31
            9 => 30
            10 => 31
            11 => 30
            12 => 31
            _ => 0

    DaysSinceEpoch<scoped {HavocVerse.Util}>(Year:int, Month:int, Day:int)<transacts>:int=
        CurrentYear := Clamp(Year, 1970, 2049)
        CurrentMonth := Clamp(Month, 1, 12)
        CurrentDay := Clamp(Day, 1, 31)

        var Days: int = 0
        for (Y := 1970..CurrentYear-1) {
            if (IsLeapYear[Y]) {
                set Days = Days + 366
            } else {
                set Days = Days + 365
            }
        }
        for (M := 1..CurrentMonth-1) {
            set Days = Days + DaysInMonth(CurrentYear, M)
        }
        Days + (CurrentDay-1)
