cancellable_event_subscription<public> := class:
    CancelEvent: event() = event(){}

    var m_Cancelled: logic = false

    Cancel<public>():void=
        set m_Cancelled = true
        CancelEvent.Signal()

SubscribeToEvent<public>(Event:event(t), Callback:type{__(:t):void} where t:type):cancellable_event_subscription=
    Subscription := cancellable_event_subscription{}
    spawn:
        SubscribeToEventSync(Event, Callback, Subscription)
    Subscription

SubscribeToEventSync<public>(Event:event(t), Callback:type{__(:t):void}, Subscription:cancellable_event_subscription where t:type)<suspends>:void=
    race:
        loop {
            Data := Event.Await()
            if (Subscription.m_Cancelled?) {
                break
            }
            Callback(Data)
        }
        block:
            Subscription.CancelEvent.Await()
