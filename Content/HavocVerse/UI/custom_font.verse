using { /Verse.org/Assets }
using { /Verse.org/Simulation }
using { /Verse.org/Colors }
using { /UnrealEngine.com/Temporary/Diagnostics }
using { /UnrealEngine.com/Temporary/UI }
using { /UnrealEngine.com/Temporary/SpatialMath }

using { HavocVerse.Logging }
using { HavocVerse.Util }

CharNames: [string]char = map{
    "num_0" => '0'
    "num_1" => '1'
    "num_2" => '2'
    "num_3" => '3'
    "num_4" => '4'
    "num_5" => '5'
    "num_6" => '6'
    "num_7" => '7'
    "num_8" => '8'
    "num_9" => '9'
    "a" => 'a'
    "b" => 'b'
    "c" => 'c'
    "d" => 'd'
    "e" => 'e'
    "f" => 'f'
    "g" => 'g'
    "h" => 'h'
    "i" => 'i'
    "j" => 'j'
    "k" => 'k'
    "l" => 'l'
    "m" => 'm'
    "n" => 'n'
    "o" => 'o'
    "p" => 'p'
    "q" => 'q'
    "r" => 'r'
    "s" => 's'
    "t" => 't'
    "u" => 'u'
    "v" => 'v'
    "w" => 'w'
    "x" => 'x'
    "y" => 'y'
    "z" => 'z'
    "a_upper" => 'A'
    "b_upper" => 'B'
    "c_upper" => 'C'
    "d_upper" => 'D'
    "e_upper" => 'E'
    "f_upper" => 'F'
    "g_upper" => 'G'
    "h_upper" => 'H'
    "i_upper" => 'I'
    "j_upper" => 'J'
    "k_upper" => 'K'
    "l_upper" => 'L'
    "m_upper" => 'M'
    "n_upper" => 'N'
    "o_upper" => 'O'
    "p_upper" => 'P'
    "q_upper" => 'Q'
    "r_upper" => 'R'
    "s_upper" => 'S'
    "t_upper" => 'T'
    "u_upper" => 'U'
    "v_upper" => 'V'
    "w_upper" => 'W'
    "x_upper" => 'X'
    "y_upper" => 'Y'
    "z_upper" => 'Z'
    "symb_bang" => '!'
    "symb_question" => '?'
    "symb_doublequote" => '"'
    "symb_singlequote" => '\''
    "symb_hash" => '&'
    "symb_percent" => '%'
    "symb_ampersand" => '&'
    "symb_backtick" => '`'
    "symb_paren_open" => '('
    "symb_paren_close" => ')'
    "symb_curly_open" => '{'
    "symb_curly_close" => '}'
    "symb_sqbracket_open" => '['
    "symb_sqbracket_close" => ']'
    "symb_angbracket_open" => '<'
    "symb_angbracket_close" => '>'
    "symb_asterisk" => '*'
    "symb_eq" => '='
    "symb_plus" => '+'
    "symb_comma" => ','
    "symb_dot" => '.'
    "symb_colon" => ':'
    "symb_semicolon" => ';'
    "symb_underscore" => '_'
    "symb_minus" => '-'
    "symb_at" => '@'
    "symb_slash" => '/'
    "symb_backslash" => '\\'
}

DefaultCustomFont<public>()<transacts>:custom_font=
    custom_font{
        FontSize := 0.0
        CharMappings := map{}
    }

custom_font<public> := class:
    FontSize<public>: float
    CharMappings<public>: [char]texture_metadata
    Spacing<public>: float = 7.0

    GetScaleForFontSize<internal>(TargetFontSize:float)<computes>:float=
        TargetFontSize / FontSize

CreateCustomFont<public>(AssetsProvider:assets_provider, AssetsPrefix:string, FontSize:float):custom_font=
    var CharMappings: [char]texture_metadata = map{}
    for (
        TextureMeta:AssetsProvider.GetTextureMetaAssets()
        StrippedName := TextureMeta.Name.StripPrefix[AssetsPrefix]
        Char := CharNames[StrippedName.ToLowercase()]
        set CharMappings[Char] = TextureMeta
    ) {}
    if (CharMappings.Length = 0) {
        Log("Custom font at assets path \"{AssetsPrefix}\" has no characters, wrong path?", ?Level := log_level.Warning)
    } else {
        Log("Custom font at assets path \"{AssetsPrefix}\" loaded {CharMappings.Length} character mappings", ?Level := log_level.Verbose)
    }
    custom_font{
        FontSize := FontSize
        CharMappings := CharMappings
    }

custom_font_text_block<public> := class:
    Widget<public>: overlay
    CustomFont<public>: custom_font

    var m_StackBox: ?stack_box = false
    var m_Text: string = ""
    var m_Color: color = NamedColors.Black
    var m_Scale: float = 1.0

    SetText<public>(Text:string):void=
        if (m_Text = Text) {
            return
        }
        set m_Text = Text
        Render()

    SetColor<public>(Color:color):void=
        if (EqColor[m_Color, Color]) {
            return
        }
        set m_Color = Color
        Render()

    SetScale<public>(Scale:float):void=
        if (m_Scale = Scale) {
            return
        }
        set m_Scale = CustomFont.GetScaleForFontSize(Scale)
        Render()

    Render<internal>():void=
        if (PrevStack := m_StackBox?) {
            # TODO : optimize by keeping the first common
            Widget.RemoveWidget(PrevStack)
            set m_StackBox = false
        }
        if (m_Text.Length = 0 or CustomFont.FontSize <= 0.0 or m_Scale <= 0.0) {
            return
        }
        var Spaces: int = 0
        var CharSlots: []stack_box_slot = array{}
        for (Char:m_Text) {
            if (TextureMeta := CustomFont.CharMappings[Char], Dimensions := TextureMeta.Dimensions?) {
                TextureBlock := texture_block:
                    DefaultImage := TextureMeta.Asset
                    DefaultDesiredSize := Dimensions * m_Scale
                    DefaultTint := m_Color
                set CharSlots += array{stack_box_slot{
                    Widget :=  TextureBlock
                    HorizontalAlignment := horizontal_alignment.Left
                    VerticalAlignment := vertical_alignment.Top
                    Padding := margin{
                        Left := Spaces * CustomFont.Spacing * m_Scale
                    }
                }}
                set Spaces = 0
            } else {
                set Spaces += 1
            }
        }
        StackBox := stack_box:
            Orientation := orientation.Horizontal
            Slots := CharSlots
        set m_StackBox = Some(StackBox)
        Widget.AddWidget(overlay_slot{
            Widget := StackBox
            HorizontalAlignment := horizontal_alignment.Fill
            VerticalAlignment := vertical_alignment.Fill
        })

CreateCustomFontTextBlock<public>(
    CustomFont:custom_font, DefaultText:string,
    ?TextColor:color = NamedColors.Black,
    ?FontSize:float = -1.0
):custom_font_text_block=
    TextBlock := custom_font_text_block{
        Widget := overlay:
        CustomFont := CustomFont
        m_Text := DefaultText
        m_Color := TextColor
        m_Scale := if (FontSize >= 0.0) {
            CustomFont.GetScaleForFontSize(FontSize)
        } else {
            1.0
        }
    }
    TextBlock.Render()
    TextBlock

GetCustomFontOrDefault<public>(Name:string)<transacts>:custom_font=
    option{DefaultUIAssetsProvider[].GetCustomFont[Name]}
        .UnwrapOr(DefaultCustomFont())
