using { /Verse.org/Simulation }
using { /Verse.org/Colors }
using { /UnrealEngine.com/Temporary/Diagnostics }
using { /UnrealEngine.com/Temporary/SpatialMath }
using { /UnrealEngine.com/Temporary/UI }
using { /Fortnite.com/UI }
using { /Fortnite.com/Devices }

using { HavocVerse.Logging }
using { HavocVerse.UI }
using { HavocVerse.Util }

persistence_recorder_load_select_menu_callback<public> := type{__(Slot:int,Reset:logic):void}

persistence_recorder_load_save_slot_state<public> := struct:
    SaveSlotIndex<public>: int
    ContainsData<public>: logic

# Menu interface for selecting a save slot to load from.
persistence_recorder_load_select_menu<public> := interface:
    # Initializes this menu
    Init<public>():void

    # Discards this menu
    Discard<public>():void

    # Opens a menu to select a save slot to load from.
    # Save slots are 0-based.
    # If an out of bounds slot is given, the slot is clamped in bounds.
    Open<public>(Player:player, SaveSlots:[]persistence_recorder_load_save_slot_state, SelectFn:persistence_recorder_load_select_menu_callback):void

persistence_recorder_popup_dialog_menu_config<public> := class<concrete>:
    @editable DialogDevice: popup_dialog_device = popup_dialog_device{}
    @editable ContinueButtonIndex: int = 0
    @editable ResetButtonIndex: int = 1

    @editable ConfirmDialogDevice: ?popup_dialog_device = false
    @editable ConfirmButtonIndex: int = 0

persistence_recorder_popup_dialog_menu_player := class:
    Player: player
    RecoverAction: tuple(int,logic)
    ResetAction: tuple(int,logic)
    DismissAction: tuple(int,logic)
    InConfirmMenu: logic

persistence_recorder_popup_dialog_menu<public> := class(persistence_recorder_load_select_menu):
    Config<public>: persistence_recorder_popup_dialog_menu_config

    var m_DialogWrapper: ?dialog_device_wrapper = false
    var m_ConfirmDialogWrapper: ?dialog_device_wrapper = false

    Init<override>():void=
        set m_DialogWrapper = Some(dialog_device_wrapper{
            DialogDevice := Config.DialogDevice
        })
        if (ConfirmDialogDevice := Config.ConfirmDialogDevice?) {
            set m_ConfirmDialogWrapper = Some(dialog_device_wrapper{
                DialogDevice := ConfirmDialogDevice
            })
        }

    Discard<override>():void=
        if (DialogWrapper := m_DialogWrapper?) {
            DialogWrapper.Cleanup()
            set m_DialogWrapper = false
        }
        if (DialogWrapper := m_ConfirmDialogWrapper?) {
            DialogWrapper.Cleanup()
            set m_ConfirmDialogWrapper = false
        }

    Open<override>(Player:player, SaveSlots:[]persistence_recorder_load_save_slot_state, SelectFn:persistence_recorder_load_select_menu_callback):void=
        if (DialogWrapper := m_DialogWrapper?) {
            Closure := persistence_recorder_popup_dialog_menu_closure{
                Menu := Self
                SelectFn := SelectFn
            }
            DialogWrapper.Show(Player, Some(Closure.HandleInteraction))
        }

persistence_recorder_popup_dialog_menu_closure := class:
    Menu: persistence_recorder_popup_dialog_menu
    SelectFn: persistence_recorder_load_select_menu_callback

    HandleInteraction<internal>(Response:dialog_response_event):void=
        Log("Persistence dialog clicked {Response}", ?Level := log_level.Debug)
        Action := if (Response.ResponseType = dialog_response_type.Clicked, Response.ButtonIndex? = Menu.Config.ResetButtonIndex) {
            if (ConfirmDialogWrapper := Menu.m_ConfirmDialogWrapper?) {
                ConfirmDialogWrapper.Show(Response.Agent, Some(Self.HandleConfirmInteraction))
                return
            }

            (0, true)
        } else {
            (0, false)
        }
        SelectFn(Action)

    HandleConfirmInteraction<private>(Response:dialog_response_event):void=
        Log("Persistence confirm dialog clicked {Response}", ?Level := log_level.Debug)
        Action := if (Response.ResponseType = dialog_response_type.Clicked, Response.ButtonIndex? = Menu.Config.ConfirmButtonIndex) {
            (0, true)
        } else {
            (0, false)
        }
        SelectFn(Action)

default_persistence_recorder_load_select_menu_config<public> := struct<concrete>:
    @editable BackgroundTexture<public>: string = ""
    @editable BackgroundSize<public>: vector2 = vector2{X := 724.0, Y := 296.0}
    @editable HeaderMessage<public>: string = "Your progress from last time was saved.\nDo you want to continue where you left off?"
    @editable HeaderMessageColor<public>: color = NamedColors.Beige
    @editable ButtonsSizeHint<public>: vector2 = vector2{X := 300.0, Y := 50.0}

default_persistence_recorder_load_select_menu<public> := class(persistence_recorder_load_select_menu):
    AssetsProvider<public>: assets_provider
    Config<public>: default_persistence_recorder_load_select_menu_config

    Init<override>():void=
        {}

    Discard<override>():void=
        {}

    Open<override>(Player:player, SaveSlots:[]persistence_recorder_load_save_slot_state, SelectFn:persistence_recorder_load_select_menu_callback):void=
        Instance := default_persistence_recorder_load_select_menu_instance{
            SelectFn := SelectFn
            Player := Player
        }

        InfoText := text_block{
            DefaultText := ToMessage(Config.HeaderMessage)
            DefaultTextColor := Config.HeaderMessageColor
            DefaultJustification := text_justification.Center
        }
        Slots := if (SaveSlots.Length = 1, State := SaveSlots[0]) {
            if (State.ContainsData?) {
                ContinueButton := button_loud{
                    DefaultText := ToMessage("Yes, Continue")
                }
                RestartButton := button_loud{
                    DefaultText := ToMessage("No, Restart")
                }

                default_persistence_recorder_load_select_menu_closure{
                    MenuInstance := Instance
                    SlotIndex := State.SaveSlotIndex
                    Reset := false
                }.Init(ContinueButton)
                default_persistence_recorder_load_select_menu_closure{
                    MenuInstance := Instance
                    SlotIndex := State.SaveSlotIndex
                    Reset := true
                }.Init(RestartButton)

                ButtonsStack := stack_box{
                    Orientation := orientation.Vertical
                    Slots := array{
                        stack_box_slot{
                            Padding := margin{
                                Top := 30.0
                            }
                            Widget := ApplySizeHint(ContinueButton, Config.ButtonsSizeHint)
                            HorizontalAlignment := horizontal_alignment.Fill
                            VerticalAlignment := vertical_alignment.Top
                        }
                        stack_box_slot{
                            Widget := ApplySizeHint(RestartButton, Config.ButtonsSizeHint)
                            Padding := margin{
                                Top := 30.0
                            }
                            HorizontalAlignment := horizontal_alignment.Fill
                            VerticalAlignment := vertical_alignment.Top
                        }
                    }
                }
                array{
                    stack_box_slot{
                        Widget := InfoText
                        HorizontalAlignment := horizontal_alignment.Center
                        VerticalAlignment := vertical_alignment.Top
                    }
                    stack_box_slot{
                        Widget := ButtonsStack
                        HorizontalAlignment := horizontal_alignment.Center
                        VerticalAlignment := vertical_alignment.Top
                    }
                }
            } else {
                NewSaveButton := button_loud{
                    DefaultText := ToMessage("Create New Save")
                }
                default_persistence_recorder_load_select_menu_closure{
                    MenuInstance := Instance
                    SlotIndex := State.SaveSlotIndex
                    Reset := true
                }.Init(NewSaveButton)

                array{
                    # size hint
                    stack_box_slot{
                        Widget := color_block{
                            DefaultDesiredSize := vector2{X := 300.0, Y := 0.0}
                        }
                    }
                    stack_box_slot{
                        Widget := NewSaveButton
                        HorizontalAlignment := horizontal_alignment.Fill
                    }
                }
            }
        } else {
            ButtonsRows := for (Index->State:SaveSlots) {
                ContinueButton := button_loud{
                    DefaultText := if (State.ContainsData?) {
                        ToMessage("Continue Save {State.SaveSlotIndex+1}")
                    } else {
                        ToMessage("Create Save {State.SaveSlotIndex+1}")
                    }
                }
                default_persistence_recorder_load_select_menu_closure{
                    MenuInstance := Instance
                    SlotIndex := State.SaveSlotIndex
                    Reset := false
                }.Init(ContinueButton)

                LineWidget := if (State.ContainsData?) {
                    RestartButton := button_regular{
                        DefaultText := ToMessage("Restart")
                    }
                    default_persistence_recorder_load_select_menu_closure{
                        MenuInstance := Instance
                        SlotIndex := State.SaveSlotIndex
                        Reset := true
                    }.Init(RestartButton)

                    ResumeButtonWidget := stack_box{
                        Orientation := orientation.Vertical
                        Slots := array{
                            # size hint
                            stack_box_slot{
                                Widget := color_block{
                                    DefaultDesiredSize := vector2{X := 300.0, Y := 0.0}
                                }
                            }
                            stack_box_slot{
                                Widget := ContinueButton
                                HorizontalAlignment := horizontal_alignment.Fill
                                VerticalAlignment := vertical_alignment.Top
                            }
                        }
                    }
                    RestartButtonWidget := stack_box{
                        Orientation := orientation.Vertical
                        Slots := array{
                            # size hint
                            stack_box_slot{
                                Widget := color_block{
                                    DefaultDesiredSize := vector2{X := 200.0, Y := 0.0}
                                }
                            }
                            stack_box_slot{
                                Widget := RestartButton
                                HorizontalAlignment := horizontal_alignment.Fill
                                VerticalAlignment := vertical_alignment.Top
                            }
                        }
                    }
                    stack_box{
                        Orientation := orientation.Horizontal
                        Slots := array{
                            stack_box_slot{
                                Widget := ResumeButtonWidget
                                VerticalAlignment := vertical_alignment.Center
                                Padding := margin{
                                    Right := 10.0
                                }
                            }
                            stack_box_slot{
                                Widget := RestartButtonWidget
                                VerticalAlignment := vertical_alignment.Center
                            }
                        }
                    }
                } else {
                    stack_box{
                        Orientation := orientation.Vertical
                        Slots := array{
                            # size hint
                            stack_box_slot{
                                Widget := color_block{
                                    DefaultDesiredSize := vector2{X := 510.0, Y := 0.0}
                                }
                            }
                            stack_box_slot{
                                Widget := ContinueButton
                                HorizontalAlignment := horizontal_alignment.Fill
                                VerticalAlignment := vertical_alignment.Top
                            }
                        }
                    }
                }
                stack_box_slot{
                    Widget := LineWidget
                    VerticalAlignment := vertical_alignment.Top
                    Padding := margin{
                        Top := 30.0
                    }
                }
            }
            array{
                stack_box_slot{
                    Widget := InfoText
                    HorizontalAlignment := horizontal_alignment.Center
                    VerticalAlignment := vertical_alignment.Top
                }
                stack_box_slot{
                    Widget := stack_box{
                        Orientation := orientation.Vertical
                        Slots := ButtonsRows
                    }
                    HorizontalAlignment := horizontal_alignment.Center
                    VerticalAlignment := vertical_alignment.Top
                }
            }
        }

        MainStack := stack_box{
            Orientation := orientation.Vertical
            Slots := Slots
        }
        Widget := if (Config.BackgroundTexture.Length > 0, BackgroundImage := AssetsProvider.GetTextureAsset[Config.BackgroundTexture]) {
            overlay{
                Slots := array{
                    overlay_slot{
                        Widget := texture_block{
                            DefaultImage := BackgroundImage
                            DefaultDesiredSize := Config.BackgroundSize
                        }
                        HorizontalAlignment := horizontal_alignment.Center
                        VerticalAlignment := vertical_alignment.Center
                    }
                    overlay_slot{
                        Widget := MainStack
                        HorizontalAlignment := horizontal_alignment.Center
                        VerticalAlignment := vertical_alignment.Center
                    }
                }
            }
        } else {
            MainStack
        }

        Canvas := canvas{
            Slots := array{
                canvas_slot{
                    Widget := Widget
                    Anchors := anchors{
                        Minimum := vector2{X := 0.5, Y := 0.5}
                        Maximum := vector2{X := 0.5, Y := 0.5}
                    }
                    Alignment := vector2{X := 0.5, Y := 0.5}
                    SizeToContent := true
                }
            }
        }
        set Instance.m_Canvas = Some(Canvas)
        GlobalUIAddWidget(Player, Canvas, player_ui_slot{
            InputMode := ui_input_mode.All
        })

default_persistence_recorder_load_select_menu_instance := class:
    SelectFn: persistence_recorder_load_select_menu_callback
    Player: player

    var m_Canvas: ?canvas = false

    Close():void=
        if (Canvas := m_Canvas?) {
            set m_Canvas = false
            GlobalUIRemoveWidget(Player, Canvas)
        }

    IsActive()<transacts><decides>:void=
        m_Canvas?

default_persistence_recorder_load_select_menu_closure := class:
    MenuInstance: default_persistence_recorder_load_select_menu_instance
    SlotIndex: int
    Reset: logic

    Init(Button:text_button_base):void=
        Button.OnClick().Subscribe(Self.OnClick)

    OnClick(Data:widget_message):void=
        if (MenuInstance.IsActive[]) {
            MenuInstance.Close()
            spawn:
                HandleSelectDelayed()
        }

    HandleSelectDelayed()<suspends>:void=
        # give a small time for the menu to initiate closing
        Sleep(0.1)
        MenuInstance.SelectFn(SlotIndex, Reset)
