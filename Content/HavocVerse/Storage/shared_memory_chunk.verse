using { /Fortnite.com/Devices }
using { /Verse.org/Simulation }
using { /UnrealEngine.com/Temporary/Diagnostics }

using { HavocVerse.Debug }
using { HavocVerse.Devices }
using { HavocVerse.Logging }
using { HavocVerse.Util }

# Device that exposes the same storage memory chunk to several devices.
# The chunk itself can then be split into multiple chunks, with the same abstraction as the storage itself.
#
# This is useful for gamemodes that have replicas of the same progress line (e.g. multiplayer tycoon).
shared_storage_chunk_device<public> := class(creative_device, storage_memory_device, aware_device, custom_device):
    @editable p_StorageBackend: creative_device = creative_device{}
    @editable p_ReserveBits: int = 48
    @editable p_ChunkLayoutId: string = "sharedchunk0"

    @editable p_FixedLayout: storage_memory_fixed_layout_config = storage_memory_fixed_layout_config{}

    @editable p_UserActions: []shared_storage_chunk_user_action_config = array{}

    var m_Initialized: logic = false
    var m_MemChunk: ?storage_memory_chunk = false
    var m_StorageMem: ?master_chunk_storage_memory = false

    DeviceMeta<override>()<transacts>:aware_device_meta=
        aware_device_meta{
            DeviceName := "shared_storage_chunk"
            DevId := GetID()
            PreferSequentialLoad := true
        }

    OnBegin<override>()<suspends>:void=
        AwareInitDevice(Self, OnInit)

    OnInit<private>()<suspends>:void=
        EnsureReady()

        for (UserActionConfig:p_UserActions) {
            UserAction := shared_storage_chunk_user_action{
                Device := Self
                Config := UserActionConfig
            }
            UserActionConfig.OnTrigger.TriggeredEvent.Subscribe(UserAction.OnTriggered)
        }

    EnsureReady<override>():void=
        if (m_Initialized?) {
            return
        }
        set m_Initialized = true
        StorageDevice := if (l_StorageDevice := storage_memory_device[p_StorageBackend]) {
            l_StorageDevice
        } else {
            Log("Shared storage chunk {GetID()} couldn't initialize (storage backend missing or not a storage device)", ?Level := log_level.Error)
            return
        }
        StorageDevice.EnsureReady()
        if (Chunk := StorageDevice.Reserve[p_ReserveBits, p_ChunkLayoutId]) {
            set m_MemChunk = Some(Chunk)
            StorageMem := master_chunk_storage_memory{
                MasterChunk := Chunk
                FixedLayout := p_FixedLayout
                UniqueReservation := false

                ContextStartSentence := "Shared storage chunk {GetID()}"
                ContextInSentence := "shared storage chunk {GetID()}"
            }
            set m_StorageMem = Some(StorageMem)
            StorageMem.EnsureReady()
        } else {
            Log("Unable to reserve memory chunk for shared storage chunk {GetID()}. Persistence will not work properly", ?Level := log_level.Error)
        }

    GetID<override>()<transacts>:string=
        if (TrackersStorage := trackers_storage_device[p_StorageBackend]) {
            "\"{TrackersStorage.p_DevId}:{p_ChunkLayoutId}\""
        } else if (VerseBitsStorage := verse_bits_storage_device[p_StorageBackend]) {
            "\"{VerseBitsStorage.p_StorageId}:{p_ChunkLayoutId}\""
        } else {
            "\"None:{p_ChunkLayoutId}\""
        }

    Size<override>()<transacts>:int=
        if (MemChunk := m_MemChunk?) {
            MemChunk.Size
        } else {
            0
        }

    Reserve<override>(ChunkSize:int, LayoutId:string)<transacts><decides>:storage_memory_chunk=
        if (not (m_Initialized?)) {
            Log("Shared storage chunk {GetID()} is not yet initialized (trying to reserve for chunk layout ID \"{LayoutId}\")", ?Level := log_level.Error)
            false?
        }
        m_StorageMem?.Reserve[ChunkSize, LayoutId]

    # Returns the raw storage memory chunk. This should be used mutually exclusively with Reserve.
    GetMemChunk<public>()<transacts><decides>:storage_memory_chunk=
        m_MemChunk?

    GetStorageMem<public>()<transacts><decides>:master_chunk_storage_memory=
        m_StorageMem?

    Dump<override>(Agent:agent, Detailed:logic):void=
        if (StorageMem := m_StorageMem?) {
            StorageMem.Dump(Agent, Detailed)
        }

    Fill<override>(Agent:agent, Value:logic):void=
        if (StorageMem := m_StorageMem?) {
            StorageMem.MasterChunk.Write(Agent, array{Value}.Repeat(StorageMem.MasterChunk.Size))
        }

    CustomTrigger<override>(Action:string, Agent:?agent):logic=
        if (Action = "dump", l_Agent := Agent?) {
            Dump(l_Agent, true)
            true
        } else {
            false
        }

shared_storage_chunk_user_action_kind := enum:
    Dump

shared_storage_chunk_user_action_config := class<concrete>:
    @editable Action: shared_storage_chunk_user_action_kind = shared_storage_chunk_user_action_kind.Dump
    @editable OnTrigger: trigger_device = trigger_device{}

shared_storage_chunk_user_action := class:
    Device: shared_storage_chunk_device
    Config: shared_storage_chunk_user_action_config

    OnTriggered(Agent:?agent):void=
        if (l_Agent := Agent?) {
            Device.Dump(l_Agent, true)
        }
