using { /Fortnite.com/Devices }
using { /Verse.org/Simulation }
using { /UnrealEngine.com/Temporary/Diagnostics }

using { HavocVerse.Debug }
using { HavocVerse.Devices }
using { HavocVerse.Logging }
using { HavocVerse.Util }

SizePerTracker: int = 31

trackers_storage_device<public> := class(creative_device, storage_memory_device, aware_device, aware_debug_device, custom_device):
    @editable p_DevId: string = "N/A"

    @editable p_BackingTrackers: []tracker_device = array{}
    @editable p_FixedLayout: storage_memory_fixed_layout_config = storage_memory_fixed_layout_config{}
    @editable p_RelyOnAutoSave: logic = false

    @editable p_UserActions: []trackers_storage_user_action_config = array{}

    var m_Cache: [agent]trackers_storage_agent_cache = map{}
    var m_StorageMem: ?master_chunk_storage_memory = false

    DeviceMeta<override>()<transacts>:aware_device_meta=
        aware_device_meta{
            DeviceName := "trackers_storage"
            DevId := p_DevId
        }

    DebugStats<override>()<transacts>:aware_debug_stats=
        aware_debug_stats{
            Length := map{
                "cache" => m_Cache.Length
            }
        }

    OnBegin<override>()<suspends>:void=
        AwareInitDevice(Self, OnInit)

    OnInit<private>()<suspends>:void=
        EnsureReady()

        for (UserActionConfig:p_UserActions) {
            UserAction := trackers_storage_user_action{
                Device := Self
                Config := UserActionConfig
            }
            UserActionConfig.OnTrigger.TriggeredEvent.Subscribe(UserAction.OnTriggered)
        }

        GetPlayspace().PlayerRemovedEvent().Subscribe(OnPlayerRemoved)

    OnPlayerRemoved<private>(Player:player):void=
        spawn:
            HandlePlayerRemoved(Player)
    
    HandlePlayerRemoved<private>(Player:player)<suspends>:void=
        Sleep(0.0)
        set m_Cache = m_Cache.RemoveByKey(Player)(0)

    EnsureReady<override>():void=
        if (not (m_StorageMem?)) {
            MasterChunk := trackers_storage_master_chunk{
                Size := Size()
                Device := Self
            }
            StorageMem := master_chunk_storage_memory{
                MasterChunk := MasterChunk
                FixedLayout := p_FixedLayout

                ContextStartSentence := "Trackers storage \"{p_DevId}\""
                ContextInSentence := "trackers storage \"{p_DevId}\""
            }
            set m_StorageMem = Some(StorageMem)
            StorageMem.EnsureReady()
            if (p_RelyOnAutoSave?) {
                Log("{Self} is relying on trackers auto-save/auto-load.", ?Level := log_level.Debug)
            }
        }

    GetID<override>()<computes>:string=
        "\"{p_DevId}\""

    Size<override>()<computes>:int=
        p_BackingTrackers.Length * SizePerTracker

    Reserve<override>(ChunkSize:int, LayoutId:string)<transacts><decides>:storage_memory_chunk=
        if (not (m_StorageMem?)) {
            Log("Trackers storage {GetID()} is not yet initialized (trying to reserve for chunk layout ID \"{LayoutId}\")", ?Level := log_level.Error)
            false?
        }
        m_StorageMem?.Reserve[ChunkSize, LayoutId]

    GetOrLoad<internal>(Agent:agent):trackers_storage_agent_cache=
        if (Cache := m_Cache[Agent]) {
            Cache
        } else {
            Cache := ReadLoad(Agent)
            if (set m_Cache[Agent] = Cache) {}
            Cache
        }

    ReadLoad<internal>(Agent:agent):trackers_storage_agent_cache=
        if (
            Player := player[Agent]
            AwareIntegration := AwareGetIntegration[]
            OnlineTime := GetSimulationElapsedTime() - AwareIntegration.GetSimulationElapsedTimeAtJoin(Player)
            OnlineTime < 0.5
        ) {
            Log(MsgLoadTooEarly(GetID(), Player, OnlineTime), ?Level := log_level.Warning)
            DumpStackTrace()
        }
        var State: []logic = array{}
        for (Tracker:p_BackingTrackers) {
            if (not (p_RelyOnAutoSave?)) {
                Tracker.Load(Agent)
            }
            Val := Tracker.GetValue(Agent)
            set State += ToBits(Val, ?PadLength := SizePerTracker)
        }
        trackers_storage_agent_cache{
            m_State := State
        }

    MsgLoadTooEarly<private><localizes>(Id:string, Player:player, OnlineTime:float):message=
        "Trackers storage {Id} is loading player '{Player}' after only {OnlineTime}s online time. This is likely to cause persistence issues."

    GetStorageMem<public>()<transacts><decides>:master_chunk_storage_memory=
        m_StorageMem?

    Dump<override>(Agent:agent, Detailed:logic):void=
        if (StorageMem := m_StorageMem?) {
            StorageMem.Dump(Agent, Detailed)
        }

    Fill<override>(Agent:agent, Value:logic):void=
        if (StorageMem := m_StorageMem?) {
            StorageMem.MasterChunk.Write(Agent, array{Value}.Repeat(StorageMem.MasterChunk.Size))
        }

    Reset<public>(Agent:agent):void=
        if (StorageMem := m_StorageMem?) {
            Log(MsgResetBackingTrackers(Agent, ToString(Self)), ?Level := log_level.Debug)
            for (Tracker:p_BackingTrackers) {
                Tracker.Reset(Agent)
                Tracker.ClearPersistence(Agent)
            }
            set m_Cache = m_Cache.RemoveByKey(Agent)(0)
        }

    MsgResetBackingTrackers<private><localizes>(Agent:agent, DeviceStr:string):message=
        "{DeviceStr} reseting backing trackers for '{Agent}'"

    CustomTrigger<override>(Action:string, Agent:?agent):logic=
        if (Action = "dump", l_Agent := Agent?) {
            Dump(l_Agent, true)
            true
        } else {
            false
        }

    DumpDebugs<public>(Agent:agent):void=
        Log(MsgDumpDebugs(ToString(Self), Agent), ?Level := log_level.Debug)
        Log("| Players in cache: {m_Cache.Length}", ?Level := log_level.Debug)
        for (Index->Tracker:p_BackingTrackers) {
            Log("| @ tracker #{Index} - Value: {Tracker.GetValue(Agent)}", ?Level := log_level.Debug)
            IsActive := if (Tracker.IsActive[Agent]). true else false
            ReachedTarget := if (Tracker.HasReachedTarget[Agent]). true else false
            Log("| QuestActive: {LogicToString(IsActive)}, ReachedTarget: {LogicToString(ReachedTarget)}", ?Level := log_level.Debug)
        }

    MsgDumpDebugs<private><localizes>(DeviceIdStr:string, Agent:agent):message=
        "{DeviceIdStr} dump more debugs for '{Agent}'"

    DebugSave<public>(Agent:agent):void=
        for (Tracker:p_BackingTrackers) {
            Tracker.Save(Agent)
        }

    DebugLoad<public>(Agent:agent):void=
        for (Tracker:p_BackingTrackers) {
            Tracker.Load(Agent)
        }
        set m_Cache = m_Cache.RemoveByKey(Agent)(0)
        GetOrLoad(Agent)

    DebugTempClear<public>(Agent:agent):void=
        ClearedState := array{false}.Repeat(Size())
        Cache := trackers_storage_agent_cache{
            m_State := ClearedState
        }
        if:
            set m_Cache[Agent] = Cache

trackers_storage_agent_cache := class:
    var m_State: []logic

trackers_storage_master_chunk := class(storage_memory_chunk):
    Device: trackers_storage_device

    Read<override>(Agent:agent, ?Offset:int = 0, ?Limit:int = -1, ?Raw:logic = false):[]logic=
        Cache := if (Raw?) {
            Device.ReadLoad(Agent)
        } else {
            Device.GetOrLoad(Agent)
        }
        ReadFrom := Offset
        ReadTo := if (Limit < 0). Size else Min(Size, ReadFrom + Limit)
        if (State := Cache.m_State.Slice[ReadFrom, ReadTo]) {
            State
        } else {
            EmptyState()
        }

    Write<override>(Agent:agent, Data:[]logic, ?Offset:int = 0):void=
        Cache := Device.GetOrLoad(Agent)
        var AffectedTrackers: [int]logic = map{}
        for (Index->Bit:Data, MemIndex := Offset + Index, MemIndex < Size) {
            if (set Cache.m_State[MemIndex] = Bit) {
                TrackerIndex := DivDefaultingFloor(MemIndex, SizePerTracker)
                if (set AffectedTrackers[TrackerIndex] = true) {}
            } else {
                # this should never happen
                Log("Storage memory size is less than expected (trackers storage {Device.GetID()})", ?Level := log_level.Error)
            }
        }
        for (
            Index->MemBlock:Cache.m_State.Chunks(SizePerTracker)
            AffectedTrackers[Index]
            Tracker := Device.p_BackingTrackers[Index])
        {
            Tracker.SetValue(Agent, FromBits(MemBlock))
            if (not (Device.p_RelyOnAutoSave?)) {
                Tracker.Save(Agent)
            }
        }

trackers_storage_user_action_kind := enum:
    Dump

trackers_storage_user_action_config := class<concrete>:
    @editable Action: trackers_storage_user_action_kind = trackers_storage_user_action_kind.Dump
    @editable OnTrigger: trigger_device = trigger_device{}

trackers_storage_user_action := class:
    Device: trackers_storage_device
    Config: trackers_storage_user_action_config

    OnTriggered<internal>(Agent:?agent):void=
        if (l_Agent := Agent?) {
            Device.Dump(l_Agent, true)
        }
