using { /Fortnite.com/Devices }
using { /Verse.org/Simulation }
using { /UnrealEngine.com/Temporary/Diagnostics }

using { HavocVerse.Devices }
using { HavocVerse.Logging }
using { HavocVerse.Util }

resource<public> := interface:
    # Queries the amount of this resource that an agent (or global) has.
    GetAmount<public>(Agent:?agent):int

    # Gives an agent (or global) an amount of this resource.
    Give<public>(Agent:?agent, Amount:int):logic

    # Removes an amount of this resource from an agent (or global).
    Take<public>(Agent:?agent, Amount:int):logic

    # Returns the custom amount formatter for this resource, if any.
    CustomAmountFormatter<public>()<transacts>:?number_formatter

transactful_resource<public> := interface(resource):
    # Queries the amount of this resource that an agent (or global) has.
    TransactfulGetAmount<public>(Agent:?agent)<transacts>:int

(Res:resource).GetResourceId<public>()<transacts>:string=
    if (PhysicalRes := physical_resource[Res]) {
        PhysicalRes.ResourceId
    } else if (VirtualRes := virtual_resource[Res]) {
        VirtualRes.ResourceId
    } else if (ProxiedRes := proxied_score_resource[Res]) {
        ProxiedRes.ResourceId
    } else {
        "N/A"
    }

resource_base<public> := class<abstract>(resource):
    AmountFormatter<public>: ?number_formatter = false

    CustomAmountFormatter<override>()<transacts>:?number_formatter=
        Self.AmountFormatter

# Formats the amount of a resource, using a custom formatter if present.
FormatResourceAmount<public>(Res:resource, Amount:int):string=
    if (Formatter := Res.CustomAmountFormatter()?) {
        Formatter.Format(Amount)
    } else {
        ToString(Amount)
    }

# Device or other kind of container that can hold resources.
resource_holder<public> := interface:
    # Fails if the resource kind is not allowed in this container.
    IsResourceAllowed<public>(ResourceKind:string)<transacts><decides>:void

    # Deposits an amount of a resource in this container.
    # May fail for various reasons.
    Deposit<public>(ResourceKind:string, Amount:int):logic

# Distribution behavior when a resource may be shared between multiple agents.
resource_distribution_policy<public> := enum:
    # Use the first agent as the sole recipient.
    FirstPlayer
    # Give the full amount of resources to each agent.
    FullAmountToEach
    # Split the amount of resources evenly between all agents.
    Split
