using { /Verse.org/Simulation }

using { HavocVerse.Util }

player_data_document<public> := class:
	Player<internal>: ?player = false
	TopLevelKey<internal>: ?string = false
	Parent<internal>: ?tuple(player_data_document,string) = false

	var m_Backend<internal>: generic_player_data_backend = generic_player_data_backend{}
	var m_Children<internal>: [string]player_data_document = map{}

	SetValue<private>(Key:string, Value:generic_player_data_backend_value)<transacts>:void=
		set m_Backend = generic_player_data_backend{
			MakeGenericPlayerDataBackend<constructor>(m_Backend)
			v0_Values := AddToMap(m_Backend.v0_Values, Key, Value)
		}
		Flush()

	SetBool<public>(Key:string, Value:logic)<transacts>:void=
		SetValue(Key, generic_player_data_backend_value{
			I := Some(LogicToInt(Value))
		})

	GetBool<public>(Key:string, Default:logic)<transacts>:logic=
		if (IntVal := m_Backend.v0_Values[Key].I?) {
			IntToLogic(IntVal)
		} else {
			Default
		}

	SetInt<public>(Key:string, Value:int)<transacts>:void=
		SetValue(Key, generic_player_data_backend_value{
			I := Some(Value)
		})

	GetInt<public>(Key:string)<transacts><decides>:int=
		m_Backend.v0_Values[Key].I?

	GetInt<public>(Key:string, Default:int)<transacts>:int=
		m_Backend.v0_Values[Key].I? or Default

	SetIntList<public>(Key:string, Value:[]int)<transacts>:void=
		ListValue := for (Element:Value) {
			generic_player_data_backend_value{I := Some(Element)}
		}
		SetValue(Key, generic_player_data_backend_value{
			Li := Some(ListValue)
		})

	GetIntList<public>(Key:string)<transacts><decides>:[]int=
		ListValue := m_Backend.v0_Values[Key].Li?
		for (Element:ListValue) {
			Element.I?
		}

	SetFloat<public>(Key:string, Value:float)<transacts>:void=
		SetValue(Key, generic_player_data_backend_value{
			F := Some(Value)
		})

	GetFloat<public>(Key:string)<transacts><decides>:float=
		m_Backend.v0_Values[Key].F?

	GetFloat<public>(Key:string, Default:float)<transacts>:float=
		m_Backend.v0_Values[Key].F? or Default

	SetString<public>(Key:string, Value:string)<transacts>:void=
		SetValue(Key, generic_player_data_backend_value{
			S := Some(Value)
		})

	GetString<public>(Key:string)<transacts><decides>:string=
		m_Backend.v0_Values[Key].S?

	GetString<public>(Key:string, Default:string)<transacts>:string=
		m_Backend.v0_Values[Key].S? or Default

	SetStringList<public>(Key:string, Value:[]string)<transacts>:void=
		ListValue := for (Element:Value) {
			generic_player_data_backend_value{S := Some(Element)}
		}
		SetValue(Key, generic_player_data_backend_value{
			Li := Some(ListValue)
		})

	GetStringList<public>(Key:string)<transacts><decides>:[]string=
		ListValue := m_Backend.v0_Values[Key].Li?
		for (Element:ListValue) {
			Element.S?
		}

	GetChild<public>(Key:string)<transacts>:player_data_document=
		if (Child := m_Children[Key]) {
			Child
		} else {
			ChildBackend := generic_player_data_backend{}
			set m_Backend = generic_player_data_backend{
				MakeGenericPlayerDataBackend<constructor>(m_Backend)
				v0_Children := AddToMap(m_Backend.v0_Children, Key, ChildBackend)
			}
			Child := player_data_document{
				Player := Player
				Parent := Some(Self, Key)
				m_Backend := ChildBackend
			}
			if. set m_Children[Key] = Child
			Child
		}

	Flush<private>()<transacts>:void=
		if (ParentRes := Parent?, ParentNode := ParentRes(0), SelfParentKey := ParentRes(1)) {
			ParentNode.SyncChild(SelfParentKey, m_Backend)
			ParentNode.Flush()
		} else if (l_Player := Player?, l_TopLevelKey := TopLevelKey?) {
			PlayerData := GetPlayerData(l_Player)
			SetPlayerData(l_Player, havoc_verse_player_data{
				MakeHavocVersePlayerData<constructor>(PlayerData)
				v0_Generic := AddToMap(PlayerData.v0_Generic, l_TopLevelKey, m_Backend)
			})
		}
	
	SyncChild<private>(Key:string, ChildBackend:generic_player_data_backend)<transacts>:void=
		set m_Backend = generic_player_data_backend{
			MakeGenericPlayerDataBackend<constructor>(m_Backend)
			v0_Children := AddToMap(m_Backend.v0_Children, Key, ChildBackend)
		}
