## Attachments {{{
attachment_kind_player_data<scoped {HavocVerse}> := class<final><persistable>:
    Version: int = 0

    v0_Unlocked<public>: [string]void = map{}
    v0_LastActive<public>: ?string = false

MakeAttachmentKindPlayerData<scoped {HavocVerse}><constructor>(
    CopyOf:attachment_kind_player_data
)<transacts> := attachment_kind_player_data:
    v0_Unlocked := CopyOf.v0_Unlocked
    v0_LastActive := CopyOf.v0_LastActive
## }}}

## Levels {{{
level_group_player_data<scoped {HavocVerse}> := class<final><persistable>:
    Version: int = 0

    v0_TotalExp<public>: int = 0
    v0_Level<public>: int = 0
    v0_Exp<public>: int = 0

MakeLevelGroupPlayerData<scoped {HavocVerse}><constructor>(
    CopyOf:level_group_player_data
)<transacts> := level_group_player_data:
    v0_TotalExp := CopyOf.v0_TotalExp
    v0_Level := CopyOf.v0_Level
    v0_Exp := CopyOf.v0_Exp
## }}}
