using { /Fortnite.com/Devices }
using { /Verse.org/Simulation }
using { /UnrealEngine.com/Temporary/Diagnostics }

using { HavocVerse.Debug }
using { HavocVerse.Devices.GameSpecific }
using { HavocVerse.Storage }
using { HavocVerse.Logging }
using { HavocVerse.Util }

active_ability_meta<public> := class<unique>:
    AbilityId<public>: string
    Level<public>: int
    TierName<public>: string
    Multiplier<public>: float

ability_hook<public> := interface:
    OnAbilityActivated<public>(Player:player, Meta:active_ability_meta):void

    OnAbilityDeactivated<public>(Player:player, Meta:active_ability_meta):void

abilities_manager_upgrade_result<public> := enum:
    Success
    FailCost
    FailMaxLevel
    NotPurchasable
    AbilityNotFound

abilities_manager_set_level_result<public> := enum:
    Success
    FailMaxLevel
    AbilityNotFound

abilities_manager_ability_persistence := enum:
    ActiveAndTiers
    OnlyTiers
    Disabled

abilities_manager_ability_config := class<concrete>:
    @editable AbilityId: string = ""
    @editable AbilityTiers: []abilities_manager_ability_tier_config = array{}
    @editable
    (abilities_manager_ability_config:)Persistence: abilities_manager_ability_persistence = abilities_manager_ability_persistence.ActiveAndTiers

    GetLevelTierConfig<internal>(Level:int)<transacts><decides>:abilities_manager_ability_tier_config=
        # Level 0 is "not owned", so index 0 is level 1
        AbilityTiers[Level-1]

abilities_manager_ability_tier_config := class<concrete>:
    @editable TierName: string = ""
    @editable Multiplier: float = 0.0
    @editable CostResource: ?string = false
    @editable CostAmount: int = 0
    @editable ReactivateOnSpawn: logic = false
    @editable ActivateTriggers: []trigger_device = array{}
    @editable DeactivateTriggers: []trigger_device = array{}

abilities_manager_player_state := class:
    var m_Abilities: [string]int = map{}
    var m_Active: [string]?active_ability_meta = map{}

abilities_manager_device<public> := class(creative_device, aware_device, aware_debug_device, persistence_recordable, persistence_resetable):
    @editable p_Abilities: []abilities_manager_ability_config = array{}
    @editable p_ResourceManager: resource_manager_device = resource_manager_device{}

    @editable p_PersistenceKeepAcrossTycoonRebirths: logic = true

    var m_AbilityConfigs: [string]abilities_manager_ability_config = map{}

    var m_Players: [player]abilities_manager_player_state = map{}
    var m_Hooks: []tuple(ability_hook,?[string]void) = array{}

    var m_PersistenceControl: persistence_recordable_control = persistence_recordable_control{}

    DeviceMeta<override>()<transacts>:aware_device_meta=
        aware_device_meta{
            DeviceName := "abilities_manager"
        }
    
    DebugStats<override>()<transacts>:aware_debug_stats=
        aware_debug_stats{
            Length := map{
                "abilityConfigs" => m_AbilityConfigs.Length
                "players" => m_Players.Length
                "hooks" => m_Hooks.Length
            }
        }

    OnBegin<override>()<suspends>:void=
        for (AbilityConfig:p_Abilities) {
            if:
                set m_AbilityConfigs[AbilityConfig.AbilityId] = AbilityConfig
        }

        AwareInitDevice(Self, OnInit)
    
    OnInit<private>()<suspends>:void=
        GetPlayspace().PlayerRemovedEvent().Subscribe(OnPlayerRemoved)

        if (Integration := AwareGetIntegration[]) {
            Integration.SubscribeToPlayerSpawnEvent(OnPlayerSpawned)
        }
        
    OnPlayerRemoved<private>(Player:player):void=
        set m_Players = m_Players.RemoveByKey(Player)(0)

    OnPlayerSpawned<private>(Player:player):void=
        for (Meta:GetActiveAbilities(Player)) {
            Log(MsgAbilityActivatedRespawn(Player, Meta.AbilityId, Meta.Level), ?Level := log_level.Debug)
            if (
                AbilityConfig := m_AbilityConfigs[Meta.AbilityId]
                TierConfig := AbilityConfig.GetLevelTierConfig[Meta.Level]
                TierConfig.ReactivateOnSpawn?
            ) {
                for (Hook:m_Hooks, (not (Hook(1)?)) or (Hook(1)?[Meta.AbilityId])) {
                    Hook(0).OnAbilityActivated(Player, Meta)
                }
                for (Trigger:TierConfig.ActivateTriggers) {
                    Trigger.Trigger(Player)
                }
            }
        }

    GetPlayerState<private>(Player:player):abilities_manager_player_state=
        if (PlayerState := m_Players[Player]) {
            PlayerState
        } else {
            PlayerState := abilities_manager_player_state{}
            if (set m_Players[Player] = PlayerState) {}
            PlayerState
        }

    # Register a global ability hook.
    RegisterHook<public>(Hook:ability_hook):void=
        set m_Hooks += array{(Hook, None())}

    # Register an ability hook bound to specific abilities.
    RegisterHook<public>(Hook:ability_hook, Abilities:[]string):void=
        var Filter: [string]void = map{}
        for (AbilityId:Abilities) {
            if:
                set Filter[AbilityId] = void
        }
        set m_Hooks += array{(Hook, Some(Filter))}


    # Fails if an ability with the given ID is not registered.
    IsKnownAbility<public>(AbilityId:string)<transacts><decides>:void=
        m_AbilityConfigs[AbilityId]

    # Returns the current level of an ability for a player.
    # Fails if the ability doesn't exist.
    GetLevel<public>(Player:player, AbilityId:string)<transacts><decides>:int=
        m_AbilityConfigs[AbilityId]
        if (AbilityLevel := m_Players[Player].m_Abilities[AbilityId]) {
            AbilityLevel
        } else {
            0
        }

    # Sets the current level of an ability for a player.
    SetLevel<public>(Player:player, AbilityId:string, Level:int):abilities_manager_set_level_result=
        PlayerState := GetPlayerState(Player)
        AbilityConfig := if (l_AbilityConfig := m_AbilityConfigs[AbilityId]) {
            l_AbilityConfig
        } else {
            return abilities_manager_set_level_result.AbilityNotFound
        }
        if (Level <= AbilityConfig.AbilityTiers.Length) {
            if (set PlayerState.m_Abilities[AbilityId] = Level) {}
            PostLevelChange(Player, AbilityId)
            abilities_manager_set_level_result.Success
        } else {
            abilities_manager_set_level_result.FailMaxLevel
        }

    # Returns the max level of an ability.
    # Fails if the ability doesn't exist.
    GetMaxLevel<public>(AbilityId:string)<transacts><decides>:int=
        m_AbilityConfigs[AbilityId].AbilityTiers.Length

    # Checks an ability is upgradable for a player.
    IsUpgradable<public>(Player:player, AbilityId:string)<transacts><decides>:void=
        CurrentLevel := GetLevel[Player, AbilityId]
        MaxLevel := GetMaxLevel[AbilityId]
        CurrentLevel < MaxLevel

    # Attempts to upgrade an ability for a player, taking the required cost.
    TryUpgrade<public>(Player:player, AbilityId:string):abilities_manager_upgrade_result=
        PlayerState := GetPlayerState(Player)
        NewLevel := if (PrevLevel := PlayerState.m_Abilities[AbilityId]) {
            PrevLevel + 1
        } else {
            1
        }
        AbilityConfig := if (l_AbilityConfig := m_AbilityConfigs[AbilityId]) {
            l_AbilityConfig
        } else {
            return abilities_manager_upgrade_result.AbilityNotFound
        }
        if (TierConfig := AbilityConfig.GetLevelTierConfig[NewLevel]) {
            CostResource := if (l_CostResource := TierConfig.CostResource?) {
                l_CostResource
            } else {
                return abilities_manager_upgrade_result.NotPurchasable
            }
            CostTaken := p_ResourceManager.Take(Player, CostResource, TierConfig.CostAmount,
                ?NotifyNotEnough := true, ?SendFeedback := true)
            if (CostTaken?) {
                if (set PlayerState.m_Abilities[AbilityId] = NewLevel) {}
                PostLevelChange(Player, AbilityId)
                abilities_manager_upgrade_result.Success
            } else {
                abilities_manager_upgrade_result.FailCost
            }
        } else {
            abilities_manager_upgrade_result.FailMaxLevel
        }

    PostLevelChange<private>(Player:player, AbilityId:string):void=
        PlayerState := GetPlayerState(Player)
        ActiveMeta := if (l_ActiveMeta := PlayerState.m_Active[AbilityId]) {
            l_ActiveMeta
        } else {
            return
        }
        NewMeta := if (
            NewLevel := GetLevel[Player, AbilityId]
            AbilityConfig := m_AbilityConfigs[AbilityId]
            TierConfig := AbilityConfig.GetLevelTierConfig[NewLevel]
        ) {
            if (Meta := ActiveMeta?, Meta.Level = NewLevel) {
                return
            }
            Some(active_ability_meta{
                AbilityId := AbilityId
                Level := NewLevel
                TierName := TierConfig.TierName
                Multiplier := TierConfig.Multiplier
            })
        } else if (ActiveMeta?) {
            None()
        } else {
            return
        }
        if:
            set PlayerState.m_Active[AbilityId] = NewMeta
        if (Meta := NewMeta?) {
            ProcessAbilityActivated(Player, Meta)
        } else if (Meta := ActiveMeta?) {
            ProcessAbilityDeactivated(Player, Meta)
        }

    ProcessAbilityActivated<private>(Player:player, Meta:active_ability_meta):void=
        Log(MsgAbilityActivated(Player, Meta.AbilityId, Meta.Level), ?Level := log_level.Debug)
        for (Hook:m_Hooks, (not (Hook(1)?)) or (Hook(1)?[Meta.AbilityId])) {
            Hook(0).OnAbilityActivated(Player, Meta)
        }
        if (AbilityConfig := m_AbilityConfigs[Meta.AbilityId], TierConfig := AbilityConfig.GetLevelTierConfig[Meta.Level]) {
            for (Trigger:TierConfig.ActivateTriggers) {
                Trigger.Trigger(Player)
            }
        }

    ProcessAbilityDeactivated<private>(Player:player, Meta:active_ability_meta):void=
        Log(MsgAbilityDeactivated(Player, Meta.AbilityId), ?Level := log_level.Debug)
        for (Hook:m_Hooks, (not (Hook(1)?)) or (Hook(1)?[Meta.AbilityId])) {
            Hook(0).OnAbilityDeactivated(Player, Meta)
        }
        if (AbilityConfig := m_AbilityConfigs[Meta.AbilityId], TierConfig := AbilityConfig.GetLevelTierConfig[Meta.Level]) {
            for (Trigger:TierConfig.DeactivateTriggers) {
                Trigger.Trigger(Player)
            }
        }

    MsgAbilityActivated<private><localizes>(Player:player, AbilityId:string, Level:int):message=
        "Activated ability {AbilityId} (level {Level}) for '{Player}'."
    MsgAbilityActivatedRespawn<private><localizes>(Player:player, AbilityId:string, Level:int):message=
        "Reactivated ability {AbilityId} (level {Level}) after respawn for '{Player}'."
    MsgAbilityDeactivated<private><localizes>(Player:player, AbilityId:string):message=
        "Deactivated ability {AbilityId} from '{Player}'."

    # Returns the meta of all active abilities for a player.
    GetActiveAbilities<public>(Player:player)<transacts>:[]active_ability_meta=
        for (__->Meta:m_Players[Player].m_Active, l_Meta := Meta?) {
            l_Meta
        }

    # Returns the meta for an active ability for a player.
    GetActiveAbility<public>(Player:player, AbilityId:string)<transacts><decides>:active_ability_meta=
        m_Players[Player].m_Active[AbilityId]?

    # Activates an ability for a player.
    Activate<public>(Player:player, AbilityId:string):void=
        PlayerState := GetPlayerState(Player)
        if (
            not PlayerState.m_Active[AbilityId]
            m_AbilityConfigs[AbilityId]
            set PlayerState.m_Active[AbilityId] = None()
        ) {
            PostLevelChange(Player, AbilityId)
        }

    # Deactivates an ability for a player.
    Deactivate<public>(Player:player, AbilityId:string):void=
        PlayerState := GetPlayerState(Player)
        if (WasActive := PlayerState.m_Active[AbilityId]) {
            set PlayerState.m_Active = PlayerState.m_Active.RemoveByKey(AbilityId)(0)
            if (Meta := WasActive?) {
                ProcessAbilityDeactivated(Player, Meta)
            }
        }

    GetPersistenceControl<override>()<transacts><decides>:persistence_recordable_control=
        m_PersistenceControl

    PersistenceReserve<override>(StorageMem:storage_memory)<transacts><decides>:persistence_recordable_state=
        var AbilityChunks: [string]storage_memory_chunk = map{}
        for (AbilityConfig:p_Abilities, AbilityConfig.Persistence <> abilities_manager_ability_persistence.Disabled) {
            PersistenceId := AbilityConfig.AbilityId
                .ReplaceAll(" ", "_")
                .ToLowercase()
            var WantedBits: int = BitsLength(AbilityConfig.AbilityTiers.Length)
            if (AbilityConfig.Persistence = abilities_manager_ability_persistence.ActiveAndTiers) {
                set WantedBits += 1
            }
            MemChunk := StorageMem.Reserve[WantedBits, "ability_{PersistenceId}"]
            set AbilityChunks = AddToMap(AbilityChunks, AbilityConfig.AbilityId, MemChunk)
        }
        abilities_manager_persistence_state{
            AbilityChunks := AbilityChunks
        }

    PersistenceSave<override>(i_State:persistence_recordable_state, Agent:agent, SaveCause:persistence_save_cause):void=
        if (State := abilities_manager_persistence_state[i_State], PlayerState := m_Players[Agent]) {
            for (AbilityId->MemChunk:State.AbilityChunks, AbilityConfig := m_AbilityConfigs[AbilityId]) {
                AbilityLevel := if (l_AbilityLevel := PlayerState.m_Abilities[AbilityId]) {
                    l_AbilityLevel
                } else {
                    0
                }
                if (AbilityConfig.Persistence = abilities_manager_ability_persistence.ActiveAndTiers) {
                    IsActive := if (PlayerState.m_Active[AbilityId]). true else false
                    MemChunk.Write(Agent, array{IsActive})
                    MemChunk.WriteInt(Agent, AbilityLevel, ?Offset := 1)
                } else {
                    MemChunk.WriteInt(Agent, AbilityLevel)
                }
            }
        }

    PersistenceLoad<override>(i_State:persistence_recordable_state, Agent:agent, Query:persistence_load_query):void=
        if (State := abilities_manager_persistence_state[i_State], Player := player[Agent]) {
            PlayerState := GetPlayerState(Player)
            for (AbilityId->MemChunk:State.AbilityChunks, AbilityConfig := m_AbilityConfigs[AbilityId]) {
                AbilityLevel := if (AbilityConfig.Persistence = abilities_manager_ability_persistence.ActiveAndTiers) {
                    IsActive := MemChunk.Read(Agent)
                    if:
                        IsActive[0]?
                        set PlayerState.m_Active[AbilityId] = None()
                    MemChunk.ReadInt(Agent, ?Offset := 1)
                } else {
                    MemChunk.ReadInt(Agent)
                }
                if:
                    AbilityLevel > 0
                    set PlayerState.m_Abilities[AbilityId] = AbilityLevel
                PostLevelChange(Player, AbilityId)
            }
        }

    PersistenceReset<override>(i_State:persistence_recordable_state, Agent:agent, Query:persistence_reset_query):void=
        if (p_PersistenceKeepAcrossTycoonRebirths?, tycoon_manager_rebirth_persistence_reset_query[Query]) {
            Log("Skipping reset of abilities manager persistence state because it is initialized by a rebirth", ?Level := log_level.Debug)
        } else {
            i_State.Reset(Agent)
            PersistenceLoad(i_State, Agent, persistence_load_query{
                Overwrite := true
                FirstLoad := false
            })
        }

abilities_manager_persistence_state := class(persistence_recordable_state):
    AbilityChunks: [string]storage_memory_chunk

    Reset<override>(Agent:agent):void=
        for (__->MemChunk:AbilityChunks) {
            MemChunk.Reset(Agent)
        }
