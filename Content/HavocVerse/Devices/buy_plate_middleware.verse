using { /Fortnite.com/Devices }
using { /Verse.org/Simulation }
using { /UnrealEngine.com/Temporary/Diagnostics }

using { HavocVerse.Debug }
using { HavocVerse.Util }

buy_plate_middleware<public> := interface:
    MayTryUnlockBuyplate<public>(Agent:agent, BuyPlate:buy_plate_device):logic

    OnUnlockBuyplate<public>(Agent:agent, BuyPlate:buy_plate_device):void

buy_plate_middleware_limit_device := class(creative_device, buy_plate_middleware, aware_device, aware_debug_device, custom_device, resetable):
    @editable p_SharingPolicy: buy_plate_middleware_limit_sharing_policy = buy_plate_middleware_limit_sharing_policy.PerPlayer
    @editable p_Limit: int = 1
    @editable p_FailHUDMessage: []hud_message_device = array{}

    @editable p_UserActions: []buy_plate_middleware_limit_action_config = array{}

    var m_Limit: ?int = false
    var m_State: [comparable]int = map{}

    DeviceMeta<override>()<transacts>:aware_device_meta=
        aware_device_meta{
            DeviceName := "buy_plate_middleware_limit"
        }

    DebugStats<override>()<transacts>:aware_debug_stats=
        aware_debug_stats{
            Length := map{
                "state" => m_State.Length
            }
        }

    OnBegin<override>()<suspends>:void=
        AwareInitDevice(Self, OnInit)

    OnInit<private>()<suspends>:void=
        for (UserActionConfig:p_UserActions) {
            UserAction := buy_plate_middleware_limit_action{
                Device := Self
                Config := UserActionConfig
            }
            UserActionConfig.OnTrigger.TriggeredEvent.Subscribe(UserAction.OnTriggered)
        }

    Reset<override>(Agent:?agent):void=
        if (l_Agent := Agent?) {
            AgentKey := GetAgentKey(l_Agent)
            set m_State = m_State.RemoveByKey(AgentKey)(0)
        } else {
            set m_Limit = false
            set m_State = map{}
        }

    GetLimit()<transacts>:int=
        if (Limit := m_Limit?). Limit else p_Limit
    
    ModifyLimit(Delta:int):void=
        set m_Limit = Some(GetLimit() + Delta)

    SetLimit(Limit:int):void=
        set m_Limit = Some(Limit)

    MayTryUnlockBuyplate<override>(Agent:agent, BuyPlate:buy_plate_device):logic=
        if (UnlockCount := m_State[GetAgentKey(Agent)]) {
            if (UnlockCount >= GetLimit()) {
                for (HUDMessage:p_FailHUDMessage) {
                    HUDMessage.Show(Agent)
                }
                return false
            }
        }
        true

    OnUnlockBuyplate<override>(Agent:agent, BuyPlate:buy_plate_device):void=
        AgentKey := GetAgentKey(Agent)
        if (set m_State[AgentKey] = m_State.GetOrDefault(AgentKey, 0)+1) {}

    GetAgentKey<private>(Agent:agent)<transacts>:comparable=
        case(p_SharingPolicy):
            buy_plate_middleware_limit_sharing_policy.Global => 0
            buy_plate_middleware_limit_sharing_policy.PerPlayer => Agent
            buy_plate_middleware_limit_sharing_policy.PerTeam => {
                if (Team := GetPlayspace().GetTeamCollection().GetTeam[Agent]) {
                    Team
                } else {
                    0
                }
            }

    CustomTrigger<override>(Input:string, Agent:?agent):logic=
        if (SplitRes := Input.SplitOnce[':']) {
            Action := SplitRes(0)
            Data := SplitRes(1)
            if (Action = "modify_limit", ModifyBy := ParseInt[Data]) {
                ModifyLimit(ModifyBy)
                return true
            } else if (Action = "set_limit", Limit := ParseInt[Data]) {
                SetLimit(Limit)
                return true
            }
        }
        false

buy_plate_middleware_limit_sharing_policy := enum:
    Global
    PerPlayer
    PerTeam

buy_plate_middleware_limit_action_kind := enum:
    SetLimit
    ModifyLimit

buy_plate_middleware_limit_action_config := class<concrete>:
    @editable Action: buy_plate_middleware_limit_action_kind = buy_plate_middleware_limit_action_kind.SetLimit
    @editable OnTrigger: trigger_device = trigger_device{}
    @editable Value: int = 1

buy_plate_middleware_limit_action := class:
    Device: buy_plate_middleware_limit_device
    Config: buy_plate_middleware_limit_action_config

    OnTriggered(Agent:?agent):void=
        case(Config.Action):
            buy_plate_middleware_limit_action_kind.SetLimit => Device.SetLimit(Config.Value)
            buy_plate_middleware_limit_action_kind.ModifyLimit => Device.ModifyLimit(Config.Value)
