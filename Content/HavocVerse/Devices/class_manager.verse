using { /Fortnite.com/Devices }
using { /Verse.org/Simulation }
using { /UnrealEngine.com/Temporary/Diagnostics }

using { HavocVerse.Debug }
using { HavocVerse.Devices }
using { HavocVerse.Logging }
using { HavocVerse.Util }

class_manager_device<public> := class(creative_device, aware_device, aware_debug_device, custom_device):
    @editable p_ResetClassSelector: class_and_team_selector_device = class_and_team_selector_device{}
    @editable p_ClassSelectors: []class_manager_entry = array{}
    @editable p_TrackPlayerClass: logic = true

    @editable p_UserActions: []class_manager_user_action_config = array{}

    var m_TrackedPlayerClass: [player]int = map{}

    DeviceMeta<override>()<transacts>:aware_device_meta=
        aware_device_meta{
            DeviceName := "class_manager"
        }

    DebugStats<override>()<transacts>:aware_debug_stats=
        aware_debug_stats{
            Length := map{
                "trackedPlayerClass" => m_TrackedPlayerClass.Length
            }
        }

    OnBegin<override>()<suspends>:void=
        AwareInitDevice(Self, OnInit)

    OnInit<private>()<suspends>:void=
        for (UserActionConfig:p_UserActions) {
            UserAction := class_manager_user_action{
                Device := Self
                Config := UserActionConfig
            }
            UserActionConfig.OnTrigger.TriggeredEvent.Subscribe(UserAction.OnTriggered)
        }

        GetPlayspace().PlayerRemovedEvent().Subscribe(OnPlayerRemoved)

    OnPlayerRemoved<private>(Player:player):void=
        set m_TrackedPlayerClass = m_TrackedPlayerClass.RemoveByKey(Player)(0)

    SetClassSlot(Agent:agent, ClassSlot:int):logic=
        for (Entry:p_ClassSelectors, Entry.ClassSlot = ClassSlot) {
            Entry.ClassSelector.ChangeClass(Agent)
            Log("Assigned class slot {ClassSlot} to player")
            return true
        }
        Log("No class selector device configured to assign class slot {ClassSlot} to player", ?Level := log_level.Error)
        false

    ResetClass(Agent:agent):void=
        p_ResetClassSelector.ChangeClass(Agent)

    CustomTrigger<override>(Input:string, Agent:?agent):logic=
        if (l_Agent := Agent?) {
            if (SplitRes := Input.SplitOnce[':']) {
                Action := SplitRes(0)
                Data := SplitRes(1)
                if (Action = "set_class", ClassSlot := ParseInt[Data]) {
                    SetClassSlot(l_Agent, ClassSlot)
                    return true
                }
            } else if (Input = "reset_class") {
                ResetClass(l_Agent)
                return true
            }
        }
        false

class_manager_entry := class<concrete>:
    @editable ClassSelector: class_and_team_selector_device = class_and_team_selector_device{}
    @editable ClassSlot: int = 1

class_manager_user_action_config := class<concrete>:
    @editable OnTrigger: trigger_device = trigger_device{}
    @editable SetClassSlot: int = -1

class_manager_user_action := class:
    Device: class_manager_device
    Config: class_manager_user_action_config

    OnTriggered(Agent:?agent):void=
        if (l_Agent := Agent?) {
            if (Config.SetClassSlot < 0) {
                Device.ResetClass(l_Agent)
            } else {
                Device.SetClassSlot(l_Agent, Config.SetClassSlot)
            }
        }

class_auto_sync_device<public> := class(creative_device):
    @editable p_ClassManager: class_manager_device = class_manager_device{}
    @editable p_AccessControls: []class_auto_sync_access_control_rule = array{}

    OnBegin<override>()<suspends>:void=
        for (Rule:p_AccessControls, AccessControl := Rule.AccessControl[0]) {
            Closure := class_auto_sync_access_control_closure{
                Device := Self
                Rule := Rule
            }
            AccessControl.SubscribeToPlayerJoinEvents(Closure.OnPlayerAdded, Closure.OnPlayerRemoved)
        }

class_auto_sync_access_control_rule := class<concrete>:
    @editable AccessControl: []access_control_device = array{}
    @editable ClassSlot: int = 1

class_auto_sync_access_control_closure := class:
    @editable Device: class_auto_sync_device
    @editable Rule: class_auto_sync_access_control_rule

    OnPlayerAdded(Player:player):void=
        Log("Set player class on access control join", ?Level := log_level.Debug)
        Device.p_ClassManager.SetClassSlot(Player, Rule.ClassSlot)

    OnPlayerRemoved(Player:player, Quit:logic):void=
        if (Quit?) {
            return
        }
        Log("Set player class on access control leave", ?Level := log_level.Debug)
        Device.p_ClassManager.ResetClass(Player)
