using { /Fortnite.com/Devices }
using { /Fortnite.com/Characters }
using { /Verse.org/Simulation }
using { /UnrealEngine.com/Temporary/Diagnostics }

using { HavocVerse.Debug }
using { HavocVerse.Logging }
using { HavocVerse.Util }

exposed_player_spawn_detection_method<public> := enum:
    AliveCheckLoop
    PlayerSpawners

player_spawn_event<public> := class:
    Player<public>: player
    Spawner<public>: ?player_spawner_device

exposed_player_spawn_device<public> := class(creative_device, aware_device, aware_debug_device):
    @editable p_DetectionMethod<public>: exposed_player_spawn_detection_method = exposed_player_spawn_detection_method.AliveCheckLoop

    @editable p_AliveCheckLoopRate: float = 0.5
    @editable p_PlayerSpawners: []player_spawner_device = array{}

    PlayerSpawnedEvent<public>: event(player_spawn_event) = event(player_spawn_event){}

    var m_Listeners: []type{__(:player_spawn_event):void} = array{}

    var m_LastPlayerStates: [player]logic = map{}

    DeviceMeta<override>()<transacts>:aware_device_meta=
        aware_device_meta{
            DeviceName := "exposed_player_spawn"
        }

    DebugStats<override>()<transacts>:aware_debug_stats=
        aware_debug_stats{
            Length := map{
                "listeners" => m_Listeners.Length
            }
        }

    OnBegin<override>()<suspends>:void=
        AwareInitDevice(Self, OnInit)

    OnInit<public>()<suspends>:void=
        if (IsMisconfigured[]) {
            Log("Exposed player spawn device is using PlayerSpawners detection method but none are configured", ?Level := log_level.Warning)
        }

        case(p_DetectionMethod):
            exposed_player_spawn_detection_method.AliveCheckLoop =>
                GetPlayspace().PlayerRemovedEvent().Subscribe(OnPlayerRemoved)
                spawn:
                    StartAliveCheckLoop()
            exposed_player_spawn_detection_method.PlayerSpawners =>
                for (PlayerSpawner:p_PlayerSpawners) {
                    Closure := exposed_player_spawn_closure{
                        Device := Self
                        PlayerSpawner := PlayerSpawner
                    }
                    PlayerSpawner.SpawnedEvent.Subscribe(Closure.OnSpawned)
                }

    IsMisconfigured<public>()<transacts><decides>:void=
        p_DetectionMethod = exposed_player_spawn_detection_method.PlayerSpawners and p_PlayerSpawners.Length = 0

    StartAliveCheckLoop<private>()<suspends>:void=
        Log("Starting alive check loop with {p_AliveCheckLoopRate}s rate", ?Level := log_level.Debug)
        loop {
            for (Player:GetPlayspace().GetPlayers(), FortCharacter := Player.GetFortCharacter[]) {
                IsAlive := if (FortCharacter.IsActive[]) {
                    true
                } else {
                    false
                }
                LastState := m_LastPlayerStates[Player] or false
                if. set m_LastPlayerStates[Player] = IsAlive
                if (LastState = false, IsAlive = true) {
                    Log("Alive check loop detected player spawn", ?Level := log_level.Debug)
                    Event := player_spawn_event{
                        Player := Player
                        Spawner := None()
                    }
                    for (Listener:m_Listeners) {
                        Listener(Event)
                    }
                    PlayerSpawnedEvent.Signal(Event)
                }
            }
            Sleep(p_AliveCheckLoopRate)
        }

    OnPlayerRemoved<private>(Player:player):void=
        set m_LastPlayerStates = m_LastPlayerStates.RemoveByKey(Player)(0)

    SubscribeToSpawnEvent<public>(Listener:type{__(:player_spawn_event):void}):void=
        set m_Listeners += array{Listener}

exposed_player_spawn_closure := class:
    Device: exposed_player_spawn_device
    PlayerSpawner: player_spawner_device

    OnSpawned(Agent:agent):void=
        if (Player := player[Agent]) {
            Device.PlayerSpawnedEvent.Signal(player_spawn_event{
                Player := Player
                Spawner := Some(PlayerSpawner)
            })
        }
