using { /Fortnite.com/Devices }
using { /Verse.org/Simulation }
using { /UnrealEngine.com/Temporary/Diagnostics }

using { HavocVerse.Debug }
using { HavocVerse.Devices.GameSpecific }
using { HavocVerse.Storage }
using { HavocVerse.Persistence }
using { HavocVerse.Logging }
using { HavocVerse.Util }

level_exp_pair<public> := class<final>:
    Level<public>: int
    ExpRest<public>: int

process_level_exp_result<public> := class<final>:
    NewState<public>: level_exp_pair
    Continue<public>: logic

level_function_closure<public> := interface:
    ProcessExpGain<public>(State:level_exp_pair)<transacts>:process_level_exp_result

    GetTargetExp<public>(State:level_exp_pair)<transacts>:int

level_group_function_provider<public> := interface:
    GetLevelGroupFunction<public>(GroupId:string)<transacts><decides>:level_function_closure

player_level_hook<public> := interface:
    HandlePlayerLevel<public>(Player:player, GroupId:string, PrevLevel:?int, Level:int):void

level_group_level_up_trigger_config := class<concrete>:
    @editable Trigger: trigger_device = trigger_device{}
    @editable ForLevel: int = -1

    IsApplicable<public>(Level:int)<transacts><decides>:void=
        ForLevel < 0 or ForLevel = Level

level_group_config_function_kind := enum:
    Flat

level_group_config_per_level_override := class<final><concrete>:
    @editable Level: int = 0
    @editable RequiredExp: int = 100

level_group_config_feedback := class<final><concrete>:
    @editable HUDMessenger: []hud_message_device = array{}
    @editable Formatter: number_formatter_config = number_formatter_config{
        Template := "+%Val%"
    }
    @editable DisplayTime: float = 1.2

level_group_config := class<final><concrete>:
    @editable GroupId: string = ""
    @editable PersistenceKeepAcrossTycoonRebirths: logic = false

    @editable StartLevel: int = 1
    @editable MaxLevel: ?int = false
    @editable FunctionKind: level_group_config_function_kind = level_group_config_function_kind.Flat
    @editable FunctionVarA: int = 0
    @editable PerLevelOverrides: []level_group_config_per_level_override = array{}
    @editable FunctionProvider: ?creative_device = false

    @editable LevelUpTriggers: []level_group_level_up_trigger_config = array{}
    @editable LevelUpEffectTriggers: []trigger_device = array{}

    @editable ExpGainFeedback: ?level_group_config_feedback = false
    @editable LevelUpFedback: ?level_group_config_feedback = false

    var m_LevelFunction<private>: ?level_function_closure = false

    GetLevelFunction<internal>()<transacts>:?level_function_closure=
        if (m_LevelFunction?) {
            return m_LevelFunction
        }

        if (FunctionProviderDevice := FunctionProvider?) {
            if (l_FunctionProvider := level_group_function_provider[FunctionProviderDevice]) {
                set m_LevelFunction = option{l_FunctionProvider.GetLevelGroupFunction[GroupId]}
            } else {
                Log("FunctionProvider for level group \"{GroupId}\" is not a level_group_function_provider", ?Level := log_level.Error)
                None()
            }
            return m_LevelFunction
        }

        var i_PerLevelOverrides: [int]int = map{}
        for (Override:PerLevelOverrides) {
            set i_PerLevelOverrides = ConcatenateMaps(i_PerLevelOverrides, map{
                Override.Level => Override.RequiredExp
            })
        }

        set m_LevelFunction = Some of case(FunctionKind):
            level_group_config_function_kind.Flat =>
                level_group_builtin_function_flat_closure{
                    PerLevelOverrides := i_PerLevelOverrides
                    MaxLevel := MaxLevel
                    ExpPerLevel := FunctionVarA
                }

level_groups_manager_hook_instance<internal> := class<final>:
    Hook: player_level_hook
    Filter: ?[string]void

    var m_ActivatedPlayers: [agent]void = map{}

    IsApplicable(GroupId:string)<decides><transacts>:void=
        (not Filter?) or (Filter?[GroupId])

level_groups_manager_device<public> := class<concrete>(creative_device, aware_device, aware_debug_device, persistence_resetable):
    @editable p_LevelGroups: []level_group_config = array{}
    @editable p_RecallHooksOnRespawn: logic = true

    var m_Hooks: []level_groups_manager_hook_instance = array{}
    var m_SubscribedToPlayerSpawn: logic = false

    var m_PlayerUpdateListeners: []type{__(:player,:level_exp_pair):void} = array{}

    DeviceMeta<override>()<transacts>:aware_device_meta=
        aware_device_meta{
            DeviceName := "level_groups_manager"
        }

    DebugStats<override>()<transacts>:aware_debug_stats=
        aware_debug_stats{
            Length := map{
                "hooks" => m_Hooks.Length
                "listeners::playerUpdate" => m_PlayerUpdateListeners.Length
            }
        }

    OnBegin<override>()<suspends>:void=
        AwareInitDevice(Self, OnInit)
    
    OnInit<private>()<suspends>:void=
        GetPlayspace().PlayerAddedEvent().Subscribe(OnPlayerAdded)
        GetPlayspace().PlayerRemovedEvent().Subscribe(OnPlayerRemoved)
        for (Player:GetPlayspace().GetPlayers()) {
            OnPlayerAdded(Player)
        }

    OnPlayerAdded<private>(Player:player):void=
        for (HookInstance:m_Hooks) {
            ProcessHookPlayer(HookInstance, Player)
        }
    
    OnPlayerRemoved<private>(Player:player):void=
        for (HookInstance:m_Hooks) {
            RemoveRes := HookInstance.m_ActivatedPlayers.RemoveByKey(Player)
            set HookInstance.m_ActivatedPlayers = RemoveRes(0)
        }

    ProcessHookPlayer<private>(HookInstance:level_groups_manager_hook_instance, Player:player):void=
        if (HookInstance.m_ActivatedPlayers[Player]) {
            return
        }
        if. set HookInstance.m_ActivatedPlayers[Player] = void

        for (
            LevelGroupConfig:p_LevelGroups
            HookInstance.IsApplicable[LevelGroupConfig.GroupId]
        ) {
            Level := GetLevel(Player, LevelGroupConfig.GroupId)
            HookInstance.Hook.HandlePlayerLevel(Player, LevelGroupConfig.GroupId, None(), Level)
        }

    GetLevelGroup<private>(GroupId:string)<transacts><decides>:level_group_config=
        var Result: ?level_group_config = false
        for (Config:p_LevelGroups, not (Result?), Config.GroupId = GroupId) {
            set Result = Some(Config)
        }
        Result?

    # Returns a group's max level if it exists.
    GetMaxLevel<public>(GroupId:string)<transacts><decides>:int=
        GetLevelGroup[GroupId].MaxLevel?

    # Registers a level hook.
    #
    # If GroupId is specified, the hook will be bound to that unique group.
    # Otherwise, the hook will be bound to all groups.
    RegisterHook<public>(Hook:player_level_hook, ?GroupId:string = ""):void=
        if (GroupId.Length > 0, not GetLevelGroup[GroupId]) {
            Log("Registering level hook for unconfigured level group \"{GroupId}\"", ?Level := log_level.Warning)
        }

        HookInstance := level_groups_manager_hook_instance{
            Hook := Hook
            Filter := if (GroupId.Length > 0) {
                Some(map{GroupId => void})
            } else {
                None()
            }
        }
        set m_Hooks += array{HookInstance}
        HandlePostHookRegister(HookInstance)

    # Registers a level hook bound to specific level groups.
    RegisterHook<public>(Hook:player_level_hook, FilterGroups:[]string):void=
        for (
            GroupId:FilterGroups
            not GetLevelGroup[GroupId]
        ) {
            Log("Registering level hook for unconfigured level group \"{GroupId}\"", ?Level := log_level.Warning)
        }

        HookInstance := level_groups_manager_hook_instance{
            Hook := Hook
            Filter := Some(FilterGroups.ToSet())
        }
        set m_Hooks += array{HookInstance}
        HandlePostHookRegister(HookInstance)

    HandlePostHookRegister<private>(HookInstance:level_groups_manager_hook_instance):void=
        for (Player:GetPlayspace().GetPlayers()) {
            ProcessHookPlayer(HookInstance, Player)
        }
        if (
            p_RecallHooksOnRespawn?
            not m_SubscribedToPlayerSpawn?
            Integration := AwareGetIntegration[]
        ) {
            # register only if there are hooks in the first place
            # to keep p_RecallHooksOnRespawn the default, while not unnecessarily warning if no exposed source is set
            Integration.SubscribeToPlayerSpawnEvent(OnPlayerSpawned)
            set m_SubscribedToPlayerSpawn = true
        }

    OnPlayerSpawned<private>(Player:player):void=
        for (HookInstance:m_Hooks) {
            for (
                LevelGroupConfig:p_LevelGroups
                HookInstance.IsApplicable[LevelGroupConfig.GroupId]
            ) {
                Level := GetLevel(Player, LevelGroupConfig.GroupId)
                PrevLevel := if (HookInstance.m_ActivatedPlayers[Player]) {
                    Some(Level)
                } else {
                    if. set HookInstance.m_ActivatedPlayers[Player] = void
                    None()
                }
                HookInstance.Hook.HandlePlayerLevel(Player, LevelGroupConfig.GroupId, PrevLevel, Level)
            }
        }

    # Returns a player's level for a specific group.
    GetLevel<public>(Player:player, GroupId:string)<transacts>:int=
        if (PlayerData := PlayerDataGetLevelGroup[Player, GroupId]) {
            PlayerData.v0_Level
        } else if (LevelGroupConfig := GetLevelGroup[GroupId]) {
            LevelGroupConfig.StartLevel
        } else {
            0
        }


    # Returns a player's level and exp pair for a specific group.
    GetLevelExp<public>(Player:player, GroupId:string)<transacts>:level_exp_pair=
        PlayerData := GetPlayerDataOrDefault(Player, GroupId)
        level_exp_pair{
            Level := PlayerData.v0_Level
            ExpRest := PlayerData.v0_Exp
        }

    # Returns a level group player data for a specific group, or default.
    GetPlayerDataOrDefault<private>(Player:player, GroupId:string)<transacts>:level_group_player_data=
        if (PlayerData := PlayerDataGetLevelGroup[Player, GroupId]) {
            PlayerData
        } else if (LevelGroupConfig := GetLevelGroup[GroupId]) {
            level_group_player_data{
                v0_Level := LevelGroupConfig.StartLevel
            }
        } else {
            level_group_player_data{}
        }

    # Returns the target required experience for the next level for a specific group.
    GetTargetExp<public>(GroupId:string, LevelExpPair:level_exp_pair)<decides><transacts>:int=
        LevelGroupConfig := GetLevelGroup[GroupId]
        LevelFunction := LevelGroupConfig.GetLevelFunction()?
        LevelFunction.GetTargetExp(LevelExpPair)

    # Grants experience to a player for a specific group.
    GiveExp<public>(Player:player, GroupId:string, ExpAmount:int):void=
        if (not Player.IsActive[]) {
            LogPlayerOffline("GiveExp({GroupId}, {ExpAmount})", Player)
            return
        }

        LevelGroupConfig := if (l_LevelGroupConfig := GetLevelGroup[GroupId]) {
            l_LevelGroupConfig
        } else {
            Log(Self, "Level group \"{GroupId}\" is not configured: GiveExp", ?Level := log_level.Error)
            return
        }

        if (ExpAmount <= 0) {
            Log(Self, "Attempting to grant {ExpAmount} exp on level group \"{GroupId}\"", ?Level := log_level.Warning)
            return
        }

        LevelFunctionRes := LevelGroupConfig.GetLevelFunction()
        LevelFunction := if (l_LevelFunction := LevelFunctionRes?) {
            l_LevelFunction
        } else {
            return
        }

        PlayerData := GetPlayerDataOrDefault(Player, GroupId)
        NewTotalExp := PlayerData.v0_TotalExp + ExpAmount
        var LevelExpPair: level_exp_pair = level_exp_pair{
            Level := PlayerData.v0_Level
            ExpRest := PlayerData.v0_Exp + ExpAmount
        }
        loop {
            Result := LevelFunction.ProcessExpGain(LevelExpPair)
            set LevelExpPair = Result.NewState
            if (not (Result.Continue?)) {
                break
            }
        }

        PlayerDataSetLevelGroup(Player, GroupId, level_group_player_data{
            MakeLevelGroupPlayerData<constructor>(PlayerData)
            v0_TotalExp := NewTotalExp
            v0_Level := LevelExpPair.Level
            v0_Exp := LevelExpPair.ExpRest
        })
        Log(MsgGiveExp(Player, GroupId, ExpAmount), ?Level := log_level.Verbose)

        for (Listener:m_PlayerUpdateListeners) {
            Listener(Player, LevelExpPair)
        }

        if (FeedbackConfig := LevelGroupConfig.ExpGainFeedback?) {
            for (HUDMessenger:FeedbackConfig.HUDMessenger) {
                Message := ToMessage(FeedbackConfig.Formatter.Format(ExpAmount))
                HUDMessenger.Show(Player, Message, ?DisplayTime := FeedbackConfig.DisplayTime)
            }
        }

        GainedLevels := LevelExpPair.Level - PlayerData.v0_Level
        if (GainedLevels <= 0) {
            return
        }

        PrevLevel := PlayerData.v0_Level
        NewLevel := LevelExpPair.Level
        Log(MsgLevelChanged(Player, GroupId, PrevLevel, NewLevel), ?Level := log_level.Verbose)
        for (HookInstance:m_Hooks, HookInstance.IsApplicable[GroupId]) {
            l_PrevLevel := if (HookInstance.m_ActivatedPlayers[Player]) {
                Some(PrevLevel)
            } else {
                if. set HookInstance.m_ActivatedPlayers[Player] = void
                None()
            }
            HookInstance.Hook.HandlePlayerLevel(Player, GroupId, l_PrevLevel, NewLevel)
        }

        if (FeedbackConfig := LevelGroupConfig.LevelUpFedback?) {
            for (HUDMessenger:FeedbackConfig.HUDMessenger) {
                Message := ToMessage(FeedbackConfig.Formatter.Format(GainedLevels))
                HUDMessenger.Show(Player, Message, ?DisplayTime := FeedbackConfig.DisplayTime)
            }
        }

        for (Level := PlayerData.v0_Level..LevelExpPair.Level-1) {
            for (TriggerConfig:LevelGroupConfig.LevelUpTriggers, TriggerConfig.IsApplicable[Level+1]) {
                TriggerConfig.Trigger.Trigger(Player)
            }
        }
        for (Trigger:LevelGroupConfig.LevelUpEffectTriggers) {
            Trigger.Trigger(Player)
        }

    MsgGiveExp<private><localizes>(Player:player, GroupId:string, ExpAmount:int):message=
        "Gave {ExpAmount} exp. to '{GroupId}' for level group \"{GroupId}\""

    MsgLevelChanged<private><localizes>(Player:player, GroupId:string, PrevLevel:int, NewLevel:int):message=
        "'{Player}' progressed from level {PrevLevel} to {NewLevel} on level group \"{GroupId}\""

    # Listen to when a player's level and exp state updates.
    SubscribeToPlayerUpdateEvent<public>(Listener:type{__(:player,:level_exp_pair):void}):void=
        set m_PlayerUpdateListeners += array{Listener}

    PersistenceReset<override>(__:persistence_recordable_state, Agent:agent, Query:persistence_reset_query):void=
        for (Config:p_LevelGroups, Player := player[Agent]) {
            if (Config.PersistenceKeepAcrossTycoonRebirths?, tycoon_manager_rebirth_persistence_reset_query[Query]) {
                Log("Skipping reset of level group \"{Config.GroupId}\" persistence state because it is initialized by a rebirth", ?Level := log_level.Debug)
            } else {
                PlayerDataSetLevelGroup(Player, Config.GroupId, level_group_player_data{})
            }
        }

level_group_builtin_function_closure<internal> := class<abstract>(level_function_closure):
    PerLevelOverrides: [int]int
    MaxLevel: ?int

    ProcessExpGain<override>(State:level_exp_pair)<transacts>:process_level_exp_result=
        if (l_MaxLevel := MaxLevel?, State.Level >= l_MaxLevel) {
            process_level_exp_result{
                NewState := State
                Continue := false
            }
        } else if (RequiredExp := PerLevelOverrides[State.Level]) {
            if (State.ExpRest >= RequiredExp) {
                if (RequiredExp <= 0) {
                    Log("Required exp override for level {State.Level} is zero or negative: ProcessExpGain", ?Level := log_level.Error)
                    return process_level_exp_result{
                        NewState := State
                        Continue := false
                    }
                }
                process_level_exp_result{
                    NewState := level_exp_pair{
                        Level := State.Level + 1
                        ExpRest := State.ExpRest - RequiredExp
                    }
                    Continue := true
                }
            } else {
                process_level_exp_result{
                    NewState := State
                    Continue := false
                }
            }
        } else {
            ChainProcessExpGain(State)
        }

    GetTargetExp<override>(State:level_exp_pair)<transacts>:int=
        if (RequiredExp := PerLevelOverrides[State.Level]) {
            RequiredExp
        } else {
            ChainGetTargetExp(State)
        }

    ChainProcessExpGain<internal>(State:level_exp_pair)<transacts>:process_level_exp_result

    ChainGetTargetExp<internal>(State:level_exp_pair)<transacts>:int

level_group_builtin_function_flat_closure<internal> := class<final>(level_group_builtin_function_closure):
    ExpPerLevel: int

    ChainProcessExpGain<override>(State:level_exp_pair)<transacts>:process_level_exp_result=
        if (State.ExpRest >= ExpPerLevel) {
            if (ExpPerLevel <= 0) {
                Log("ExpPerLevel is zero or negative: ChainProcessExpGain", ?Level := log_level.Error)
                return process_level_exp_result{
                    NewState := State
                    Continue := false
                }
            }

            process_level_exp_result{
                NewState := level_exp_pair{
                    Level := State.Level + 1
                    ExpRest := State.ExpRest - ExpPerLevel
                }
                Continue := true
            }
        } else {
            process_level_exp_result{
                NewState := State
                Continue := false
            }
        }

    ChainGetTargetExp<override>(State:level_exp_pair)<transacts>:int=
        ExpPerLevel
