using { /Fortnite.com/Devices }
using { /Verse.org/Simulation }
using { /Verse.org/Colors }
using { /UnrealEngine.com/Temporary/Diagnostics }
using { /UnrealEngine.com/Temporary/SpatialMath }
using { /UnrealEngine.com/Temporary/UI }
using { /Fortnite.com/UI }

using { HavocVerse.Debug }
using { HavocVerse.UI }
using { HavocVerse.Logging }
using { HavocVerse.Util }

on_screen_timer_visibility := enum:
    Always
    HideWhenComplete

on_screen_timer_direction := enum:
    CountDown
    CountUp

on_screen_timer_device := class(creative_device, aware_device):
    @editable p_AccessControl: access_control_device = access_control_device{}

    @editable p_Timer: timer_device = timer_device{}
    @editable p_TimerDirection: on_screen_timer_direction = on_screen_timer_direction.CountDown

    @editable p_Visibility: on_screen_timer_visibility = on_screen_timer_visibility.HideWhenComplete
    @editable p_TextTemplate: string = "In %Timer%"
    @editable p_HUDMessage: ?hud_message_device = false

    @editable p_EnableVerseText: logic = true
    @editable p_TextPlacement: ui_canvas_placement_config = ui_canvas_placement_config{}
    @editable p_TextStyle: ui_text_style_config = ui_text_style_config{}

    @editable p_HUDManagerConfig: hud_manager_base_config = hud_manager_base_config{
        RefreshRate := 1.0
    }
    @editable p_BackgroundImage: string = ""
    @editable p_BackgroundSize: vector2 = vector2{X := 100.0, Y := 50.0}
    @editable p_BackgroundPlacement: ui_canvas_placement_config = ui_canvas_placement_config{}

    var m_HudManager: ?on_screen_timer_hud_manager = false
    var m_ActiveHUDViewers: [player]void = map{}

    DeviceMeta<override>()<transacts>:aware_device_meta=
        aware_device_meta{
            DeviceName := "on_screen_timer"
        }

    OnBegin<override>()<suspends>:void=
        AwareInitDevice(Self, OnInit)

    OnInit<private>()<suspends>:void=
        p_AccessControl.RegisterUsingDevice(Self)

        if (p_EnableVerseText?) {
            HudManager := on_screen_timer_hud_manager{
                AccessControl := Some(p_AccessControl)
                RefreshRate := p_HUDManagerConfig.RefreshRate
                ZOrder := p_HUDManagerConfig.ZOrder
                Device := Self
            }
            set m_HudManager = Some(HudManager)
            spawn:
                HudManager.Run(GetPlayspace())
        }

        if (HudMessage := p_HUDMessage?) {
            p_AccessControl.SubscribeToPlayerJoinEvents(OnAccessControlAdded, OnAccessControlRemoved)
            spawn:
                RunHudMessageLoop(HudMessage)
        }

    OnAccessControlAdded<private>(Player:player):void=
        {}

    OnAccessControlRemoved<private>(Player:player, GameQuit:logic):void=
        RemoveRes := m_ActiveHUDViewers.RemoveByKey(Player)
        set m_ActiveHUDViewers = RemoveRes(0)
        if (RemoveRes(1)?, HudMessage := p_HUDMessage?) {
            HudMessage.Hide(Player)
        }

    RunHudMessageLoop<private>(HudMessage:hud_message_device)<suspends>:void=
        loop {
            for (Agent:p_AccessControl.ListAllowedAgents(), Player := player[Agent]) {
                UpdateHudMessage(HudMessage, Player)
            }
            Sleep (1.0)
        }
    
    UpdateHudMessage<private>(HudMessage:hud_message_device, Player:player):void=
        RemainingSeconds := if (p_Timer.IsStatePerAgent[]) {
            p_Timer.GetActiveDuration(Player)
        } else {
            p_Timer.GetActiveDuration()
        }
        if (RemainingSeconds > 0.0 or p_Visibility = on_screen_timer_visibility.Always) {
            TimerText := GetTimerText(RemainingSeconds)
            HudMessage.Show(Player, ToMessage(TimerText))
        } else if (m_ActiveHUDViewers[Player]) {
            set m_ActiveHUDViewers = m_ActiveHUDViewers.RemoveByKey(Player)(0)
            HudMessage.Hide(Player)
        }

    GetTimerText<internal>(RemainingSeconds:float)<transacts>:string=
        TimerSeconds := case(p_TimerDirection):
            on_screen_timer_direction.CountDown => RemainingSeconds
            on_screen_timer_direction.CountUp => p_Timer.GetMaxDuration() - RemainingSeconds
        TimerText := ToTimerMMSS(DefaultingCeil(TimerSeconds))
        p_TextTemplate.ReplaceAll("%Timer%", TimerText)

on_screen_timer_hud_manager := class(hud_manager_base):
    Device: on_screen_timer_device

    CreateHUD<override>(Player:player):?hud_manager_player=
        BackgroundWidget := if (
            Device.p_BackgroundImage.Length > 0
            Texture := AwareDefaultAssetsProvider[].GetTextureAsset[Device.p_BackgroundImage]
        ) {
            option{texture_block{
                DefaultImage := Texture
                DefaultDesiredSize := Device.p_BackgroundSize
            }}
        } else {
            None()
        }
        TextWidget := Device.p_TextStyle.ToTextBlock("")
        Canvas := canvas{
            Slots := array{
                Device.p_TextPlacement.ToCanvasSlot(TextWidget)
            }
        }
        if (l_BackgroundWidget := BackgroundWidget?) {
            Canvas.AddWidget(Device.p_BackgroundPlacement.ToCanvasSlot(l_BackgroundWidget))
        }
        Element := on_screen_timer_hud_element{
            Widget := Canvas
            Device := Device
            Player := Player
            TextWidget := MakeCachedTextBlock(TextWidget)
        }
        Some(MakeSingleElementHUDPlayer(Player, Canvas, Element))

on_screen_timer_hud_element := class(hud_element):
    Device: on_screen_timer_device
    Player: player

    TextWidget: cached_text_block

    Update<override>():void=
        RemainingSeconds := if (Device.p_Timer.IsStatePerAgent[]) {
            Device.p_Timer.GetActiveDuration(Player)
        } else {
            Device.p_Timer.GetActiveDuration()
        }
        if (RemainingSeconds > 0.0 or Device.p_Visibility = on_screen_timer_visibility.Always) {
            Show()
            TimerText := Device.GetTimerText(RemainingSeconds)
            TextWidget.SetText(TimerText)
        } else {
            Hide()
        }
