using { /Fortnite.com/Devices }
using { /Verse.org/Simulation }
using { /UnrealEngine.com/Temporary/Diagnostics }

using { HavocVerse.Debug }
using { HavocVerse.Logging }
using { HavocVerse.Util }

# Device that can be reset to its initial state.
resetable<public> := interface:
    # Reset the device to its initial state.
    # If an agent is provided, the reset is meant to only affect that agent.
    Reset<public>(Agent:?agent):void

quit_rollback_device_mode := enum:
    SubscribedPlayers
    AnyPlayer

quit_rollback_device_granularity := enum:
    Player
    Omitted

# Device that tries to rollback a player's edits when they leave the game.
quit_rollback_device<public> := class(creative_device, aware_device, aware_debug_device, custom_device, resetable):
    @editable p_DevId: string = "N/A"

    @editable p_Mode: quit_rollback_device_mode = quit_rollback_device_mode.SubscribedPlayers
    @editable p_Granularity: quit_rollback_device_granularity = quit_rollback_device_granularity.Player
    @editable p_Devices: []creative_device = array{}

    @editable p_UserActions: []quit_rollback_user_action_config = array{}

    var m_SubscribedPlayers: [player]void = map{}

    DeviceMeta<override>()<transacts>:aware_device_meta=
        aware_device_meta{
            DeviceName := "quit_rollback"
            DevId := p_DevId
        }

    DebugStats<override>()<transacts>:aware_debug_stats=
        aware_debug_stats{
            Length := map{
                "subscribedPlayers" => m_SubscribedPlayers.Length
            }
        }

    OnBegin<override>()<suspends>:void=
        AwareInitDevice(Self, OnInit)

    OnInit<private>()<suspends>:void=
        for (Index->Device:p_Devices, not resetable[Device]) {
            Log("Quit rollback device \"{p_DevId}\" is configured to reset a non-resetable device (index {Index})", ?Level := log_level.Error)
        }
        GetPlayspace().PlayerRemovedEvent().Subscribe(OnPlayerRemoved)

    Reset<override>(Agent:?agent):void=
        if (l_Agent := Agent?) {
            if (Player := player[l_Agent]) {
                set m_SubscribedPlayers = m_SubscribedPlayers.RemoveByKey(Player)(0)
            }
        } else if (not (Agent?)) {
            set m_SubscribedPlayers = map{}
        }

    OnPlayerRemoved<private>(Player:player):void=
        case(p_Mode):
            quit_rollback_device_mode.SubscribedPlayers => {
                if (m_SubscribedPlayers[Player]) {
                    set m_SubscribedPlayers = m_SubscribedPlayers.RemoveByKey(Player)(0)
                    PerformRollback(Player)
                }
            }
            quit_rollback_device_mode.AnyPlayer => PerformRollback(Player)

    PerformRollback<private>(Player:player):void=
        Agent: ?agent = case(p_Granularity):
            quit_rollback_device_granularity.Player => Some(Player)
            quit_rollback_device_granularity.Omitted => false
        for (Device:p_Devices, CustomDevice := resetable[Device]) {
            CustomDevice.Reset(Agent)
        }

    SubscribePlayer(Player:player):void=
        if. set m_SubscribedPlayers[Player] = void

    UnsubscribePlayer(Player:player):void=
        set m_SubscribedPlayers = m_SubscribedPlayers.RemoveByKey(Player)(0)

    CustomTrigger<override>(Action:string, Agent:?agent):logic=
        if (Action = "subscribe", Player := player[Agent?]) {
            SubscribePlayer(Player)
            true
        } else if (Action = "unsubscribe", Player := player[Agent?]) {
            UnsubscribePlayer(Player)
            true
        } else {
            false
        }

quit_rollback_user_action_kind := enum:
    Subscribe
    Unsubscribe

quit_rollback_user_action_config := class<concrete>:
    @editable Action: quit_rollback_user_action_kind = quit_rollback_user_action_kind.Subscribe
    @editable OnTrigger: trigger_device = trigger_device{}

quit_rollback_user_action := class:
    Device: quit_rollback_device
    Config: quit_rollback_user_action_config

    OnTriggered(Agent:?agent):void=
        if (Player := player[Agent?]) {
            case(Config.Action):
                quit_rollback_user_action_kind.Subscribe => Device.SubscribePlayer(Player)
                quit_rollback_user_action_kind.Unsubscribe => Device.UnsubscribePlayer(Player)
        }
