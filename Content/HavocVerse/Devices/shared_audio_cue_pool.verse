using { /Fortnite.com/Devices }
using { /Fortnite.com/Characters }
using { /Verse.org/Simulation }
using { /UnrealEngine.com/Temporary/Diagnostics }
using { /UnrealEngine.com/Temporary/SpatialMath }

using { HavocVerse.Debug }
using { HavocVerse.Logging }
using { HavocVerse.Util }

TooltipReturnToPoolAfter<internal><localizes>:message = "Delay before making a player available again after playing a sound. This should be longer (or equal) to the sound's duration"

shared_audio_cue_pool_device<public> := class(creative_device, aware_device, aware_debug_device):
    @editable p_AudioPlayers: []audio_player_device = array{}
    @editable p_AudioPlayersFromTag: ?string = false

    @editable{ToolTip := TooltipReturnToPoolAfter} p_ReturnToPoolAfter: float = 1.5

    DevicePool: object_pool = object_pool{}

    var m_Initialized: logic = false

    DeviceMeta<override>()<transacts>:aware_device_meta=
        aware_device_meta{
            DeviceName := "shared_audio_cue_pool"
        }

    DebugStats<override>()<transacts>:aware_debug_stats=
        aware_debug_stats{
            Length := map{
                "devicePool" => DevicePool.MemLength()
            }
        }

    OnBegin<override>()<suspends>:void=
        AwareInitDevice(Self, OnInit)

    OnInit<private>()<suspends>:void=
        EnsureReady()

    EnsureReady<private>():void=
        if (m_Initialized?) {
            return
        }
        set m_Initialized = true
        for (AudioPlayer:p_AudioPlayers) {
            DevicePool.Add(AudioPlayer)
        }
        if (
            FromTag := p_AudioPlayersFromTag?
            FromTag.Length > 0
            TagsProvider := AwareDefaultTagsProvider[]
        ) {
            if (Tag := TagsProvider.GetTag[FromTag]) {
                DevicePool.AddFromTag(Tag, FilterAudioPlayers)
            }
        } else if (FromTag := p_AudioPlayersFromTag?) {
            Log("{Self} could not load audio players from misconfigured tag \"{FromTag}\"", ?Level := log_level.Error)
        }

    # Plays an audio cue at an agent's position.
    Play<public>(Agent:agent):void=
        if (FortCharacter := Agent.GetFortCharacter[]) {
            Play(FortCharacter.GetTransform().Translation)
        }

    # Plays an audio cue at a given position.
    Play<public>(Position:vector3):void=
        if (Entry := DevicePool.Reserve[]) {
            spawn:
                Play0(Entry, Position)
        }

    Play0<private>(Entry:object_pool_entry, Position:vector3)<suspends>:void=
        if (
            AudioPlayer := audio_player_device[Entry.Object]
            NewTransform := AudioPlayer.GetTransform().WithTranslation(Position)
            AudioPlayer.TeleportTo[NewTransform]
        ) {
            AudioPlayer.Play()
            if (p_ReturnToPoolAfter >= 0.0) {
                Sleep(p_ReturnToPoolAfter)
                AudioPlayer.Stop()
            }
        }
        Entry.Release()
