using { /Fortnite.com/Devices }
using { /Verse.org/Simulation }
using { /UnrealEngine.com/Temporary/Diagnostics }
using { /UnrealEngine.com/Temporary/SpatialMath }

using { HavocVerse.Debug }
using { HavocVerse.Logging }
using { HavocVerse.Util }

shared_prop_pool_device<public> := class(creative_device, aware_device, aware_debug_device):
    @editable p_PropInstances: []creative_prop = array{}
    @editable p_PropInstancesFromTag: ?string = false

    @editable p_TeleportBack: logic = true
    @editable p_ScaleDownBeforeTeleport: logic = true
    
    var m_PropPool: ?prop_pool = false

    DeviceMeta<override>()<transacts>:aware_device_meta=
        aware_device_meta{
            DeviceName := "shared_prop_pool"
        }

    DebugStats<override>()<transacts>:aware_debug_stats=
        aware_debug_stats{
            Length := map{
                "propPool" => if (PropPool := m_PropPool?) {
                    PropPool.MemLength()
                } else {
                    0
                }
            }
        }

    OnBegin<override>()<suspends>:void=
        AwareInitDevice(Self, OnInit)
    
    OnInit<private>()<suspends>:void=
        EnsureReady()

    EnsureReady<private>():void=
        if (m_PropPool?) {
            return
        }
        PropPool := prop_pool{
            Hide := true
            TeleportBack := p_TeleportBack
            ScaleDownBeforeTeleport := p_ScaleDownBeforeTeleport
        }
        set m_PropPool = Some(PropPool)
        for (PropInstance:p_PropInstances) {
            PropPool.Add(PropInstance)
        }
        if (
            FromTag := p_PropInstancesFromTag?
            FromTag.Length > 0
            TagsProvider := AwareDefaultTagsProvider[]
        ) {
            if (Tag := TagsProvider.GetTag[FromTag]) {
                PropPool.AddFromTag(Tag)
            }
        } else if (FromTag := p_PropInstancesFromTag?) {
            Log("{Self} could not load props from misconfigured tag \"{FromTag}\"", ?Level := log_level.Error)
        }

    Reserve<public>()<transacts><decides>:prop_pool_entry=
        m_PropPool?.Reserve[]

    Reserve<public>(Position:vector3)<transacts><decides>:prop_pool_entry=
        m_PropPool?.Reserve[Position]

    Reserve<public>(Transform:transform)<transacts><decides>:prop_pool_entry=
        m_PropPool?.Reserve[Transform]
