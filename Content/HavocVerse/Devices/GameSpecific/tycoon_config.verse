using { /Verse.org/Simulation }
using { /UnrealEngine.com/Temporary/Diagnostics }

using { HavocVerse.Logging }
using { HavocVerse.UI }
using { HavocVerse.Util }

tycoon_manager_config := class<concrete>:
    BeaconLines: [string]tycoon_manager_beacon_line_config = map{}
    Rebirth: ?tycoon_manager_rebirth_config = false

tycoon_manager_beacon_line_config := class<final>:
    InOrderBuyPlates: []tycoon_manager_beacon_line_config_entry

    ContainsBuyPlate(BuyPlateId:string)<decides><transacts>:void=
        var Found: logic = false
        for (LineEntry:InOrderBuyPlates, not Found?, LineEntry.ContainsBuyPlate[BuyPlateId]) {
            set Found = true
        }
        Found?

tycoon_manager_beacon_line_config_entry := class<final>:
    AvailableOptions: [string]void

    ContainsBuyPlate(BuyPlateId:string)<decides><transacts>:void=
        AvailableOptions[BuyPlateId]

    ContainsAnyBuyPlateOf(Ids:[]string)<decides><transacts>:void=
        var Found: logic = false
        for (BuyPlateId:Ids, not Found?, AvailableOptions[BuyPlateId]) {
            set Found = true
        }
        Found?

tycoon_manager_buyplates_config := struct:
    Entries: []tycoon_manager_buyplates_entry = array{}

tycoon_manager_buyplates_entry := class<final>:
    BuyplateId: string
    Cost: ?tycoon_manager_buyplates_entry_resource
    Generates: ?tycoon_manager_buyplates_entry_resource
    Unlocks: []tycoon_manager_unlock_entry

    var m_StorageIndex: int = -1

tycoon_manager_unlock_entry := struct:
    TargetRoutingId: string
    Payload: ?string

tycoon_manager_buyplates_entry_resource := struct:
    ResourceId: string
    Amount: int
    CheckDontTake: logic

tycoon_manager_rebirth_config := class<final>:
    MaxLevel: int
    DefaultToLast: logic
    LevelConfigs: [int]tycoon_manager_rebirth_level_config

    GetLevelConfig(Level:int)<transacts><decides>:tycoon_manager_rebirth_level_config=
        GetLevelConfig0(Level)?

    GetLevelConfig0<private>(Level:int)<transacts>:?tycoon_manager_rebirth_level_config=
        if (DefaultToLast?) {
            var Index: int = Level
            loop {
                if (LevelConfig := LevelConfigs[Index]) {
                    return Some(LevelConfig)
                }
                if (Index <= 0) {
                    break
                }
                set Index -= 1
            }
        } else if (LevelConfig := LevelConfigs[Level]) {
            Some(LevelConfig)
        }
        None()

tycoon_manager_rebirth_level_config := class:
    HudImageTexture: ?string = false
    ResourceMultipliers: [string]float = map{}

## ToString {{{
ToString(Obj:tycoon_manager_buyplates_config):string=
    var Entries: string = ""
    for (Index->Entry:Obj.Entries) {
        if (Index > 0) {
            set Entries += ", "
        }
        set Entries += ToString(Entry)
    }
    "\{Entries: [{Entries}]\}"

ToString(Obj:tycoon_manager_buyplates_entry):string=
    Cost := if (l_Cost := Obj.Cost?) {
        ToString(l_Cost)
    } else {
        "None"
    }
    var Unlocks: string = ""
    for (Index->Unlock:Obj.Unlocks) {
        if (Index > 0) {
            set Unlocks += ", "
        }
        set Unlocks += "{Unlock}"
    }
    "\{BuyplateId: \"{Obj.BuyplateId}\", Cost: {Cost}, Unlocks: [{Unlocks}]\}"

ToString(Obj:tycoon_manager_buyplates_entry_resource):string=
    "\{ResourceId: \"{Obj.ResourceId}\", Amount: {Obj.Amount}\}"

ToString(Obj:tycoon_manager_unlock_entry):string=
    PayloadStr := if (Payload := Obj.Payload?) {
        "Some(\"{Payload}\")"
    } else {
        "None"
    }
    "\{TargetRoutingId: \"{Obj.TargetRoutingId}\", Payload: {PayloadStr}\}"
## }}}
