using { /Fortnite.com/Devices }
using { /Fortnite.com/Characters }
using { /Verse.org/Simulation }
using { /UnrealEngine.com/Temporary/Diagnostics }
using { /UnrealEngine.com/Temporary/SpatialMath }
using { /UnrealEngine.com/Temporary/UI }

using { HavocVerse.Debug }
using { HavocVerse.Devices }
using { HavocVerse.Storage }
using { HavocVerse.JSON }
using { HavocVerse.UI }
using { HavocVerse.Util }
using { HavocVerse.Util.Task }
using { HavocVerse.Logging }

tycoon_manager_rebirth_reset_trigger_config := class<concrete>:
    @editable Trigger: trigger_device = trigger_device{}
    @editable ForLevel: int = -1

    IsApplicable<public>(RebirthLevel:int)<transacts><decides>:void=
        ForLevel < 0 or ForLevel = RebirthLevel

tycoon_manager_progress_line<public> := class<concrete>:
    @editable Name<public>: string = "N/A"
    @editable ClassSlot: int = -1
    @editable AccessControl<public>: ?access_control_device = false
    @editable FrontCheckpoint: ?player_checkpoint_device = false
    @editable FrontTeleporter: ?teleporter_device = false
    @editable AutoGenerationManager: ?auto_generation_manager = false
    @editable ClaimBuyplate: ?buy_plate_device = false
    @editable ExtraBuyplates: []creative_device = array{}
    @editable ExtraRollbackDevices: []creative_device = array{}
    @editable ExtraRollbackTriggers: []trigger_device = array{}
    @editable BeaconLines: []tycoon_manager_progress_line_beacon_line = array{}

    @editable RebirthResetTriggers: []tycoon_manager_rebirth_reset_trigger_config = array{}
    @editable RebirthBillboard: []billboard_device = array{}
    @editable RebirthBillboardTemplate: string = "Rebirths: %CurrentLevel%"
    @editable RebirthBillboardHideOnZero: logic = false

    var m_Index: int = -1
    var m_AutoGenerationTickets: [string]auto_generation_ticket = map{}
    var m_PersistenceRecorder: ?persistence_recorder_device = false
    var m_RebirthRecordableHandle: ?persistence_recorder_recordable_handle = false
    var m_BuyPlatesBackrefCache: [string]buy_plate_device = map{}

    var m_RebirthLevel: int = 0
    var m_SetResourceMultipliers: [string]float = map{}

tycoon_manager_progress_line_beacon_line := class<concrete>:
    @editable BeaconLineId: string = "beaconline1"
    @editable MapIndicator: map_indicator_device = map_indicator_device{}
    @editable MapIndicatorUnmetCost: ?map_indicator_device = false

    var m_CurrentEntry: int = 0
    var m_LockedTargetBuyPlate: ?string = false

    var m_MapIndicatorState: ?tycoon_manager_beacon_line_buy_plate_state = false
    var m_MapIndicatorEnabled: ?tycoon_manager_beacon_line_map_indicator_variant = false
    var m_MapIndicatorPulseActive: logic = false
    var m_MapIndicatorPulseToAgents: [agent]void = map{}

    GetActiveMapIndicator()<decides><transacts>:map_indicator_device=
        case(m_MapIndicatorEnabled?):
            tycoon_manager_beacon_line_map_indicator_variant.Primary => MapIndicator
            tycoon_manager_beacon_line_map_indicator_variant.UnmetCost => MapIndicatorUnmetCost?

    DisablePulse(ProgressLine:tycoon_manager_progress_line):void=
        if (not m_MapIndicatorPulseActive?) {
            return
        }
        set m_MapIndicatorPulseActive = false
        if (l_MapIndicator := GetActiveMapIndicator[]) {
            for (Agent:ProgressLine.AccessControl?.ListAllowedAgents()) {
                if. set m_MapIndicatorPulseToAgents[Agent] = void
            }
            for (Agent->__:m_MapIndicatorPulseToAgents) {
                l_MapIndicator.DeactivateObjectivePulse(Agent)
            }
        }
        set m_MapIndicatorPulseToAgents = map{}

tycoon_manager_beacon_line_map_indicator_variant := enum:
    Primary
    UnmetCost

ToString(Variant:tycoon_manager_beacon_line_map_indicator_variant)<transacts>:string=
    case(Variant):
        tycoon_manager_beacon_line_map_indicator_variant.Primary => "Primary"
        tycoon_manager_beacon_line_map_indicator_variant.UnmetCost => "UnmetCost"

tycoon_manager_beacon_line_buy_plate_state := class<final>:
    BuyPlate: buy_plate_device
    UnmetCost: logic

tycoon_manager_player_state := class<final>:
    ProgressLine: tycoon_manager_progress_line

tycoon_manager_rebirth_pet_config<public> := class<concrete>:
    @editable PetId<public>: string = ""
    @editable RebirthLevel<public>: int = 1

tycoon_manager_beacon_line_area_config := class<concrete>:
    @editable Volume: volume_device = volume_device{}
    @editable OuterVolume: ?volume_device = false

    var m_InVolume: [agent]void = map{}

tycoon_claim_or_teleport_result<public> := enum:
    Claimed
    Teleported
    Failed

tycoon_rebirth_event<public> := class<final>:
    Player<public>: player
    NewRebirthLevel<public>: int
    Virtual<public>: logic

tycoon_manager_device<public> := class(creative_device, aware_device, aware_debug_device, pet_condition_provider):
    @editable p_ResourceManager<public>: resource_manager_device = resource_manager_device{}
    @editable p_AnalyticsController<public>: analytics_controller_device = analytics_controller_device{}
    @editable p_ClassManager<public>: class_manager_device = class_manager_device{}
    @editable p_ProgressLines: []tycoon_manager_progress_line = array{}
    @editable p_TimeRewardsDevice: []time_rewards_device = array{}
    @editable p_PetController: []pet_controller_device = array{}

    @editable p_RoutingIdSingleMatch: logic = true
    @editable p_AlwaysSetCostOverride: logic = true

    @editable p_ClaimFailHUDMessage: []hud_message_device = array{}
    @editable p_MiscPlayerHeight<public>: float = 180.0

    @editable p_BeaconLinesHideAreas: []tycoon_manager_beacon_line_area_config = array{}
    @editable p_BeaconLinesMapIndicatorOffset: vector3 = vector3{}

    @editable p_EnableQuitRollback: logic = true
    @editable p_QuitRollbackGranularity: quit_rollback_device_granularity = quit_rollback_device_granularity.Player

    @editable p_EnablePersistence: logic = false
    @editable p_PersistenceSaveSlots: []shared_storage_chunk_device = array{}
    @editable p_PersistenceSavePoints: []exposed_save_point_device = array{}
    @editable p_PersistenceAutoSaveInternal: float = 3.0
    @editable p_PersistenceBuyPlatesReserveBits: int = -1
    @editable p_PersistenceIncludeResourceManager: logic = true
    @editable p_PersistenceExtraDevices: []creative_device = array{}
    @editable p_PersistenceSelectMenuEnabled: logic = true
    @editable p_PersistenceUseDialogSelectMenu: []persistence_recorder_popup_dialog_menu_config = array{}
    @editable p_PersistenceSelectMenuConfig: default_persistence_recorder_load_select_menu_config = default_persistence_recorder_load_select_menu_config{}
    @editable p_PersistenceBuyPlateRecoveryChunkSize: int = -1
    @editable p_PersistenceBuyPlateRecoveryChunkDelay: float = 0.1

    @editable p_RebirthRetainResources<public>: []string = array{}
    @editable p_RebirthPets<public>: []tycoon_manager_rebirth_pet_config = array{}
    @editable p_RebirthLevelReserveBits: int = 4
    @editable p_RebirthNextLevelOnTrigger: []trigger_device = array{}
    @editable p_RebirthNextLevelDialog: []popup_dialog_device = array{}
    @editable p_RebirthNextLevelDialogPerLevel: logic = false
    @editable p_RebirthNextLevelDialogPerLevelDefaultLast: logic = true
    @editable p_RebirthNextLevelDialogConfirmButtonIndex: int = 0
    @editable p_RebirthPlaceholdersMultiplierFormatter: number_formatter_config = number_formatter_config{
        Template := "%Val%x"
    }
    @editable p_RebirthResetTriggers: []tycoon_manager_rebirth_reset_trigger_config = array{}
    @editable p_RebirthMultipliersChannel: string = "tycoon_rebirth"
    @editable p_RebirthHUDManagerConfig: hud_manager_base_config = hud_manager_base_config{
        RefreshRate := -1.0
    }
    @editable p_RebirthHUDImageSize: vector2 = vector2{X := 100.0, Y := 50.0}
    @editable p_RebirthHUDImagePlacement: ui_canvas_placement_config = ui_canvas_placement_config{}

    @editable p_BuyPlateCostMetTrackingRefreshRate: float = 0.5

    PlayerSaveSlotSelectedEvent<public>: event(tuple(tycoon_manager_progress_line,player_save_slot_selected_event)) = event(tuple(tycoon_manager_progress_line,player_save_slot_selected_event)){}
    PlayerJoinProgressLineEvent<public>: event(tuple(player,tycoon_manager_progress_line)) = event(tuple(player,tycoon_manager_progress_line)){}

    var m_TycoonConfig<private>: tycoon_manager_config = tycoon_manager_config{}
    var m_BuyplatesConfigs<private>: [string]tycoon_manager_buyplates_entry = map{}
    var m_Initialized<private>: logic = false

    var m_Players<private>: [player]tycoon_manager_player_state = map{}
    var m_RebirthHudManager<internal>: ?hud_manager_base = false
    var m_RebirthDialogs<private>: [int]tuple(string,dialog_device_wrapper) = map{}
    var m_RebirthPets<private>: [string]int = map{}

    var m_PersistenceControl<private>: persistence_recordable_control = persistence_recordable_control{}

    var m_RebirthListeners<private>: []type{__(:tycoon_rebirth_event):void} = array{}

    DeviceMeta<override>()<transacts>:aware_device_meta=
        aware_device_meta{
            DeviceName := "tycoon_manager"
        }

    DebugStats<override>()<transacts>:aware_debug_stats=
        aware_debug_stats{
            Length := map{
                "buyplatesConfigs" => m_BuyplatesConfigs.Length
                "players" => m_Players.Length
                "rebirthDialogs" => m_RebirthDialogs.Length
                "listeners::rebirth" => m_RebirthListeners.Length
            }
        }

    OnBegin<override>()<suspends>:void=
        AwareInitDevice(Self, OnInit)

    OnInit<private>()<suspends>:void=
        set m_TycoonConfig = tycoon_manager_config{  
            BeaconLines := map{
              "beaconline1" => tycoon_manager_beacon_line_config{InOrderBuyPlates := array{
                tycoon_manager_beacon_line_config_entry{AvailableOptions := map{"bp_400"=>void}}
                tycoon_manager_beacon_line_config_entry{AvailableOptions := map{"bp_300"=>void}}
                tycoon_manager_beacon_line_config_entry{AvailableOptions := map{"bp_001"=>void}}
                tycoon_manager_beacon_line_config_entry{AvailableOptions := map{"bp_401"=>void}}
                tycoon_manager_beacon_line_config_entry{AvailableOptions := map{"bp_200"=>void}}
                tycoon_manager_beacon_line_config_entry{AvailableOptions := map{"bp_002"=>void}}
                tycoon_manager_beacon_line_config_entry{AvailableOptions := map{"bp_402"=>void}}
                tycoon_manager_beacon_line_config_entry{AvailableOptions := map{"bp_003"=>void}}
                tycoon_manager_beacon_line_config_entry{AvailableOptions := map{"bp_201"=>void}}
                tycoon_manager_beacon_line_config_entry{AvailableOptions := map{"bp_004"=>void}}
                tycoon_manager_beacon_line_config_entry{AvailableOptions := map{"bp_005"=>void}}
                tycoon_manager_beacon_line_config_entry{AvailableOptions := map{"bp_202"=>void}}
                tycoon_manager_beacon_line_config_entry{AvailableOptions := map{"bp_006"=>void}}
                tycoon_manager_beacon_line_config_entry{AvailableOptions := map{"bp_007"=>void}}
                tycoon_manager_beacon_line_config_entry{AvailableOptions := map{"bp_203"=>void}}
                tycoon_manager_beacon_line_config_entry{AvailableOptions := map{"bp_008"=>void}}
                tycoon_manager_beacon_line_config_entry{AvailableOptions := map{"bp_009"=>void}}
                tycoon_manager_beacon_line_config_entry{AvailableOptions := map{"bp_204"=>void}}
                tycoon_manager_beacon_line_config_entry{AvailableOptions := map{"bp_010"=>void}}
                tycoon_manager_beacon_line_config_entry{AvailableOptions := map{"bp_011"=>void}}
                tycoon_manager_beacon_line_config_entry{AvailableOptions := map{"bp_205"=>void}}
                tycoon_manager_beacon_line_config_entry{AvailableOptions := map{"bp_012"=>void}}
                tycoon_manager_beacon_line_config_entry{AvailableOptions := map{"bp_013"=>void}}
                tycoon_manager_beacon_line_config_entry{AvailableOptions := map{"bp_206"=>void}}
                tycoon_manager_beacon_line_config_entry{AvailableOptions := map{"bp_014"=>void}}
                tycoon_manager_beacon_line_config_entry{AvailableOptions := map{"bp_015"=>void}}
                tycoon_manager_beacon_line_config_entry{AvailableOptions := map{"bp_207"=>void}}
                tycoon_manager_beacon_line_config_entry{AvailableOptions := map{"bp_016"=>void}}
                tycoon_manager_beacon_line_config_entry{AvailableOptions := map{"bp_017"=>void}}
                tycoon_manager_beacon_line_config_entry{AvailableOptions := map{"bp_208"=>void}}
                tycoon_manager_beacon_line_config_entry{AvailableOptions := map{"bp_018"=>void}}
                tycoon_manager_beacon_line_config_entry{AvailableOptions := map{"bp_019"=>void}}
                tycoon_manager_beacon_line_config_entry{AvailableOptions := map{"bp_209"=>void}}
                tycoon_manager_beacon_line_config_entry{AvailableOptions := map{"bp_020"=>void}}
                tycoon_manager_beacon_line_config_entry{AvailableOptions := map{"bp_021"=>void}}
                tycoon_manager_beacon_line_config_entry{AvailableOptions := map{"bp_022"=>void}}
                tycoon_manager_beacon_line_config_entry{AvailableOptions := map{"bp_301"=>void}}
                tycoon_manager_beacon_line_config_entry{AvailableOptions := map{"bp_210"=>void}}
                tycoon_manager_beacon_line_config_entry{AvailableOptions := map{"bp_023"=>void}}
                tycoon_manager_beacon_line_config_entry{AvailableOptions := map{"bp_024"=>void}}
                tycoon_manager_beacon_line_config_entry{AvailableOptions := map{"bp_211"=>void}}
                tycoon_manager_beacon_line_config_entry{AvailableOptions := map{"bp_025"=>void}}
                tycoon_manager_beacon_line_config_entry{AvailableOptions := map{"bp_026"=>void}}
                tycoon_manager_beacon_line_config_entry{AvailableOptions := map{"bp_212"=>void}}
                tycoon_manager_beacon_line_config_entry{AvailableOptions := map{"bp_027"=>void}}
                tycoon_manager_beacon_line_config_entry{AvailableOptions := map{"bp_028"=>void}}
                tycoon_manager_beacon_line_config_entry{AvailableOptions := map{"bp_213"=>void}}
                tycoon_manager_beacon_line_config_entry{AvailableOptions := map{"bp_029"=>void}}
                tycoon_manager_beacon_line_config_entry{AvailableOptions := map{"bp_030"=>void}}
                tycoon_manager_beacon_line_config_entry{AvailableOptions := map{"bp_214"=>void}}
                tycoon_manager_beacon_line_config_entry{AvailableOptions := map{"bp_031"=>void}}
                tycoon_manager_beacon_line_config_entry{AvailableOptions := map{"bp_032"=>void}}
                tycoon_manager_beacon_line_config_entry{AvailableOptions := map{"bp_215"=>void}}
                tycoon_manager_beacon_line_config_entry{AvailableOptions := map{"bp_033"=>void}}
                tycoon_manager_beacon_line_config_entry{AvailableOptions := map{"bp_034"=>void}}
                tycoon_manager_beacon_line_config_entry{AvailableOptions := map{"bp_216"=>void}}
                tycoon_manager_beacon_line_config_entry{AvailableOptions := map{"bp_035"=>void}}
                tycoon_manager_beacon_line_config_entry{AvailableOptions := map{"bp_036"=>void}}
                tycoon_manager_beacon_line_config_entry{AvailableOptions := map{"bp_217"=>void}}
                tycoon_manager_beacon_line_config_entry{AvailableOptions := map{"bp_037"=>void}}
                tycoon_manager_beacon_line_config_entry{AvailableOptions := map{"bp_038"=>void}}
                tycoon_manager_beacon_line_config_entry{AvailableOptions := map{"bp_218"=>void}}
                tycoon_manager_beacon_line_config_entry{AvailableOptions := map{"bp_039"=>void}}
                tycoon_manager_beacon_line_config_entry{AvailableOptions := map{"bp_040"=>void}}
                tycoon_manager_beacon_line_config_entry{AvailableOptions := map{"bp_219"=>void}}
                tycoon_manager_beacon_line_config_entry{AvailableOptions := map{"bp_041"=>void}}
                tycoon_manager_beacon_line_config_entry{AvailableOptions := map{"bp_042"=>void}}
                tycoon_manager_beacon_line_config_entry{AvailableOptions := map{"bp_043"=>void}}
                tycoon_manager_beacon_line_config_entry{AvailableOptions := map{"bp_044"=>void}}
                tycoon_manager_beacon_line_config_entry{AvailableOptions := map{"bp_302"=>void}}
                tycoon_manager_beacon_line_config_entry{AvailableOptions := map{"bp_220"=>void}}
                tycoon_manager_beacon_line_config_entry{AvailableOptions := map{"bp_045"=>void}}
                tycoon_manager_beacon_line_config_entry{AvailableOptions := map{"bp_046"=>void}}
                tycoon_manager_beacon_line_config_entry{AvailableOptions := map{"bp_221"=>void}}
                tycoon_manager_beacon_line_config_entry{AvailableOptions := map{"bp_047"=>void}}
                tycoon_manager_beacon_line_config_entry{AvailableOptions := map{"bp_048"=>void}}
                tycoon_manager_beacon_line_config_entry{AvailableOptions := map{"bp_222"=>void}}
                tycoon_manager_beacon_line_config_entry{AvailableOptions := map{"bp_049"=>void}}
                tycoon_manager_beacon_line_config_entry{AvailableOptions := map{"bp_050"=>void}}
                tycoon_manager_beacon_line_config_entry{AvailableOptions := map{"bp_223"=>void}}
                tycoon_manager_beacon_line_config_entry{AvailableOptions := map{"bp_051"=>void}}
                tycoon_manager_beacon_line_config_entry{AvailableOptions := map{"bp_052"=>void}}
                tycoon_manager_beacon_line_config_entry{AvailableOptions := map{"bp_224"=>void}}
                tycoon_manager_beacon_line_config_entry{AvailableOptions := map{"bp_053"=>void}}
                tycoon_manager_beacon_line_config_entry{AvailableOptions := map{"bp_054"=>void}}
                tycoon_manager_beacon_line_config_entry{AvailableOptions := map{"bp_225"=>void}}
                tycoon_manager_beacon_line_config_entry{AvailableOptions := map{"bp_055"=>void}}
                tycoon_manager_beacon_line_config_entry{AvailableOptions := map{"bp_056"=>void}}
                tycoon_manager_beacon_line_config_entry{AvailableOptions := map{"bp_226"=>void}}
                tycoon_manager_beacon_line_config_entry{AvailableOptions := map{"bp_057"=>void}}
                tycoon_manager_beacon_line_config_entry{AvailableOptions := map{"bp_058"=>void}}
                tycoon_manager_beacon_line_config_entry{AvailableOptions := map{"bp_227"=>void}}
                tycoon_manager_beacon_line_config_entry{AvailableOptions := map{"bp_059"=>void}}
                tycoon_manager_beacon_line_config_entry{AvailableOptions := map{"bp_060"=>void}}
                tycoon_manager_beacon_line_config_entry{AvailableOptions := map{"bp_228"=>void}}
                tycoon_manager_beacon_line_config_entry{AvailableOptions := map{"bp_061"=>void}}
                tycoon_manager_beacon_line_config_entry{AvailableOptions := map{"bp_062"=>void}}
                tycoon_manager_beacon_line_config_entry{AvailableOptions := map{"bp_229"=>void}}
                tycoon_manager_beacon_line_config_entry{AvailableOptions := map{"bp_063"=>void}}
                tycoon_manager_beacon_line_config_entry{AvailableOptions := map{"bp_064"=>void}}
                tycoon_manager_beacon_line_config_entry{AvailableOptions := map{"bp_065"=>void}}
                tycoon_manager_beacon_line_config_entry{AvailableOptions := map{"bp_066"=>void}}
                tycoon_manager_beacon_line_config_entry{AvailableOptions := map{"bp_303"=>void}}
                tycoon_manager_beacon_line_config_entry{AvailableOptions := map{"bp_230"=>void}}
                tycoon_manager_beacon_line_config_entry{AvailableOptions := map{"bp_067"=>void}}
                tycoon_manager_beacon_line_config_entry{AvailableOptions := map{"bp_068"=>void}}
                tycoon_manager_beacon_line_config_entry{AvailableOptions := map{"bp_231"=>void}}
                tycoon_manager_beacon_line_config_entry{AvailableOptions := map{"bp_069"=>void}}
                tycoon_manager_beacon_line_config_entry{AvailableOptions := map{"bp_070"=>void}}
                tycoon_manager_beacon_line_config_entry{AvailableOptions := map{"bp_232"=>void}}
                tycoon_manager_beacon_line_config_entry{AvailableOptions := map{"bp_071"=>void}}
                tycoon_manager_beacon_line_config_entry{AvailableOptions := map{"bp_072"=>void}}
                tycoon_manager_beacon_line_config_entry{AvailableOptions := map{"bp_233"=>void}}
                tycoon_manager_beacon_line_config_entry{AvailableOptions := map{"bp_073"=>void}}
                tycoon_manager_beacon_line_config_entry{AvailableOptions := map{"bp_074"=>void}}
                tycoon_manager_beacon_line_config_entry{AvailableOptions := map{"bp_234"=>void}}
                tycoon_manager_beacon_line_config_entry{AvailableOptions := map{"bp_075"=>void}}
                tycoon_manager_beacon_line_config_entry{AvailableOptions := map{"bp_076"=>void}}
                tycoon_manager_beacon_line_config_entry{AvailableOptions := map{"bp_235"=>void}}
                tycoon_manager_beacon_line_config_entry{AvailableOptions := map{"bp_077"=>void}}
                tycoon_manager_beacon_line_config_entry{AvailableOptions := map{"bp_078"=>void}}
                tycoon_manager_beacon_line_config_entry{AvailableOptions := map{"bp_236"=>void}}
                tycoon_manager_beacon_line_config_entry{AvailableOptions := map{"bp_079"=>void}}
                tycoon_manager_beacon_line_config_entry{AvailableOptions := map{"bp_080"=>void}}
                tycoon_manager_beacon_line_config_entry{AvailableOptions := map{"bp_237"=>void}}
                tycoon_manager_beacon_line_config_entry{AvailableOptions := map{"bp_081"=>void}}
                tycoon_manager_beacon_line_config_entry{AvailableOptions := map{"bp_082"=>void}}
                tycoon_manager_beacon_line_config_entry{AvailableOptions := map{"bp_238"=>void}}
                tycoon_manager_beacon_line_config_entry{AvailableOptions := map{"bp_083"=>void}}
                tycoon_manager_beacon_line_config_entry{AvailableOptions := map{"bp_084"=>void}}
                tycoon_manager_beacon_line_config_entry{AvailableOptions := map{"bp_239"=>void}}
                tycoon_manager_beacon_line_config_entry{AvailableOptions := map{"bp_085"=>void}}
                tycoon_manager_beacon_line_config_entry{AvailableOptions := map{"bp_086"=>void}}
                tycoon_manager_beacon_line_config_entry{AvailableOptions := map{"bp_087"=>void}}
                tycoon_manager_beacon_line_config_entry{AvailableOptions := map{"bp_088"=>void}}
                tycoon_manager_beacon_line_config_entry{AvailableOptions := map{"bp_304"=>void}}
                tycoon_manager_beacon_line_config_entry{AvailableOptions := map{"bp_240"=>void}}
                tycoon_manager_beacon_line_config_entry{AvailableOptions := map{"bp_089"=>void}}
                tycoon_manager_beacon_line_config_entry{AvailableOptions := map{"bp_090"=>void}}
                tycoon_manager_beacon_line_config_entry{AvailableOptions := map{"bp_241"=>void}}
                tycoon_manager_beacon_line_config_entry{AvailableOptions := map{"bp_091"=>void}}
                tycoon_manager_beacon_line_config_entry{AvailableOptions := map{"bp_092"=>void}}
                tycoon_manager_beacon_line_config_entry{AvailableOptions := map{"bp_242"=>void}}
                tycoon_manager_beacon_line_config_entry{AvailableOptions := map{"bp_093"=>void}}
                tycoon_manager_beacon_line_config_entry{AvailableOptions := map{"bp_094"=>void}}
                tycoon_manager_beacon_line_config_entry{AvailableOptions := map{"bp_243"=>void}}
                tycoon_manager_beacon_line_config_entry{AvailableOptions := map{"bp_095"=>void}}
                tycoon_manager_beacon_line_config_entry{AvailableOptions := map{"bp_096"=>void}}
                tycoon_manager_beacon_line_config_entry{AvailableOptions := map{"bp_244"=>void}}
                tycoon_manager_beacon_line_config_entry{AvailableOptions := map{"bp_097"=>void}}
                tycoon_manager_beacon_line_config_entry{AvailableOptions := map{"bp_098"=>void}}
                tycoon_manager_beacon_line_config_entry{AvailableOptions := map{"bp_245"=>void}}
                tycoon_manager_beacon_line_config_entry{AvailableOptions := map{"bp_099"=>void}}
                tycoon_manager_beacon_line_config_entry{AvailableOptions := map{"bp_100"=>void}}
                tycoon_manager_beacon_line_config_entry{AvailableOptions := map{"bp_246"=>void}}
                tycoon_manager_beacon_line_config_entry{AvailableOptions := map{"bp_101"=>void}}
                tycoon_manager_beacon_line_config_entry{AvailableOptions := map{"bp_102"=>void}}
                tycoon_manager_beacon_line_config_entry{AvailableOptions := map{"bp_247"=>void}}
                tycoon_manager_beacon_line_config_entry{AvailableOptions := map{"bp_103"=>void}}
                tycoon_manager_beacon_line_config_entry{AvailableOptions := map{"bp_104"=>void}}
                tycoon_manager_beacon_line_config_entry{AvailableOptions := map{"bp_248"=>void}}
                tycoon_manager_beacon_line_config_entry{AvailableOptions := map{"bp_105"=>void}}
                tycoon_manager_beacon_line_config_entry{AvailableOptions := map{"bp_106"=>void}}
                tycoon_manager_beacon_line_config_entry{AvailableOptions := map{"bp_249"=>void}}
                tycoon_manager_beacon_line_config_entry{AvailableOptions := map{"bp_107"=>void}}
                tycoon_manager_beacon_line_config_entry{AvailableOptions := map{"bp_108"=>void}}
                tycoon_manager_beacon_line_config_entry{AvailableOptions := map{"bp_109"=>void}}
                tycoon_manager_beacon_line_config_entry{AvailableOptions := map{"bp_110"=>void}}}}
            }  
            Rebirth := option{tycoon_manager_rebirth_config{
              MaxLevel := 16000
              DefaultToLast := true
              LevelConfigs := map{
                0 => tycoon_manager_rebirth_level_config{HudImageTexture := option{""}, ResourceMultipliers := map{}}
                1 => tycoon_manager_rebirth_level_config{HudImageTexture := option{"UI/Rebirths/rebirth_01"}, ResourceMultipliers := map{"silver"=>2.0, "gems"=>2.0, "ruby"=>2.0, "treasure"=>2.0, "treasurechest"=>2.0}}
                2 => tycoon_manager_rebirth_level_config{HudImageTexture := option{"UI/Rebirths/rebirth_02"}, ResourceMultipliers := map{"silver"=>3.0, "gems"=>3.0, "ruby"=>3.0, "treasure"=>3.0, "treasurechest"=>3.0}}
                3 => tycoon_manager_rebirth_level_config{HudImageTexture := option{"UI/Rebirths/rebirth_03"}, ResourceMultipliers := map{"silver"=>4.0, "gems"=>4.0, "ruby"=>4.0, "treasure"=>4.0, "treasurechest"=>4.0}}
                4 => tycoon_manager_rebirth_level_config{HudImageTexture := option{"UI/Rebirths/rebirth_04"}, ResourceMultipliers := map{"silver"=>5.0, "gems"=>5.0, "ruby"=>5.0, "treasure"=>5.0, "treasurechest"=>5.0}}
                5 => tycoon_manager_rebirth_level_config{HudImageTexture := option{"UI/Rebirths/rebirth_05"}, ResourceMultipliers := map{"silver"=>5.0, "gems"=>5.0, "ruby"=>5.0, "treasure"=>5.0, "treasurechest"=>5.0}}
                6 => tycoon_manager_rebirth_level_config{HudImageTexture := option{"UI/Rebirths/rebirth_06"}, ResourceMultipliers := map{"silver"=>5.0, "gems"=>5.0, "ruby"=>5.0, "treasure"=>5.0, "treasurechest"=>5.0}}
                7 => tycoon_manager_rebirth_level_config{HudImageTexture := option{"UI/Rebirths/rebirth_07"}, ResourceMultipliers := map{"silver"=>5.0, "gems"=>5.0, "ruby"=>5.0, "treasure"=>5.0, "treasurechest"=>5.0}}
                8 => tycoon_manager_rebirth_level_config{HudImageTexture := option{"UI/Rebirths/rebirth_08"}, ResourceMultipliers := map{"silver"=>5.0, "gems"=>5.0, "ruby"=>5.0, "treasure"=>5.0, "treasurechest"=>5.0}}
                9 => tycoon_manager_rebirth_level_config{HudImageTexture := option{"UI/Rebirths/rebirth_09"}, ResourceMultipliers := map{"silver"=>5.0, "gems"=>5.0, "ruby"=>5.0, "treasure"=>5.0, "treasurechest"=>5.0}}
                10 => tycoon_manager_rebirth_level_config{HudImageTexture := option{"UI/Rebirths/rebirth_10"}, ResourceMultipliers := map{"silver"=>5.0, "gems"=>5.0, "ruby"=>5.0, "treasure"=>5.0, "treasurechest"=>5.0}}
                11 => tycoon_manager_rebirth_level_config{HudImageTexture := option{"UI/Rebirths/rebirth_11"}, ResourceMultipliers := map{"silver"=>5.0, "gems"=>5.0, "ruby"=>5.0, "treasure"=>5.0, "treasurechest"=>5.0}}
                12 => tycoon_manager_rebirth_level_config{HudImageTexture := option{"UI/Rebirths/rebirth_12"}, ResourceMultipliers := map{"silver"=>5.0, "gems"=>5.0, "ruby"=>5.0, "treasure"=>5.0, "treasurechest"=>5.0}}
                13 => tycoon_manager_rebirth_level_config{HudImageTexture := option{"UI/Rebirths/rebirth_13"}, ResourceMultipliers := map{"silver"=>5.0, "gems"=>5.0, "ruby"=>5.0, "treasure"=>5.0, "treasurechest"=>5.0}}
                14 => tycoon_manager_rebirth_level_config{HudImageTexture := option{"UI/Rebirths/rebirth_14"}, ResourceMultipliers := map{"silver"=>5.0, "gems"=>5.0, "ruby"=>5.0, "treasure"=>5.0, "treasurechest"=>5.0}}
                15 => tycoon_manager_rebirth_level_config{HudImageTexture := option{"UI/Rebirths/rebirth_15"}, ResourceMultipliers := map{"silver"=>5.0, "gems"=>5.0, "ruby"=>5.0, "treasure"=>5.0, "treasurechest"=>5.0}}
                16 => tycoon_manager_rebirth_level_config{HudImageTexture := option{"UI/Rebirths/rebirth_16"}, ResourceMultipliers := map{"silver"=>5.0, "gems"=>5.0, "ruby"=>5.0, "treasure"=>5.0, "treasurechest"=>5.0}}
                17 => tycoon_manager_rebirth_level_config{HudImageTexture := option{"UI/Rebirths/rebirth_17"}, ResourceMultipliers := map{"silver"=>5.0, "gems"=>5.0, "ruby"=>5.0, "treasure"=>5.0, "treasurechest"=>5.0}}
                18 => tycoon_manager_rebirth_level_config{HudImageTexture := option{"UI/Rebirths/rebirth_18"}, ResourceMultipliers := map{"silver"=>5.0, "gems"=>5.0, "ruby"=>5.0, "treasure"=>5.0, "treasurechest"=>5.0}}
                19 => tycoon_manager_rebirth_level_config{HudImageTexture := option{"UI/Rebirths/rebirth_19"}, ResourceMultipliers := map{"silver"=>5.0, "gems"=>5.0, "ruby"=>5.0, "treasure"=>5.0, "treasurechest"=>5.0}}
                20 => tycoon_manager_rebirth_level_config{HudImageTexture := option{"UI/Rebirths/rebirth_20"}, ResourceMultipliers := map{"silver"=>5.0, "gems"=>5.0, "ruby"=>5.0, "treasure"=>5.0, "treasurechest"=>5.0}}
                21 => tycoon_manager_rebirth_level_config{HudImageTexture := option{"UI/Rebirths/rebirth_21"}, ResourceMultipliers := map{"silver"=>5.0, "gems"=>5.0, "ruby"=>5.0, "treasure"=>5.0, "treasurechest"=>5.0}}
                22 => tycoon_manager_rebirth_level_config{HudImageTexture := option{"UI/Rebirths/rebirth_22"}, ResourceMultipliers := map{"silver"=>5.0, "gems"=>5.0, "ruby"=>5.0, "treasure"=>5.0, "treasurechest"=>5.0}}
                23 => tycoon_manager_rebirth_level_config{HudImageTexture := option{"UI/Rebirths/rebirth_23"}, ResourceMultipliers := map{"silver"=>5.0, "gems"=>5.0, "ruby"=>5.0, "treasure"=>5.0, "treasurechest"=>5.0}}
                24 => tycoon_manager_rebirth_level_config{HudImageTexture := option{"UI/Rebirths/rebirth_24"}, ResourceMultipliers := map{"silver"=>5.0, "gems"=>5.0, "ruby"=>5.0, "treasure"=>5.0, "treasurechest"=>5.0}}
                25 => tycoon_manager_rebirth_level_config{HudImageTexture := option{"UI/Rebirths/rebirth_25"}, ResourceMultipliers := map{"silver"=>5.0, "gems"=>5.0, "ruby"=>5.0, "treasure"=>5.0, "treasurechest"=>5.0}}
                26 => tycoon_manager_rebirth_level_config{HudImageTexture := option{"UI/Rebirths/rebirth_26"}, ResourceMultipliers := map{"silver"=>5.0, "gems"=>5.0, "ruby"=>5.0, "treasure"=>5.0, "treasurechest"=>5.0}}
                50 => tycoon_manager_rebirth_level_config{HudImageTexture := option{"UI/Rebirths/rebirth_50"}, ResourceMultipliers := map{"silver"=>5.0, "gems"=>5.0, "ruby"=>5.0, "treasure"=>5.0, "treasurechest"=>5.0}}
                100 => tycoon_manager_rebirth_level_config{HudImageTexture := option{"UI/Rebirths/rebirth_100"}, ResourceMultipliers := map{"silver"=>5.0, "gems"=>5.0, "ruby"=>5.0, "treasure"=>5.0, "treasurechest"=>5.0}}
                200 => tycoon_manager_rebirth_level_config{HudImageTexture := option{"UI/Rebirths/rebirth_200"}, ResourceMultipliers := map{"silver"=>5.0, "gems"=>5.0, "ruby"=>5.0, "treasure"=>5.0, "treasurechest"=>5.0}}
                300 => tycoon_manager_rebirth_level_config{HudImageTexture := option{"UI/Rebirths/rebirth_300"}, ResourceMultipliers := map{"silver"=>5.0, "gems"=>5.0, "ruby"=>5.0, "treasure"=>5.0, "treasurechest"=>5.0}}
                400 => tycoon_manager_rebirth_level_config{HudImageTexture := option{"UI/Rebirths/rebirth_400"}, ResourceMultipliers := map{"silver"=>5.0, "gems"=>5.0, "ruby"=>5.0, "treasure"=>5.0, "treasurechest"=>5.0}}
                500 => tycoon_manager_rebirth_level_config{HudImageTexture := option{"UI/Rebirths/rebirth_500"}, ResourceMultipliers := map{"silver"=>5.0, "gems"=>5.0, "ruby"=>5.0, "treasure"=>5.0, "treasurechest"=>5.0}}
              }
            }}
          }
          
          set m_BuyplatesConfigs = map{
            "bp_claim" => tycoon_manager_buyplates_entry{
              BuyplateId := "bp_claim"
              Cost := false
              Generates := option{tycoon_manager_buyplates_entry_resource{ResourceId := "silver", Amount := 1, CheckDontTake := false}}
              Unlocks := array{
                tycoon_manager_unlock_entry{TargetRoutingId := "bp_300", Payload := false}
                tycoon_manager_unlock_entry{TargetRoutingId := "bp_400", Payload := false}
              }
              m_StorageIndex := 0
            }
            "bp_400" => tycoon_manager_buyplates_entry{
              BuyplateId := "bp_400"
              Cost := false
              Generates := false
              Unlocks := array{
              }
              m_StorageIndex := 1
            }
            "bp_300" => tycoon_manager_buyplates_entry{
              BuyplateId := "bp_300"
              Cost := option{tycoon_manager_buyplates_entry_resource{ResourceId := "silver", Amount := 5, CheckDontTake := false}}
              Generates := option{tycoon_manager_buyplates_entry_resource{ResourceId := "silver", Amount := 4, CheckDontTake := false}}
              Unlocks := array{
                tycoon_manager_unlock_entry{TargetRoutingId := "bp_001", Payload := false}
                tycoon_manager_unlock_entry{TargetRoutingId := "bp_200", Payload := false}
                tycoon_manager_unlock_entry{TargetRoutingId := "bp_401", Payload := false}
              }
              m_StorageIndex := 2
            }
            "bp_001" => tycoon_manager_buyplates_entry{
              BuyplateId := "bp_001"
              Cost := option{tycoon_manager_buyplates_entry_resource{ResourceId := "money", Amount := 750, CheckDontTake := false}}
              Generates := option{tycoon_manager_buyplates_entry_resource{ResourceId := "silver", Amount := 8, CheckDontTake := false}}
              Unlocks := array{
                tycoon_manager_unlock_entry{TargetRoutingId := "bp_002", Payload := false}
              }
              m_StorageIndex := 3
            }
            "bp_401" => tycoon_manager_buyplates_entry{
              BuyplateId := "bp_401"
              Cost := false
              Generates := false
              Unlocks := array{
              }
              m_StorageIndex := 4
            }
            "bp_002" => tycoon_manager_buyplates_entry{
              BuyplateId := "bp_002"
              Cost := option{tycoon_manager_buyplates_entry_resource{ResourceId := "money", Amount := 3450, CheckDontTake := false}}
              Generates := option{tycoon_manager_buyplates_entry_resource{ResourceId := "silver", Amount := 18, CheckDontTake := false}}
              Unlocks := array{
                tycoon_manager_unlock_entry{TargetRoutingId := "bp_003", Payload := false}
                tycoon_manager_unlock_entry{TargetRoutingId := "bp_402", Payload := false}
              }
              m_StorageIndex := 5
            }
            "bp_402" => tycoon_manager_buyplates_entry{
              BuyplateId := "bp_402"
              Cost := false
              Generates := false
              Unlocks := array{
                tycoon_manager_unlock_entry{TargetRoutingId := "garage_device", Payload := option{"enable"}}
              }
              m_StorageIndex := 6
            }
            "bp_003" => tycoon_manager_buyplates_entry{
              BuyplateId := "bp_003"
              Cost := option{tycoon_manager_buyplates_entry_resource{ResourceId := "money", Amount := 8200, CheckDontTake := false}}
              Generates := option{tycoon_manager_buyplates_entry_resource{ResourceId := "silver", Amount := 21, CheckDontTake := false}}
              Unlocks := array{
                tycoon_manager_unlock_entry{TargetRoutingId := "bp_004", Payload := false}
              }
              m_StorageIndex := 7
            }
            "bp_004" => tycoon_manager_buyplates_entry{
              BuyplateId := "bp_004"
              Cost := option{tycoon_manager_buyplates_entry_resource{ResourceId := "money", Amount := 17200, CheckDontTake := false}}
              Generates := option{tycoon_manager_buyplates_entry_resource{ResourceId := "silver", Amount := 26, CheckDontTake := false}}
              Unlocks := array{
                tycoon_manager_unlock_entry{TargetRoutingId := "bp_005", Payload := false}
              }
              m_StorageIndex := 8
            }
            "bp_005" => tycoon_manager_buyplates_entry{
              BuyplateId := "bp_005"
              Cost := option{tycoon_manager_buyplates_entry_resource{ResourceId := "money", Amount := 28000, CheckDontTake := false}}
              Generates := option{tycoon_manager_buyplates_entry_resource{ResourceId := "silver", Amount := 27, CheckDontTake := false}}
              Unlocks := array{
                tycoon_manager_unlock_entry{TargetRoutingId := "bp_006", Payload := false}
              }
              m_StorageIndex := 9
            }
            "bp_006" => tycoon_manager_buyplates_entry{
              BuyplateId := "bp_006"
              Cost := option{tycoon_manager_buyplates_entry_resource{ResourceId := "money", Amount := 42000, CheckDontTake := false}}
              Generates := option{tycoon_manager_buyplates_entry_resource{ResourceId := "silver", Amount := 31, CheckDontTake := false}}
              Unlocks := array{
                tycoon_manager_unlock_entry{TargetRoutingId := "bp_007", Payload := false}
              }
              m_StorageIndex := 10
            }
            "bp_007" => tycoon_manager_buyplates_entry{
              BuyplateId := "bp_007"
              Cost := option{tycoon_manager_buyplates_entry_resource{ResourceId := "money", Amount := 49800, CheckDontTake := false}}
              Generates := option{tycoon_manager_buyplates_entry_resource{ResourceId := "silver", Amount := 33, CheckDontTake := false}}
              Unlocks := array{
                tycoon_manager_unlock_entry{TargetRoutingId := "bp_008", Payload := false}
              }
              m_StorageIndex := 11
            }
            "bp_008" => tycoon_manager_buyplates_entry{
              BuyplateId := "bp_008"
              Cost := option{tycoon_manager_buyplates_entry_resource{ResourceId := "money", Amount := 66800, CheckDontTake := false}}
              Generates := option{tycoon_manager_buyplates_entry_resource{ResourceId := "silver", Amount := 37, CheckDontTake := false}}
              Unlocks := array{
                tycoon_manager_unlock_entry{TargetRoutingId := "bp_009", Payload := false}
              }
              m_StorageIndex := 12
            }
            "bp_009" => tycoon_manager_buyplates_entry{
              BuyplateId := "bp_009"
              Cost := option{tycoon_manager_buyplates_entry_resource{ResourceId := "money", Amount := 76000, CheckDontTake := false}}
              Generates := option{tycoon_manager_buyplates_entry_resource{ResourceId := "silver", Amount := 40, CheckDontTake := false}}
              Unlocks := array{
                tycoon_manager_unlock_entry{TargetRoutingId := "bp_010", Payload := false}
              }
              m_StorageIndex := 13
            }
            "bp_010" => tycoon_manager_buyplates_entry{
              BuyplateId := "bp_010"
              Cost := option{tycoon_manager_buyplates_entry_resource{ResourceId := "money", Amount := 116000, CheckDontTake := false}}
              Generates := option{tycoon_manager_buyplates_entry_resource{ResourceId := "silver", Amount := 46, CheckDontTake := false}}
              Unlocks := array{
                tycoon_manager_unlock_entry{TargetRoutingId := "bp_011", Payload := false}
              }
              m_StorageIndex := 14
            }
            "bp_011" => tycoon_manager_buyplates_entry{
              BuyplateId := "bp_011"
              Cost := option{tycoon_manager_buyplates_entry_resource{ResourceId := "money", Amount := 130000, CheckDontTake := false}}
              Generates := option{tycoon_manager_buyplates_entry_resource{ResourceId := "silver", Amount := 49, CheckDontTake := false}}
              Unlocks := array{
                tycoon_manager_unlock_entry{TargetRoutingId := "bp_012", Payload := false}
              }
              m_StorageIndex := 15
            }
            "bp_012" => tycoon_manager_buyplates_entry{
              BuyplateId := "bp_012"
              Cost := option{tycoon_manager_buyplates_entry_resource{ResourceId := "money", Amount := 161000, CheckDontTake := false}}
              Generates := option{tycoon_manager_buyplates_entry_resource{ResourceId := "silver", Amount := 57, CheckDontTake := false}}
              Unlocks := array{
                tycoon_manager_unlock_entry{TargetRoutingId := "bp_013", Payload := false}
              }
              m_StorageIndex := 16
            }
            "bp_013" => tycoon_manager_buyplates_entry{
              BuyplateId := "bp_013"
              Cost := option{tycoon_manager_buyplates_entry_resource{ResourceId := "money", Amount := 178000, CheckDontTake := false}}
              Generates := option{tycoon_manager_buyplates_entry_resource{ResourceId := "silver", Amount := 61, CheckDontTake := false}}
              Unlocks := array{
                tycoon_manager_unlock_entry{TargetRoutingId := "bp_014", Payload := false}
              }
              m_StorageIndex := 17
            }
            "bp_014" => tycoon_manager_buyplates_entry{
              BuyplateId := "bp_014"
              Cost := option{tycoon_manager_buyplates_entry_resource{ResourceId := "money", Amount := 216000, CheckDontTake := false}}
              Generates := option{tycoon_manager_buyplates_entry_resource{ResourceId := "silver", Amount := 71, CheckDontTake := false}}
              Unlocks := array{
                tycoon_manager_unlock_entry{TargetRoutingId := "bp_015", Payload := false}
              }
              m_StorageIndex := 18
            }
            "bp_015" => tycoon_manager_buyplates_entry{
              BuyplateId := "bp_015"
              Cost := option{tycoon_manager_buyplates_entry_resource{ResourceId := "money", Amount := 237000, CheckDontTake := false}}
              Generates := option{tycoon_manager_buyplates_entry_resource{ResourceId := "silver", Amount := 76, CheckDontTake := false}}
              Unlocks := array{
                tycoon_manager_unlock_entry{TargetRoutingId := "bp_016", Payload := false}
              }
              m_StorageIndex := 19
            }
            "bp_016" => tycoon_manager_buyplates_entry{
              BuyplateId := "bp_016"
              Cost := option{tycoon_manager_buyplates_entry_resource{ResourceId := "money", Amount := 284000, CheckDontTake := false}}
              Generates := option{tycoon_manager_buyplates_entry_resource{ResourceId := "silver", Amount := 88, CheckDontTake := false}}
              Unlocks := array{
                tycoon_manager_unlock_entry{TargetRoutingId := "bp_017", Payload := false}
              }
              m_StorageIndex := 20
            }
            "bp_017" => tycoon_manager_buyplates_entry{
              BuyplateId := "bp_017"
              Cost := option{tycoon_manager_buyplates_entry_resource{ResourceId := "money", Amount := 311000, CheckDontTake := false}}
              Generates := option{tycoon_manager_buyplates_entry_resource{ResourceId := "silver", Amount := 94, CheckDontTake := false}}
              Unlocks := array{
                tycoon_manager_unlock_entry{TargetRoutingId := "bp_018", Payload := false}
              }
              m_StorageIndex := 21
            }
            "bp_018" => tycoon_manager_buyplates_entry{
              BuyplateId := "bp_018"
              Cost := option{tycoon_manager_buyplates_entry_resource{ResourceId := "money", Amount := 369000, CheckDontTake := false}}
              Generates := option{tycoon_manager_buyplates_entry_resource{ResourceId := "silver", Amount := 106, CheckDontTake := false}}
              Unlocks := array{
                tycoon_manager_unlock_entry{TargetRoutingId := "bp_019", Payload := false}
              }
              m_StorageIndex := 22
            }
            "bp_019" => tycoon_manager_buyplates_entry{
              BuyplateId := "bp_019"
              Cost := option{tycoon_manager_buyplates_entry_resource{ResourceId := "money", Amount := 401000, CheckDontTake := false}}
              Generates := option{tycoon_manager_buyplates_entry_resource{ResourceId := "silver", Amount := 113, CheckDontTake := false}}
              Unlocks := array{
                tycoon_manager_unlock_entry{TargetRoutingId := "bp_020", Payload := false}
              }
              m_StorageIndex := 23
            }
            "bp_020" => tycoon_manager_buyplates_entry{
              BuyplateId := "bp_020"
              Cost := option{tycoon_manager_buyplates_entry_resource{ResourceId := "money", Amount := 472000, CheckDontTake := false}}
              Generates := option{tycoon_manager_buyplates_entry_resource{ResourceId := "silver", Amount := 130, CheckDontTake := false}}
              Unlocks := array{
                tycoon_manager_unlock_entry{TargetRoutingId := "bp_021", Payload := false}
              }
              m_StorageIndex := 24
            }
            "bp_021" => tycoon_manager_buyplates_entry{
              BuyplateId := "bp_021"
              Cost := option{tycoon_manager_buyplates_entry_resource{ResourceId := "money", Amount := 511000, CheckDontTake := false}}
              Generates := option{tycoon_manager_buyplates_entry_resource{ResourceId := "silver", Amount := 146, CheckDontTake := false}}
              Unlocks := array{
                tycoon_manager_unlock_entry{TargetRoutingId := "bp_022", Payload := false}
              }
              m_StorageIndex := 25
            }
            "bp_022" => tycoon_manager_buyplates_entry{
              BuyplateId := "bp_022"
              Cost := option{tycoon_manager_buyplates_entry_resource{ResourceId := "money", Amount := 555000, CheckDontTake := false}}
              Generates := option{tycoon_manager_buyplates_entry_resource{ResourceId := "silver", Amount := 150, CheckDontTake := false}}
              Unlocks := array{
                tycoon_manager_unlock_entry{TargetRoutingId := "bp_301", Payload := false}
              }
              m_StorageIndex := 26
            }
            "bp_301" => tycoon_manager_buyplates_entry{
              BuyplateId := "bp_301"
              Cost := option{tycoon_manager_buyplates_entry_resource{ResourceId := "money", Amount := 600000, CheckDontTake := false}}
              Generates := option{tycoon_manager_buyplates_entry_resource{ResourceId := "gems", Amount := 10, CheckDontTake := false}}
              Unlocks := array{
                tycoon_manager_unlock_entry{TargetRoutingId := "bp_023", Payload := false}
                tycoon_manager_unlock_entry{TargetRoutingId := "bp_210", Payload := false}
              }
              m_StorageIndex := 27
            }
            "bp_023" => tycoon_manager_buyplates_entry{
              BuyplateId := "bp_023"
              Cost := option{tycoon_manager_buyplates_entry_resource{ResourceId := "money", Amount := 713000, CheckDontTake := false}}
              Generates := option{tycoon_manager_buyplates_entry_resource{ResourceId := "gems", Amount := 18, CheckDontTake := false}}
              Unlocks := array{
                tycoon_manager_unlock_entry{TargetRoutingId := "bp_024", Payload := false}
              }
              m_StorageIndex := 28
            }
            "bp_024" => tycoon_manager_buyplates_entry{
              BuyplateId := "bp_024"
              Cost := option{tycoon_manager_buyplates_entry_resource{ResourceId := "money", Amount := 794000, CheckDontTake := false}}
              Generates := option{tycoon_manager_buyplates_entry_resource{ResourceId := "gems", Amount := 21, CheckDontTake := false}}
              Unlocks := array{
                tycoon_manager_unlock_entry{TargetRoutingId := "bp_025", Payload := false}
              }
              m_StorageIndex := 29
            }
            "bp_025" => tycoon_manager_buyplates_entry{
              BuyplateId := "bp_025"
              Cost := option{tycoon_manager_buyplates_entry_resource{ResourceId := "money", Amount := 996000, CheckDontTake := false}}
              Generates := option{tycoon_manager_buyplates_entry_resource{ResourceId := "gems", Amount := 27, CheckDontTake := false}}
              Unlocks := array{
                tycoon_manager_unlock_entry{TargetRoutingId := "bp_026", Payload := false}
              }
              m_StorageIndex := 30
            }
            "bp_026" => tycoon_manager_buyplates_entry{
              BuyplateId := "bp_026"
              Cost := option{tycoon_manager_buyplates_entry_resource{ResourceId := "money", Amount := 1120000, CheckDontTake := false}}
              Generates := option{tycoon_manager_buyplates_entry_resource{ResourceId := "gems", Amount := 30, CheckDontTake := false}}
              Unlocks := array{
                tycoon_manager_unlock_entry{TargetRoutingId := "bp_027", Payload := false}
              }
              m_StorageIndex := 31
            }
            "bp_027" => tycoon_manager_buyplates_entry{
              BuyplateId := "bp_027"
              Cost := option{tycoon_manager_buyplates_entry_resource{ResourceId := "money", Amount := 1400000, CheckDontTake := false}}
              Generates := option{tycoon_manager_buyplates_entry_resource{ResourceId := "gems", Amount := 37, CheckDontTake := false}}
              Unlocks := array{
                tycoon_manager_unlock_entry{TargetRoutingId := "bp_028", Payload := false}
              }
              m_StorageIndex := 32
            }
            "bp_028" => tycoon_manager_buyplates_entry{
              BuyplateId := "bp_028"
              Cost := option{tycoon_manager_buyplates_entry_resource{ResourceId := "money", Amount := 1570000, CheckDontTake := false}}
              Generates := option{tycoon_manager_buyplates_entry_resource{ResourceId := "gems", Amount := 41, CheckDontTake := false}}
              Unlocks := array{
                tycoon_manager_unlock_entry{TargetRoutingId := "bp_029", Payload := false}
              }
              m_StorageIndex := 33
            }
            "bp_029" => tycoon_manager_buyplates_entry{
              BuyplateId := "bp_029"
              Cost := option{tycoon_manager_buyplates_entry_resource{ResourceId := "money", Amount := 1950000, CheckDontTake := false}}
              Generates := option{tycoon_manager_buyplates_entry_resource{ResourceId := "gems", Amount := 50, CheckDontTake := false}}
              Unlocks := array{
                tycoon_manager_unlock_entry{TargetRoutingId := "bp_030", Payload := false}
              }
              m_StorageIndex := 34
            }
            "bp_030" => tycoon_manager_buyplates_entry{
              BuyplateId := "bp_030"
              Cost := option{tycoon_manager_buyplates_entry_resource{ResourceId := "money", Amount := 2180000, CheckDontTake := false}}
              Generates := option{tycoon_manager_buyplates_entry_resource{ResourceId := "gems", Amount := 55, CheckDontTake := false}}
              Unlocks := array{
                tycoon_manager_unlock_entry{TargetRoutingId := "bp_031", Payload := false}
              }
              m_StorageIndex := 35
            }
            "bp_031" => tycoon_manager_buyplates_entry{
              BuyplateId := "bp_031"
              Cost := option{tycoon_manager_buyplates_entry_resource{ResourceId := "money", Amount := 2700000, CheckDontTake := false}}
              Generates := option{tycoon_manager_buyplates_entry_resource{ResourceId := "gems", Amount := 68, CheckDontTake := false}}
              Unlocks := array{
                tycoon_manager_unlock_entry{TargetRoutingId := "bp_032", Payload := false}
              }
              m_StorageIndex := 36
            }
            "bp_032" => tycoon_manager_buyplates_entry{
              BuyplateId := "bp_032"
              Cost := option{tycoon_manager_buyplates_entry_resource{ResourceId := "money", Amount := 3010000, CheckDontTake := false}}
              Generates := option{tycoon_manager_buyplates_entry_resource{ResourceId := "gems", Amount := 75, CheckDontTake := false}}
              Unlocks := array{
                tycoon_manager_unlock_entry{TargetRoutingId := "bp_033", Payload := false}
              }
              m_StorageIndex := 37
            }
            "bp_033" => tycoon_manager_buyplates_entry{
              BuyplateId := "bp_033"
              Cost := option{tycoon_manager_buyplates_entry_resource{ResourceId := "money", Amount := 3710000, CheckDontTake := false}}
              Generates := option{tycoon_manager_buyplates_entry_resource{ResourceId := "gems", Amount := 88, CheckDontTake := false}}
              Unlocks := array{
                tycoon_manager_unlock_entry{TargetRoutingId := "bp_034", Payload := false}
              }
              m_StorageIndex := 38
            }
            "bp_034" => tycoon_manager_buyplates_entry{
              BuyplateId := "bp_034"
              Cost := option{tycoon_manager_buyplates_entry_resource{ResourceId := "money", Amount := 4100000, CheckDontTake := false}}
              Generates := option{tycoon_manager_buyplates_entry_resource{ResourceId := "gems", Amount := 97, CheckDontTake := false}}
              Unlocks := array{
                tycoon_manager_unlock_entry{TargetRoutingId := "bp_035", Payload := false}
              }
              m_StorageIndex := 39
            }
            "bp_035" => tycoon_manager_buyplates_entry{
              BuyplateId := "bp_035"
              Cost := option{tycoon_manager_buyplates_entry_resource{ResourceId := "money", Amount := 5020000, CheckDontTake := false}}
              Generates := option{tycoon_manager_buyplates_entry_resource{ResourceId := "gems", Amount := 118, CheckDontTake := false}}
              Unlocks := array{
                tycoon_manager_unlock_entry{TargetRoutingId := "bp_036", Payload := false}
              }
              m_StorageIndex := 40
            }
            "bp_036" => tycoon_manager_buyplates_entry{
              BuyplateId := "bp_036"
              Cost := option{tycoon_manager_buyplates_entry_resource{ResourceId := "money", Amount := 5550000, CheckDontTake := false}}
              Generates := option{tycoon_manager_buyplates_entry_resource{ResourceId := "gems", Amount := 130, CheckDontTake := false}}
              Unlocks := array{
                tycoon_manager_unlock_entry{TargetRoutingId := "bp_037", Payload := false}
              }
              m_StorageIndex := 41
            }
            "bp_037" => tycoon_manager_buyplates_entry{
              BuyplateId := "bp_037"
              Cost := option{tycoon_manager_buyplates_entry_resource{ResourceId := "money", Amount := 6790000, CheckDontTake := false}}
              Generates := option{tycoon_manager_buyplates_entry_resource{ResourceId := "gems", Amount := 160, CheckDontTake := false}}
              Unlocks := array{
                tycoon_manager_unlock_entry{TargetRoutingId := "bp_038", Payload := false}
              }
              m_StorageIndex := 42
            }
            "bp_038" => tycoon_manager_buyplates_entry{
              BuyplateId := "bp_038"
              Cost := option{tycoon_manager_buyplates_entry_resource{ResourceId := "money", Amount := 7510000, CheckDontTake := false}}
              Generates := option{tycoon_manager_buyplates_entry_resource{ResourceId := "gems", Amount := 176, CheckDontTake := false}}
              Unlocks := array{
                tycoon_manager_unlock_entry{TargetRoutingId := "bp_039", Payload := false}
              }
              m_StorageIndex := 43
            }
            "bp_039" => tycoon_manager_buyplates_entry{
              BuyplateId := "bp_039"
              Cost := option{tycoon_manager_buyplates_entry_resource{ResourceId := "money", Amount := 9170000, CheckDontTake := false}}
              Generates := option{tycoon_manager_buyplates_entry_resource{ResourceId := "gems", Amount := 215, CheckDontTake := false}}
              Unlocks := array{
                tycoon_manager_unlock_entry{TargetRoutingId := "bp_040", Payload := false}
              }
              m_StorageIndex := 44
            }
            "bp_040" => tycoon_manager_buyplates_entry{
              BuyplateId := "bp_040"
              Cost := option{tycoon_manager_buyplates_entry_resource{ResourceId := "money", Amount := 10100000, CheckDontTake := false}}
              Generates := option{tycoon_manager_buyplates_entry_resource{ResourceId := "gems", Amount := 238, CheckDontTake := false}}
              Unlocks := array{
                tycoon_manager_unlock_entry{TargetRoutingId := "bp_041", Payload := false}
              }
              m_StorageIndex := 45
            }
            "bp_041" => tycoon_manager_buyplates_entry{
              BuyplateId := "bp_041"
              Cost := option{tycoon_manager_buyplates_entry_resource{ResourceId := "money", Amount := 12400000, CheckDontTake := false}}
              Generates := option{tycoon_manager_buyplates_entry_resource{ResourceId := "gems", Amount := 300, CheckDontTake := false}}
              Unlocks := array{
                tycoon_manager_unlock_entry{TargetRoutingId := "bp_042", Payload := false}
              }
              m_StorageIndex := 46
            }
            "bp_042" => tycoon_manager_buyplates_entry{
              BuyplateId := "bp_042"
              Cost := option{tycoon_manager_buyplates_entry_resource{ResourceId := "money", Amount := 13800000, CheckDontTake := false}}
              Generates := option{tycoon_manager_buyplates_entry_resource{ResourceId := "gems", Amount := 325, CheckDontTake := false}}
              Unlocks := array{
                tycoon_manager_unlock_entry{TargetRoutingId := "bp_043", Payload := false}
              }
              m_StorageIndex := 47
            }
            "bp_043" => tycoon_manager_buyplates_entry{
              BuyplateId := "bp_043"
              Cost := option{tycoon_manager_buyplates_entry_resource{ResourceId := "money", Amount := 15200000, CheckDontTake := false}}
              Generates := option{tycoon_manager_buyplates_entry_resource{ResourceId := "gems", Amount := 355, CheckDontTake := false}}
              Unlocks := array{
                tycoon_manager_unlock_entry{TargetRoutingId := "bp_044", Payload := false}
              }
              m_StorageIndex := 48
            }
            "bp_044" => tycoon_manager_buyplates_entry{
              BuyplateId := "bp_044"
              Cost := option{tycoon_manager_buyplates_entry_resource{ResourceId := "money", Amount := 16800000, CheckDontTake := false}}
              Generates := option{tycoon_manager_buyplates_entry_resource{ResourceId := "gems", Amount := 390, CheckDontTake := false}}
              Unlocks := array{
                tycoon_manager_unlock_entry{TargetRoutingId := "bp_302", Payload := false}
              }
              m_StorageIndex := 49
            }
            "bp_302" => tycoon_manager_buyplates_entry{
              BuyplateId := "bp_302"
              Cost := option{tycoon_manager_buyplates_entry_resource{ResourceId := "money", Amount := 21700000, CheckDontTake := false}}
              Generates := option{tycoon_manager_buyplates_entry_resource{ResourceId := "ruby", Amount := 50, CheckDontTake := false}}
              Unlocks := array{
                tycoon_manager_unlock_entry{TargetRoutingId := "bp_045", Payload := false}
                tycoon_manager_unlock_entry{TargetRoutingId := "bp_220", Payload := false}
              }
              m_StorageIndex := 50
            }
            "bp_045" => tycoon_manager_buyplates_entry{
              BuyplateId := "bp_045"
              Cost := option{tycoon_manager_buyplates_entry_resource{ResourceId := "money", Amount := 28900000, CheckDontTake := false}}
              Generates := option{tycoon_manager_buyplates_entry_resource{ResourceId := "ruby", Amount := 56, CheckDontTake := false}}
              Unlocks := array{
                tycoon_manager_unlock_entry{TargetRoutingId := "bp_046", Payload := false}
              }
              m_StorageIndex := 51
            }
            "bp_046" => tycoon_manager_buyplates_entry{
              BuyplateId := "bp_046"
              Cost := option{tycoon_manager_buyplates_entry_resource{ResourceId := "money", Amount := 32800000, CheckDontTake := false}}
              Generates := option{tycoon_manager_buyplates_entry_resource{ResourceId := "ruby", Amount := 59, CheckDontTake := false}}
              Unlocks := array{
                tycoon_manager_unlock_entry{TargetRoutingId := "bp_047", Payload := false}
              }
              m_StorageIndex := 52
            }
            "bp_047" => tycoon_manager_buyplates_entry{
              BuyplateId := "bp_047"
              Cost := option{tycoon_manager_buyplates_entry_resource{ResourceId := "money", Amount := 41300000, CheckDontTake := false}}
              Generates := option{tycoon_manager_buyplates_entry_resource{ResourceId := "ruby", Amount := 67, CheckDontTake := false}}
              Unlocks := array{
                tycoon_manager_unlock_entry{TargetRoutingId := "bp_048", Payload := false}
              }
              m_StorageIndex := 53
            }
            "bp_048" => tycoon_manager_buyplates_entry{
              BuyplateId := "bp_048"
              Cost := option{tycoon_manager_buyplates_entry_resource{ResourceId := "money", Amount := 46000000, CheckDontTake := false}}
              Generates := option{tycoon_manager_buyplates_entry_resource{ResourceId := "ruby", Amount := 73, CheckDontTake := false}}
              Unlocks := array{
                tycoon_manager_unlock_entry{TargetRoutingId := "bp_049", Payload := false}
              }
              m_StorageIndex := 54
            }
            "bp_049" => tycoon_manager_buyplates_entry{
              BuyplateId := "bp_049"
              Cost := option{tycoon_manager_buyplates_entry_resource{ResourceId := "money", Amount := 56600000, CheckDontTake := false}}
              Generates := option{tycoon_manager_buyplates_entry_resource{ResourceId := "ruby", Amount := 87, CheckDontTake := false}}
              Unlocks := array{
                tycoon_manager_unlock_entry{TargetRoutingId := "bp_050", Payload := false}
              }
              m_StorageIndex := 55
            }
            "bp_050" => tycoon_manager_buyplates_entry{
              BuyplateId := "bp_050"
              Cost := option{tycoon_manager_buyplates_entry_resource{ResourceId := "money", Amount := 62700000, CheckDontTake := false}}
              Generates := option{tycoon_manager_buyplates_entry_resource{ResourceId := "ruby", Amount := 95, CheckDontTake := false}}
              Unlocks := array{
                tycoon_manager_unlock_entry{TargetRoutingId := "bp_051", Payload := false}
              }
              m_StorageIndex := 56
            }
            "bp_051" => tycoon_manager_buyplates_entry{
              BuyplateId := "bp_051"
              Cost := option{tycoon_manager_buyplates_entry_resource{ResourceId := "money", Amount := 76400000, CheckDontTake := false}}
              Generates := option{tycoon_manager_buyplates_entry_resource{ResourceId := "ruby", Amount := 108, CheckDontTake := false}}
              Unlocks := array{
                tycoon_manager_unlock_entry{TargetRoutingId := "bp_052", Payload := false}
              }
              m_StorageIndex := 57
            }
            "bp_052" => tycoon_manager_buyplates_entry{
              BuyplateId := "bp_052"
              Cost := option{tycoon_manager_buyplates_entry_resource{ResourceId := "money", Amount := 84000000, CheckDontTake := false}}
              Generates := option{tycoon_manager_buyplates_entry_resource{ResourceId := "ruby", Amount := 116, CheckDontTake := false}}
              Unlocks := array{
                tycoon_manager_unlock_entry{TargetRoutingId := "bp_053", Payload := false}
              }
              m_StorageIndex := 58
            }
            "bp_053" => tycoon_manager_buyplates_entry{
              BuyplateId := "bp_053"
              Cost := option{tycoon_manager_buyplates_entry_resource{ResourceId := "money", Amount := 101000000, CheckDontTake := false}}
              Generates := option{tycoon_manager_buyplates_entry_resource{ResourceId := "ruby", Amount := 133, CheckDontTake := false}}
              Unlocks := array{
                tycoon_manager_unlock_entry{TargetRoutingId := "bp_054", Payload := false}
              }
              m_StorageIndex := 59
            }
            "bp_054" => tycoon_manager_buyplates_entry{
              BuyplateId := "bp_054"
              Cost := option{tycoon_manager_buyplates_entry_resource{ResourceId := "money", Amount := 110000000, CheckDontTake := false}}
              Generates := option{tycoon_manager_buyplates_entry_resource{ResourceId := "ruby", Amount := 141, CheckDontTake := false}}
              Unlocks := array{
                tycoon_manager_unlock_entry{TargetRoutingId := "bp_055", Payload := false}
              }
              m_StorageIndex := 60
            }
            "bp_055" => tycoon_manager_buyplates_entry{
              BuyplateId := "bp_055"
              Cost := option{tycoon_manager_buyplates_entry_resource{ResourceId := "money", Amount := 130000000, CheckDontTake := false}}
              Generates := option{tycoon_manager_buyplates_entry_resource{ResourceId := "ruby", Amount := 157, CheckDontTake := false}}
              Unlocks := array{
                tycoon_manager_unlock_entry{TargetRoutingId := "bp_056", Payload := false}
              }
              m_StorageIndex := 61
            }
            "bp_056" => tycoon_manager_buyplates_entry{
              BuyplateId := "bp_056"
              Cost := option{tycoon_manager_buyplates_entry_resource{ResourceId := "money", Amount := 141000000, CheckDontTake := false}}
              Generates := option{tycoon_manager_buyplates_entry_resource{ResourceId := "ruby", Amount := 167, CheckDontTake := false}}
              Unlocks := array{
                tycoon_manager_unlock_entry{TargetRoutingId := "bp_057", Payload := false}
              }
              m_StorageIndex := 62
            }
            "bp_057" => tycoon_manager_buyplates_entry{
              BuyplateId := "bp_057"
              Cost := option{tycoon_manager_buyplates_entry_resource{ResourceId := "money", Amount := 166000000, CheckDontTake := false}}
              Generates := option{tycoon_manager_buyplates_entry_resource{ResourceId := "ruby", Amount := 191, CheckDontTake := false}}
              Unlocks := array{
                tycoon_manager_unlock_entry{TargetRoutingId := "bp_058", Payload := false}
              }
              m_StorageIndex := 63
            }
            "bp_058" => tycoon_manager_buyplates_entry{
              BuyplateId := "bp_058"
              Cost := option{tycoon_manager_buyplates_entry_resource{ResourceId := "money", Amount := 179000000, CheckDontTake := false}}
              Generates := option{tycoon_manager_buyplates_entry_resource{ResourceId := "ruby", Amount := 205, CheckDontTake := false}}
              Unlocks := array{
                tycoon_manager_unlock_entry{TargetRoutingId := "bp_059", Payload := false}
              }
              m_StorageIndex := 64
            }
            "bp_059" => tycoon_manager_buyplates_entry{
              BuyplateId := "bp_059"
              Cost := option{tycoon_manager_buyplates_entry_resource{ResourceId := "money", Amount := 209000000, CheckDontTake := false}}
              Generates := option{tycoon_manager_buyplates_entry_resource{ResourceId := "ruby", Amount := 237, CheckDontTake := false}}
              Unlocks := array{
                tycoon_manager_unlock_entry{TargetRoutingId := "bp_060", Payload := false}
              }
              m_StorageIndex := 65
            }
            "bp_060" => tycoon_manager_buyplates_entry{
              BuyplateId := "bp_060"
              Cost := option{tycoon_manager_buyplates_entry_resource{ResourceId := "money", Amount := 225000000, CheckDontTake := false}}
              Generates := option{tycoon_manager_buyplates_entry_resource{ResourceId := "ruby", Amount := 253, CheckDontTake := false}}
              Unlocks := array{
                tycoon_manager_unlock_entry{TargetRoutingId := "bp_061", Payload := false}
              }
              m_StorageIndex := 66
            }
            "bp_061" => tycoon_manager_buyplates_entry{
              BuyplateId := "bp_061"
              Cost := option{tycoon_manager_buyplates_entry_resource{ResourceId := "money", Amount := 262000000, CheckDontTake := false}}
              Generates := option{tycoon_manager_buyplates_entry_resource{ResourceId := "ruby", Amount := 289, CheckDontTake := false}}
              Unlocks := array{
                tycoon_manager_unlock_entry{TargetRoutingId := "bp_062", Payload := false}
              }
              m_StorageIndex := 67
            }
            "bp_062" => tycoon_manager_buyplates_entry{
              BuyplateId := "bp_062"
              Cost := option{tycoon_manager_buyplates_entry_resource{ResourceId := "money", Amount := 282000000, CheckDontTake := false}}
              Generates := option{tycoon_manager_buyplates_entry_resource{ResourceId := "ruby", Amount := 311, CheckDontTake := false}}
              Unlocks := array{
                tycoon_manager_unlock_entry{TargetRoutingId := "bp_063", Payload := false}
              }
              m_StorageIndex := 68
            }
            "bp_063" => tycoon_manager_buyplates_entry{
              BuyplateId := "bp_063"
              Cost := option{tycoon_manager_buyplates_entry_resource{ResourceId := "money", Amount := 327000000, CheckDontTake := false}}
              Generates := option{tycoon_manager_buyplates_entry_resource{ResourceId := "ruby", Amount := 365, CheckDontTake := false}}
              Unlocks := array{
                tycoon_manager_unlock_entry{TargetRoutingId := "bp_064", Payload := false}
              }
              m_StorageIndex := 69
            }
            "bp_064" => tycoon_manager_buyplates_entry{
              BuyplateId := "bp_064"
              Cost := option{tycoon_manager_buyplates_entry_resource{ResourceId := "money", Amount := 353000000, CheckDontTake := false}}
              Generates := option{tycoon_manager_buyplates_entry_resource{ResourceId := "ruby", Amount := 390, CheckDontTake := false}}
              Unlocks := array{
                tycoon_manager_unlock_entry{TargetRoutingId := "bp_065", Payload := false}
              }
              m_StorageIndex := 70
            }
            "bp_065" => tycoon_manager_buyplates_entry{
              BuyplateId := "bp_065"
              Cost := option{tycoon_manager_buyplates_entry_resource{ResourceId := "money", Amount := 380000000, CheckDontTake := false}}
              Generates := option{tycoon_manager_buyplates_entry_resource{ResourceId := "ruby", Amount := 425, CheckDontTake := false}}
              Unlocks := array{
                tycoon_manager_unlock_entry{TargetRoutingId := "bp_066", Payload := false}
              }
              m_StorageIndex := 71
            }
            "bp_066" => tycoon_manager_buyplates_entry{
              BuyplateId := "bp_066"
              Cost := option{tycoon_manager_buyplates_entry_resource{ResourceId := "money", Amount := 410000000, CheckDontTake := false}}
              Generates := option{tycoon_manager_buyplates_entry_resource{ResourceId := "ruby", Amount := 455, CheckDontTake := false}}
              Unlocks := array{
                tycoon_manager_unlock_entry{TargetRoutingId := "bp_303", Payload := false}
              }
              m_StorageIndex := 72
            }
            "bp_303" => tycoon_manager_buyplates_entry{
              BuyplateId := "bp_303"
              Cost := option{tycoon_manager_buyplates_entry_resource{ResourceId := "money", Amount := 505000000, CheckDontTake := false}}
              Generates := option{tycoon_manager_buyplates_entry_resource{ResourceId := "treasure", Amount := 100, CheckDontTake := false}}
              Unlocks := array{
                tycoon_manager_unlock_entry{TargetRoutingId := "bp_067", Payload := false}
                tycoon_manager_unlock_entry{TargetRoutingId := "bp_230", Payload := false}
              }
              m_StorageIndex := 73
            }
            "bp_067" => tycoon_manager_buyplates_entry{
              BuyplateId := "bp_067"
              Cost := option{tycoon_manager_buyplates_entry_resource{ResourceId := "money", Amount := 587000000, CheckDontTake := false}}
              Generates := option{tycoon_manager_buyplates_entry_resource{ResourceId := "treasure", Amount := 110, CheckDontTake := false}}
              Unlocks := array{
                tycoon_manager_unlock_entry{TargetRoutingId := "bp_068", Payload := false}
              }
              m_StorageIndex := 74
            }
            "bp_068" => tycoon_manager_buyplates_entry{
              BuyplateId := "bp_068"
              Cost := option{tycoon_manager_buyplates_entry_resource{ResourceId := "money", Amount := 631000000, CheckDontTake := false}}
              Generates := option{tycoon_manager_buyplates_entry_resource{ResourceId := "treasure", Amount := 115, CheckDontTake := false}}
              Unlocks := array{
                tycoon_manager_unlock_entry{TargetRoutingId := "bp_069", Payload := false}
              }
              m_StorageIndex := 75
            }
            "bp_069" => tycoon_manager_buyplates_entry{
              BuyplateId := "bp_069"
              Cost := option{tycoon_manager_buyplates_entry_resource{ResourceId := "money", Amount := 725000000, CheckDontTake := false}}
              Generates := option{tycoon_manager_buyplates_entry_resource{ResourceId := "treasure", Amount := 126, CheckDontTake := false}}
              Unlocks := array{
                tycoon_manager_unlock_entry{TargetRoutingId := "bp_070", Payload := false}
              }
              m_StorageIndex := 76
            }
            "bp_070" => tycoon_manager_buyplates_entry{
              BuyplateId := "bp_070"
              Cost := option{tycoon_manager_buyplates_entry_resource{ResourceId := "money", Amount := 775000000, CheckDontTake := false}}
              Generates := option{tycoon_manager_buyplates_entry_resource{ResourceId := "treasure", Amount := 132, CheckDontTake := false}}
              Unlocks := array{
                tycoon_manager_unlock_entry{TargetRoutingId := "bp_071", Payload := false}
              }
              m_StorageIndex := 77
            }
            "bp_071" => tycoon_manager_buyplates_entry{
              BuyplateId := "bp_071"
              Cost := option{tycoon_manager_buyplates_entry_resource{ResourceId := "money", Amount := 884000000, CheckDontTake := false}}
              Generates := option{tycoon_manager_buyplates_entry_resource{ResourceId := "treasure", Amount := 146, CheckDontTake := false}}
              Unlocks := array{
                tycoon_manager_unlock_entry{TargetRoutingId := "bp_072", Payload := false}
              }
              m_StorageIndex := 78
            }
            "bp_072" => tycoon_manager_buyplates_entry{
              BuyplateId := "bp_072"
              Cost := option{tycoon_manager_buyplates_entry_resource{ResourceId := "money", Amount := 942000000, CheckDontTake := false}}
              Generates := option{tycoon_manager_buyplates_entry_resource{ResourceId := "treasure", Amount := 152, CheckDontTake := false}}
              Unlocks := array{
                tycoon_manager_unlock_entry{TargetRoutingId := "bp_073", Payload := false}
              }
              m_StorageIndex := 79
            }
            "bp_073" => tycoon_manager_buyplates_entry{
              BuyplateId := "bp_073"
              Cost := option{tycoon_manager_buyplates_entry_resource{ResourceId := "money", Amount := 1070000000, CheckDontTake := false}}
              Generates := option{tycoon_manager_buyplates_entry_resource{ResourceId := "treasure", Amount := 168, CheckDontTake := false}}
              Unlocks := array{
                tycoon_manager_unlock_entry{TargetRoutingId := "bp_074", Payload := false}
              }
              m_StorageIndex := 80
            }
            "bp_074" => tycoon_manager_buyplates_entry{
              BuyplateId := "bp_074"
              Cost := option{tycoon_manager_buyplates_entry_resource{ResourceId := "money", Amount := 1130000000, CheckDontTake := false}}
              Generates := option{tycoon_manager_buyplates_entry_resource{ResourceId := "treasure", Amount := 177, CheckDontTake := false}}
              Unlocks := array{
                tycoon_manager_unlock_entry{TargetRoutingId := "bp_075", Payload := false}
              }
              m_StorageIndex := 81
            }
            "bp_075" => tycoon_manager_buyplates_entry{
              BuyplateId := "bp_075"
              Cost := option{tycoon_manager_buyplates_entry_resource{ResourceId := "money", Amount := 1280000000, CheckDontTake := false}}
              Generates := option{tycoon_manager_buyplates_entry_resource{ResourceId := "treasure", Amount := 194, CheckDontTake := false}}
              Unlocks := array{
                tycoon_manager_unlock_entry{TargetRoutingId := "bp_076", Payload := false}
              }
              m_StorageIndex := 82
            }
            "bp_076" => tycoon_manager_buyplates_entry{
              BuyplateId := "bp_076"
              Cost := option{tycoon_manager_buyplates_entry_resource{ResourceId := "money", Amount := 1360000000, CheckDontTake := false}}
              Generates := option{tycoon_manager_buyplates_entry_resource{ResourceId := "treasure", Amount := 204, CheckDontTake := false}}
              Unlocks := array{
                tycoon_manager_unlock_entry{TargetRoutingId := "bp_077", Payload := false}
              }
              m_StorageIndex := 83
            }
            "bp_077" => tycoon_manager_buyplates_entry{
              BuyplateId := "bp_077"
              Cost := option{tycoon_manager_buyplates_entry_resource{ResourceId := "money", Amount := 1710000000, CheckDontTake := false}}
              Generates := option{tycoon_manager_buyplates_entry_resource{ResourceId := "treasure", Amount := 225, CheckDontTake := false}}
              Unlocks := array{
                tycoon_manager_unlock_entry{TargetRoutingId := "bp_078", Payload := false}
              }
              m_StorageIndex := 84
            }
            "bp_078" => tycoon_manager_buyplates_entry{
              BuyplateId := "bp_078"
              Cost := option{tycoon_manager_buyplates_entry_resource{ResourceId := "money", Amount := 1820000000, CheckDontTake := false}}
              Generates := option{tycoon_manager_buyplates_entry_resource{ResourceId := "treasure", Amount := 238, CheckDontTake := false}}
              Unlocks := array{
                tycoon_manager_unlock_entry{TargetRoutingId := "bp_079", Payload := false}
              }
              m_StorageIndex := 85
            }
            "bp_079" => tycoon_manager_buyplates_entry{
              BuyplateId := "bp_079"
              Cost := option{tycoon_manager_buyplates_entry_resource{ResourceId := "money", Amount := 2030000000, CheckDontTake := false}}
              Generates := option{tycoon_manager_buyplates_entry_resource{ResourceId := "treasure", Amount := 265, CheckDontTake := false}}
              Unlocks := array{
                tycoon_manager_unlock_entry{TargetRoutingId := "bp_080", Payload := false}
              }
              m_StorageIndex := 86
            }
            "bp_080" => tycoon_manager_buyplates_entry{
              BuyplateId := "bp_080"
              Cost := option{tycoon_manager_buyplates_entry_resource{ResourceId := "money", Amount := 2150000000, CheckDontTake := false}}
              Generates := option{tycoon_manager_buyplates_entry_resource{ResourceId := "treasure", Amount := 280, CheckDontTake := false}}
              Unlocks := array{
                tycoon_manager_unlock_entry{TargetRoutingId := "bp_081", Payload := false}
              }
              m_StorageIndex := 87
            }
            "bp_081" => tycoon_manager_buyplates_entry{
              BuyplateId := "bp_081"
              Cost := option{tycoon_manager_buyplates_entry_resource{ResourceId := "money", Amount := 2410000000, CheckDontTake := false}}
              Generates := option{tycoon_manager_buyplates_entry_resource{ResourceId := "treasure", Amount := 309, CheckDontTake := false}}
              Unlocks := array{
                tycoon_manager_unlock_entry{TargetRoutingId := "bp_082", Payload := false}
              }
              m_StorageIndex := 88
            }
            "bp_082" => tycoon_manager_buyplates_entry{
              BuyplateId := "bp_082"
              Cost := option{tycoon_manager_buyplates_entry_resource{ResourceId := "money", Amount := 2550000000, CheckDontTake := false}}
              Generates := option{tycoon_manager_buyplates_entry_resource{ResourceId := "treasure", Amount := 330, CheckDontTake := false}}
              Unlocks := array{
                tycoon_manager_unlock_entry{TargetRoutingId := "bp_083", Payload := false}
              }
              m_StorageIndex := 89
            }
            "bp_083" => tycoon_manager_buyplates_entry{
              BuyplateId := "bp_083"
              Cost := option{tycoon_manager_buyplates_entry_resource{ResourceId := "money", Amount := 2860000000, CheckDontTake := false}}
              Generates := option{tycoon_manager_buyplates_entry_resource{ResourceId := "treasure", Amount := 362, CheckDontTake := false}}
              Unlocks := array{
                tycoon_manager_unlock_entry{TargetRoutingId := "bp_084", Payload := false}
              }
              m_StorageIndex := 90
            }
            "bp_084" => tycoon_manager_buyplates_entry{
              BuyplateId := "bp_084"
              Cost := option{tycoon_manager_buyplates_entry_resource{ResourceId := "money", Amount := 3020000000, CheckDontTake := false}}
              Generates := option{tycoon_manager_buyplates_entry_resource{ResourceId := "treasure", Amount := 380, CheckDontTake := false}}
              Unlocks := array{
                tycoon_manager_unlock_entry{TargetRoutingId := "bp_085", Payload := false}
              }
              m_StorageIndex := 91
            }
            "bp_085" => tycoon_manager_buyplates_entry{
              BuyplateId := "bp_085"
              Cost := option{tycoon_manager_buyplates_entry_resource{ResourceId := "money", Amount := 3360000000, CheckDontTake := false}}
              Generates := option{tycoon_manager_buyplates_entry_resource{ResourceId := "treasure", Amount := 410, CheckDontTake := false}}
              Unlocks := array{
                tycoon_manager_unlock_entry{TargetRoutingId := "bp_086", Payload := false}
              }
              m_StorageIndex := 92
            }
            "bp_086" => tycoon_manager_buyplates_entry{
              BuyplateId := "bp_086"
              Cost := option{tycoon_manager_buyplates_entry_resource{ResourceId := "money", Amount := 3550000000, CheckDontTake := false}}
              Generates := option{tycoon_manager_buyplates_entry_resource{ResourceId := "treasure", Amount := 430, CheckDontTake := false}}
              Unlocks := array{
                tycoon_manager_unlock_entry{TargetRoutingId := "bp_087", Payload := false}
              }
              m_StorageIndex := 93
            }
            "bp_087" => tycoon_manager_buyplates_entry{
              BuyplateId := "bp_087"
              Cost := option{tycoon_manager_buyplates_entry_resource{ResourceId := "money", Amount := 3740000000, CheckDontTake := false}}
              Generates := option{tycoon_manager_buyplates_entry_resource{ResourceId := "treasure", Amount := 455, CheckDontTake := false}}
              Unlocks := array{
                tycoon_manager_unlock_entry{TargetRoutingId := "bp_088", Payload := false}
              }
              m_StorageIndex := 94
            }
            "bp_088" => tycoon_manager_buyplates_entry{
              BuyplateId := "bp_088"
              Cost := option{tycoon_manager_buyplates_entry_resource{ResourceId := "money", Amount := 3950000000, CheckDontTake := false}}
              Generates := option{tycoon_manager_buyplates_entry_resource{ResourceId := "treasure", Amount := 490, CheckDontTake := false}}
              Unlocks := array{
                tycoon_manager_unlock_entry{TargetRoutingId := "bp_304", Payload := false}
              }
              m_StorageIndex := 95
            }
            "bp_304" => tycoon_manager_buyplates_entry{
              BuyplateId := "bp_304"
              Cost := option{tycoon_manager_buyplates_entry_resource{ResourceId := "money", Amount := 4630000000, CheckDontTake := false}}
              Generates := option{tycoon_manager_buyplates_entry_resource{ResourceId := "treasurechest", Amount := 150, CheckDontTake := false}}
              Unlocks := array{
                tycoon_manager_unlock_entry{TargetRoutingId := "bp_089", Payload := false}
                tycoon_manager_unlock_entry{TargetRoutingId := "bp_240", Payload := false}
              }
              m_StorageIndex := 96
            }
            "bp_089" => tycoon_manager_buyplates_entry{
              BuyplateId := "bp_089"
              Cost := option{tycoon_manager_buyplates_entry_resource{ResourceId := "money", Amount := 5010000000, CheckDontTake := false}}
              Generates := option{tycoon_manager_buyplates_entry_resource{ResourceId := "treasurechest", Amount := 164, CheckDontTake := false}}
              Unlocks := array{
                tycoon_manager_unlock_entry{TargetRoutingId := "bp_090", Payload := false}
              }
              m_StorageIndex := 97
            }
            "bp_090" => tycoon_manager_buyplates_entry{
              BuyplateId := "bp_090"
              Cost := option{tycoon_manager_buyplates_entry_resource{ResourceId := "money", Amount := 5220000000, CheckDontTake := false}}
              Generates := option{tycoon_manager_buyplates_entry_resource{ResourceId := "treasurechest", Amount := 171, CheckDontTake := false}}
              Unlocks := array{
                tycoon_manager_unlock_entry{TargetRoutingId := "bp_091", Payload := false}
              }
              m_StorageIndex := 98
            }
            "bp_091" => tycoon_manager_buyplates_entry{
              BuyplateId := "bp_091"
              Cost := option{tycoon_manager_buyplates_entry_resource{ResourceId := "money", Amount := 5660000000, CheckDontTake := false}}
              Generates := option{tycoon_manager_buyplates_entry_resource{ResourceId := "treasurechest", Amount := 186, CheckDontTake := false}}
              Unlocks := array{
                tycoon_manager_unlock_entry{TargetRoutingId := "bp_092", Payload := false}
              }
              m_StorageIndex := 99
            }
            "bp_092" => tycoon_manager_buyplates_entry{
              BuyplateId := "bp_092"
              Cost := option{tycoon_manager_buyplates_entry_resource{ResourceId := "money", Amount := 5890000000, CheckDontTake := false}}
              Generates := option{tycoon_manager_buyplates_entry_resource{ResourceId := "treasurechest", Amount := 194, CheckDontTake := false}}
              Unlocks := array{
                tycoon_manager_unlock_entry{TargetRoutingId := "bp_093", Payload := false}
              }
              m_StorageIndex := 100
            }
            "bp_093" => tycoon_manager_buyplates_entry{
              BuyplateId := "bp_093"
              Cost := option{tycoon_manager_buyplates_entry_resource{ResourceId := "money", Amount := 7020000000, CheckDontTake := false}}
              Generates := option{tycoon_manager_buyplates_entry_resource{ResourceId := "treasurechest", Amount := 210, CheckDontTake := false}}
              Unlocks := array{
                tycoon_manager_unlock_entry{TargetRoutingId := "bp_094", Payload := false}
              }
              m_StorageIndex := 101
            }
            "bp_094" => tycoon_manager_buyplates_entry{
              BuyplateId := "bp_094"
              Cost := option{tycoon_manager_buyplates_entry_resource{ResourceId := "money", Amount := 7310000000, CheckDontTake := false}}
              Generates := option{tycoon_manager_buyplates_entry_resource{ResourceId := "treasurechest", Amount := 218, CheckDontTake := false}}
              Unlocks := array{
                tycoon_manager_unlock_entry{TargetRoutingId := "bp_095", Payload := false}
              }
              m_StorageIndex := 102
            }
            "bp_095" => tycoon_manager_buyplates_entry{
              BuyplateId := "bp_095"
              Cost := option{tycoon_manager_buyplates_entry_resource{ResourceId := "money", Amount := 7920000000, CheckDontTake := false}}
              Generates := option{tycoon_manager_buyplates_entry_resource{ResourceId := "treasurechest", Amount := 238, CheckDontTake := false}}
              Unlocks := array{
                tycoon_manager_unlock_entry{TargetRoutingId := "bp_096", Payload := false}
              }
              m_StorageIndex := 103
            }
            "bp_096" => tycoon_manager_buyplates_entry{
              BuyplateId := "bp_096"
              Cost := option{tycoon_manager_buyplates_entry_resource{ResourceId := "money", Amount := 8250000000, CheckDontTake := false}}
              Generates := option{tycoon_manager_buyplates_entry_resource{ResourceId := "treasurechest", Amount := 250, CheckDontTake := false}}
              Unlocks := array{
                tycoon_manager_unlock_entry{TargetRoutingId := "bp_097", Payload := false}
              }
              m_StorageIndex := 104
            }
            "bp_097" => tycoon_manager_buyplates_entry{
              BuyplateId := "bp_097"
              Cost := option{tycoon_manager_buyplates_entry_resource{ResourceId := "money", Amount := 8960000000, CheckDontTake := false}}
              Generates := option{tycoon_manager_buyplates_entry_resource{ResourceId := "treasurechest", Amount := 275, CheckDontTake := false}}
              Unlocks := array{
                tycoon_manager_unlock_entry{TargetRoutingId := "bp_098", Payload := false}
              }
              m_StorageIndex := 105
            }
            "bp_098" => tycoon_manager_buyplates_entry{
              BuyplateId := "bp_098"
              Cost := option{tycoon_manager_buyplates_entry_resource{ResourceId := "money", Amount := 10200000000, CheckDontTake := false}}
              Generates := option{tycoon_manager_buyplates_entry_resource{ResourceId := "treasurechest", Amount := 282, CheckDontTake := false}}
              Unlocks := array{
                tycoon_manager_unlock_entry{TargetRoutingId := "bp_099", Payload := false}
              }
              m_StorageIndex := 106
            }
            "bp_099" => tycoon_manager_buyplates_entry{
              BuyplateId := "bp_099"
              Cost := option{tycoon_manager_buyplates_entry_resource{ResourceId := "money", Amount := 11000000000, CheckDontTake := false}}
              Generates := option{tycoon_manager_buyplates_entry_resource{ResourceId := "treasurechest", Amount := 305, CheckDontTake := false}}
              Unlocks := array{
                tycoon_manager_unlock_entry{TargetRoutingId := "bp_100", Payload := false}
              }
              m_StorageIndex := 107
            }
            "bp_100" => tycoon_manager_buyplates_entry{
              BuyplateId := "bp_100"
              Cost := option{tycoon_manager_buyplates_entry_resource{ResourceId := "money", Amount := 11500000000, CheckDontTake := false}}
              Generates := option{tycoon_manager_buyplates_entry_resource{ResourceId := "treasurechest", Amount := 317, CheckDontTake := false}}
              Unlocks := array{
                tycoon_manager_unlock_entry{TargetRoutingId := "bp_101", Payload := false}
              }
              m_StorageIndex := 108
            }
            "bp_101" => tycoon_manager_buyplates_entry{
              BuyplateId := "bp_101"
              Cost := option{tycoon_manager_buyplates_entry_resource{ResourceId := "money", Amount := 12500000000, CheckDontTake := false}}
              Generates := option{tycoon_manager_buyplates_entry_resource{ResourceId := "treasurechest", Amount := 341, CheckDontTake := false}}
              Unlocks := array{
                tycoon_manager_unlock_entry{TargetRoutingId := "bp_102", Payload := false}
              }
              m_StorageIndex := 109
            }
            "bp_102" => tycoon_manager_buyplates_entry{
              BuyplateId := "bp_102"
              Cost := option{tycoon_manager_buyplates_entry_resource{ResourceId := "money", Amount := 13000000000, CheckDontTake := false}}
              Generates := option{tycoon_manager_buyplates_entry_resource{ResourceId := "treasurechest", Amount := 354, CheckDontTake := false}}
              Unlocks := array{
                tycoon_manager_unlock_entry{TargetRoutingId := "bp_103", Payload := false}
              }
              m_StorageIndex := 110
            }
            "bp_103" => tycoon_manager_buyplates_entry{
              BuyplateId := "bp_103"
              Cost := option{tycoon_manager_buyplates_entry_resource{ResourceId := "money", Amount := 16400000000, CheckDontTake := false}}
              Generates := option{tycoon_manager_buyplates_entry_resource{ResourceId := "treasurechest", Amount := 375, CheckDontTake := false}}
              Unlocks := array{
                tycoon_manager_unlock_entry{TargetRoutingId := "bp_104", Payload := false}
              }
              m_StorageIndex := 111
            }
            "bp_104" => tycoon_manager_buyplates_entry{
              BuyplateId := "bp_104"
              Cost := option{tycoon_manager_buyplates_entry_resource{ResourceId := "money", Amount := 17100000000, CheckDontTake := false}}
              Generates := option{tycoon_manager_buyplates_entry_resource{ResourceId := "treasurechest", Amount := 390, CheckDontTake := false}}
              Unlocks := array{
                tycoon_manager_unlock_entry{TargetRoutingId := "bp_105", Payload := false}
              }
              m_StorageIndex := 112
            }
            "bp_105" => tycoon_manager_buyplates_entry{
              BuyplateId := "bp_105"
              Cost := option{tycoon_manager_buyplates_entry_resource{ResourceId := "money", Amount := 18400000000, CheckDontTake := false}}
              Generates := option{tycoon_manager_buyplates_entry_resource{ResourceId := "treasurechest", Amount := 416, CheckDontTake := false}}
              Unlocks := array{
                tycoon_manager_unlock_entry{TargetRoutingId := "bp_106", Payload := false}
              }
              m_StorageIndex := 113
            }
            "bp_106" => tycoon_manager_buyplates_entry{
              BuyplateId := "bp_106"
              Cost := option{tycoon_manager_buyplates_entry_resource{ResourceId := "money", Amount := 19200000000, CheckDontTake := false}}
              Generates := option{tycoon_manager_buyplates_entry_resource{ResourceId := "treasurechest", Amount := 440, CheckDontTake := false}}
              Unlocks := array{
                tycoon_manager_unlock_entry{TargetRoutingId := "bp_107", Payload := false}
              }
              m_StorageIndex := 114
            }
            "bp_107" => tycoon_manager_buyplates_entry{
              BuyplateId := "bp_107"
              Cost := option{tycoon_manager_buyplates_entry_resource{ResourceId := "money", Amount := 23700000000, CheckDontTake := false}}
              Generates := option{tycoon_manager_buyplates_entry_resource{ResourceId := "treasurechest", Amount := 514, CheckDontTake := false}}
              Unlocks := array{
                tycoon_manager_unlock_entry{TargetRoutingId := "bp_108", Payload := false}
              }
              m_StorageIndex := 115
            }
            "bp_108" => tycoon_manager_buyplates_entry{
              BuyplateId := "bp_108"
              Cost := option{tycoon_manager_buyplates_entry_resource{ResourceId := "money", Amount := 24800000000, CheckDontTake := false}}
              Generates := option{tycoon_manager_buyplates_entry_resource{ResourceId := "treasurechest", Amount := 564, CheckDontTake := false}}
              Unlocks := array{
                tycoon_manager_unlock_entry{TargetRoutingId := "bp_109", Payload := false}
              }
              m_StorageIndex := 116
            }
            "bp_109" => tycoon_manager_buyplates_entry{
              BuyplateId := "bp_109"
              Cost := option{tycoon_manager_buyplates_entry_resource{ResourceId := "money", Amount := 29100000000, CheckDontTake := false}}
              Generates := option{tycoon_manager_buyplates_entry_resource{ResourceId := "treasurechest", Amount := 760, CheckDontTake := false}}
              Unlocks := array{
                tycoon_manager_unlock_entry{TargetRoutingId := "bp_110", Payload := false}
              }
              m_StorageIndex := 117
            }
            "bp_110" => tycoon_manager_buyplates_entry{
              BuyplateId := "bp_110"
              Cost := option{tycoon_manager_buyplates_entry_resource{ResourceId := "money", Amount := 30000000000, CheckDontTake := false}}
              Generates := false
              Unlocks := array{
              }
              m_StorageIndex := 118
            }
            "bp_200" => tycoon_manager_buyplates_entry{
              BuyplateId := "bp_200"
              Cost := option{tycoon_manager_buyplates_entry_resource{ResourceId := "money", Amount := 1950, CheckDontTake := false}}
              Generates := option{tycoon_manager_buyplates_entry_resource{ResourceId := "silver", Amount := 10, CheckDontTake := false}}
              Unlocks := array{
                tycoon_manager_unlock_entry{TargetRoutingId := "bp_201", Payload := false}
              }
              m_StorageIndex := 119
            }
            "bp_201" => tycoon_manager_buyplates_entry{
              BuyplateId := "bp_201"
              Cost := option{tycoon_manager_buyplates_entry_resource{ResourceId := "money", Amount := 12400, CheckDontTake := false}}
              Generates := option{tycoon_manager_buyplates_entry_resource{ResourceId := "silver", Amount := 24, CheckDontTake := false}}
              Unlocks := array{
                tycoon_manager_unlock_entry{TargetRoutingId := "bp_202", Payload := false}
              }
              m_StorageIndex := 120
            }
            "bp_202" => tycoon_manager_buyplates_entry{
              BuyplateId := "bp_202"
              Cost := option{tycoon_manager_buyplates_entry_resource{ResourceId := "money", Amount := 34800, CheckDontTake := false}}
              Generates := option{tycoon_manager_buyplates_entry_resource{ResourceId := "silver", Amount := 29, CheckDontTake := false}}
              Unlocks := array{
                tycoon_manager_unlock_entry{TargetRoutingId := "bp_203", Payload := false}
              }
              m_StorageIndex := 121
            }
            "bp_203" => tycoon_manager_buyplates_entry{
              BuyplateId := "bp_203"
              Cost := option{tycoon_manager_buyplates_entry_resource{ResourceId := "money", Amount := 58000, CheckDontTake := false}}
              Generates := option{tycoon_manager_buyplates_entry_resource{ResourceId := "silver", Amount := 35, CheckDontTake := false}}
              Unlocks := array{
                tycoon_manager_unlock_entry{TargetRoutingId := "bp_204", Payload := false}
              }
              m_StorageIndex := 122
            }
            "bp_204" => tycoon_manager_buyplates_entry{
              BuyplateId := "bp_204"
              Cost := option{tycoon_manager_buyplates_entry_resource{ResourceId := "money", Amount := 86000, CheckDontTake := false}}
              Generates := option{tycoon_manager_buyplates_entry_resource{ResourceId := "silver", Amount := 43, CheckDontTake := false}}
              Unlocks := array{
                tycoon_manager_unlock_entry{TargetRoutingId := "bp_205", Payload := false}
              }
              m_StorageIndex := 123
            }
            "bp_205" => tycoon_manager_buyplates_entry{
              BuyplateId := "bp_205"
              Cost := option{tycoon_manager_buyplates_entry_resource{ResourceId := "money", Amount := 145000, CheckDontTake := false}}
              Generates := option{tycoon_manager_buyplates_entry_resource{ResourceId := "silver", Amount := 53, CheckDontTake := false}}
              Unlocks := array{
                tycoon_manager_unlock_entry{TargetRoutingId := "bp_206", Payload := false}
              }
              m_StorageIndex := 124
            }
            "bp_206" => tycoon_manager_buyplates_entry{
              BuyplateId := "bp_206"
              Cost := option{tycoon_manager_buyplates_entry_resource{ResourceId := "money", Amount := 196000, CheckDontTake := false}}
              Generates := option{tycoon_manager_buyplates_entry_resource{ResourceId := "silver", Amount := 66, CheckDontTake := false}}
              Unlocks := array{
                tycoon_manager_unlock_entry{TargetRoutingId := "bp_207", Payload := false}
              }
              m_StorageIndex := 125
            }
            "bp_207" => tycoon_manager_buyplates_entry{
              BuyplateId := "bp_207"
              Cost := option{tycoon_manager_buyplates_entry_resource{ResourceId := "money", Amount := 260000, CheckDontTake := false}}
              Generates := option{tycoon_manager_buyplates_entry_resource{ResourceId := "silver", Amount := 82, CheckDontTake := false}}
              Unlocks := array{
                tycoon_manager_unlock_entry{TargetRoutingId := "bp_208", Payload := false}
              }
              m_StorageIndex := 126
            }
            "bp_208" => tycoon_manager_buyplates_entry{
              BuyplateId := "bp_208"
              Cost := option{tycoon_manager_buyplates_entry_resource{ResourceId := "money", Amount := 339000, CheckDontTake := false}}
              Generates := option{tycoon_manager_buyplates_entry_resource{ResourceId := "silver", Amount := 101, CheckDontTake := false}}
              Unlocks := array{
                tycoon_manager_unlock_entry{TargetRoutingId := "bp_209", Payload := false}
              }
              m_StorageIndex := 127
            }
            "bp_209" => tycoon_manager_buyplates_entry{
              BuyplateId := "bp_209"
              Cost := option{tycoon_manager_buyplates_entry_resource{ResourceId := "money", Amount := 435000, CheckDontTake := false}}
              Generates := option{tycoon_manager_buyplates_entry_resource{ResourceId := "silver", Amount := 124, CheckDontTake := false}}
              Unlocks := array{
              }
              m_StorageIndex := 128
            }
            "bp_210" => tycoon_manager_buyplates_entry{
              BuyplateId := "bp_210"
              Cost := option{tycoon_manager_buyplates_entry_resource{ResourceId := "money", Amount := 645000, CheckDontTake := false}}
              Generates := option{tycoon_manager_buyplates_entry_resource{ResourceId := "gems", Amount := 15, CheckDontTake := false}}
              Unlocks := array{
                tycoon_manager_unlock_entry{TargetRoutingId := "bp_211", Payload := false}
              }
              m_StorageIndex := 129
            }
            "bp_211" => tycoon_manager_buyplates_entry{
              BuyplateId := "bp_211"
              Cost := option{tycoon_manager_buyplates_entry_resource{ResourceId := "money", Amount := 888000, CheckDontTake := false}}
              Generates := option{tycoon_manager_buyplates_entry_resource{ResourceId := "gems", Amount := 24, CheckDontTake := false}}
              Unlocks := array{
                tycoon_manager_unlock_entry{TargetRoutingId := "bp_212", Payload := false}
              }
              m_StorageIndex := 130
            }
            "bp_212" => tycoon_manager_buyplates_entry{
              BuyplateId := "bp_212"
              Cost := option{tycoon_manager_buyplates_entry_resource{ResourceId := "money", Amount := 1250000, CheckDontTake := false}}
              Generates := option{tycoon_manager_buyplates_entry_resource{ResourceId := "gems", Amount := 33, CheckDontTake := false}}
              Unlocks := array{
                tycoon_manager_unlock_entry{TargetRoutingId := "bp_213", Payload := false}
              }
              m_StorageIndex := 131
            }
            "bp_213" => tycoon_manager_buyplates_entry{
              BuyplateId := "bp_213"
              Cost := option{tycoon_manager_buyplates_entry_resource{ResourceId := "money", Amount := 1750000, CheckDontTake := false}}
              Generates := option{tycoon_manager_buyplates_entry_resource{ResourceId := "gems", Amount := 45, CheckDontTake := false}}
              Unlocks := array{
                tycoon_manager_unlock_entry{TargetRoutingId := "bp_214", Payload := false}
              }
              m_StorageIndex := 132
            }
            "bp_214" => tycoon_manager_buyplates_entry{
              BuyplateId := "bp_214"
              Cost := option{tycoon_manager_buyplates_entry_resource{ResourceId := "money", Amount := 2430000, CheckDontTake := false}}
              Generates := option{tycoon_manager_buyplates_entry_resource{ResourceId := "gems", Amount := 61, CheckDontTake := false}}
              Unlocks := array{
                tycoon_manager_unlock_entry{TargetRoutingId := "bp_215", Payload := false}
              }
              m_StorageIndex := 133
            }
            "bp_215" => tycoon_manager_buyplates_entry{
              BuyplateId := "bp_215"
              Cost := option{tycoon_manager_buyplates_entry_resource{ResourceId := "money", Amount := 3350000, CheckDontTake := false}}
              Generates := option{tycoon_manager_buyplates_entry_resource{ResourceId := "gems", Amount := 80, CheckDontTake := false}}
              Unlocks := array{
                tycoon_manager_unlock_entry{TargetRoutingId := "bp_216", Payload := false}
              }
              m_StorageIndex := 134
            }
            "bp_216" => tycoon_manager_buyplates_entry{
              BuyplateId := "bp_216"
              Cost := option{tycoon_manager_buyplates_entry_resource{ResourceId := "money", Amount := 4540000, CheckDontTake := false}}
              Generates := option{tycoon_manager_buyplates_entry_resource{ResourceId := "gems", Amount := 107, CheckDontTake := false}}
              Unlocks := array{
                tycoon_manager_unlock_entry{TargetRoutingId := "bp_217", Payload := false}
              }
              m_StorageIndex := 135
            }
            "bp_217" => tycoon_manager_buyplates_entry{
              BuyplateId := "bp_217"
              Cost := option{tycoon_manager_buyplates_entry_resource{ResourceId := "money", Amount := 6140000, CheckDontTake := false}}
              Generates := option{tycoon_manager_buyplates_entry_resource{ResourceId := "gems", Amount := 145, CheckDontTake := false}}
              Unlocks := array{
                tycoon_manager_unlock_entry{TargetRoutingId := "bp_218", Payload := false}
              }
              m_StorageIndex := 136
            }
            "bp_218" => tycoon_manager_buyplates_entry{
              BuyplateId := "bp_218"
              Cost := option{tycoon_manager_buyplates_entry_resource{ResourceId := "money", Amount := 8300000, CheckDontTake := false}}
              Generates := option{tycoon_manager_buyplates_entry_resource{ResourceId := "gems", Amount := 194, CheckDontTake := false}}
              Unlocks := array{
                tycoon_manager_unlock_entry{TargetRoutingId := "bp_219", Payload := false}
              }
              m_StorageIndex := 137
            }
            "bp_219" => tycoon_manager_buyplates_entry{
              BuyplateId := "bp_219"
              Cost := option{tycoon_manager_buyplates_entry_resource{ResourceId := "money", Amount := 11200000, CheckDontTake := false}}
              Generates := option{tycoon_manager_buyplates_entry_resource{ResourceId := "gems", Amount := 272, CheckDontTake := false}}
              Unlocks := array{
              }
              m_StorageIndex := 138
            }
            "bp_220" => tycoon_manager_buyplates_entry{
              BuyplateId := "bp_220"
              Cost := option{tycoon_manager_buyplates_entry_resource{ResourceId := "money", Amount := 25200000, CheckDontTake := false}}
              Generates := option{tycoon_manager_buyplates_entry_resource{ResourceId := "ruby", Amount := 53, CheckDontTake := false}}
              Unlocks := array{
                tycoon_manager_unlock_entry{TargetRoutingId := "bp_221", Payload := false}
              }
              m_StorageIndex := 139
            }
            "bp_221" => tycoon_manager_buyplates_entry{
              BuyplateId := "bp_221"
              Cost := option{tycoon_manager_buyplates_entry_resource{ResourceId := "money", Amount := 37000000, CheckDontTake := false}}
              Generates := option{tycoon_manager_buyplates_entry_resource{ResourceId := "ruby", Amount := 62, CheckDontTake := false}}
              Unlocks := array{
                tycoon_manager_unlock_entry{TargetRoutingId := "bp_222", Payload := false}
              }
              m_StorageIndex := 140
            }
            "bp_222" => tycoon_manager_buyplates_entry{
              BuyplateId := "bp_222"
              Cost := option{tycoon_manager_buyplates_entry_resource{ResourceId := "money", Amount := 51100000, CheckDontTake := false}}
              Generates := option{tycoon_manager_buyplates_entry_resource{ResourceId := "ruby", Amount := 78, CheckDontTake := false}}
              Unlocks := array{
                tycoon_manager_unlock_entry{TargetRoutingId := "bp_223", Payload := false}
              }
              m_StorageIndex := 141
            }
            "bp_223" => tycoon_manager_buyplates_entry{
              BuyplateId := "bp_223"
              Cost := option{tycoon_manager_buyplates_entry_resource{ResourceId := "money", Amount := 69300000, CheckDontTake := false}}
              Generates := option{tycoon_manager_buyplates_entry_resource{ResourceId := "ruby", Amount := 102, CheckDontTake := false}}
              Unlocks := array{
                tycoon_manager_unlock_entry{TargetRoutingId := "bp_224", Payload := false}
              }
              m_StorageIndex := 142
            }
            "bp_224" => tycoon_manager_buyplates_entry{
              BuyplateId := "bp_224"
              Cost := option{tycoon_manager_buyplates_entry_resource{ResourceId := "money", Amount := 92100000, CheckDontTake := false}}
              Generates := option{tycoon_manager_buyplates_entry_resource{ResourceId := "ruby", Amount := 124, CheckDontTake := false}}
              Unlocks := array{
                tycoon_manager_unlock_entry{TargetRoutingId := "bp_225", Payload := false}
              }
              m_StorageIndex := 143
            }
            "bp_225" => tycoon_manager_buyplates_entry{
              BuyplateId := "bp_225"
              Cost := option{tycoon_manager_buyplates_entry_resource{ResourceId := "money", Amount := 120000000, CheckDontTake := false}}
              Generates := option{tycoon_manager_buyplates_entry_resource{ResourceId := "ruby", Amount := 149, CheckDontTake := false}}
              Unlocks := array{
                tycoon_manager_unlock_entry{TargetRoutingId := "bp_226", Payload := false}
              }
              m_StorageIndex := 144
            }
            "bp_226" => tycoon_manager_buyplates_entry{
              BuyplateId := "bp_226"
              Cost := option{tycoon_manager_buyplates_entry_resource{ResourceId := "money", Amount := 153000000, CheckDontTake := false}}
              Generates := option{tycoon_manager_buyplates_entry_resource{ResourceId := "ruby", Amount := 178, CheckDontTake := false}}
              Unlocks := array{
                tycoon_manager_unlock_entry{TargetRoutingId := "bp_227", Payload := false}
              }
              m_StorageIndex := 145
            }
            "bp_227" => tycoon_manager_buyplates_entry{
              BuyplateId := "bp_227"
              Cost := option{tycoon_manager_buyplates_entry_resource{ResourceId := "money", Amount := 193000000, CheckDontTake := false}}
              Generates := option{tycoon_manager_buyplates_entry_resource{ResourceId := "ruby", Amount := 219, CheckDontTake := false}}
              Unlocks := array{
                tycoon_manager_unlock_entry{TargetRoutingId := "bp_228", Payload := false}
              }
              m_StorageIndex := 146
            }
            "bp_228" => tycoon_manager_buyplates_entry{
              BuyplateId := "bp_228"
              Cost := option{tycoon_manager_buyplates_entry_resource{ResourceId := "money", Amount := 243000000, CheckDontTake := false}}
              Generates := option{tycoon_manager_buyplates_entry_resource{ResourceId := "ruby", Amount := 270, CheckDontTake := false}}
              Unlocks := array{
                tycoon_manager_unlock_entry{TargetRoutingId := "bp_229", Payload := false}
              }
              m_StorageIndex := 147
            }
            "bp_229" => tycoon_manager_buyplates_entry{
              BuyplateId := "bp_229"
              Cost := option{tycoon_manager_buyplates_entry_resource{ResourceId := "money", Amount := 304000000, CheckDontTake := false}}
              Generates := option{tycoon_manager_buyplates_entry_resource{ResourceId := "ruby", Amount := 335, CheckDontTake := false}}
              Unlocks := array{
              }
              m_StorageIndex := 148
            }
            "bp_230" => tycoon_manager_buyplates_entry{
              BuyplateId := "bp_230"
              Cost := option{tycoon_manager_buyplates_entry_resource{ResourceId := "money", Amount := 545000000, CheckDontTake := false}}
              Generates := option{tycoon_manager_buyplates_entry_resource{ResourceId := "treasure", Amount := 105, CheckDontTake := false}}
              Unlocks := array{
                tycoon_manager_unlock_entry{TargetRoutingId := "bp_231", Payload := false}
              }
              m_StorageIndex := 149
            }
            "bp_231" => tycoon_manager_buyplates_entry{
              BuyplateId := "bp_231"
              Cost := option{tycoon_manager_buyplates_entry_resource{ResourceId := "money", Amount := 677000000, CheckDontTake := false}}
              Generates := option{tycoon_manager_buyplates_entry_resource{ResourceId := "treasure", Amount := 120, CheckDontTake := false}}
              Unlocks := array{
                tycoon_manager_unlock_entry{TargetRoutingId := "bp_232", Payload := false}
              }
              m_StorageIndex := 150
            }
            "bp_232" => tycoon_manager_buyplates_entry{
              BuyplateId := "bp_232"
              Cost := option{tycoon_manager_buyplates_entry_resource{ResourceId := "money", Amount := 828000000, CheckDontTake := false}}
              Generates := option{tycoon_manager_buyplates_entry_resource{ResourceId := "treasure", Amount := 139, CheckDontTake := false}}
              Unlocks := array{
                tycoon_manager_unlock_entry{TargetRoutingId := "bp_233", Payload := false}
              }
              m_StorageIndex := 151
            }
            "bp_233" => tycoon_manager_buyplates_entry{
              BuyplateId := "bp_233"
              Cost := option{tycoon_manager_buyplates_entry_resource{ResourceId := "money", Amount := 1000000000, CheckDontTake := false}}
              Generates := option{tycoon_manager_buyplates_entry_resource{ResourceId := "treasure", Amount := 160, CheckDontTake := false}}
              Unlocks := array{
                tycoon_manager_unlock_entry{TargetRoutingId := "bp_234", Payload := false}
              }
              m_StorageIndex := 152
            }
            "bp_234" => tycoon_manager_buyplates_entry{
              BuyplateId := "bp_234"
              Cost := option{tycoon_manager_buyplates_entry_resource{ResourceId := "money", Amount := 1200000000, CheckDontTake := false}}
              Generates := option{tycoon_manager_buyplates_entry_resource{ResourceId := "treasure", Amount := 185, CheckDontTake := false}}
              Unlocks := array{
                tycoon_manager_unlock_entry{TargetRoutingId := "bp_235", Payload := false}
              }
              m_StorageIndex := 153
            }
            "bp_235" => tycoon_manager_buyplates_entry{
              BuyplateId := "bp_235"
              Cost := option{tycoon_manager_buyplates_entry_resource{ResourceId := "money", Amount := 1620000000, CheckDontTake := false}}
              Generates := option{tycoon_manager_buyplates_entry_resource{ResourceId := "treasure", Amount := 214, CheckDontTake := false}}
              Unlocks := array{
                tycoon_manager_unlock_entry{TargetRoutingId := "bp_236", Payload := false}
              }
              m_StorageIndex := 154
            }
            "bp_236" => tycoon_manager_buyplates_entry{
              BuyplateId := "bp_236"
              Cost := option{tycoon_manager_buyplates_entry_resource{ResourceId := "money", Amount := 1920000000, CheckDontTake := false}}
              Generates := option{tycoon_manager_buyplates_entry_resource{ResourceId := "treasure", Amount := 250, CheckDontTake := false}}
              Unlocks := array{
                tycoon_manager_unlock_entry{TargetRoutingId := "bp_237", Payload := false}
              }
              m_StorageIndex := 155
            }
            "bp_237" => tycoon_manager_buyplates_entry{
              BuyplateId := "bp_237"
              Cost := option{tycoon_manager_buyplates_entry_resource{ResourceId := "money", Amount := 2280000000, CheckDontTake := false}}
              Generates := option{tycoon_manager_buyplates_entry_resource{ResourceId := "treasure", Amount := 294, CheckDontTake := false}}
              Unlocks := array{
                tycoon_manager_unlock_entry{TargetRoutingId := "bp_238", Payload := false}
              }
              m_StorageIndex := 156
            }
            "bp_238" => tycoon_manager_buyplates_entry{
              BuyplateId := "bp_238"
              Cost := option{tycoon_manager_buyplates_entry_resource{ResourceId := "money", Amount := 2700000000, CheckDontTake := false}}
              Generates := option{tycoon_manager_buyplates_entry_resource{ResourceId := "treasure", Amount := 345, CheckDontTake := false}}
              Unlocks := array{
                tycoon_manager_unlock_entry{TargetRoutingId := "bp_239", Payload := false}
              }
              m_StorageIndex := 157
            }
            "bp_239" => tycoon_manager_buyplates_entry{
              BuyplateId := "bp_239"
              Cost := option{tycoon_manager_buyplates_entry_resource{ResourceId := "money", Amount := 3190000000, CheckDontTake := false}}
              Generates := option{tycoon_manager_buyplates_entry_resource{ResourceId := "treasure", Amount := 390, CheckDontTake := false}}
              Unlocks := array{
              }
              m_StorageIndex := 158
            }
            "bp_240" => tycoon_manager_buyplates_entry{
              BuyplateId := "bp_240"
              Cost := option{tycoon_manager_buyplates_entry_resource{ResourceId := "money", Amount := 4820000000, CheckDontTake := false}}
              Generates := option{tycoon_manager_buyplates_entry_resource{ResourceId := "treasurechest", Amount := 156, CheckDontTake := false}}
              Unlocks := array{
                tycoon_manager_unlock_entry{TargetRoutingId := "bp_241", Payload := false}
              }
              m_StorageIndex := 159
            }
            "bp_241" => tycoon_manager_buyplates_entry{
              BuyplateId := "bp_241"
              Cost := option{tycoon_manager_buyplates_entry_resource{ResourceId := "money", Amount := 5430000000, CheckDontTake := false}}
              Generates := option{tycoon_manager_buyplates_entry_resource{ResourceId := "treasurechest", Amount := 179, CheckDontTake := false}}
              Unlocks := array{
                tycoon_manager_unlock_entry{TargetRoutingId := "bp_242", Payload := false}
              }
              m_StorageIndex := 160
            }
            "bp_242" => tycoon_manager_buyplates_entry{
              BuyplateId := "bp_242"
              Cost := option{tycoon_manager_buyplates_entry_resource{ResourceId := "money", Amount := 6130000000, CheckDontTake := false}}
              Generates := option{tycoon_manager_buyplates_entry_resource{ResourceId := "treasurechest", Amount := 202, CheckDontTake := false}}
              Unlocks := array{
                tycoon_manager_unlock_entry{TargetRoutingId := "bp_243", Payload := false}
              }
              m_StorageIndex := 161
            }
            "bp_243" => tycoon_manager_buyplates_entry{
              BuyplateId := "bp_243"
              Cost := option{tycoon_manager_buyplates_entry_resource{ResourceId := "money", Amount := 7610000000, CheckDontTake := false}}
              Generates := option{tycoon_manager_buyplates_entry_resource{ResourceId := "treasurechest", Amount := 227, CheckDontTake := false}}
              Unlocks := array{
                tycoon_manager_unlock_entry{TargetRoutingId := "bp_244", Payload := false}
              }
              m_StorageIndex := 162
            }
            "bp_244" => tycoon_manager_buyplates_entry{
              BuyplateId := "bp_244"
              Cost := option{tycoon_manager_buyplates_entry_resource{ResourceId := "money", Amount := 8590000000, CheckDontTake := false}}
              Generates := option{tycoon_manager_buyplates_entry_resource{ResourceId := "treasurechest", Amount := 265, CheckDontTake := false}}
              Unlocks := array{
                tycoon_manager_unlock_entry{TargetRoutingId := "bp_245", Payload := false}
              }
              m_StorageIndex := 163
            }
            "bp_245" => tycoon_manager_buyplates_entry{
              BuyplateId := "bp_245"
              Cost := option{tycoon_manager_buyplates_entry_resource{ResourceId := "money", Amount := 10600000000, CheckDontTake := false}}
              Generates := option{tycoon_manager_buyplates_entry_resource{ResourceId := "treasurechest", Amount := 294, CheckDontTake := false}}
              Unlocks := array{
                tycoon_manager_unlock_entry{TargetRoutingId := "bp_246", Payload := false}
              }
              m_StorageIndex := 164
            }
            "bp_246" => tycoon_manager_buyplates_entry{
              BuyplateId := "bp_246"
              Cost := option{tycoon_manager_buyplates_entry_resource{ResourceId := "money", Amount := 12000000000, CheckDontTake := false}}
              Generates := option{tycoon_manager_buyplates_entry_resource{ResourceId := "treasurechest", Amount := 329, CheckDontTake := false}}
              Unlocks := array{
                tycoon_manager_unlock_entry{TargetRoutingId := "bp_247", Payload := false}
              }
              m_StorageIndex := 165
            }
            "bp_247" => tycoon_manager_buyplates_entry{
              BuyplateId := "bp_247"
              Cost := option{tycoon_manager_buyplates_entry_resource{ResourceId := "money", Amount := 13500000000, CheckDontTake := false}}
              Generates := option{tycoon_manager_buyplates_entry_resource{ResourceId := "treasurechest", Amount := 362, CheckDontTake := false}}
              Unlocks := array{
                tycoon_manager_unlock_entry{TargetRoutingId := "bp_248", Payload := false}
              }
              m_StorageIndex := 166
            }
            "bp_248" => tycoon_manager_buyplates_entry{
              BuyplateId := "bp_248"
              Cost := option{tycoon_manager_buyplates_entry_resource{ResourceId := "money", Amount := 17700000000, CheckDontTake := false}}
              Generates := option{tycoon_manager_buyplates_entry_resource{ResourceId := "treasurechest", Amount := 402, CheckDontTake := false}}
              Unlocks := array{
                tycoon_manager_unlock_entry{TargetRoutingId := "bp_249", Payload := false}
              }
              m_StorageIndex := 167
            }
            "bp_249" => tycoon_manager_buyplates_entry{
              BuyplateId := "bp_249"
              Cost := option{tycoon_manager_buyplates_entry_resource{ResourceId := "money", Amount := 22800000000, CheckDontTake := false}}
              Generates := option{tycoon_manager_buyplates_entry_resource{ResourceId := "treasurechest", Amount := 470, CheckDontTake := false}}
              Unlocks := array{
              }
              m_StorageIndex := 168
            }
          }

        if (
            RebirthConfig := m_TycoonConfig.Rebirth?
            MaxPersistenceRebirthLevel := FromBits(array{true}.Repeat(p_RebirthLevelReserveBits))
            RebirthConfig.MaxLevel > MaxPersistenceRebirthLevel
        ) {
            Log("Tycoon managed has max rebirth level to {RebirthConfig.MaxLevel} but only {p_RebirthLevelReserveBits} bits (max {MaxPersistenceRebirthLevel}) are allocated for rebirth persistence", ?Level := log_level.Error)
        }

        for (Index->ProgressLine:p_ProgressLines) {
            set ProgressLine.m_Index = Index
            if (AccessControl := ProgressLine.AccessControl?, AccessControl.p_Mode = access_control_mode.Whitelist) {
                JoinClosure := tycoon_manager_player_closure{
                    Device := Self
                    ProgressLine := ProgressLine
                }
                # since access control is in whitelist mode, and these events are backed by ordered arrays,
                # we can make the assumption that OnPlayerAdded is processed before persistence load
                AccessControl.SubscribeToPlayerJoinEvents(JoinClosure.OnPlayerAdded, JoinClosure.OnPlayerRemoved)

                if (ClaimBuyplate := ProgressLine.ClaimBuyplate?) {
                    Middleware := tycoon_manager_claim_buyplate_middleware{
                        Device := Self
                        ProgressLine := ProgressLine
                    }
                    ClaimBuyplate.AddMiddleware(Middleware)
                }
                if (p_BuyPlateCostMetTrackingRefreshRate > 0.0) {
                    spawn:
                        StartCostMetTrackingLoop(ProgressLine, AccessControl)
                    Closure := tycoon_manager_buyplate_cost_met_reflection_closure{
                        Device := Self
                        ProgressLine := ProgressLine
                    }
                    p_ResourceManager.SubscribeToResourceTake(Closure.OnResourceTake)
                }

                if (p_EnablePersistence?) {
                    PersistenceRecorder := persistence_recorder_device{
                        p_DevId := "progress line {ProgressLine.Name}"
                        p_PlayerTrackingMode := persistence_recorder_player_tracking_mode.AccessControlWhitelist
                        p_AccessControl := AccessControl
                        p_StorageSaveSlots := p_PersistenceSaveSlots
                        p_SavePoints := p_PersistenceSavePoints
                        p_AutoSaveInterval := p_PersistenceAutoSaveInternal
                        p_UseDefaultSelectMenu := p_PersistenceSelectMenuEnabled
                        p_DefaultSelectMenuConfig := p_PersistenceSelectMenuConfig
                        p_UseDialogSelectMenu := p_PersistenceUseDialogSelectMenu
                        p_SkipSelectMenuOnFirstPlay := true
                        p_CustomBuyPlatesReserveBits := p_PersistenceBuyPlatesReserveBits
                        p_BuyPlateRecoveryChunkSize := p_PersistenceBuyPlateRecoveryChunkSize
                        p_BuyPlateRecoveryChunkDelay := p_PersistenceBuyPlateRecoveryChunkDelay
                    }
                    if (m_TycoonConfig.Rebirth?) {
                        RebirthDummyDevice := tycoon_manager_rebirth_dummy_device{
                            TycoonManager := Self
                            ProgressLine := ProgressLine
                        }
                        Handle := PersistenceRecorder.RegisterRecordable(RebirthDummyDevice)
                        set ProgressLine.m_RebirthRecordableHandle = Some(Handle)
                    }
                    if (p_PersistenceIncludeResourceManager?) {
                        PersistenceRecorder.RegisterDeviceToRecord(p_ResourceManager)
                    }
                    for (TimeRewardsDevice:p_TimeRewardsDevice) {
                        PersistenceRecorder.RegisterDeviceToRecord(TimeRewardsDevice)
                    }
                    for (Device:p_PersistenceExtraDevices, persistence_recordable_base[Device]) {
                        PersistenceRecorder.RegisterDeviceToRecord(Device)
                    }
                    spawn:
                        PersistenceRecorder.OnBegin()
                    set ProgressLine.m_PersistenceRecorder = Some(PersistenceRecorder)

                    Closure := tycoon_manager_player_persistence_load_closure{
                        Callback := OnPlayerSaveSlotSelected
                        ProgressLine := ProgressLine
                    }
                    SubscribeToEvent(PersistenceRecorder.PlayerSaveSlotSelectedEvent, Closure.Call)
                }

                DeviceRegisterClosure := tycoon_manager_device_register_closure{
                    Device := Self
                    ProgressLine := ProgressLine
                }
                Subscribed := AccessControl.SubscribeToDevicesBackrefs(DeviceRegisterClosure.Call)
                if (Subscribed?) {
                    if (ClaimBuyplate := ProgressLine.ClaimBuyplate?) {
                        OnDeviceRegistered(ProgressLine, ClaimBuyplate)
                    }
                    for (BuyPlate:ProgressLine.ExtraBuyplates) {
                        OnDeviceRegistered(ProgressLine, BuyPlate)
                    }
                } else {
                    Log("Could not subscribe to access control device of progress line \"{ProgressLine.Name}\" because backrefs building is not enabled", ?Level := log_level.Error)
                }

                ResetBeaconLines(ProgressLine)
            } else if (ProgressLine.AccessControl?) {
                Log("Progress line \"{ProgressLine.Name}\" has an access control, but its mode is not Whitelist", ?Level := log_level.Error)
            } else {
                Log("Progress line \"{ProgressLine.Name}\" is missing an access control device", ?Level := log_level.Error)
            }
        }

        GetPlayspace().PlayerRemovedEvent().Subscribe(OnPlayerQuit)
        for (BeaconLineHideArea:p_BeaconLinesHideAreas) {
            Closure := tycoon_manager_beacon_line_area_closure{
                Device := Self
                Config := BeaconLineHideArea
            }

            BeaconLineHideArea.Volume.AgentEntersEvent.Subscribe(Closure.OnAgentEntered)
            OuterVolume := BeaconLineHideArea.OuterVolume? or BeaconLineHideArea.Volume
            OuterVolume.AgentExitsEvent.Subscribe(Closure.OnAgentExited)
        }

        if (p_EnablePersistence?) {
            for (Index->Device:p_PersistenceExtraDevices, not persistence_recordable_base[Device]) {
                Log("Tycoon manager is configured with an extra persistence device that does not support persistence (index {Index})", ?Level := log_level.Error)
            }
        }

        if (RebirthConfig := m_TycoonConfig.Rebirth?) {
            for (Trigger:p_RebirthNextLevelOnTrigger) {
                Trigger.TriggeredEvent.Subscribe(OnRebirthNextLevelTriggerCalled)
            }

            HudManager := tycoon_manager_rebirth_hud_manager{
                RefreshRate := p_RebirthHUDManagerConfig.RefreshRate
                ZOrder := p_RebirthHUDManagerConfig.ZOrder
                Device := Self
                Config := RebirthConfig
            }
            set m_RebirthHudManager = Some(HudManager)
            spawn:
                HudManager.Run(GetPlayspace())
        }

        for (
            PetRebirthConfig:p_RebirthPets
            set m_RebirthPets[PetRebirthConfig.PetId] = PetRebirthConfig.RebirthLevel
        ) {}
        for (PetController:p_PetController) {
            for (PetId->__:m_RebirthPets) {
                PetController.SetPetConditionProvider(PetId, Self)
            }
        }

        AwareSubscribeToPostInit(OnPostInit)

    OnPostInit<private>():void=
        set m_Initialized = true
        for (ProgressLine:p_ProgressLines) {
            for (BeaconLine:ProgressLine.BeaconLines) {
                set BeaconLine.m_CurrentEntry = 0
                set BeaconLine.m_LockedTargetBuyPlate = None()
                UpdateBeaconLine(ProgressLine, BeaconLine)
            }
        }

    OnPlayerQuit<private>(Player:player):void=
        for (BeaconLineArea:p_BeaconLinesHideAreas) {
            set BeaconLineArea.m_InVolume = BeaconLineArea.m_InVolume.RemoveByKey(Player)(0)
        }
        for (ProgressLine:p_ProgressLines) {
            for (BeaconLine:ProgressLine.BeaconLines) {
                set BeaconLine.m_MapIndicatorPulseToAgents = BeaconLine.m_MapIndicatorPulseToAgents.RemoveByKey(Player)(0)
            }
        }
        if (PlayerState := m_Players[Player]) {
            OnPlayerRemoved(PlayerState.ProgressLine, Player)
            set m_Players = m_Players.RemoveByKey(Player)(0)
        }

    OnPlayerAdded<internal>(ProgressLine:tycoon_manager_progress_line, Player:player):void=
        PlayerState := tycoon_manager_player_state{
            ProgressLine := ProgressLine
        }
        if. set m_Players[Player] = PlayerState
        if (ProgressLine.ClassSlot > 0) {
            p_ClassManager.SetClassSlot(Player, ProgressLine.ClassSlot)
        }
        SyncPlayerBeaconLines(Player)
        if (AccessControl := ProgressLine.AccessControl?) {
            UpdateCostMetTracking(ProgressLine, AccessControl)
        }
        PlayerJoinProgressLineEvent.Signal(Player, ProgressLine)

    OnPlayerRemoved<internal>(ProgressLine:tycoon_manager_progress_line, Player:player):void=
        SyncPlayerBeaconLines(Player)
        DoQuitRollback(ProgressLine, Player)
        if (ProgressLine.ClassSlot > 0) {
            p_ClassManager.ResetClass(Player)
        }
        if (PlayerState := m_Players[Player], PlayerState.ProgressLine.m_Index = ProgressLine.m_Index) {
            set m_Players = m_Players.RemoveByKey(Player)(0)
        }

    SubscribeToRebirthEvent<public>(Listener:type{__(:tycoon_rebirth_event):void}):void=
        set m_RebirthListeners += array{Listener}

    IsPlaying<public>(Player:player)<transacts><decides>:void=
        m_Players[Player]

    ForceExpelPlayer<public>(Player:player):logic=
        if (
            PlayerState := m_Players[Player]
            AccessControl := PlayerState.ProgressLine.AccessControl?
            AccessControl.IsAllowed[Player]
        ) {
            AccessControl.RemoveFromWhitelist(Player, false)
            true
        } else {
            false
        }

    TeleportHome<public>(Player:player):logic=
        if (ProgressLine := GetPlayerState[Player].ProgressLine, Teleporter := ProgressLine.FrontTeleporter?) {
            Teleporter.Teleport(Player)
            true
        } else if (TeleportHome0[Player]) {
            true
        } else {
            false
        }

    # Attempts to teleport the player to the front checkpoint of their progress line.
    # Fails if the player is not part of a progress line, if no front checkpoint device was linked, or if .TeleportTo[] fails.
    TeleportHome0<private>(Player:player)<decides><transacts>:void=
        ProgressLine := GetPlayerState[Player].ProgressLine
        Checkpoint := ProgressLine.FrontCheckpoint?
        FortCharacter := Player.GetFortCharacter[]
        TargetTransform := Checkpoint.GetTransform()
        TargetPosition := TargetTransform.Translation.AddZ(p_MiscPlayerHeight)
        FortCharacter.StallingTeleportTo[TargetPosition, TargetTransform.Rotation]

    ClaimOrTeleportHome<public>(Player:player):tycoon_claim_or_teleport_result=
        if (IsPlaying[Player]) {
            TPSuccess := TeleportHome(Player)
            if (TPSuccess?) {
                return tycoon_claim_or_teleport_result.Teleported
            } else {
                Log("Couldn't teleport '{Player}' to their progress line, check editor logs for reason", ?Level := log_level.Error)
            }
        } else {
            var InUseProgressLines: [int]void = map{}
            for (__->PlayerState:m_Players) {
                if. set InUseProgressLines[PlayerState.ProgressLine.m_Index] = void
            }
            for (ProgressLine:p_ProgressLines, not InUseProgressLines[ProgressLine.m_Index]) {
                if (ClaimBuyplate := ProgressLine.ClaimBuyplate?) {
                    ClaimBuyplate.Unlock(Player, ?UserTriggered := false, ?FirstUnlock := false, ?Force := true)
                } else {
                    ClaimProgressLineFromBuyplate(ProgressLine, Player)
                }
                TPSuccess := TeleportHome(Player)
                if (not TPSuccess?) {
                    Log("Couldn't teleport '{Player}' to their progress line after claiming, check editor logs for reason", ?Level := log_level.Error)
                }
                return tycoon_claim_or_teleport_result.Claimed
            }
            Log("Couldn't auto claim a progress line for '{Player}': all progress lines are in use", ?Level := log_level.Error)
        }
        tycoon_claim_or_teleport_result.Failed

    ClaimProgressLineFromBuyplate<internal>(ProgressLine:tycoon_manager_progress_line, Player:agent):void=
        if (AccessControl := ProgressLine.AccessControl?) {
            AccessControl.AddToWhitelist(Player)
        } else {
            Log("Couldn't claim progress line #{ProgressLine.m_Index} for '{Player}': missing access control", ?Level := log_level.Error)
        }

    DoQuitRollback<private>(ProgressLine:tycoon_manager_progress_line, Player:player):void=
        if (not (p_EnableQuitRollback?) or ProgressLine.AccessControl?.ListAllowedAgents().Length > 0) {
            return
        }
        set ProgressLine.m_RebirthLevel = 0
        RollbackAgent: ?agent = case(p_QuitRollbackGranularity):
            quit_rollback_device_granularity.Player => Some(Player)
            quit_rollback_device_granularity.Omitted => false
        DoRollback(ProgressLine, RollbackAgent)
        if (ClaimBuyplate := ProgressLine.ClaimBuyplate?) {
            ClaimBuyplate.Reset(RollbackAgent)
        }
        EnsureRebirthEffects(ProgressLine)

    DoRollback<private>(ProgressLine:tycoon_manager_progress_line, RollbackAgent:?agent):void=
        for (Ticket:ProgressLine.m_AutoGenerationTickets) {
            Ticket.Disable()
        }
        var RollbackDeviceCount: int = 0
        if (DeviceBackrefs := ProgressLine.AccessControl?.GetUsingDevices[]) {
            for (Device:DeviceBackrefs, Resetable := resetable[Device]) {
                Resetable.Reset(RollbackAgent)
                set RollbackDeviceCount += 1
            }
        }
        for (BuyPlateDevice:ProgressLine.ExtraBuyplates, BuyPlate := buy_plate_device[BuyPlateDevice]) {
            BuyPlate.Reset(RollbackAgent)
            set RollbackDeviceCount += 1
        }
        for (Device:ProgressLine.ExtraRollbackDevices, Resetable := resetable[Device]) {
            Resetable.Reset(RollbackAgent)
            set RollbackDeviceCount += 1
        }
        ResetBeaconLines(ProgressLine)
        if (l_Agent := RollbackAgent?) {
            for (Trigger:ProgressLine.ExtraRollbackTriggers) {
                Trigger.Trigger(l_Agent)
            }
        } else {
            for (Trigger:ProgressLine.ExtraRollbackTriggers) {
                Trigger.Trigger()
            }
        }
        Log("Rolled back progress line \"{ProgressLine.Name}\", {RollbackDeviceCount} affected devices", ?Level := log_level.Debug)

    StartCostMetTrackingLoop<private>(
        ProgressLine:tycoon_manager_progress_line,
        AccessControl:access_control_device
    )<suspends>:void=
        loop {
            UpdateCostMetTracking(ProgressLine, AccessControl)
            Sleep(p_BuyPlateCostMetTrackingRefreshRate)
        }

    UpdateCostMetTracking<internal>(
        ProgressLine:tycoon_manager_progress_line,
        AccessControl:access_control_device
    ):void=
        if (Agent := AccessControl.ListAllowedAgents()[0]) {
            var NewStateBy: []string = array{}
            for (Device:AccessControl.GetUsingDevices[], BuyPlate := buy_plate_device[Device]) {
                NewState := BuyPlate.UpdateCostMetTracking(Agent)
                if (NewState?) {
                    set NewStateBy += array{BuyPlate.GetRoutingId()}
                }
            }
            for (
                BeaconLine:ProgressLine.BeaconLines
                BeaconLineConfig := m_TycoonConfig.BeaconLines[BeaconLine.BeaconLineId]
                LineEntry := BeaconLineConfig.InOrderBuyPlates[BeaconLine.m_CurrentEntry]
                LineEntry.ContainsAnyBuyPlateOf[NewStateBy]
            ) {
                UpdateBeaconLine(ProgressLine, BeaconLine)
            }
        }

    ResetBeaconLines<private>(ProgressLine:tycoon_manager_progress_line):void=
        for (BeaconLine:ProgressLine.BeaconLines) {
            set BeaconLine.m_CurrentEntry = 0
            set BeaconLine.m_LockedTargetBuyPlate = None()
            UpdateBeaconLine(ProgressLine, BeaconLine)
        }

    IsBeaconLineEntryComplete<private>(
        ProgressLine:tycoon_manager_progress_line,
        LineEntry:tycoon_manager_beacon_line_config_entry
    )<decides><transacts>:void=
        var AllUnlocked: logic = true
        for (BuyPlateId->__:LineEntry.AvailableOptions, AllUnlocked?) {
            if (BuyPlate := FindBuyPlateById[ProgressLine, BuyPlateId]) {
                if (not BuyPlate.IsUnlocked[]) {
                    set AllUnlocked = false
                }
            } else {
                set AllUnlocked = false
            }
        }
        AllUnlocked?

    BestRankingBeaconLineEntryBuyPlate<private>(
        ProgressLine:tycoon_manager_progress_line,
        LineEntry:tycoon_manager_beacon_line_config_entry
    ):?buy_plate_device=
        Agent := option{ProgressLine.AccessControl?.ListAllowedAgents()[0]}

        var AnyCostMet: logic = false
        EligibleEntries := for (
            BuyPlateId->__:LineEntry.AvailableOptions
            BuyPlate := FindBuyPlateById[ProgressLine, BuyPlateId]
            BuyPlate.IsEnabled[]
            not BuyPlate.IsUnlocked[]
        ) {
            CostMet := if (l_Agent := Agent?) {
                BuyPlate.DoesAgentMeetsCostCached(l_Agent)
            } else {
                false
            }
            if (CostMet?, not AnyCostMet?) {
                set AnyCostMet = true
            }
            (BuyPlate, CostMet)
        }

        if (
            AnyCostMet?
            Player := player[Agent?]
            FortPlayer := Player.GetFortCharacter[]
            AgentPosition := FortPlayer.GetTransform().Translation
        ) {
            var TopRanking: ?tuple(buy_plate_device,float) = false
            for (Entry:EligibleEntries, Entry(1)?) {
                BuyPlate := Entry(0)
                Dist := Distance(AgentPosition, BuyPlate.GetPosition())
                if ((not TopRanking?) or (Dist < TopRanking?(1))) {
                    set TopRanking = Some(Entry(0), Dist)
                }
            }
            option{TopRanking?(0)}
        } else {
            option{EligibleEntries[0](0)}
        }

    UpdateBeaconLine<private>(
        ProgressLine:tycoon_manager_progress_line,
        BeaconLine:tycoon_manager_progress_line_beacon_line
    ):void=
        var NewEntry: logic = false
        loop {
            if (
                BeaconLineConfig := m_TycoonConfig.BeaconLines[BeaconLine.BeaconLineId]
                LineEntry := BeaconLineConfig.InOrderBuyPlates[BeaconLine.m_CurrentEntry]
                IsBeaconLineEntryComplete[ProgressLine, LineEntry]
            ) {
                # handle unordered unlocking
                set BeaconLine.m_CurrentEntry += 1
                set BeaconLine.m_LockedTargetBuyPlate = None()
                set NewEntry = true
            } else {
                break
            }
        }

        var MapIndicatorPointTo: ?tycoon_manager_beacon_line_buy_plate_state = false
        if (
            BeaconLineConfig := m_TycoonConfig.BeaconLines[BeaconLine.BeaconLineId]
            LineEntry := BeaconLineConfig.InOrderBuyPlates[BeaconLine.m_CurrentEntry]
        ) {
            if (NewEntry?) {
                for (BuyPlateId->__:LineEntry.AvailableOptions, not FindBuyPlateById[ProgressLine, BuyPlateId]) {
                    Log("Buy plate \"{BuyPlateId}\" is part of the now active entry in beacon line \"{BeaconLine.BeaconLineId}\" but is not registered", ?Level := log_level.Warning)
                }
            }

            TargetBuyplate := if (
                LockedTargetBuyPlateId := BeaconLine.m_LockedTargetBuyPlate?
                LineEntry.ContainsBuyPlate[LockedTargetBuyPlateId]
                BuyPlate := FindBuyPlateById[ProgressLine, LockedTargetBuyPlateId]
                BuyPlate.IsEnabled[]
                not BuyPlate.IsUnlocked[]
            ) {
                Some(BuyPlate)
            } else {
                TopRanking := BestRankingBeaconLineEntryBuyPlate(ProgressLine, LineEntry)
                TopRankingStr := if (BuyPlate := TopRanking?) {
                    ToString(BuyPlate)
                } else {
                    "None"
                }
                Log(MsgRankedBeaconLineBuyPlates(BeaconLine.BeaconLineId, TopRankingStr), ?Level := log_level.Verbose)
                TopRanking
            }

            if (BuyPlate := TargetBuyplate?) {
                UnmetCost := if (Agent := ProgressLine.AccessControl?.ListAllowedAgents()[0]) {
                    InvertLogic(BuyPlate.DoesAgentMeetsCostCached(Agent))
                } else {
                    false
                }
                if (not UnmetCost?) {
                    set BeaconLine.m_LockedTargetBuyPlate = Some(BuyPlate.GetRoutingId())
                }

                if (
                    BuyPlate = BeaconLine.m_MapIndicatorState?.BuyPlate
                    UnmetCost = BeaconLine.m_MapIndicatorState?.UnmetCost
                ) {
                    return
                }
                set MapIndicatorPointTo = Some(tycoon_manager_beacon_line_buy_plate_state{
                    BuyPlate := BuyPlate
                    UnmetCost := UnmetCost
                })
            }
        }

        if (
            BuyPlateState := MapIndicatorPointTo?
            Position := BuyPlateState.BuyPlate.GetPosition(?ExtraOffset := p_BeaconLinesMapIndicatorOffset)
            Transform := transform{
                Translation := Position
                Rotation := IdentityRotation()
                Scale := MakeUniformVec3(1.0)
            }
            MapIndicatorVariant := if (BuyPlateState.UnmetCost?, BeaconLine.MapIndicatorUnmetCost?) {
                tycoon_manager_beacon_line_map_indicator_variant.UnmetCost
            } else {
                tycoon_manager_beacon_line_map_indicator_variant.Primary
            }
            MapIndicator := case(MapIndicatorVariant):
                tycoon_manager_beacon_line_map_indicator_variant.Primary => BeaconLine.MapIndicator
                tycoon_manager_beacon_line_map_indicator_variant.UnmetCost => BeaconLine.MapIndicatorUnmetCost?
            MapIndicator.TeleportTo[Transform]
        ) {
            set BeaconLine.m_MapIndicatorState = Some(BuyPlateState)

            if (PrevVariant := BeaconLine.m_MapIndicatorEnabled?, PrevVariant <> MapIndicatorVariant) {
                BeaconLine.DisablePulse(ProgressLine)
                if (PrevMapIndicator := BeaconLine.GetActiveMapIndicator[]) {
                    PrevMapIndicator.Disable()
                }
                set BeaconLine.m_MapIndicatorEnabled = false
            }
            if (not BeaconLine.m_MapIndicatorEnabled?) {
                set BeaconLine.m_MapIndicatorEnabled = Some(MapIndicatorVariant)
                MapIndicator.Enable()
                Log(MsgBeaconLineMapIndicatorVariantEnabled(BeaconLine.BeaconLineId, ToString(MapIndicatorVariant)),
                    ?Level := log_level.Verbose)
            }
            if (not BeaconLine.m_MapIndicatorPulseActive?) {
                set BeaconLine.m_MapIndicatorPulseActive = true
                for (
                    Agent:ProgressLine.AccessControl?.ListAllowedAgents()
                    not IsWithinBeaconLineHideArea[Agent]
                ) {
                    spawn:
                        ActivateObjectivePulseWithWarmup(Agent, BeaconLine)
                }
            }
        } else {
            set BeaconLine.m_MapIndicatorState = false

            BeaconLine.DisablePulse(ProgressLine)
            if (MapIndicator := BeaconLine.GetActiveMapIndicator[], MapIndicatorVariant := BeaconLine.m_MapIndicatorEnabled?) {
                MapIndicator.Disable()
                set BeaconLine.m_MapIndicatorEnabled = false
                Log(MsgBeaconLineMapIndicatorVariantDisabled(BeaconLine.BeaconLineId, ToString(MapIndicatorVariant)),
                    ?Level := log_level.Verbose)
            }
        }

    MsgRankedBeaconLineBuyPlates<private><localizes>(BeaconLineId:string, BuyPlateStr:string):message=
        "Beacon line \"{BeaconLineId}\" elected {BuyPlateStr} as best ranking buy plate"
    MsgBeaconLineMapIndicatorVariantEnabled<private><localizes>(BeaconLineId:string, VariantStr:string):message=
        "Beacon line \"{BeaconLineId}\" enabled map indicator variant {VariantStr}"
    MsgBeaconLineMapIndicatorVariantDisabled<private><localizes>(BeaconLineId:string, VariantStr:string):message=
        "Beacon line \"{BeaconLineId}\" disabled map indicator"

    IsWithinBeaconLineHideArea<private>(Agent:agent)<transacts><decides>:void=
        var Within: logic = false
        for (BeaconLineArea:p_BeaconLinesHideAreas, BeaconLineArea.m_InVolume[Agent]) {
            set Within = true
        }
        Within?

    SyncPlayerBeaconLines<internal>(Player:player):void=
        if (
            PlayerState := m_Players[Player]
            ProgressLine := PlayerState.ProgressLine
            AccessControl := ProgressLine.AccessControl?
            AccessControl.IsAllowed[Player]
            not IsWithinBeaconLineHideArea[Player]
        ) {
            for (
                BeaconLine:ProgressLine.BeaconLines
                BeaconLine.m_MapIndicatorPulseActive?
                not BeaconLine.m_MapIndicatorPulseToAgents[Player]
            ) {
                spawn:
                    ActivateObjectivePulseWithWarmup(Player, BeaconLine)
            }
            Log(MsgSyncPlayerBeaconLines(Player, "active"), ?Level := log_level.Verbose)
        } else if (
            PlayerState := m_Players[Player]
            ProgressLine := PlayerState.ProgressLine
        ) {
            for (
                BeaconLine:ProgressLine.BeaconLines
                BeaconLine.m_MapIndicatorPulseToAgents[Player]
                MapIndicator := BeaconLine.GetActiveMapIndicator[]
            ) {
                MapIndicator.DeactivateObjectivePulse(Player)
                set BeaconLine.m_MapIndicatorPulseToAgents = BeaconLine.m_MapIndicatorPulseToAgents.RemoveByKey(Player)(0)
            }
            Log(MsgSyncPlayerBeaconLines(Player, "inactive"), ?Level := log_level.Verbose)
        }

    MsgSyncPlayerBeaconLines<private><localizes>(Agent:agent, StateStr:string):message=
        "Syncing beacon lines for '{Agent}'. New State: {StateStr}"

    ActivateObjectivePulseWithWarmup<private>(
        Agent:agent,
        BeaconLine:tycoon_manager_progress_line_beacon_line
    )<suspends>:void=
        if. set BeaconLine.m_MapIndicatorPulseToAgents[Agent] = void
        var It: int = 0
        loop {
            if (
                BeaconLine.m_MapIndicatorPulseActive?
                It < 10
                MapIndicator := BeaconLine.GetActiveMapIndicator[]
            ) {
                MapIndicator.ActivateObjectivePulse(Agent)
                Sleep(0.1)
            } else {
                break
            }
            set It += 1
        }

    FindBuyPlateById<private>(ProgressLine:tycoon_manager_progress_line, BuyPlateId:string)<transacts><decides>:buy_plate_device=
        ProgressLine.m_BuyPlatesBackrefCache[BuyPlateId] or FindBuyPlateById0[ProgressLine, BuyPlateId]

    FindBuyPlateById0<private>(ProgressLine:tycoon_manager_progress_line, BuyPlateId:string)<transacts><decides>:buy_plate_device=
        var Result: ?buy_plate_device = false
        if (DeviceBackrefs := ProgressLine.AccessControl?.GetUsingDevices[]) {
            for (
                Device:DeviceBackrefs,
                not (Result?),
                BuyPlate := buy_plate_device[Device],
                BuyPlate.GetRoutingId() = BuyPlateId
            ) {
                set Result = Some(BuyPlate)
            }
        }
        Result?

    OnDeviceRegistered<internal>(ProgressLine:tycoon_manager_progress_line, Device:creative_device):void=
        if (BuyPlate := buy_plate_device[Device]) {
            BuyPlateId := BuyPlate.GetRoutingId()
            if. set ProgressLine.m_BuyPlatesBackrefCache[BuyPlateId] = BuyPlate
            if (Config := m_BuyplatesConfigs[BuyPlate.GetRoutingId()]) {
                Closure := tycoon_manager_configured_buy_plate_closure{
                    Device := Self
                    ProgressLine := ProgressLine
                    Config := Config
                }
                if. BuyPlate.SubscribeToEvent["unlock", Closure.OnUnlock]
                if. BuyPlate.SubscribeToEvent["enable", Closure.OnEnable]

                if (Cost := Config.Cost?) {
                    CostOverride := tycoon_manager_buy_plate_cost_override{
                        Device := Self
                        Cost := Cost
                    }
                    BuyPlate.SetCostOverride(CostOverride)
                } else if (BuyPlate.p_ExpectsCostOverride? or p_AlwaysSetCostOverride?) {
                    BuyPlate.SetCostOverride(buy_plate_free_cost_override{})
                }
            } else {
                Closure := tycoon_manager_any_buy_plate_closure{
                    Device := Self
                    ProgressLine := ProgressLine
                    BuyPlateId := BuyPlate.GetRoutingId()
                }
                if. BuyPlate.SubscribeToEvent["unlock", Closure.OnUnlock]
                if. BuyPlate.SubscribeToEvent["enable", Closure.OnEnable]
            }
            if (m_Initialized?) {
                for (
                    BeaconLineId->BeaconLineConfig:m_TycoonConfig.BeaconLines
                    LineEntry := BeaconLineConfig.InOrderBuyPlates[0]
                    LineEntry.ContainsBuyPlate[BuyPlateId]
                ) {
                    for (BeaconLine:ProgressLine.BeaconLines, BeaconLine.BeaconLineId = BeaconLineId) {
                        UpdateBeaconLine(ProgressLine, BeaconLine)
                    }
                }
            }
        }
        if (PersistenceRecorder := ProgressLine.m_PersistenceRecorder?) {
            if (BuyPlate := buy_plate_device[Device], BuyPlateId := BuyPlate.GetRoutingId()) {
                FixedIndex := if (Index := m_BuyplatesConfigs[BuyPlateId].m_StorageIndex) {
                    Index
                } else {
                    Log("Tycoon manager is to persist buyplate \"{BuyPlateId}\" but its ordering cannot be ensured as it is not present in the config", ?Level := log_level.Warning)
                    -1
                }
                PersistenceRecorder.RegisterBuyPlateToRecord(BuyPlate, ?FixedIndex := FixedIndex)
            } else {
                PersistenceRecorder.RegisterDeviceToRecord(Device, ?Silent := true)
            }
        }

    OnAnyBuyPlateEnabled<internal>(ProgressLine:tycoon_manager_progress_line, BuyplateId:string):void=
        for (
            BeaconLine:ProgressLine.BeaconLines
            BeaconLineConfig := m_TycoonConfig.BeaconLines[BeaconLine.BeaconLineId]
            LineEntry := BeaconLineConfig.InOrderBuyPlates[BeaconLine.m_CurrentEntry]
            LineEntry.ContainsBuyPlate[BuyplateId]
        ) {
            UpdateBeaconLine(ProgressLine, BeaconLine)
        }
        if (AccessControl := ProgressLine.AccessControl?) {
            UpdateCostMetTracking(ProgressLine, AccessControl)
        }

    OnAnyBuyPlateUnlocked<internal>(ProgressLine:tycoon_manager_progress_line, BuyplateId:string, Agent:?agent):void=
        for (
            BeaconLine:ProgressLine.BeaconLines
            BeaconLineConfig := m_TycoonConfig.BeaconLines[BeaconLine.BeaconLineId]
        ) {
            if (
                (LineEntry := BeaconLineConfig.InOrderBuyPlates[BeaconLine.m_CurrentEntry],
                 LineEntry.ContainsBuyPlate[BuyplateId])
            ) {
                set BeaconLine.m_LockedTargetBuyPlate = None()
                UpdateBeaconLine(ProgressLine, BeaconLine)
            } else if (
                BeaconLine.m_LockedTargetBuyPlate?
                BeaconLineConfig.ContainsBuyPlate[BuyplateId]
            ) {
                set BeaconLine.m_LockedTargetBuyPlate = None()
                UpdateBeaconLine(ProgressLine, BeaconLine)
            }
        }

    OnConfiguredBuyPlateUnlocked<internal>(ProgressLine:tycoon_manager_progress_line, Config:tycoon_manager_buyplates_entry, Agent:?agent):void=
        if (Generation := Config.Generates?) {
            if (Ticket := ProgressLine.m_AutoGenerationTickets[Generation.ResourceId]) {
                Ticket.ModifyRate(ToFloat(Generation.Amount))
            } else if (
                Ticket := ProgressLine.AutoGenerationManager?.CreateTicket[Generation.ResourceId, false]
                set ProgressLine.m_AutoGenerationTickets[Generation.ResourceId] = Ticket
            ) {
                Ticket.SetRate(ToFloat(Generation.Amount))
            } else {
                Log("Failed to create auto generation ticket for resource \"{Generation.ResourceId}\" for generation of progress line \"{ProgressLine.Name}\". Is auto generation manager set?", ?Level := log_level.Error)
            }
        }
        if (Config.Unlocks.Length > 0, DeviceBackrefs := ProgressLine.AccessControl?.GetUsingDevices[]) {
            var Unlocked: [string]logic = map{}
            for (
                Device:DeviceBackrefs
                DeviceRoutingId := aware_device[Device].DeviceMeta().GetRoutingId()
            ) {
                for (
                    Unlock:Config.Unlocks
                    DeviceRoutingId = Unlock.TargetRoutingId
                    (not Unlocked[DeviceRoutingId]) or (not p_RoutingIdSingleMatch?) or (Unlock.Payload?)
                    set Unlocked[DeviceRoutingId] = true
                ) {
                    TriggerData := Unlock.Payload? or "enable"
                    Device.AttemptCustomTrigger(TriggerData, Agent, ?Context := "progress line \"{ProgressLine.Name}\"")
                }
            }
            for (Unlock:Config.Unlocks, not Unlocked[Unlock.TargetRoutingId]) {
                Log("Failed to process unlock \"{Unlock}\" after \"{Config.BuyplateId}\" unlocked for progress line \"{ProgressLine.Name}\". No matching device found in backrefs.", ?Level := log_level.Error)
            }
        }

    OnPlayerSaveSlotSelected<private>(ProgressLine:tycoon_manager_progress_line, Event:player_save_slot_selected_event):void=
        PlayerSaveSlotSelectedEvent.Signal(ProgressLine, Event)

        EventName := if ((Event.AlreadyPlayed?) and (Event.Reset?)) {
            "claim_reset"
        } else if (Event.AlreadyPlayed?) {
            "claim_continue"
        } else {
            "claim_first"
        }
        p_AnalyticsController.Emit(Event.Player, EventName)

    OnRebirthNextLevelTriggerCalled<private>(Agent:?agent):void=
        Player := if (l_Player := player[Agent?]) {
            l_Player
        } else {
            Log("Tycoon manager rebirth next level OnTrigger does nothing without a player instigator", ?Level := log_level.Warning)
            return
        }
        PromptRebirth(Player)

    PromptRebirth<public>(Player:player):void=
        if (RebirthConfig := m_TycoonConfig.Rebirth?, PlayerState := GetPlayerState[Player]) {
            RebirthLevel := PlayerState.ProgressLine.m_RebirthLevel
            if (RebirthLevel >= RebirthConfig.MaxLevel) {
                Log(MsgRebirthMaxLevelOnTrigger(Player), ?Level := log_level.Debug)
                return
            }
            DialogIndex := if (p_RebirthNextLevelDialogPerLevel?) {
                if (p_RebirthNextLevelDialogPerLevelDefaultLast?) {
                    var Index: int = RebirthLevel
                    loop {
                        if (Index <= 0 or p_RebirthNextLevelDialog[Index]) {
                            break
                        }
                        set Index -= 1
                    }
                    Index
                } else {
                    RebirthLevel
                }
            } else {
                0
            }
            if (DialogDevice := p_RebirthNextLevelDialog[DialogIndex]) {
                StateRes := if (l_StateRes := m_RebirthDialogs[DialogIndex]) {
                    l_StateRes
                } else {
                    InitialTemplate := DialogDevice.GetDescriptionText()
                    Wrapper := dialog_device_wrapper{
                        DialogDevice := DialogDevice
                    }
                    l_StateRes := (InitialTemplate, Wrapper)
                    if (set m_RebirthDialogs[DialogIndex] = l_StateRes) {}
                    l_StateRes
                }
                var Template: string = StateRes(0)
                set Template = Template
                    .ReplaceAll("%CurrentLevel%", ToString(RebirthLevel))
                    .ReplaceAll("%NextLevel%", ToString(RebirthLevel + 1))
                if (NextLevelConfig := RebirthConfig.GetLevelConfig[RebirthLevel + 1]) {
                    NumberFormatter := p_RebirthPlaceholdersMultiplierFormatter.CreateFormatter()
                    for (ResourceId->Multiplier:NextLevelConfig.ResourceMultipliers) {
                        set Template = Template.ReplaceAll("%Multiplier:{ResourceId}%", NumberFormatter.Format(Multiplier))
                    }
                }
                Title := DialogDevice.GetTitleText()
                DialogDevice.SetTitleText(ToMessage(Title)) 
                DialogDevice.SetDescriptionText(ToMessage(Template))
                Closure := tycoon_manager_rebirth_dialog_closure{
                    Device := Self
                    Player := Player
                    ProgressLine := PlayerState.ProgressLine
                    ExpectedCurrentLevel := RebirthLevel
                }
                StateRes(1).Show(Player, Some(Closure.Call))
            } else {
                RebirthPlayer(Player)
            }
        } else {
            Log("Tycoon manager rebirth next level OnTrigger does nothing when rebirth is not configured", ?Level := log_level.Warning)
        }

    GetPlayerState<internal>(Player:player)<transacts><decides>:tycoon_manager_player_state=
        m_Players[Player]

    GetRebirthLevel<public>(Player:player)<transacts><decides>:int=
        m_Players[Player].ProgressLine.m_RebirthLevel

    RebirthPlayer<internal>(Player:player):void=
        if (RebirthConfig := m_TycoonConfig.Rebirth?, PlayerState := m_Players[Player]) {
            ProgressLine := PlayerState.ProgressLine
            if (ProgressLine.m_RebirthLevel >= RebirthConfig.MaxLevel) {
                Log(MsgRebirthMaxLevel(Player), ?Level := log_level.Debug)
                return
            }
            set ProgressLine.m_RebirthLevel += 1
            RebirthLevel := ProgressLine.m_RebirthLevel
            Log(MsgRebirth(Player, RebirthLevel, "manual"), ?Level := log_level.Debug)
            if (PersistenceRecorder := PlayerState.ProgressLine.m_PersistenceRecorder?) {
                # save rebirth level
                if (Handle := PlayerState.ProgressLine.m_RebirthRecordableHandle?) {
                    Handle.Save(Player)
                } else {
                    PersistenceRecorder.SaveAgent(Player)
                }
                # reset persistence state
                PersistenceRecorder.ResetAgent(Player, tycoon_manager_rebirth_persistence_reset_query{
                    SourceDevice := Self
                    ResetFilter := Some(RebirthPersistenceResetFilter)
                })
            }
            DoRollback(PlayerState.ProgressLine, Some(Player))
            if (ClaimBuyplate := ProgressLine.ClaimBuyplate?) {
                ClaimBuyplate.Unlock(Player, ?UserTriggered := false, ?FirstUnlock := false, ?Force := true)
            }
            for (TriggerConfig:p_RebirthResetTriggers, TriggerConfig.IsApplicable[RebirthLevel]) {
                TriggerConfig.Trigger.Trigger(Player)
            }
            for (TriggerConfig:ProgressLine.RebirthResetTriggers, TriggerConfig.IsApplicable[RebirthLevel]) {
                TriggerConfig.Trigger.Trigger(Player)
            }
            if (m_RebirthListeners.Length > 0) {
                Event := tycoon_rebirth_event{
                    Player := Player
                    NewRebirthLevel := RebirthLevel
                    Virtual := false
                }
                for (Listener:m_RebirthListeners) {
                    Listener(Event)
                }
            }
            if (PersistenceRecorder := PlayerState.ProgressLine.m_PersistenceRecorder?) {
                # save persistence in case triggers caused relevant changes
                PersistenceRecorder.SaveAgent(Player)
            }
            EnsureRebirthEffects(PlayerState.ProgressLine)
            if (AccessControl := ProgressLine.AccessControl?) {
                UpdateCostMetTracking(ProgressLine, AccessControl)
            }
        }

    ForceVirtualRebirth<public>(Player:player):void=
        if (RebirthConfig := m_TycoonConfig.Rebirth?, PlayerState := m_Players[Player]) {
            ProgressLine := PlayerState.ProgressLine
            if (ProgressLine.m_RebirthLevel >= RebirthConfig.MaxLevel) {
                Log(MsgRebirthMaxLevel(Player), ?Level := log_level.Debug)
                return
            }
            set ProgressLine.m_RebirthLevel += 1
            RebirthLevel := ProgressLine.m_RebirthLevel
            Log(MsgRebirth(Player, RebirthLevel, "forced virtual"), ?Level := log_level.Debug)
            if (PersistenceRecorder := PlayerState.ProgressLine.m_PersistenceRecorder?) {
                # save rebirth level
                if (Handle := PlayerState.ProgressLine.m_RebirthRecordableHandle?) {
                    Handle.Save(Player)
                } else {
                    PersistenceRecorder.SaveAgent(Player)
                }
            }
            if (m_RebirthListeners.Length > 0) {
                Event := tycoon_rebirth_event{
                    Player := Player
                    NewRebirthLevel := RebirthLevel
                    Virtual := true
                }
                for (Listener:m_RebirthListeners) {
                    Listener(Event)
                }
            }
            EnsureRebirthEffects(PlayerState.ProgressLine)
        }

    RebirthPersistenceResetFilter<private>(Recordable:persistence_recordable_base)<transacts><decides>:void=
        not pet_controller_device[Recordable]

    EnsureRebirthEffects<internal>(ProgressLine:tycoon_manager_progress_line):void=
        if (HudManager := m_RebirthHudManager?) {
            HudManager.Refresh()
        }
        if (RebirthConfig := m_TycoonConfig.Rebirth?, AutoGenerationManager := ProgressLine.AutoGenerationManager?) {
            ResourceMultipliers := if (LevelConfig := RebirthConfig.GetLevelConfig[ProgressLine.m_RebirthLevel]) {
                LevelConfig.ResourceMultipliers
            } else {
                map{}
            }
            for (ResourceId->Multiplier:ResourceMultipliers) {
                AutoGenerationManager.GetResourceMultiplierInstance(ResourceId).SetMultiplier(p_RebirthMultipliersChannel, Multiplier)
                if (set ProgressLine.m_SetResourceMultipliers[ResourceId] = Multiplier) {}
            }
            # remove multipliers that are no longer wanted
            for (ResourceId->__:ProgressLine.m_SetResourceMultipliers, not ResourceMultipliers[ResourceId]) {
                AutoGenerationManager.GetResourceMultiplierInstance(ResourceId).SetMultiplier(p_RebirthMultipliersChannel, None())
            }
        }
        if (Billboard := ProgressLine.RebirthBillboard[0]) {
            if (ProgressLine.m_RebirthLevel > 0 or not (ProgressLine.RebirthBillboardHideOnZero?)) {
                Text := ProgressLine.RebirthBillboardTemplate
                    .ReplaceAll("%CurrentLevel%", ToString(ProgressLine.m_RebirthLevel))
                Billboard.ShowText()
                Billboard.SetText(ToMessage(Text))
            } else {
                Billboard.HideText()
                Billboard.SetText(ToMessage(""))
            }
        }

    MsgRebirth<private><localizes>(Player:player, NewLevel:int, Context:string):message=
        "Rebirthing '{Player}' to level {NewLevel} ({Context})"
    MsgRebirthMaxLevel<private><localizes>(Player:player):message=
        "Ignoring next rebirth level for '{Player}' as they reached max level"
    MsgRebirthMaxLevelOnTrigger<private><localizes>(Player:player):message=
        "Ignoring next rebirth level OnTrigger for '{Player}' as they reached max level"

    IsPetUnlocked<override>(Player:player, PetId:string)<transacts><decides>:void=
        RebirthLevel := m_Players[Player].ProgressLine.m_RebirthLevel
        RebirthLevel >= m_RebirthPets[PetId]

tycoon_manager_beacon_line_area_closure := class<final>:
    Device: tycoon_manager_device
    Config: tycoon_manager_beacon_line_area_config

    OnAgentEntered<internal>(Agent:agent):void=
        if. set Config.m_InVolume[Agent] = void
        if (Player := player[Agent]) {
            Device.SyncPlayerBeaconLines(Player)
        }

    OnAgentExited<internal>(Agent:agent):void=
        set Config.m_InVolume = Config.m_InVolume.RemoveByKey(Agent)(0)
        if (Player := player[Agent]) {
            Device.SyncPlayerBeaconLines(Player)
        }

tycoon_manager_player_closure := class:
    Device: tycoon_manager_device
    ProgressLine: tycoon_manager_progress_line

    OnPlayerAdded(Player:player):void=
        Device.OnPlayerAdded(ProgressLine, Player)

    OnPlayerRemoved(Player:player, GameQuit:logic):void=
        Device.OnPlayerRemoved(ProgressLine, Player)

## Persistence {{{
tycoon_manager_player_persistence_load_closure := class:
    Callback: type{__(:tycoon_manager_progress_line,:player_save_slot_selected_event):void}
    ProgressLine: tycoon_manager_progress_line

    Call(Event:player_save_slot_selected_event):void=
        Callback(ProgressLine, Event)

tycoon_manager_rebirth_dummy_device := class(persistence_recordable, persistence_resetable):
    TycoonManager: tycoon_manager_device
    ProgressLine: tycoon_manager_progress_line

    GetPersistenceControl<override>()<transacts><decides>:persistence_recordable_control=
        None()?

    PersistenceReserve<override>(StorageMem:storage_memory)<transacts><decides>:tycoon_manager_rebirth_persistence_state=
        TycoonManager.p_EnablePersistence?
        tycoon_manager_rebirth_persistence_state{
            RebirthLevelChunk := StorageMem.Reserve[TycoonManager.p_RebirthLevelReserveBits, "tycoon_rebirth_level"]
        }

    PersistenceSave<override>(i_State:persistence_recordable_state, Agent:agent, SaveCause:persistence_save_cause):void=
        if (SaveCause = persistence_save_cause.AutoSave) {
            return
        }
        if (
            State := tycoon_manager_rebirth_persistence_state[i_State]
            PlayerState := TycoonManager.GetPlayerState[player[Agent]]
            PlayerState.ProgressLine.m_Index = ProgressLine.m_Index
        ) {
            State.RebirthLevelChunk.WriteInt(Agent, ProgressLine.m_RebirthLevel)
        }

    PersistenceLoad<override>(i_State:persistence_recordable_state, Agent:agent, Query:persistence_load_query):void=
        if (
            State := tycoon_manager_rebirth_persistence_state[i_State]
            PlayerState := TycoonManager.GetPlayerState[player[Agent]]
            PlayerState.ProgressLine.m_Index = ProgressLine.m_Index
        ) {
            SavedLevel := State.RebirthLevelChunk.ReadInt(Agent)
            PrevRebirthLevel := ProgressLine.m_RebirthLevel
            set ProgressLine.m_RebirthLevel = if (Query.Overwrite?) {
                SavedLevel
            } else {
                Max(ProgressLine.m_RebirthLevel, SavedLevel)
            }
            if (PrevRebirthLevel <> ProgressLine.m_RebirthLevel) {
                TycoonManager.EnsureRebirthEffects(ProgressLine)
            }
        }

    PersistenceReset<override>(i_State:persistence_recordable_state, Agent:agent, Query:persistence_reset_query):void=
        if (tycoon_manager_rebirth_persistence_reset_query[Query]) {
            Log("Skipping reset of rebirth persistence state because it is initialized by a rebirth", ?Level := log_level.Debug)
        } else {
            i_State.Reset(Agent)
            PersistenceLoad(i_State, Agent, persistence_load_query{
                Overwrite := true
                FirstLoad := false
            })
        }

tycoon_manager_rebirth_persistence_reset_query<public> := class(persistence_reset_query):

tycoon_manager_rebirth_persistence_state<public> := class(persistence_recordable_state):
    RebirthLevelChunk<public>: storage_memory_chunk

    Reset<override>(Agent:agent):void=
        RebirthLevelChunk.Reset(Agent)
## }}}

## Buy plate features {{{
tycoon_manager_buyplate_cost_met_reflection_closure := class:
    Device: tycoon_manager_device
    ProgressLine: tycoon_manager_progress_line

    OnResourceTake(Event:resource_take_event):void=
        if (AccessControl := ProgressLine.AccessControl?, AccessControl.IsAllowed[Event.Agent?]) {
            Device.UpdateCostMetTracking(ProgressLine, AccessControl)
        }

tycoon_manager_claim_buyplate_middleware := class(buy_plate_middleware):
    Device: tycoon_manager_device
    ProgressLine: tycoon_manager_progress_line

    MayTryUnlockBuyplate<override>(Agent:agent, BuyPlate:buy_plate_device):logic=
        if (Player := player[Agent], Device.GetPlayerState[Player]) {
            for (HUDMessage:Device.p_ClaimFailHUDMessage) {
                HUDMessage.Show(Agent)
            }
            false
        } else {
            true
        }

    OnUnlockBuyplate<override>(Agent:agent, BuyPlate:buy_plate_device):void=
        if (Player := player[Agent]) {
            Device.ClaimProgressLineFromBuyplate(ProgressLine, Player)
        }

tycoon_manager_buy_plate_cost_override := class(buy_plate_cost_override):
    Device: tycoon_manager_device
    Cost: tycoon_manager_buyplates_entry_resource

    TryUnlock<override>(Agent:agent, Silent:logic):logic=
        if (Cost.Amount <= 0) {
            return true
        }
        Notify := InvertLogic(Silent)
        if (Cost.CheckDontTake?) {
            Amount := Device.p_ResourceManager.GetAmount(Agent, Cost.ResourceId)
            if (Amount >= Cost.Amount) {
                return true
            } else if (Notify?) {
                Device.p_ResourceManager.SendNotEnoughNotification(Agent, Cost.ResourceId)
            }
        } else {
            Success := Device.p_ResourceManager.Take(Some(Agent), Cost.ResourceId, Cost.Amount,
                ?NotifyNotEnough := Notify, ?SendFeedback := Notify)
            return Success
        }
        false

    HasEnough<override>(Agent:agent):logic=
        if (Cost.Amount <= 0) {
            return true
        }
        Amount := Device.p_ResourceManager.GetAmount(Agent, Cost.ResourceId)
        if (Amount >= Cost.Amount) {
            true
        } else {
            false
        }
## }}}

tycoon_manager_device_register_closure := class:
    Device: tycoon_manager_device
    ProgressLine: tycoon_manager_progress_line

    Call(RegisteredDevice:creative_device):void=
        Device.OnDeviceRegistered(ProgressLine, RegisteredDevice)

## Buy plates closures {{{
tycoon_manager_any_buy_plate_closure := class:
    Device: tycoon_manager_device
    ProgressLine: tycoon_manager_progress_line
    BuyPlateId: string

    OnUnlock(Agent:?agent):void=
        Device.OnAnyBuyPlateUnlocked(ProgressLine, BuyPlateId, Agent)

    OnEnable(Agent:?agent):void=
        Device.OnAnyBuyPlateEnabled(ProgressLine, BuyPlateId)

tycoon_manager_configured_buy_plate_closure := class:
    Device: tycoon_manager_device
    ProgressLine: tycoon_manager_progress_line
    Config: tycoon_manager_buyplates_entry

    OnUnlock(Agent:?agent):void=
        Device.OnConfiguredBuyPlateUnlocked(ProgressLine, Config, Agent)
        Device.OnAnyBuyPlateUnlocked(ProgressLine, Config.BuyplateId, Agent)

    OnEnable(Agent:?agent):void=
        Device.OnAnyBuyPlateEnabled(ProgressLine, Config.BuyplateId)
## }}}
## Rebirth UI {{{
tycoon_manager_rebirth_dialog_closure := class:
    Device: tycoon_manager_device
    Player: player

    ProgressLine: tycoon_manager_progress_line
    ExpectedCurrentLevel: int

    Call(Event:dialog_response_event):void=
        Log("Rebirth dialog event {Event}", ?Level := log_level.Debug)
        if (
            Event.ResponseType = dialog_response_type.Clicked
            Event.ButtonIndex? = Device.p_RebirthNextLevelDialogConfirmButtonIndex
            player[Event.Agent] = Player
            PlayerState := Device.GetPlayerState[Player]
            PlayerState.ProgressLine.m_Index = ProgressLine.m_Index
            PlayerState.ProgressLine.m_RebirthLevel = ExpectedCurrentLevel
        ) {
            Device.RebirthPlayer(Player)
        }

tycoon_manager_rebirth_hud_manager := class(hud_manager_base):
    Device: tycoon_manager_device
    Config: tycoon_manager_rebirth_config

    CreateHUD<override>(Player:player):?hud_manager_player=
        Image := CreateDefaultingTextureBlock(Device.p_RebirthHUDImageSize, None())
        Canvas := canvas{
            Slots := array{
                Device.p_RebirthHUDImagePlacement.ToCanvasSlot(Image.Widget())
            }
        }
        Element := tycoon_manager_rebirth_hud_image{
            Widget := Canvas
            Manager := Self
            Player := Player
            Image:= Image
        }
        Some(MakeSingleElementHUDPlayer(Player, Canvas, Element))

tycoon_manager_rebirth_hud_image := class(hud_element):
    Manager: tycoon_manager_rebirth_hud_manager
    Player: player

    Image: defaulting_texture_block

    var m_LastLevel: ?int = false

    Update<override>():void=
        Level := option{Manager.Device.GetRebirthLevel[Player]}
        if (m_LastLevel = Level) {
            return
        }
        set m_LastLevel = Level
        if (
            l_Level := Level?
            TextureName := Manager.Config.GetLevelConfig[l_Level].HudImageTexture?
            Texture := AwareDefaultAssetsProvider[].GetTextureAsset[TextureName]
        ) {
            Show()
            Image.SetTexture(Texture)
            Log("Rebirth HUD image switching to level {l_Level}", ?Level := log_level.Debug)
        } else {
            Hide()
            Log("Rebirth HUD image switching to no level", ?Level := log_level.Debug)
        }
## }}}
