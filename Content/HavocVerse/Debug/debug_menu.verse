using { /Fortnite.com/Devices }
using { /Verse.org/Simulation }
using { /Verse.org/Colors }
using { /UnrealEngine.com/Temporary/Diagnostics }
using { /UnrealEngine.com/Temporary/SpatialMath }
using { /UnrealEngine.com/Temporary/UI }
using { /Fortnite.com/UI }

using { HavocVerse.Devices }
using { HavocVerse.Devices.GameSpecific }
using { HavocVerse.Resource }
using { HavocVerse.Storage }
using { HavocVerse.UI }
using { HavocVerse.Util }
using { HavocVerse.Logging }

## Debug Option Types {{{
debug_menu_option<public> := class:
    Label<public>: string
    Provider<public>: debug_menu_option_provider

debug_menu_option_provider<public> := interface:

debug_menu_checkbox_provider<public> := interface(debug_menu_option_provider):
    GetState<public>(Player:player)<transacts><decides>:void

    SetState<public>(Player:player, Value:logic):void

debug_menu_checkbox_labels<public> := interface:
    GetEnabledLabel<public>()<transacts>:string

    GetDisabledLabel<public>()<transacts>:string

debug_menu_choices_provider<public> := interface(debug_menu_option_provider):
    GetSelected<public>(Player:player)<transacts>:int

    SetSelected<public>(Player:player, Index:int)<transacts><decides>:void

    FormatOption<public>(Index:int)<transacts><decides>:string

debug_menu_slider_provider<public> := interface(debug_menu_option_provider):
    GetValue<public>(Player:player)<transacts><decides>:float

    SetValue<public>(Player:player, Value:float):void

    # min, max, step
    GetBounds<public>()<transacts>:tuple(float,float,float)

debug_menu_choices_provider_default<public> := class(debug_menu_choices_provider):
    Options<public>: []string

    var m_Selected<public>: int = 0

    GetSelected<override>(Player:player)<transacts>:int=
        m_Selected

    SetSelected<override>(Player:player, Index:int)<transacts><decides>:void=
        Options[Index]
        set m_Selected = Index

    FormatOption<override>(Index:int)<transacts><decides>:string=
        Options[Index]
## }}}

debug_menu_config := struct<concrete>:
    @editable PromptBackgroundTexture<public>: string = "HavocVerseUI/ui_debug_menu_code_prompt"
    @editable PromptBackgroundSize<public>: vector2 = vector2{X := 450.0, Y := 542.0}
    @editable BackgroundTexture<public>: string = "HavocVerseUI/ui_debug_menu"
    @editable BackgroundSize<public>: vector2 = vector2{X := 1280.0, Y := 720.0}

debug_menu := class(ui_menu):
    Device: debug_device
    Config: debug_menu_config
    DebugPlayer: debug_player
    SudoInstigator: ?player

    Categories: []string = array{"General", "Resource", "Storage", "Custom", "More"}

    ## Open code prompt {{{
    OpenCodePrompt(Player:player):void=
        Instance := debug_menu_code_instance{
            Menu := Self
            Player := Player
            Device := Device
            DebugPlayer := DebugPlayer
            CodeTextWidget := CreateTextBlock("")
        }

        Digits := array{1, 2, 3, 4, 5, 6, 7, 8, 9, -2, 0, -1}
        DigitsStack := stack_box{
            Orientation := orientation.Vertical
            Slots := for (RowIndex->Row:Digits.Chunks(3)) {
                RowWidget := stack_box{
                    Orientation := orientation.Horizontal
                    Slots := for (Index->Digit:Row) {
                        Button := if (Digit = -1) {
                            button_loud{
                                DefaultText := ToMessage("▶")
                            }
                        } else if (Digit = -2) {
                            button_regular{
                                DefaultText := ToMessage("◀")
                            }
                        } else {
                            button_regular{
                                DefaultText := ToMessage(ToString(Digit))
                            }
                        }
                        Closure := debug_menu_code_closure{
                            Instance := Instance
                            Digit := Digit
                        }
                        Button.OnClick().Subscribe(Closure.OnClick)
                        stack_box_slot{
                            Widget := ApplySizeHint(Button, vector2{X := 106.0, Y := 50.0})
                            HorizontalAlignment := horizontal_alignment.Left
                            Padding := margin{
                                Left := if (Index > 0). 20.0 else 0.0
                            }
                        }
                    }
                }
                stack_box_slot{
                    Widget := RowWidget
                    VerticalAlignment := vertical_alignment.Top
                    HorizontalAlignment := horizontal_alignment.Left
                    Padding := margin{
                        Top := if (RowIndex > 0). 20.0 else 0.0
                    }
                }
            }
        }
        Widget := stack_box{
            Orientation := orientation.Vertical
            Slots := array{
                stack_box_slot{
                    Widget := Instance.CodeTextWidget
                    VerticalAlignment := vertical_alignment.Top
                    HorizontalAlignment := horizontal_alignment.Center
                }
                stack_box_slot{
                    Widget := DigitsStack
                    VerticalAlignment := vertical_alignment.Top
                    HorizontalAlignment := horizontal_alignment.Center
                    Padding := margin{
                        Top := 60.0
                    }
                }
            }
        }

        CommonOpen(Player, Instance, CommonCanvas(Widget,
            ?BackgroundTexture := Config.PromptBackgroundTexture, ?BackgroundSize := Config.PromptBackgroundSize,
            ?DefaultToColorBlock := true))
    ## }}}
    
    ## Open menu {{{
    CreateInstance<override>(Player:player):tuple(ui_menu_instance,canvas)=
        Instance := debug_menu_instance{
            Menu := Self
            Player := Player
            DebugPlayer := DebugPlayer
            Device := Device
            RightPanel := canvas{}
            SudoInstigator := SudoInstigator
        }
        set DebugPlayer.m_OpenedMenu = Some(Instance)

        CategoryButtonSizeHint := vector2{X := 260.0, Y := 50.0}
        CategoryStack := stack_box{
            Orientation := orientation.Vertical
            Slots := for (Index->Category:Categories) {
                Button := button_loud:
                    DefaultText := ToMessage(Category)
                if (Category = "General", SudoInstigator?) {
                    Button.SetEnabled(false)
                } else {
                    Closure := debug_menu_category_closure{
                        Instance := Instance
                        Category := Category
                    }
                    Button.OnClick().Subscribe(Closure.OnClick)
                }
                stack_box_slot{
                    Widget := ApplySizeHint(Button, CategoryButtonSizeHint)
                    VerticalAlignment := vertical_alignment.Top
                    Padding := margin{
                        Top := if (Index > 0). 15.0 else 0.0
                    }
                }
            }
        }

        TopLeftCategories := if (SudoInstigator?) {
            array{"Close"}
        } else {
            array{"Close", "Logs"}
        }
        TopLeftStack := stack_box{
            Orientation := orientation.Vertical
            Slots := for (Index->Category:TopLeftCategories) {
                Button := button_regular{
                    DefaultText := ToMessage(Category)
                }
                case(Category):
                    "Close" => Button.OnClick().Subscribe(Instance.OnCloseClick)
                    "Logs" => Button.OnClick().Subscribe(Instance.OpenLogsClick)
                    _ => {}
                stack_box_slot{
                    Widget := ApplySizeHint(Button, CategoryButtonSizeHint)
                    VerticalAlignment := vertical_alignment.Top
                    Padding := margin{
                        Top := if (Index > 0). 15.0 else 0.0
                    }
                }
            }
        }
        LeftPanel := canvas{
            Slots := array{
                canvas_slot{
                    Widget := TopLeftStack
                    Anchors := anchors{
                        Minimum := vector2{X := 0.5, Y := 0.0}
                        Maximum := vector2{X := 0.5, Y := 0.0}
                    }
                    Alignment := vector2{X := 0.5, Y := 0.0}
                    Offsets := margin{
                        Top := 20.0
                    }
                }
                canvas_slot{
                    Widget := CategoryStack
                    Anchors := anchors{
                        Minimum := vector2{X := 0.5, Y := 1.0}
                        Maximum := vector2{X := 0.5, Y := 1.0}
                    }
                    Alignment := vector2{X := 0.5, Y := 1.0}
                    Offsets := margin{
                        Top := -20.0
                    }
                }
            }
        }

        Widget := stack_box{
            Orientation := orientation.Horizontal
            Slots := array{
                stack_box_slot{
                    Widget := ApplySizeHint(LeftPanel, vector2{X := 296.0, Y := 690.0})
                    HorizontalAlignment := horizontal_alignment.Left
                }
                stack_box_slot{
                    Widget := ApplySizeHint(Instance.RightPanel, vector2{X := 948.0, Y := 690.0})
                    HorizontalAlignment := horizontal_alignment.Left
                    Padding := margin{
                        Left := 6.0
                    }
                }
            }
        }

        ScrollControls := stack_box{
            Orientation := orientation.Vertical
            Slots := for (Index->Direction:array{"↑", "↓"}) {
                Button := button_regular{
                    DefaultText := ToMessage(Direction)
                }
                case(Index):
                    0 => Button.OnClick().Subscribe(Instance.OnScrollUpClick)
                    1 => Button.OnClick().Subscribe(Instance.OnScrollDownClick)
                    _ => {}
                stack_box_slot{
                    Widget := ApplySizeHint(Button, vector2{X := 60.0, Y := (Config.BackgroundSize.Y - 10.0) / 2.0})
                    HorizontalAlignment := horizontal_alignment.Left
                    Padding := margin{
                        Top := if (Index > 0). 10.0 else 0.0
                    }
                }
            }
        }
        set Instance.m_ScrollControls = Some(ScrollControls)

        MainMenuWidget := PlaceOnCenteredBackground(Widget, Config.BackgroundTexture, Config.BackgroundSize,
            ?DefaultToColorBlock := true)
        Canvas := canvas{
            Slots := array{
                canvas_slot{
                    Widget := MainMenuWidget
                    Anchors := anchors{
                        Minimum := vector2{X := 0.5, Y := 0.5}
                        Maximum := vector2{X := 0.5, Y := 0.5}
                    }
                    Alignment := vector2{X := 0.5, Y := 0.5}
                }
                canvas_slot{
                    Widget := ScrollControls
                    Anchors := anchors{
                        Minimum := vector2{X := 0.5, Y := 0.5}
                        Maximum := vector2{X := 0.5, Y := 0.5}
                    }
                    Alignment := vector2{X := 0.0, Y := 0.5}
                    Offsets := margin{
                        Left := (Config.BackgroundSize.X / 2.0) + 15.0
                    }
                }
            }
        }

        if (SudoPlayer := SudoInstigator?) {
            Canvas.AddWidget(canvas_slot{
                Widget := stack_box:
                    Orientation := orientation.Horizontal
                    Slots := array{
                        stack_box_slot:
                            Widget := CreateTextBlock("SUDOING ", ?TextColor := NamedColors.White)
                            HorizontalAlignment := horizontal_alignment.Left
                            VerticalAlignment := vertical_alignment.Top
                        stack_box_slot:
                            Widget := CreateTextBlock(MsgSudoingAgent(SudoPlayer), ?TextColor := NamedColors.Crimson)
                            HorizontalAlignment := horizontal_alignment.Left
                            VerticalAlignment := vertical_alignment.Top
                            Padding := margin{
                                Left := 10.0
                            }
                    }
                Anchors := anchors:
                    Minimum := vector2{X := 0.5, Y := 0.5}
                    Maximum := vector2{X := 0.5, Y := 0.5}
                Alignment := vector2{X := 0.5, Y := 1.0}
                Offsets := margin{
                    Top := -((Config.BackgroundSize.Y / 2.0) + 10.0)
                }
            })

            Instance.SelectCategory("Resource")
        } else {
            Instance.SelectCategory("General")
        }
        (Instance, Canvas)

    MsgSudoingAgent<private><localizes>(Agent:agent):message=
        "{Agent}"
    ## }}}

debug_menu_instance := class(ui_menu_instance):
    Device: debug_device
    DebugPlayer: debug_player

    SudoInstigator: ?player = false

    RightPanel: canvas

    var m_ActiveCategory: string = ""
    var m_ActiveWidget: ?widget = false

    var m_ScrollControls: ?widget = false
    var m_ScrollPosition: int = 0

    var m_UptimeTextWidget: ?text_block = false
    var m_TickMonitoringTextWidget: ?text_block = false

    Instigator<public>():player=
        if (l_SudoInstigator := SudoInstigator?) {
            l_SudoInstigator
        } else {
            Player
        }

    OnClose<override>():void=
        set DebugPlayer.m_OpenedMenu = None()

    OnScrollUpClick(Data:widget_message):void=
        set m_ScrollPosition = Max(m_ScrollPosition-1, 0)
        PostScrollUpdate()

    OnScrollDownClick(Data:widget_message):void=
        set m_ScrollPosition += 1
        PostScrollUpdate()

    PostScrollUpdate<private>():void=
        if (ActiveWidget := m_ActiveWidget?) {
            RightPanel.RemoveWidget(ActiveWidget)
            set m_ActiveWidget = false
            SetActiveWidget(ActiveWidget)
        }

    SetScrollingEnabled<private>(Enabled:logic):void=
        if (Enabled?, ScrollControls := m_ScrollControls?) {
            ScrollControls.SetVisibility(widget_visibility.Visible)
        } else if (not (Enabled?), ScrollControls := m_ScrollControls?) {
            ScrollControls.SetVisibility(widget_visibility.Hidden)
        }

    OpenLogsClick(Data:widget_message):void=
        Close()

        if (AssetsProvider := AwareDefaultAssetsProvider[]) {
            LogsMenu := debug_logs_menu{
                AssetsProvider := AssetsProvider
                Device := Device
            }
            LogsMenu.Open(Player)
        }

    SelectCategory(Category:string):void=
        if (Category = m_ActiveCategory) {
            return
        }
        set m_ActiveCategory = Category
        set m_ScrollPosition = 0
        SetScrollingEnabled(false)
        if (Widget := m_ActiveWidget?) {
            RightPanel.RemoveWidget(Widget)
            set m_ActiveWidget = false
        }

        CreateFn: ?type{__():widget} = case(Category):
            "General" => Some(CreateGeneralCategory)
            "Resource" => Some(CreateResourceCategory)
            "Storage" => Some(CreateStorageCategory)
            "Custom" => Some(CreateCustomCategory)
            "More" => Some(CreateExtraCategory)
            _ => false
        if (Category <> "General") {
            set m_TickMonitoringTextWidget = None()
        }

        if (l_CreateFn := CreateFn?) {
            Widget := l_CreateFn()
            SetActiveWidget(Widget)
        }

    OpenPanel<private>(Widget:widget):void=
        set m_ActiveCategory = ""
        set m_ScrollPosition = 0
        if (l_Widget := m_ActiveWidget?) {
            RightPanel.RemoveWidget(l_Widget)
            set m_ActiveWidget = false
        }
        SetActiveWidget(Widget)

    SetActiveWidget<private>(Widget:widget):void=
        RightPanel.AddWidget(canvas_slot{
            Widget := Widget
            SizeToContent := false
            Anchors := anchors{
                Minimum := vector2{X := 0.0, Y := 0.0}
                Maximum := vector2{X := 1.0, Y := 1.0}
            }
            Alignment := vector2{X := 0.0, Y := 0.0}
            Offsets := margin{
                Left := 20.0
                Right := 20.0
                Top := 20.0 - (m_ScrollPosition * 250.0)
                Bottom := 20.0
            }
        })
        set m_ActiveWidget = Some(Widget)

    UpdateUptime<internal>(UptimeText:string):void=
        if (TextWidget := m_UptimeTextWidget?) {
            TextWidget.SetText(ToMessage("Uptime : {UptimeText}"))
        }

    UpdateTickMonitoring<internal>(LastTickDuration:float, AvgTick:float, MinTick: float, MaxTick:float):void=
        if (TextWidget := m_TickMonitoringTextWidget?) {
            TextWidget.SetText(ToMessage("TickM : {LastTickDuration} last - {AvgTick} avg\n       {MinTick} min - {MaxTick} max"))
        }

    ## General Category {{{
    CreateGeneralCategory<private>():widget=
        var Slots: []stack_box_slot = array{}
        UptimeWidget := m_UptimeTextWidget? or (
            l_UptimeWidget := CreateTextBlock("")
            set m_UptimeTextWidget = Some(l_UptimeWidget)
            UpdateUptime(Device.FormatUptime())
            l_UptimeWidget
        )
        set Slots += array{
            stack_box_slot{
                Widget := UptimeWidget
                VerticalAlignment := vertical_alignment.Top
                HorizontalAlignment := horizontal_alignment.Left
            }
        }
        if (Device.p_TickMonitoring?) {
            TextWidget := CreateTextBlock("")
            set Slots += array{
                stack_box_slot{
                    Widget := TextWidget
                    VerticalAlignment := vertical_alignment.Top
                    HorizontalAlignment := horizontal_alignment.Left
                }
            }
            set m_TickMonitoringTextWidget = Some(TextWidget)
        } else {
            set m_TickMonitoringTextWidget = None()
        }
        set Slots += array{
            stack_box_slot{
                Widget := CreateTextBlock("Shared", ?TextColor := NamedColors.Gray)
                VerticalAlignment := vertical_alignment.Top
                HorizontalAlignment := horizontal_alignment.Left
                Padding := margin{
                    Top := 30.0
                }
            }
        }
        for (Index->Option:Device.m_MenuGeneralSharedOptions) {
            set Slots += array{stack_box_slot{
                Widget := CreateOptionLine(Option)
                VerticalAlignment := vertical_alignment.Top
                HorizontalAlignment := horizontal_alignment.Left
                Padding := margin{
                    Top := if (Index > 0). 15.0 else 5.0
                }
            }}
        }
        set Slots += array{
            stack_box_slot{
                Widget := CreateTextBlock("Only you", ?TextColor := NamedColors.Gray)
                VerticalAlignment := vertical_alignment.Top
                HorizontalAlignment := horizontal_alignment.Left
                Padding := margin{
                    Top := 30.0
                }
            }
        }
        for (Index->Option:Device.m_MenuGeneralIndividualOptions) {
            set Slots += array{stack_box_slot{
                Widget := CreateOptionLine(Option)
                VerticalAlignment := vertical_alignment.Top
                HorizontalAlignment := horizontal_alignment.Left
                Padding := margin{
                    Top := if (Index > 0). 15.0 else 5.0
                }
            }}
        }
        stack_box{
            Orientation := orientation.Vertical
            Slots := Slots
        }
    ## }}}

    ## Resource Category {{{
    CreateResourceCategory<private>():widget=
        var Entries: []tuple(resource_manager_device,string) = array{}
        for (AwareNetDevice:AwareListDevices(), ResourceManager := resource_manager_device[AwareNetDevice]) {
            for (Res:ResourceManager.GetResources[]) {
                set Entries += array{(ResourceManager, Res.GetResourceId())}
            }
        }
        stack_box{
            Orientation := orientation.Vertical
            Slots := for (RowIndex->Row:Entries.Chunks(3)) {
                RowWidget := stack_box{
                    Orientation := orientation.Horizontal
                    Slots := for (Index->Entry:Row) {
                        ResourceManager := Entry(0)
                        ResourceId := Entry(1)
                        ButtonType := if (ResourceManager.IsProxiedResource[ResourceId]) {
                            fort_button_type.Regular
                        } else {
                            fort_button_type.Loud
                        }
                        Button := CreateFortButton(ButtonType, ToMessage(ResourceId))
                        Closure := debug_menu_resource_closure{
                            Instance := Self
                            ResourceManager := ResourceManager
                            ResourceId := ResourceId
                        }
                        Button.OnClick().Subscribe(Closure.OnClick)
                        stack_box_slot{
                            Widget := ApplySizeHint(Button, vector2{X := 288.0, Y := 50.0})
                            HorizontalAlignment := horizontal_alignment.Left
                            Padding := margin{
                                Left := if (Index > 0). 20.0 else 0.0
                            }
                        }
                    }
                }
                stack_box_slot{
                    Widget := RowWidget
                    VerticalAlignment := vertical_alignment.Top
                    HorizontalAlignment := horizontal_alignment.Left
                    Padding := margin{
                        Top := if (RowIndex > 0). 20.0 else 0.0
                    }
                }
            }
        }
    ## }}}

    ## Storage Category {{{
    CreateStorageCategory<private>():widget=
        StorageDevices := for (AwareNetDevice:AwareListDevices(), StorageDevice := storage_memory_device[AwareNetDevice]) {
            StorageDevice
        }
        Slots := for (Index->StorageDevice:StorageDevices) {
            Widget := CreateStorageEntry(StorageDevice)
            stack_box_slot{
                Widget := Widget
                VerticalAlignment := vertical_alignment.Top
                HorizontalAlignment := horizontal_alignment.Left
                Padding := margin{
                    Top := if (Index > 0). 30.0 else 0.0
                }
            }
        }
        stack_box{
            Orientation := orientation.Vertical
            Slots := Slots
        }

    CreateStorageEntry<private>(StorageDevice:storage_memory_device):widget=
        StorageMem := if (l_StorageMem := trackers_storage_device[StorageDevice].GetStorageMem[]) {
            l_StorageMem
        } else if (l_StorageMem := verse_bits_storage_device[StorageDevice].GetStorageMem[]) {
            l_StorageMem
        } else if (l_StorageMem := shared_storage_chunk_device[StorageDevice].GetStorageMem[]) {
            l_StorageMem
        } else {
            return overlay{}
        }

        DeviceType := if (trackers_storage_device[StorageDevice]) {
            "Trackers Storage"
        } else if (verse_bits_storage_device[StorageDevice]) {
            "Verse Bits Storage"
        } else if (shared_storage_chunk_device[StorageDevice]) {
            "Shared Chunk"
        } else {
            "Storage"
        }
        MoreButton := button_regular{
            DefaultText := ToMessage("+")
        }
        MoreClosure := debug_menu_storage_closure{
            Instance := Self
            StorageDevice := StorageDevice
            Action := "+"
        }
        MoreButton.OnClick().Subscribe(MoreClosure.OnClick)

        StorageSize := StorageMem.Size()
        TitleRow := canvas{
            Slots := array{
                canvas_slot{
                    Widget := CreateTextBlock("{DeviceType} {StorageDevice.GetID()} ({StorageSize}b)")
                    SizeToContent := true
                    Anchors := anchors{
                        Minimum := vector2{X := 0.0, Y := 0.5}
                        Maximum := vector2{X := 0.0, Y := 0.5}
                    }
                    Alignment := vector2{X := 0.0, Y := 0.5}
                }
                canvas_slot{
                    Widget := ApplySizeHint(MoreButton, vector2{X := 60.0, Y := 50.0})
                    SizeToContent := true
                    Anchors := anchors{
                        Minimum := vector2{X := 1.0, Y := 0.5}
                        Maximum := vector2{X := 1.0, Y := 0.5}
                    }
                    Alignment := vector2{X := 1.0, Y := 0.5}
                }
            }
        }
        ActionsRow := stack_box{
            Orientation := orientation.Horizontal
            Slots := for (Index->Action:array{"Dump", "Fill", "Reset"}) {
                Button := button_loud{
                    DefaultText := ToMessage(Action)
                }
                Closure := debug_menu_storage_closure{
                    Instance := Self
                    StorageDevice := StorageDevice
                    Action := Action
                }
                Button.OnClick().Subscribe(Closure.OnClick)
                stack_box_slot{
                    Widget := ApplySizeHint(Button, vector2{X := 288.0, Y := 50.0})
                    HorizontalAlignment := horizontal_alignment.Left
                    Padding := margin{
                        Left := if (Index > 0). 20.0 else 0.0
                    }
                }
            }
        }
        stack_box{
            Orientation := orientation.Vertical
            Slots := array{
                stack_box_slot{
                    Widget := TitleRow
                    VerticalAlignment := vertical_alignment.Top
                    HorizontalAlignment := horizontal_alignment.Fill
                }
                stack_box_slot{
                    Widget := ActionsRow
                    VerticalAlignment := vertical_alignment.Top
                    HorizontalAlignment := horizontal_alignment.Left
                    Padding := margin{
                        Top := 10.0
                    }
                }
            }
        }
    ## }}}

    ## Debug Draw Panel {{{
    OpenDebugDrawPanel<private>():void=
        DebugDrawControl := stack_box{
            Orientation := orientation.Horizontal
            Slots := for (Index->Action:array{"Show", "Hide", "Clear"}) {
                Button := button_regular{
                    DefaultText := ToMessage(Action)
                }
                Closure := debug_menu_debug_draw_control_closure{
                    Instance := Self
                    Action := Action
                }
                Button.OnClick().Subscribe(Closure.OnClick)
                stack_box_slot{
                    Widget := ApplySizeHint(Button, vector2{X := 288.0, Y := 50.0})
                    HorizontalAlignment := horizontal_alignment.Left
                    Padding := margin{
                        Left := if (Index > 0). 20.0 else 0.0
                    }
                }
            }
        }

        ActionsWidget := stack_box{
            Orientation := orientation.Vertical
            Slots := for (RowIndex->Row:Device.m_DebugDrawActions.Chunks(2)) {
                RowWidget := stack_box{
                    Orientation := orientation.Horizontal
                    Slots := for (Index->DebugAction:Row) {
                        Button := button_loud{
                            DefaultText := ToMessage(DebugAction.Name)
                        }
                        Closure := debug_menu_debug_action_closure{
                            Instance := Self
                            DebugAction := DebugAction
                        }
                        Button.OnClick().Subscribe(Closure.OnClick)
                        stack_box_slot{
                            Widget := ApplySizeHint(Button, vector2{X := 444.0, Y := 50.0})
                            HorizontalAlignment := horizontal_alignment.Left
                            Padding := margin{
                                Left := if (Index > 0). 20.0 else 0.0
                            }
                        }
                    }
                }
                stack_box_slot{
                    Widget := RowWidget
                    VerticalAlignment := vertical_alignment.Top
                    HorizontalAlignment := horizontal_alignment.Left
                    Padding := margin{
                        Top := if (RowIndex > 0). 20.0 else 0.0
                    }
                }
            }
        }

        Widget := stack_box{
            Orientation := orientation.Vertical
            Slots := array{
                stack_box_slot{
                    Widget := CreateOptionLine(debug_menu_option{
                        Label := "Debug Draw"
                        Provider := debug_device_debug_draw_duration_policy_choice{
                            Device := Device
                        }
                    })
                    VerticalAlignment := vertical_alignment.Top
                    HorizontalAlignment := horizontal_alignment.Left
                }
                stack_box_slot{
                    Widget := CreateOptionLine(debug_menu_option{
                        Label := "Thickness"
                        Provider := debug_device_debug_draw_thickness_slider{
                            Device := Device
                        }
                    })
                    VerticalAlignment := vertical_alignment.Top
                    HorizontalAlignment := horizontal_alignment.Left
                    Padding := margin{
                        Top := 15.0
                    }
                }
                stack_box_slot{
                    Widget := DebugDrawControl
                    VerticalAlignment := vertical_alignment.Top
                    HorizontalAlignment := horizontal_alignment.Left
                    Padding := margin{
                        Top := 15.0
                    }
                }
                stack_box_slot{
                    Widget := CreateTextBlock("Actions", ?TextColor := NamedColors.Gray)
                    VerticalAlignment := vertical_alignment.Top
                    HorizontalAlignment := horizontal_alignment.Left
                    Padding := margin{
                        Top := 15.0
                    }
                }
                stack_box_slot{
                    Widget := ActionsWidget
                    VerticalAlignment := vertical_alignment.Top
                    HorizontalAlignment := horizontal_alignment.Left
                    Padding := margin{
                        Top := 10.0
                    }
                }
            }
        }
        OpenPanel(Widget)
    ## }}}

    ## Custom Category {{{
    CreateCustomCategory<private>():widget=
        stack_box{
            Orientation := orientation.Vertical
            Slots := for (RowIndex->Row:Device.p_CustomActions.Chunks(2)) {
                RowWidget := stack_box{
                    Orientation := orientation.Horizontal
                    Slots := for (Index->CustomAction:Row) {
                        Button := button_loud{
                            DefaultText := ToMessage(CustomAction.Label)
                        }
                        Closure := debug_menu_custom_action_closure{
                            Instance := Self
                            CustomAction := CustomAction
                        }
                        Button.OnClick().Subscribe(Closure.OnClick)
                        stack_box_slot{
                            Widget := ApplySizeHint(Button, vector2{X := 444.0, Y := 50.0})
                            HorizontalAlignment := horizontal_alignment.Left
                            Padding := margin{
                                Left := if (Index > 0). 20.0 else 0.0
                            }
                        }
                    }
                }
                stack_box_slot{
                    Widget := RowWidget
                    VerticalAlignment := vertical_alignment.Top
                    HorizontalAlignment := horizontal_alignment.Left
                    Padding := margin{
                        Top := if (RowIndex > 0). 20.0 else 0.0
                    }
                }
            }
        }
    ## }}}

    ## Extra Category {{{
    ExtraCategories: []string = array{"Debug Draw", "Sudo", "Memory"}

    CreateExtraCategory<private>():widget=
        var Buttons: []text_button_base = for (ExtraCategory:ExtraCategories) {
            Button := button_loud{
                DefaultText := ToMessage(ExtraCategory)
            }
            Closure := debug_menu_extra_category_closure{
                Instance := Self
                ExtraCategory := ExtraCategory
            }
            Button.OnClick().Subscribe(Closure.OnClick)
            Button
        }
        for (AwareDevice:AwareListDevices(), TycoonManager := tycoon_manager_device[AwareDevice]) {
            Button := button_loud{
                DefaultText := ToMessage("Tycoon")
            }
            Closure := debug_menu_tycoon_closure{
                Instance := Self
                TycoonManager := TycoonManager
            }
            Button.OnClick().Subscribe(Closure.OnClick)
            set Buttons += array{Button}
        }

        stack_box{
            Orientation := orientation.Vertical
            Slots := for (RowIndex->Row:Buttons.Chunks(2)) {
                RowWidget := stack_box{
                    Orientation := orientation.Horizontal
                    Slots := for (Index->Button:Row) {
                        stack_box_slot{
                            Widget := ApplySizeHint(Button, vector2{X := 444.0, Y := 50.0})
                            HorizontalAlignment := horizontal_alignment.Left
                            Padding := margin{
                                Left := if (Index > 0). 20.0 else 0.0
                            }
                        }
                    }
                }
                stack_box_slot{
                    Widget := RowWidget
                    VerticalAlignment := vertical_alignment.Top
                    HorizontalAlignment := horizontal_alignment.Left
                    Padding := margin{
                        Top := if (RowIndex > 0). 20.0 else 0.0
                    }
                }
            }
        }

    OpenExtraCategory<internal>(Category:string):void=
        case(Category):
            "Debug Draw" => OpenDebugDrawPanel()
            "Sudo" => OpenSudoPanel()
            "Memory" => OpenMemoryMenu()
            _ => {}
    ## }}}

    ## Tycoon Panel {{{
    OpenTycoonPanel<internal>(TycoonManager:tycoon_manager_device):void=
        Panel := debug_menu_tycoon_panel{
            Instance := Self
            TycoonManager := TycoonManager
        }

        ExpelButton := button_loud:
            DefaultText := ToMessage("Expel")
        ExpelButton.OnClick().Subscribe(Panel.OnExpelClick)

        Line1 := stack_box:
            Orientation := orientation.Horizontal
            Slots := array{
                stack_box_slot:
                    Widget := ApplySizeHint(ExpelButton, vector2{X := 440.0, Y := 50.0})
                    HorizontalAlignment := horizontal_alignment.Left
                    VerticalAlignment := vertical_alignment.Top
            }

        PromptRebirthButton := button_loud:
            DefaultText := ToMessage("Prompt Rebirth")
        PromptRebirthButton.OnClick().Subscribe(Panel.OnPromptRebirthClick)
        ForceRebirthButton := button_loud:
            DefaultText := ToMessage("Virtual Rebirth")
        ForceRebirthButton.OnClick().Subscribe(Panel.OnForceRebirthClick)

        Line2 := stack_box:
            Orientation := orientation.Horizontal
            Slots := array{
                stack_box_slot:
                    Widget := ApplySizeHint(PromptRebirthButton, vector2{X := 440.0, Y := 50.0})
                    HorizontalAlignment := horizontal_alignment.Left
                    VerticalAlignment := vertical_alignment.Top
                stack_box_slot:
                    Widget := ApplySizeHint(ForceRebirthButton, vector2{X := 440.0, Y := 50.0})
                    HorizontalAlignment := horizontal_alignment.Left
                    VerticalAlignment := vertical_alignment.Top
                    Padding := margin{
                        Left := 15.0
                    }
            }

        Widget := stack_box:
            Orientation := orientation.Vertical
            Slots := array{
                stack_box_slot:
                    Widget := Line1
                    HorizontalAlignment := horizontal_alignment.Left
                    VerticalAlignment := vertical_alignment.Top
                stack_box_slot:
                    Widget := Line2
                    HorizontalAlignment := horizontal_alignment.Left
                    VerticalAlignment := vertical_alignment.Top
                    Padding := margin{
                        Top := 20.0
                    }
            }

        OpenPanel(Widget)
    ## }}}

    ## Open Memory Menu {{{
    OpenMemoryMenu<private>():void=
        Close()

        MemoryMenu := memory_tool_menu{
            AssetsProvider := Menu.AssetsProvider
        }
        MemoryMenu.Open(Player)
    ## }}}

    ## Sudo Panel {{{
    OpenSudoPanel<private>():void=
        Players := Device.GetPlayspace().GetPlayers()
        PlayersGrid := stack_box{
            Orientation := orientation.Vertical
            Slots := for (RowIndex->Row:Players.Chunks(2)) {
                RowWidget := stack_box{
                    Orientation := orientation.Horizontal
                    Slots := for (Index->SudoPlayer:Row) {
                        Button := button_loud:
                            DefaultText := MsgSudoButtonName(SudoPlayer)
                        Closure := debug_menu_sudo_open_closure{
                            Instance := Self
                            SudoPlayer := SudoPlayer
                        }
                        Button.OnClick().Subscribe(Closure.OnSelectClick)
                        stack_box_slot:
                            Widget := ApplySizeHint(Button, vector2{X := 444.0, Y := 50.0})
                            HorizontalAlignment := horizontal_alignment.Left
                            Padding := margin{
                                Left := if (Index > 0). 20.0 else 0.0
                            }
                    }
                }
                stack_box_slot:
                    Widget := RowWidget
                    VerticalAlignment := vertical_alignment.Top
                    HorizontalAlignment := horizontal_alignment.Left
                    Padding := margin{
                        Top := if (RowIndex > 0). 20.0 else 0.0
                    }
            }
        }

        Widget := stack_box:
            Orientation := orientation.Vertical
            Slots := array{
                stack_box_slot:
                    Widget := CreateTextBlock("Open Menu As:")
                    VerticalAlignment := vertical_alignment.Top
                    HorizontalAlignment := horizontal_alignment.Center
                stack_box_slot:
                    Widget := PlayersGrid
                    VerticalAlignment := vertical_alignment.Top
                    HorizontalAlignment := horizontal_alignment.Fill
                    Padding := margin{
                        Top := 20.0
                    }
            }

        OpenPanel(Widget)

    MsgSudoButtonName<private><localizes>(Agent:agent):message=
        "{Agent}"
    ## }}}

    ## Resource Panel {{{
    OpenResourcePanel<internal>(ResourceManager:resource_manager_device, ResourceId:string):void=
        Panel := debug_menu_resource_panel{
            Instance := Self
            ResourceManager := ResourceManager
            ResourceId := ResourceId
            ModeChoices := debug_menu_choices_provider_default{
                Options := array{"Give", "Take"}
            }
            CurrentAmountTextBlock := MakeCachedTextBlock(CreateTextBlock("",
                ?TextColor := NamedColors.CornflowerBlue))
        }
        ResourceAmounts: []int = array{
            1, 10, 100,
            1000, 10000, 100000,
            1000000, 10000000, 100000000,
            1000000000, 10000000000, 100000000000,
            1000000000000, 10000000000000, 100000000000000
        }
        AmountSelection := stack_box{
            Orientation := orientation.Vertical
            Slots := for (RowIndex->Row:ResourceAmounts.Chunks(3)) {
                RowWidget := stack_box{
                    Orientation := orientation.Horizontal
                    Slots := for (Index->Amount:Row) {
                        Button := button_regular{
                            DefaultText := ToMessage(FormatNumber(ToFloat(Amount), ?Style := number_format_style.KMBT))
                        }
                        Closure := debug_menu_resource_amount_closure{
                            Panel := Panel
                            Amount := Amount
                        }
                        Button.OnClick().Subscribe(Closure.OnClick)
                        stack_box_slot{
                            Widget := ApplySizeHint(Button, vector2{X := 176.0, Y := 50.0})
                            HorizontalAlignment := horizontal_alignment.Left
                            Padding := margin{
                                Left := if (Index > 0). 10.0 else 0.0
                            }
                        }
                    }
                }
                stack_box_slot{
                    Widget := RowWidget
                    VerticalAlignment := vertical_alignment.Top
                    HorizontalAlignment := horizontal_alignment.Left
                    Padding := margin{
                        Top := if (RowIndex > 0). 10.0 else 0.0
                    }
                }
            }
        }
        ClearButton := button_regular{
            DefaultText := ToMessage("Clear")
        }
        ClearButton.OnClick().Subscribe(Panel.OnClickClear)
        AmountSelection.AddWidget(stack_box_slot{
            Widget := ApplySizeHint(ClearButton, vector2{X := 560.0, Y := 50.0})
            VerticalAlignment := vertical_alignment.Top
            HorizontalAlignment := horizontal_alignment.Left
            Padding := margin{
                Top := 40.0
            }
        })

        Panel.UpdateCurrentAmount()
        HeaderLine := stack_box{
            Orientation := orientation.Horizontal
            Slots := array{
                stack_box_slot:
                    Widget := CreateTextBlock("Resource: {ResourceId}")
                    VerticalAlignment := vertical_alignment.Top
                    HorizontalAlignment := horizontal_alignment.Left
                stack_box_slot:
                    Widget := Panel.CurrentAmountTextBlock.Widget
                    VerticalAlignment := vertical_alignment.Top
                    HorizontalAlignment := horizontal_alignment.Left
                    Padding := margin{
                        Left := 20.0
                    }
            }
        }

        Widget := stack_box:
            Orientation := orientation.Vertical
            Slots := array{
                stack_box_slot:
                    Widget := HeaderLine
                    VerticalAlignment := vertical_alignment.Top
                    HorizontalAlignment := horizontal_alignment.Left
                stack_box_slot:
                    Widget := CreateOptionLine(debug_menu_option{
                        Label := "Mode"
                        Provider := Panel.ModeChoices
                    })
                    VerticalAlignment := vertical_alignment.Top
                    HorizontalAlignment := horizontal_alignment.Left
                    Padding := margin{
                        Top := 10.0
                    }
                stack_box_slot:
                    Widget := AmountSelection
                    VerticalAlignment := vertical_alignment.Top
                    HorizontalAlignment := horizontal_alignment.Left
                    Padding := margin{
                        Top := 40.0
                    }
            }
        OpenPanel(Widget)
    ## }}}

    ## Storage More Panel {{{
    OpenStorageMorePanel<internal>(StorageDevice:storage_memory_device):void=
        StorageMem := if (l_StorageMem := trackers_storage_device[StorageDevice].GetStorageMem[]) {
            l_StorageMem
        } else if (l_StorageMem := verse_bits_storage_device[StorageDevice].GetStorageMem[]) {
            l_StorageMem
        } else if (l_StorageMem := shared_storage_chunk_device[StorageDevice].GetStorageMem[]) {
            l_StorageMem
        } else {
            return
        }

        Entries := for (MemChunk:StorageMem.GetChunks()) {
            ChunkSize := MemChunk.Size
            TitleRow := canvas:
                Slots := array{
                    canvas_slot:
                        Widget := CreateTextBlock("{MemChunk.LayoutId} ({ChunkSize}b)", ?TextColor := NamedColors.White)
                        SizeToContent := true
                        Anchors := anchors{
                            Minimum := vector2{X := 0.0, Y := 0.5}
                            Maximum := vector2{X := 0.0, Y := 0.5}
                        }
                        Alignment := vector2{X := 0.0, Y := 0.5}
                }

            ActionsClosure := debug_menu_storage_more_closure{
                Instance := Self
                StorageDevice := StorageDevice
                MemChunk := MemChunk
            }
            TitleRowActions := stack_box{
                Orientation := orientation.Horizontal
                Slots := for (ActionIndex->Action:array{"Fill", "Reset"}) {
                    Button := button_loud{
                        DefaultText := ToMessage(Action)
                    }
                    case(ActionIndex):
                        0 => Button.OnClick().Subscribe(ActionsClosure.OnFillClick)
                        1 => Button.OnClick().Subscribe(ActionsClosure.OnResetClick)
                        _ => {}
                    stack_box_slot:
                        Widget := ApplySizeHint(Button, vector2{X := 140.0, Y := 50.0})
                        HorizontalAlignment := horizontal_alignment.Left
                        Padding := margin{
                            Left := if (ActionIndex > 0). 20.0 else 0.0
                        }
                }
            }
            TitleRow.AddWidget(canvas_slot{
                Widget := TitleRowActions
                SizeToContent := true
                Anchors := anchors{
                    Minimum := vector2{X := 1.0, Y := 0.5}
                    Maximum := vector2{X := 1.0, Y := 0.5}
                }
                Alignment := vector2{X := 1.0, Y := 0.5}
            })

            BitsStr := BitsToString(MemChunk.Read(Instigator())).Chunks(40).Join("\n")
            stack_box{
                Orientation := orientation.Vertical
                Slots := array{
                    stack_box_slot:
                        Widget := TitleRow
                        HorizontalAlignment := horizontal_alignment.Fill
                    stack_box_slot:
                        Widget := CreateTextBlock(BitsStr, ?TextColor := NamedColors.CornflowerBlue)
                        HorizontalAlignment := horizontal_alignment.Left
                        Padding := margin{
                            Top := 10.0
                        }
                }
            }
        }

        ChunksStack := stack_box{
            Orientation := orientation.Vertical
            Slots := for (Index->Entry:Entries) {
                stack_box_slot:
                    Widget := Entry
                    HorizontalAlignment := horizontal_alignment.Fill
                    Padding := margin{
                        Top := if (Index > 0). 20.0 else 0.0
                    }
            }
        }

        Panel := debug_menu_storage_more_panel{
            Instance := Self
            StorageDevice := StorageDevice
        }

        DumpLayoutButton := button_loud:
            DefaultText := ToMessage("Dump Layout")
        DumpLayoutButton.OnClick().Subscribe(Panel.OnDumpLayoutClick)

        ExtraActionsSlots := for (
            ActionIndex->Action:array{"Dump Layout", "Dump Debugs"}
            ActionIndex = 0 or trackers_storage_device[StorageDevice]
        ) {
            Button := button_loud:
                DefaultText := ToMessage(Action)
            case(ActionIndex):
                0 => Button.OnClick().Subscribe(Panel.OnDumpLayoutClick)
                1 => Button.OnClick().Subscribe(Panel.OnDumpDebugsClick)
                _ => {}
            stack_box_slot:
                Widget := ApplySizeHint(Button, vector2{X := 440.0, Y := 50.0})
                VerticalAlignment := vertical_alignment.Top
                Padding := margin{
                    Left := if (ActionIndex > 0). 20.0 else 0.0
                }
        }
        ExtraActionsStack := stack_box:
            Orientation := orientation.Vertical
            Slots := array{
                stack_box_slot:
                    Widget := stack_box:
                        Orientation := orientation.Horizontal
                        Slots := ExtraActionsSlots
                    HorizontalAlignment := horizontal_alignment.Left
            }

        if (TrackersStorage := trackers_storage_device[StorageDevice]) {
            ExtraActionsStack.AddWidget(stack_box_slot{
                Widget := stack_box:
                    Orientation := orientation.Horizontal
                    Slots := for (ActionIndex->Action:array{"Save", "Load", "TmpClr"}) {
                        Button := button_regular:
                            DefaultText := ToMessage(Action)
                        case(ActionIndex):
                            0 => Button.OnClick().Subscribe(Panel.OnSaveClick)
                            1 => Button.OnClick().Subscribe(Panel.OnLoadClick)
                            2 => Button.OnClick().Subscribe(Panel.OnTempClearClick)
                            _ => {}
                        stack_box_slot:
                            Widget := ApplySizeHint(Button, vector2{X := 200.0, Y := 50.0})
                            VerticalAlignment := vertical_alignment.Top
                            Padding := margin{
                                Left := if (ActionIndex > 0). 10.0 else 0.0
                            }
                    } 
                HorizontalAlignment := horizontal_alignment.Left
                Padding := margin{
                    Top := 10.0
                }
            })
        }

        Widget := stack_box:
            Orientation := orientation.Vertical
            Slots := array{
                stack_box_slot:
                    Widget := ExtraActionsStack
                    HorizontalAlignment := horizontal_alignment.Fill
                stack_box_slot:
                    Widget := ChunksStack
                    HorizontalAlignment := horizontal_alignment.Fill
                    Padding := margin{
                        Top := 10.0
                    }
            }

        OpenPanel(Widget)
        SetScrollingEnabled(true)
    ## }}}

    ## Option Line {{{
    CreateOptionLine<private>(Option:debug_menu_option):widget=
        LabelWidget := CreateTextBlock(Option.Label)
        Widget := if (Checkbox := debug_menu_checkbox_provider[Option.Provider]) {
            ToggleButton := button_regular{
                DefaultText := ToMessage("")
            }
            CheckboxElement := debug_menu_checkbox_element{
                Provider := Checkbox
                ToggleButton := ToggleButton
            }
            ToggleButton.OnClick().Subscribe(CheckboxElement.OnClick)
            CheckboxElement.UpdateButton(DebugPlayer.Player)
            ApplySizeHint(ToggleButton, vector2{X := 160.0, Y := 50.0})
        } else if (Choices := debug_menu_choices_provider[Option.Provider]) {
            CycleButton := button_regular{
                DefaultText := ToMessage("")
            }
            ChoicesElement := debug_menu_choices_element{
                Provider := Choices
                CycleButton := CycleButton
            }
            CycleButton.OnClick().Subscribe(ChoicesElement.OnClick)
            ChoicesElement.UpdateButton(DebugPlayer.Player)
            ApplySizeHint(CycleButton, vector2{X := 260.0, Y := 50.0})
        } else if (Slider := debug_menu_slider_provider[Option.Provider]) {
            Bounds := Slider.GetBounds()
            SliderWidget := slider_regular{
                DefaultMinValue := Bounds(0)
                DefaultMaxValue := Bounds(1)
                DefaultStepSize := Bounds(2)
            }
            SliderElement := debug_menu_slider_element{
                Provider := Slider
                SliderWidget := SliderWidget
                Step := Bounds(2)
            }
            SliderWidget.OnValueChanged().Subscribe(SliderElement.OnValueChanged)
            SliderElement.UpdateValue(DebugPlayer.Player)
            ApplySizeHint(SliderWidget, vector2{X := 360.0, Y := 50.0})
        } else {
            return LabelWidget
        }
        stack_box{
            Orientation := orientation.Horizontal
            Slots := array{
                stack_box_slot{
                    Widget := LabelWidget 
                    HorizontalAlignment := horizontal_alignment.Left
                    VerticalAlignment := vertical_alignment.Center
                }
                stack_box_slot{
                    Widget := Widget
                    Padding := margin{
                        Left := 20.0
                    }
                    HorizontalAlignment := horizontal_alignment.Left
                    VerticalAlignment := vertical_alignment.Center
                }
            }
        }
    ## }}}

## Code Menu {{{
debug_menu_code_instance := class(ui_menu_instance):
    Device: debug_device
    DebugPlayer: debug_player

    CodeTextWidget: text_block

    var m_Typed: string = ""

    TypedDigit(Digit:int):void=
        if (m_Typed.Length >= 10) {
            return
        }
        set m_Typed += ToString(Digit)
        CodeTextWidget.SetText(ToMessage(m_Typed))

    Backspace():void=
        if (l_Slice := m_Typed.Slice[0, m_Typed.Length - 1]) {
            set m_Typed = l_Slice
        } else {
            set m_Typed = array{}
        }
        CodeTextWidget.SetText(ToMessage(m_Typed))

    Submit():void=
        Close()
        if (m_Typed = Device.p_SecretCode) {
            set DebugPlayer.m_TypedCode = true
            if (DebugPlayer.m_RemoteOpen?, InputTrigger := Device.p_RemoteOpenInputTrigger?) {
                InputTrigger.Register(Player)
            }
            Device.PersistDebugPlayer(DebugPlayer)
            Log("Attempted code: {m_Typed}. Success.", ?Level := log_level.Debug)
            Menu.Open(Player)
        } else {
            set DebugPlayer.m_FailedCodeAttempts += 1
            Log("Attempted code: {m_Typed}. Failed, {DebugPlayer.m_FailedCodeAttempts} attempts", ?Level := log_level.Debug)
        }

debug_menu_code_closure := class:
    Instance: debug_menu_code_instance
    Digit: int

    OnClick(Data:widget_message):void=
        if (not Instance.IsActive[]) {
            return
        }
        if (Digit = -1) {
            Instance.Submit()
        } else if (Digit = -2) {
            Instance.Backspace()
        } else {
            Instance.TypedDigit(Digit)
        }
## }}}

debug_menu_category_closure := class:
    Instance: debug_menu_instance
    Category: string

    OnClick(Data:widget_message):void=
        Instance.SelectCategory(Category)

debug_menu_resource_closure := class:
    Instance: debug_menu_instance
    ResourceManager: resource_manager_device
    ResourceId: string

    OnClick(Data:widget_message):void=
        Instance.OpenResourcePanel(ResourceManager, ResourceId)

debug_menu_tycoon_closure := class:
    Instance: debug_menu_instance
    TycoonManager: tycoon_manager_device

    OnClick(Data:widget_message):void=
        Instance.OpenTycoonPanel(TycoonManager)

debug_menu_storage_closure := class:
    Instance: debug_menu_instance
    StorageDevice: storage_memory_device
    Action: string

    OnClick(Data:widget_message):void=
        Instigator := Instance.Instigator()
        case(Action):
            "Dump" => StorageDevice.Dump(Instigator, true)
            "Fill" => StorageDevice.Fill(Instigator, true)
            "Reset" =>
                if (TrackersStorage := trackers_storage_device[StorageDevice]) {
                    TrackersStorage.Reset(Instigator)
                } else if (VerseBitsStorage := verse_bits_storage_device[StorageDevice]) {
                    VerseBitsStorage.Reset(Instigator)
                } else {
                    StorageDevice.Fill(Instigator, false)
                }
            "+" => Instance.OpenStorageMorePanel(StorageDevice)
            _ => {}

debug_menu_storage_more_closure := class:
    Instance: debug_menu_instance
    StorageDevice: storage_memory_device
    MemChunk: storage_memory_subchunk

    OnFillClick(Data:widget_message):void=
        MemChunk.Write(Instance.Instigator(), array{true}.Repeat(MemChunk.Size))
        Refresh()

    OnResetClick(Data:widget_message):void=
        MemChunk.Reset(Instance.Instigator())
        Refresh()

    Refresh<private>():void=
        Instance.OpenStorageMorePanel(StorageDevice)

debug_menu_debug_draw_control_closure := class:
    Instance: debug_menu_instance
    Action: string

    OnClick(Data:widget_message):void=
        DebugDraw := Instance.Device.DebugDraw
        case(Action):
            "Show" => DebugDraw.ShowChannel()
            "Hide" => DebugDraw.HideChannel()
            "Clear" => DebugDraw.Clear()
            _ => {}

debug_menu_debug_action_closure := class:
    Instance: debug_menu_instance
    DebugAction: debug_action

    OnClick(Data:widget_message):void=
        DebugAction.Executor.Execute(Instance.Instigator())

debug_menu_custom_action_closure := class:
    Instance: debug_menu_instance
    CustomAction: debug_device_custom_action_config

    OnClick(Data:widget_message):void=
        CustomAction.Trigger.Trigger(Instance.Instigator())

debug_menu_extra_category_closure := class:
    Instance: debug_menu_instance
    ExtraCategory: string

    OnClick(Data:widget_message):void=
        Instance.OpenExtraCategory(ExtraCategory)

debug_menu_resource_panel := class:
    Instance: debug_menu_instance
    ResourceManager: resource_manager_device
    ResourceId: string

    CurrentAmountTextBlock: cached_text_block

    ModeChoices: debug_menu_choices_provider

    SelectAmount(Amount:int):void=
        Instigator := Instance.Instigator()
        if (ModeChoices.GetSelected(Instance.Player) = 0) {
            ResourceManager.Give(Instigator, ResourceId, Amount)
        } else {
            ResourceManager.Take(Instigator, ResourceId, Amount)
        }
        UpdateCurrentAmount()

    OnClickClear(Data:widget_message):void=
        Instigator := Instance.Instigator()
        Amount := ResourceManager.GetAmount(Instigator, ResourceId)
        if (Amount > 0) {
            ResourceManager.Take(Instigator, ResourceId, Amount)
        }
        UpdateCurrentAmount()

    UpdateCurrentAmount():void=
        FormattedAmount := ResourceManager.GetFormattedAmount(Instance.Instigator(), ResourceId)
        CurrentAmountTextBlock.SetText(FormattedAmount)

debug_menu_tycoon_panel := class:
    Instance: debug_menu_instance
    TycoonManager: tycoon_manager_device

    OnExpelClick(Data:widget_message):void=
        TycoonManager.ForceExpelPlayer(Instance.Instigator())

    OnPromptRebirthClick(Data:widget_message):void=
        TycoonManager.PromptRebirth(Instance.Instigator())

    OnForceRebirthClick(Data:widget_message):void=
        TycoonManager.ForceVirtualRebirth(Instance.Instigator())

debug_menu_storage_more_panel := class:
    Instance: debug_menu_instance
    StorageDevice: storage_memory_device

    OnDumpLayoutClick(Data:widget_message):void=
        StorageDevice.Dump(Instance.Instigator(), false)

    OnDumpDebugsClick(Data:widget_message):void=
        if (TrackersStorage := trackers_storage_device[StorageDevice]) {
            TrackersStorage.DumpDebugs(Instance.Instigator())
        }
    
    OnSaveClick(Data:widget_message):void=
        if (TrackersStorage := trackers_storage_device[StorageDevice]) {
            TrackersStorage.DebugSave(Instance.Instigator())
        }

    OnLoadClick(Data:widget_message):void=
        if (TrackersStorage := trackers_storage_device[StorageDevice]) {
            TrackersStorage.DebugLoad(Instance.Instigator())
        }
        Instance.OpenStorageMorePanel(StorageDevice)

    OnTempClearClick(Data:widget_message):void=
        if (TrackersStorage := trackers_storage_device[StorageDevice]) {
            TrackersStorage.DebugTempClear(Instance.Instigator())
        }
        Instance.OpenStorageMorePanel(StorageDevice)

debug_menu_resource_amount_closure := class:
    Panel: debug_menu_resource_panel
    Amount: int

    OnClick(Data:widget_message):void=
        Panel.SelectAmount(Amount)

debug_menu_sudo_open_closure := class:
    Instance: debug_menu_instance
    SudoPlayer: player

    OnSelectClick(Data:widget_message):void=
        if (SourceMenu := debug_menu[Instance.Menu]) {
            Menu := debug_menu{
                AssetsProvider := SourceMenu.AssetsProvider
                Device := SourceMenu.Device
                Config := SourceMenu.Config
                DebugPlayer := SourceMenu.DebugPlayer
                SudoInstigator := Some(SudoPlayer)
            }
            Instance.Close()
            Menu.Open(Instance.Player)
        }

## Debug option element implementations {{{
debug_menu_checkbox_element := class:
    Provider: debug_menu_checkbox_provider
    ToggleButton: text_button_base

    OnClick(Data:widget_message):void=
        NewState := if (Provider.GetState[Data.Player]). false else true
        Provider.SetState(Data.Player, NewState)
        UpdateButton(Data.Player)

    UpdateButton(Player:player):void=
        Text := if (Labels := debug_menu_checkbox_labels[Provider]) {
            if (Provider.GetState[Player]) {
                Labels.GetEnabledLabel()
            } else {
                Labels.GetDisabledLabel()
            }
        } else {
            if (Provider.GetState[Player]). "On" else "Off"
        }
        ToggleButton.SetText(ToMessage(Text))

debug_menu_choices_element := class:
    Provider: debug_menu_choices_provider
    CycleButton: text_button_base

    OnClick(Data:widget_message):void=
        if (Provider.SetSelected[Data.Player, Provider.GetSelected(Data.Player) + 1]) {
        } else if (Provider.SetSelected[Data.Player, 0]) {
        }
        UpdateButton(Data.Player)

    UpdateButton(Player:player):void=
        Text := if (FormattedOption := Provider.FormatOption[Provider.GetSelected(Player)]) {
            FormattedOption
        } else {
            "N/A"
        }
        CycleButton.SetText(ToMessage(Text))

debug_menu_slider_element := class:
    Provider: debug_menu_slider_provider
    SliderWidget: slider_regular
    Step: float

    OnValueChanged(Data:widget_message):void=
        SliderValue := SliderWidget.GetValue()
        if (Step - ToFloat(Floor[Step]) > 0.0) {
            Provider.SetValue(Data.Player, SliderValue)
            return
        }

        Value := if (
            ValInt := Round[SliderValue]
            StepInt := Floor[Step]
            ModRes := Mod[ValInt, StepInt]
            LowerBound := ValInt - ModRes
            UpperBound := LowerBound + StepInt
            l_Value := if (ValInt - LowerBound < UpperBound - ValInt). LowerBound else UpperBound
        ) {
            ToFloat(l_Value)
        } else {
            SliderWidget.GetMinValue()
        }
        Provider.SetValue(Data.Player, Value)
        if (SliderValue <> Value) {
            SliderWidget.SetValue(Value)
        }

    UpdateValue(Player:player):void=
        if (Value := Provider.GetValue[Player]) {
            SliderWidget.SetValue(Value)
        } else {
            SliderWidget.SetValue(SliderWidget.GetMinValue())
        }
## }}}

## Logs menu {{{
debug_logs_menu := class(ui_menu):
    Device: debug_device

    CreateInstance<override>(Player:player):tuple(ui_menu_instance,canvas)=
        Instance := debug_logs_menu_instance{
            Menu := Self
            Player := Player
            Device := Device
        }

        TopRowBackground := color_block{
            DefaultColor := NamedColors.CornflowerBlue
            DefaultOpacity := 1.0
        }
        Background := color_block{
            DefaultColor := NamedColors.Black
            DefaultOpacity := 0.85
        }

        ## Top Row {{{
        CloseButton := button_regular{
            DefaultText := ToMessage("Close")
        }
        CloseButton.OnClick().Subscribe(Instance.OnCloseClick)
        CloseButtonWidget := ApplySizeHint(CloseButton, vector2{X := 200.0, Y := 50.0})

        PageTextWidget := text_block{
            DefaultText := ToMessage("")
            DefaultTextColor := NamedColors.White
        }
        set Instance.m_PageTextWidget = Some(PageTextWidget)

        ScrollActions := stack_box{
            Orientation := orientation.Horizontal
            Slots := array{
                stack_box_slot{
                    Widget := PageTextWidget
                    VerticalAlignment := vertical_alignment.Center
                    HorizontalAlignment := horizontal_alignment.Left
                }
            }
        }
        for (Index->Text:array{"Up", "Down"}) {
            Button := button_regular{
                DefaultText := ToMessage(Text)
            }
            case(Index):
                0 => Button.OnClick().Subscribe(Instance.OnScrollUpClick)
                1 => Button.OnClick().Subscribe(Instance.OnScrollDownClick)
                _ => {}
            ScrollActions.AddWidget(stack_box_slot{
                Widget := ApplySizeHint(Button, vector2{X := 160.0, Y := 50.0})
                HorizontalAlignment := horizontal_alignment.Left
                Padding := margin{
                    Left := if (Index > 0). 10.0 else 20.0
                }
            })
        }

        SortOrderCycleButton := button_regular{
            DefaultText := ToMessage("")
        }
        SortOrderChoicesElement := debug_menu_choices_element{
            Provider := debug_logs_menu_sort_order_choice{
                Instance := Instance
            }
            CycleButton := SortOrderCycleButton
        }
        set Instance.m_SortOrderChoicesElement = Some(SortOrderChoicesElement)
        SortOrderCycleButton.OnClick().Subscribe(Instance.OnSortOrderCycleButtonClick)
        SortOrderChoicesElement.UpdateButton(Player)
        SortOrderCycleWidget := ApplySizeHint(SortOrderCycleButton, vector2{X := 160.0, Y := 50.0})

        LevelCycleButton := button_regular{
            DefaultText := ToMessage("")
        }
        LevelChoicesElement := debug_menu_choices_element{
            Provider := debug_logs_menu_level_choice{
                Instance := Instance
            }
            CycleButton := LevelCycleButton
        }
        set Instance.m_LevelChoicesElement = Some(LevelChoicesElement)
        LevelCycleButton.OnClick().Subscribe(Instance.OnLevelCycleButtonClick)
        LevelChoicesElement.UpdateButton(Player)
        LevelCycleWidget := ApplySizeHint(LevelCycleButton, vector2{X := 240.0, Y := 50.0})

        SyncButton := button_loud{
            DefaultText := ToMessage("Refresh")
        }
        SyncButton.OnClick().Subscribe(Instance.OnRefreshClick)
        SyncButtonWidget := ApplySizeHint(SyncButton, vector2{X := 240.0, Y := 50.0})

        TopRowMiddle := stack_box{
            Orientation := orientation.Horizontal
            Slots := array{
                stack_box_slot{
                    Widget := SortOrderCycleWidget
                    VerticalAlignment := vertical_alignment.Center
                    HorizontalAlignment := horizontal_alignment.Left
                }
                stack_box_slot{
                    Widget := LevelCycleWidget
                    VerticalAlignment := vertical_alignment.Center
                    HorizontalAlignment := horizontal_alignment.Left
                    Padding := margin{
                        Left := 15.0
                    }
                }
                stack_box_slot{
                    Widget := SyncButtonWidget
                    VerticalAlignment := vertical_alignment.Center
                    HorizontalAlignment := horizontal_alignment.Left
                    Padding := margin{
                        Left := 15.0
                    }
                }
            }
        }

        TopRow := canvas{
            Slots := array{
                canvas_slot{
                    Widget := CloseButtonWidget
                    SizeToContent := true
                }
                canvas_slot{
                    Widget := TopRowMiddle
                    SizeToContent := true
                    Anchors := anchors{
                        Minimum := vector2{X := 0.5, Y := 0.0}
                        Maximum := vector2{X := 0.5, Y := 0.0}
                    }
                    Alignment := vector2{X := 0.5, Y := 0.0}
                }
                canvas_slot{
                    Widget := ScrollActions
                    SizeToContent := true
                    Anchors := anchors{
                        Minimum := vector2{X := 1.0, Y := 0.0}
                        Maximum := vector2{X := 1.0, Y := 0.0}
                    }
                    Alignment := vector2{X := 1.0, Y := 0.0}
                }
            }
        }

        InputTopRow := overlay{
            Slots := array{
                overlay_slot{
                    Widget := TopRowBackground
                    HorizontalAlignment := horizontal_alignment.Fill
                    VerticalAlignment := vertical_alignment.Fill
                }
                overlay_slot{
                    Widget := TopRow
                    HorizontalAlignment := horizontal_alignment.Fill
                    VerticalAlignment := vertical_alignment.Top
                    Padding := margin{
                        Left := 20.0
                        Right := 20.0
                        Top := 20.0
                        Bottom := 20.0
                    }
                }
            }
        }
        ## }}}

        ## Place Content {{{
        ContentStack := stack_box{
            Orientation := orientation.Vertical
        }
        MainStack := stack_box{
            Orientation := orientation.Vertical
            Slots := array{
                stack_box_slot{
                    Widget := InputTopRow
                    VerticalAlignment := vertical_alignment.Top
                    HorizontalAlignment := horizontal_alignment.Fill
                }
                stack_box_slot{
                    Widget := ContentStack
                    VerticalAlignment := vertical_alignment.Fill
                    HorizontalAlignment := horizontal_alignment.Fill
                    Padding := margin{
                        Left := 20.0
                        Right := 20.0
                        Top := 10.0
                        Bottom := 20.0
                    }
                }
            }
        }

        InputCanvas := canvas{
            Slots := array{
                canvas_slot{
                    Widget := Background
                    SizeToContent := false
                    Anchors := anchors{
                        Minimum := vector2{X := 0.0, Y := 0.0}
                        Maximum := vector2{X := 1.0, Y := 1.0}
                    }
                    Alignment := vector2{X := 0.0, Y := 0.0}
                }
                canvas_slot{
                    Widget := MainStack
                    SizeToContent := false
                    Anchors := anchors{
                        Minimum := vector2{X := 0.0, Y := 0.0}
                        Maximum := vector2{X := 1.0, Y := 1.0}
                    }
                    Alignment := vector2{X := 0.0, Y := 0.0}
                }
            }
        }
        ## }}}

        ## Horizontal Scroll {{{
        HorizontalScrollActions := stack_box{
            Orientation := orientation.Horizontal
            Slots := for (Index->Text:array{"<", ">"}) {
                Button := button_regular{
                    DefaultText := ToMessage(Text)
                }
                case(Index):
                    0 =>
                        Button.OnClick().Subscribe(Instance.OnScrollLeftClick)
                        set Instance.m_LeftScrollWidget = Some(Button)
                    1 => Button.OnClick().Subscribe(Instance.OnScrollRightClick)
                    _ => {}
                stack_box_slot{
                    Widget := ApplySizeHint(Button, vector2{X := 50.0, Y := 30.0})
                    HorizontalAlignment := horizontal_alignment.Left
                    Padding := margin{
                        Left := if (Index > 0). 10.0 else 0.0
                    }
                }
            }
        }
        InputCanvas.AddWidget(canvas_slot{
            Widget := HorizontalScrollActions
            SizeToContent := true
            Anchors := anchors{
                Minimum := vector2{X := 1.0, Y := 1.0}
                Maximum := vector2{X := 1.0, Y := 1.0}
            }
            Alignment := vector2{X := 1.0, Y := 1.0}
            Offsets := margin{
                Left := -20.0
                Top := -20.0
            }
        })
        ## }}}

        set Instance.m_ContentStack = Some(ContentStack)
        Instance.Refresh()
        (Instance, InputCanvas)

logs_menu_log_level := enum:
    L_All
    L_Verbose
    L_Debug
    L_Normal
    L_Warning
    L_Error

GetLogsMenuLogLevelTranspilingIndex(LogLevel:logs_menu_log_level)<computes>:int=
    case(LogLevel):
        logs_menu_log_level.L_All => -1
        logs_menu_log_level.L_Verbose => 0
        logs_menu_log_level.L_Debug => 1
        logs_menu_log_level.L_Normal => 2
        logs_menu_log_level.L_Warning => 3
        logs_menu_log_level.L_Error => 4

debug_logs_menu_instance := class(ui_menu_instance):
    Device: debug_device

    var m_LevelChoicesElement: ?debug_menu_choices_element = false
    var m_SortOrderChoicesElement: ?debug_menu_choices_element = false

    var m_PageTextWidget: ?text_block = false

    var m_LogStack: []debug_log_message = array{}

    var m_ContentStack: ?stack_box = false
    var m_LeftScrollWidget: ?widget = false

    var m_EntriesWidget: ?widget = false

    var m_Level: logs_menu_log_level = logs_menu_log_level.L_All
    var m_SortOrder: sort_order = sort_order.Descending

    var m_ScrollPosition: int = 0
    var m_HorizontalScrollPosition: int = 0

    Refresh<override>():void=
        set m_ScrollPosition = 0
        set m_HorizontalScrollPosition = 0
        set m_LogStack = Device.GetLogStack()
        UpdateEntries()

    UpdateEntries():void=
        OrderedLogStack := for (
                LogMessage:m_LogStack
                (m_Level = logs_menu_log_level.L_All) or (
                    GetLogsMenuLogLevelTranspilingIndex(m_Level) = GetLogLevelIndex(LogMessage.Level))
            ) {
                LogMessage
            }
            .QuickSortWithOpts(CmpLogMessageByTime, quick_sort_opts{
                Order := m_SortOrder
            })

        OnScreen := 20
        PerPage := 10

        PageCount := DivDefaultingCeil(OrderedLogStack.Length, PerPage)
        PageIndex := if (m_ScrollPosition >= PageCount) {
            PageCount-1
        } else {
            m_ScrollPosition
        }

        LogsSliceStartIndex := PageIndex*PerPage
        LogsSliceStopIndex := Min(LogsSliceStartIndex + OnScreen, OrderedLogStack.Length)
        RelevantLogs := if (Res := OrderedLogStack.Slice[LogsSliceStartIndex, LogsSliceStopIndex]) {
            Res
        } else {
            array{}
        }

        TimeFormatOpts := number_format_options{
            MinDecimals := 2
            MaxDecimals := 2
        }
        LogsStackWidget := stack_box{
            Orientation := orientation.Vertical
            Slots := for (LogMessage:RelevantLogs) {
                TimeText := FormatNumber(LogMessage.SentAt, TimeFormatOpts, ?Style := number_format_style.Literal)
                Prefix := "{TimeText}: "
                Content := if (Message := LogMessage.OriginalMessage?) {
                    ConcatMessage(Prefix, Message)
                } else {
                    ToMessage("{Prefix}{LogMessage.StrMessage}")
                }
                Color := case(LogMessage.Level):
                    log_level.Verbose => NamedColors.Plum
                    log_level.Debug => NamedColors.Silver
                    log_level.Normal => NamedColors.White
                    log_level.Warning => NamedColors.Orange
                    log_level.Error => NamedColors.Red
                TextWidget := text_block{
                    DefaultText := Content
                    DefaultTextColor := Color
                    DefaultJustification := text_justification.Left
                    DefaultShadowOffset := Some(vector2{
                        X := 2.0
                        Y := 2.0
                    })
                    DefaultShadowColor := NamedColors.Black
                }
                TextWidget.SetShadowOpacity(0.65)
                stack_box_slot{
                    Widget := TextWidget
                    VerticalAlignment := vertical_alignment.Top
                    HorizontalAlignment := horizontal_alignment.Left
                    Padding := margin{
                        Top := 5.0
                    }
                }
            } 
        }

        if (PageTextWidget := m_PageTextWidget?) {
            Text := "{PageIndex+1}/{PageCount}"
            PageTextWidget.SetText(ToMessage(Text))
        }
        if (ContentStack := m_ContentStack?) {
            if (PrevWidget := m_EntriesWidget?) {
                ContentStack.RemoveWidget(PrevWidget)
            }
            ContentStack.AddWidget(stack_box_slot{
                Widget := LogsStackWidget
                VerticalAlignment := vertical_alignment.Top
                HorizontalAlignment := horizontal_alignment.Left
                Padding := margin{
                    Left := -(m_HorizontalScrollPosition * 250.0)
                }
            })
            set m_EntriesWidget = Some(LogsStackWidget)
        }

        if (LeftScrollWidget := m_LeftScrollWidget?) {
            LeftScrollWidget.SetVisibility(if (m_HorizontalScrollPosition > 0) {
                widget_visibility.Visible
            } else {
                widget_visibility.Hidden
            })
        }

    OnLevelCycleButtonClick(Data:widget_message):void=
        if (LevelChoicesElement := m_LevelChoicesElement?) {
            LevelChoicesElement.OnClick(Data)
            UpdateEntries()
        }

    OnSortOrderCycleButtonClick(Data:widget_message):void=
        if (SortOrderChoicesElement := m_SortOrderChoicesElement?) {
            SortOrderChoicesElement.OnClick(Data)
            UpdateEntries()
        }

    OnRefreshClick(Data:widget_message):void=
        Refresh()

    OnScrollUpClick(Data:widget_message):void=
        set m_ScrollPosition = Max(m_ScrollPosition-1, 0)
        UpdateEntries()

    OnScrollDownClick(Data:widget_message):void=
        set m_ScrollPosition += 1
        UpdateEntries()

    OnScrollLeftClick(Data:widget_message):void=
        set m_HorizontalScrollPosition = Max(m_HorizontalScrollPosition-1, 0)
        UpdateEntries()

    OnScrollRightClick(Data:widget_message):void=
        set m_HorizontalScrollPosition += 1
        UpdateEntries()

debug_logs_menu_sort_order_choice := class(debug_menu_choices_provider):
    Instance: debug_logs_menu_instance

    GetSelected<override>(Player:player)<transacts>:int=
        case(Instance.m_SortOrder):
            sort_order.Ascending => 0
            sort_order.Descending => 1

    SetSelected<override>(Player:player, Index:int)<transacts><decides>:void=
        SortOrder := case(Index):
            0 => sort_order.Ascending
            1 => sort_order.Descending
        set Instance.m_SortOrder = SortOrder

    FormatOption<override>(Index:int)<transacts><decides>:string=
        case(Index):
            0 => "ASC"
            1 => "DESC"

debug_logs_menu_level_choice := class(debug_menu_choices_provider):
    Instance: debug_logs_menu_instance

    GetSelected<override>(Player:player)<transacts>:int=
        case(Instance.m_Level):
            logs_menu_log_level.L_All => 0
            logs_menu_log_level.L_Verbose => 1
            logs_menu_log_level.L_Debug => 2
            logs_menu_log_level.L_Normal => 3
            logs_menu_log_level.L_Warning => 4
            logs_menu_log_level.L_Error => 5

    SetSelected<override>(Player:player, Index:int)<transacts><decides>:void=
        LogLevel := FromIndex[Index]
        set Instance.m_Level = LogLevel

    FormatOption<override>(Index:int)<transacts><decides>:string=
        LogLevel := FromIndex[Index]
        case(LogLevel):
            logs_menu_log_level.L_All => "All"
            logs_menu_log_level.L_Verbose => "Verbose"
            logs_menu_log_level.L_Debug => "Debug"
            logs_menu_log_level.L_Normal => "Normal"
            logs_menu_log_level.L_Warning => "Warning"
            logs_menu_log_level.L_Error => "Error"

    FromIndex<private>(Index:int)<computes><decides>:logs_menu_log_level=
        case(Index):
            0 => logs_menu_log_level.L_All
            1 => logs_menu_log_level.L_Verbose
            2 => logs_menu_log_level.L_Debug
            3 => logs_menu_log_level.L_Normal
            4 => logs_menu_log_level.L_Warning
            5 => logs_menu_log_level.L_Error
## }}}
