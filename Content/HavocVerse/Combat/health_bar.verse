using { /Verse.org/Simulation }

using { HavocVerse.UI }

health_bar_last_damage_retain_mode<public> := enum:
    OnlyLast
    AllWithinTime

health_bar_damage_result<public> := enum:
    DidNothing
    Damaged
    Died

health_bar_damage_history_entry := class:
    DamageAmount<public>: float
    Time<public>: float

combat_health_bar<public> := class(hud_health_bar_provider):
    LastDamageRetainMode<public>: health_bar_last_damage_retain_mode = health_bar_last_damage_retain_mode.AllWithinTime
    LastDamageRetainTime<public>: float = 1.0
    LastDamageClearedByHealing<public>: logic = false

    var RefreshHudManager<public>: ?hud_manager_base = false

    var Health<public>: float
    var MaxHealth<public>: float

    var m_LastHealth<private>: float = 0.0
    var m_LastHealthAt<private>: float = -1.0
    var m_DamageHistory<private>: []health_bar_damage_history_entry = array{}

    var m_PostHealListeners<private>: []type{__(:float):void} = array{}
    var m_PostDamageListeners<private>: []type{__(:float,:health_bar_damage_result):void} = array{}

    PostHealthChange<public>():void=
        if (HudManager := RefreshHudManager?) {
            HudManager.Refresh()
        }

    Heal<public>(Amount:float):void=
        PrevHealth := Health
        set Health = Min(Health + Amount, MaxHealth)
        ActualAmount := Health - PrevHealth
        if (LastDamageClearedByHealing?) {
            set m_DamageHistory = array{}
        }
        if (ActualAmount > 0.0) {
            for (Listener:m_PostHealListeners) {
                Listener(ActualAmount)
            }
        }
        PostHealthChange()

    Damage<public>(Amount:float):health_bar_damage_result=
        if (Amount <= 0.0 or not IsAlive[]) {
            return health_bar_damage_result.DidNothing
        }
        set m_LastHealth = Health
        set m_LastHealthAt = GetSimulationElapsedTime()
        set Health = Max(Health - Amount, 0.0)
        ActualDamage := m_LastHealth - Health
        Result := if (IsAlive[]) {
            health_bar_damage_result.Damaged
        } else {
            health_bar_damage_result.Died
        }
        if (LastDamageRetainMode = health_bar_last_damage_retain_mode.AllWithinTime) {
            Now := GetSimulationElapsedTime()
            set m_DamageHistory = for (Entry:m_DamageHistory, Now - Entry.Time < LastDamageRetainTime) {
                Entry
            }
            set m_DamageHistory += array{health_bar_damage_history_entry{
                DamageAmount := ActualDamage
                Time := Now
            }}
        }
        for (Listener:m_PostDamageListeners) {
            Listener(ActualDamage, Result)
        }
        PostHealthChange()
        Result

    IsAlive<public>()<decides><transacts>:void=
        Health > 0.0

    SubscribeToPostHealEvent<public>(Listener:type{__(Amount:float):void}):void=
        set m_PostHealListeners += array{Listener}

    SubscribeToPostDamageEvent<public>(Listener:type{__(DamageAmount:float,Result:health_bar_damage_result):void}):void=
        set m_PostDamageListeners += array{Listener}

    GetMaxHealth<override>():float=
        MaxHealth

    GetPrevHealth<override>():float=
        Now := GetSimulationElapsedTime()
        case(LastDamageRetainMode):
            health_bar_last_damage_retain_mode.OnlyLast =>
                if (m_LastHealthAt < 0.0 or m_LastHealthAt + LastDamageRetainTime < Now) {
                    Health
                } else {
                    m_LastHealth
                }
            health_bar_last_damage_retain_mode.AllWithinTime =>
                var TotalDamage: float = 0.0
                set m_DamageHistory = for (Entry:m_DamageHistory, Now - Entry.Time < LastDamageRetainTime) {
                    set TotalDamage += Entry.DamageAmount
                    Entry
                }
                Health + TotalDamage

    GetHealth<override>():float=
        Health
