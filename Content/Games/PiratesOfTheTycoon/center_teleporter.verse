using { /Fortnite.com/Devices }
using { /Fortnite.com/Characters }
using { /Verse.org/Simulation }
using { /UnrealEngine.com/Temporary/Diagnostics }

using { HavocVerse.Debug }
using { HavocVerse.Devices }
using { HavocVerse.Devices.GameSpecific }
using { HavocVerse.Util }
using { HavocVerse.Logging }

center_teleporter_device := class<final><concrete>(creative_device, aware_device):
    @editable p_TycoonManager: tycoon_manager_device = tycoon_manager_device{}
    @editable p_Volume: volume_device = volume_device{}
    @editable p_SpawnMutatorZone: mutator_zone_device = mutator_zone_device{}
    @editable p_ReTeleportPad: player_spawner_device = player_spawner_device{}
    @editable p_PostClaimCinematicInstances: []cinematic_sequence_device = array{}

    var m_CyclicSeqIndex: int = 0

    DeviceMeta<override>()<transacts>:aware_device_meta=
        aware_device_meta{
            DeviceName := "center_teleporter"
        }

    OnBegin<override>()<suspends>:void=
        AwareInitDevice(Self, OnInit)
    
    OnInit<private>()<suspends>:void=
        p_Volume.AgentEntersEvent.Subscribe(OnAgentEntersTeleporter)
        p_SpawnMutatorZone.AgentExitsEvent.Subscribe(OnAgentLeavesSpawn)

    OnAgentEntersTeleporter<private>(Agent:agent):void=
        if (Player := player[Agent]) {
            Result := p_TycoonManager.ClaimOrTeleportHome(Player)
            if (
                Result = tycoon_claim_or_teleport_result.Claimed
                SeqControllerIndex := m_CyclicSeqIndex
                SeqController := p_PostClaimCinematicInstances[SeqControllerIndex]
                NextIndex := Mod[m_CyclicSeqIndex + 1, p_PostClaimCinematicInstances.Length]
                set m_CyclicSeqIndex = NextIndex
            ) {
                SeqController.Play(Player)
                Log("Playing postclaim for '{Agent}'; instance index: {SeqControllerIndex}", ?Level := log_level.Debug)
            }
        }

    OnAgentLeavesSpawn<private>(Agent:agent):void=
        if:
            Player := player[Agent]
            not p_TycoonManager.IsPlaying[Player]
            FortCharacter := Player.GetFortCharacter[]
            TargetTransform := p_ReTeleportPad.GetTransform()
            TargetPosition := TargetTransform.Translation.AddZ(p_TycoonManager.p_MiscPlayerHeight)
            FortCharacter.StallingTeleportTo[TargetPosition, TargetTransform.Rotation]
            Log("Bringing '{Agent}' back to spawn because they left spawn without claiming", ?Level := log_level.Debug)
