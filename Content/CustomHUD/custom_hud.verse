using { /Fortnite.com/Devices }
using { /Verse.org/Simulation }
using { /UnrealEngine.com/Temporary/Diagnostics }

using { HavocVerse.Debug }
using { HavocVerse.Devices }
using { HavocVerse.Util }
using { HavocVerse.Logging }

custom_hud_device := class<concrete><final>(creative_device, aware_device, aware_debug_device):
    @editable p_AccessControl: access_control_device = access_control_device{}

    @editable p_ResourceManager: resource_manager_device = resource_manager_device{}
    @editable p_LevelGroupsManager: level_groups_manager_device = level_groups_manager_device{}
    @editable p_LevelGroupId: string = ""

    @editable p_PrimaryResourcesHUD: hud_message_device = hud_message_device{}
    @editable p_PrimaryResourceFormatter: number_formatter_config = number_formatter_config{}

    @editable p_LevelInfoHUD: hud_message_device = hud_message_device{}
    @editable p_LevelLabelNumberFormatter: number_formatter_config = number_formatter_config{}
    @editable p_LevelProgresssSeqControllers: []cinematic_sequence_device = array{}
    @editable p_BoosterLevelSeqControllers: []cinematic_sequence_device = array{}

    @editable p_ResourceHUDs: []hud_message_device = array{}
    @editable p_BoosterInfoHUD: hud_message_device = hud_message_device{}
    @editable p_ResourceAmountFormatter: number_formatter_config = number_formatter_config{}
    @editable p_ResourceRateFormatter: number_formatter_config = number_formatter_config{
        Template := "%Val%/s"
    }

    OrderedResources<private>: []string = array{"silver", "gems", "ruby", "treasure", "treasurechest"}
    OrderedLevels<private>: []string = array{
        "Landlubber",
        "Swabbie",
        "Deckhand",
        "Seadog",
        "Helmsman",
        "Quartermaster",
        "First Mate",
        "Captain",
        "Buccaneer",
        "Pirate Lord"
    }

    var m_ResourceBoosterSources<private>: []resource_booster_device = array{}
    var m_ActiveBoosterHUD<private>: [player]void = map{}

    LevelProgressMPC<private>: mpc_handle = mpc_handle{Context := "Level Progress"}
    BoosterLevelMPC<private>: mpc_handle = mpc_handle{Context := "Booster Level"}

    DeviceMeta<override>()<transacts>:aware_device_meta=
        aware_device_meta{
            DeviceName := "custom_hud_device"
        }

    DebugStats<override>()<transacts>:aware_debug_stats=
        aware_debug_stats{
            Length := map{
                "resourceBoosterSources" => m_ResourceBoosterSources.Length
                "activeBoosterPlayers" => m_ActiveBoosterHUD.Length
            }
        }

    OnBegin<override>()<suspends>:void=
        AwareInitDevice(Self, OnInit)
    
    OnInit<private>()<suspends>:void=
        LevelProgressMPC.Setup(GetPlayspace(), p_LevelProgresssSeqControllers)
        BoosterLevelMPC.Setup(GetPlayspace(), p_BoosterLevelSeqControllers)

        p_AccessControl.SubscribeToPlayerJoinEvents(OnPlayerAdded, OnPlayerRemoved)
        p_LevelGroupsManager.SubscribeToPlayerUpdateEvent(OnPlayerExpStateChanged)

        spawn:
            StartUpdateLoop()

        AwareSubscribeToPostInit(OnPostInit)

    OnPostInit<private>():void=
        set m_ResourceBoosterSources = array{}
        for (AwareDevice:AwareListDevices(), ResourceBoosterDevice := resource_booster_device[AwareDevice]) {
            set m_ResourceBoosterSources += array{ResourceBoosterDevice}
        }

    StartUpdateLoop<private>()<suspends>:void=
        loop {
            Sleep(0.5)
            for (Player:GetPlayspace().GetPlayers()) {
                UpdatePlayer(Player)
            }
        }

    OnPlayerAdded<private>(Player:player):void=
        UpdatePlayer(Player)

    OnPlayerRemoved<private>(Player:player, GameQuit:logic):void=
        HideUMGWidgets(Player)

    OnPlayerExpStateChanged<private>(Player:player, __:level_exp_pair):void=
        if (p_AccessControl.IsAllowed[Player]) {
            UpdateLevelInfo(Player)
        }

    UpdatePlayer<private>(Player:player):void=
        if (not p_AccessControl.IsAllowed[Player]) {
            return
        }
        UpdatePrimaryResourcesWidget(Player)
        UpdateLevelInfo(Player)
        UpdateBoosterInfo(Player)
        UpdateResourceWidgets(Player)

    HideUMGWidgets<private>(Player:player):void=
        p_PrimaryResourcesHUD.Hide(Player)
        p_LevelInfoHUD.Hide(Player)
        for (ResourceHUD:p_ResourceHUDs) {
            ResourceHUD.Hide(Player)
        }
        p_BoosterInfoHUD.Hide(Player)
        set m_ActiveBoosterHUD = m_ActiveBoosterHUD.RemoveByKey(Player)(0)

    UpdatePrimaryResourcesWidget<private>(Player:player):void=
        GoldCount := p_ResourceManager.GetAmount(Player, "money")
        BoosterCount := p_ResourceManager.GetAmount(Player, "skulls")

        GoldCountStr := p_PrimaryResourceFormatter.Format(GoldCount)
        BoosterCountStr := p_PrimaryResourceFormatter.Format(BoosterCount)

        Payload := "{GoldCountStr}\n{BoosterCountStr}"
        p_PrimaryResourcesHUD.Show(Player, ToMessage(Payload))

    UpdateLevelInfo<private>(Player:player):void=
        LevelExpPair := p_LevelGroupsManager.GetLevelExp(Player, p_LevelGroupId)
        ExpectedExp := p_LevelGroupsManager.GetTargetExp[p_LevelGroupId, LevelExpPair] or -1

        RawProgress := ToFloat(LevelExpPair.ExpRest) / ToFloat(ExpectedExp)
        Progress := Clamp(RawProgress, 0.0, 1.0)
        Frame := DefaultingFloor(Progress * 99)
        LevelProgressMPC.SetValue(Player, Frame)

        LevelTitle := if (NextLevelName := OrderedLevels[LevelExpPair.Level], RawProgress <= 1.0) {
            "Next Rank: {NextLevelName}"
        } else if (CurrentLevelName := OrderedLevels[LevelExpPair.Level-1]) {
            CurrentLevelName
        } else {
            "Rank {LevelExpPair.Level}"
        }
        LevelProgressInfo := if (RawProgress > 1.0) {
            "(MAX)"
        } else {
            CurrentExpStr := p_LevelLabelNumberFormatter.Format(LevelExpPair.ExpRest)
            ExpectedExpStr := p_LevelLabelNumberFormatter.Format(ExpectedExp)
            "{CurrentExpStr}/{ExpectedExpStr}"
        }

        Payload := "{LevelTitle}\n{LevelProgressInfo}"
        p_LevelInfoHUD.Show(Player, ToMessage(Payload))

    UpdateBoosterInfo<private>(Player:player):void=
        Multiplier := p_ResourceManager.GetEffectiveMultiplier(Player, "silver")
        BoosterLevel := Clamp(DefaultingFloor(Multiplier), 0, 99)
        BoosterLevelMPC.SetValue(Player, BoosterLevel)

        if (BoosterLevel <= 1) {
            if (m_ActiveBoosterHUD[Player]) {
                p_BoosterInfoHUD.Hide(Player)
                set m_ActiveBoosterHUD = m_ActiveBoosterHUD.RemoveByKey(Player)(0)
            }
            return
        }

        var FoundRemainingTime: ?float = None()
        for (
            ResourceBoosterDevice:m_ResourceBoosterSources
            ResourceBoosterDevice.p_MultiplierChannel = "default"
            ResourceBoosterDevice.p_AutoGenerationManager.IncludesPlayer[Player]
        ) {
            DeviceRemainingTime := ResourceBoosterDevice.GetRemainingTime(Player)
            if (PrevRemainingTime := FoundRemainingTime?) {
                set FoundRemainingTime = Some(PrevRemainingTime + DeviceRemainingTime)
            } else {
                set FoundRemainingTime = Some(DeviceRemainingTime)
            }
        }

        var Payload: string = "{BoosterLevel}x Boost!\n"
        if (RemainingTime := FoundRemainingTime?, RemainingTime > 0.0) {
            TimerText := ToTimerMMSS(DefaultingCeil(RemainingTime))
            set Payload += "{TimerText}\n"
        }
        p_BoosterInfoHUD.Show(Player, ToMessage(Payload), ?DisplayTime := 0.0)
        if:
            not m_ActiveBoosterHUD[Player]
            set m_ActiveBoosterHUD[Player] = void

    UpdateResourceWidgets<private>(Player:player):void=
        for (Index->ResourceHUD:p_ResourceHUDs, ResourceId := OrderedResources[Index]) {
            Amount := p_ResourceManager.GetAmount(Player, ResourceId)
            Rate := p_ResourceManager.GetAutoGenerationRate(Player, ResourceId)

            AmountStr := p_ResourceAmountFormatter.Format(Amount)
            RateStr := p_ResourceRateFormatter.Format(Rate)

            Payload := "{AmountStr}\n{RateStr}"
            ResourceHUD.Show(Player, ToMessage(Payload))
        }
